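# Sidebar action menu for a joinable party, showing portraits, names, presence, interaction buttons and a scrollable member list.
# The native FortJoinablePartyActionMenu parent supplies the key widgets, including dynamic entry boxes for positive and negative interactions and party members.
# A widget switcher holds the interaction list and a yes/no confirmation panel.
# The Blueprint graph only plays slide-and-fade animations: Opened on open, ConfirmationOpened when confirmation toggles visible.
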WidgetBlueprint /FortniteCoreUI/UI/Sidebar/SidebarJoinablePartyActionMenu
animations [
 Opened {
  binding_count=1
  display_label=""
  length_seconds=0.3
  object=&self:Opened
  sequence {bindings [InteractionsBorder {
     id=#1
     display_name=VerticalBox_Interactions
     possessed_class=Border
     tracks [
      {
       type=MovieScene2DTransformTrack
       sections [{
         type=MovieScene2DTransformSection
         channels [
          Translation.X {type=MovieSceneScriptingFloatChannel}
          Translation.Y {type=MovieSceneScriptingFloatChannel, keys [[0,0],[1,-72],[3,0],[4,0]]}
          Angle {type=MovieSceneScriptingFloatChannel}
          Scale.X {type=MovieSceneScriptingFloatChannel}
          Scale.Y {type=MovieSceneScriptingFloatChannel}
          Shear.X {type=MovieSceneScriptingFloatChannel}
          Shear.Y {type=MovieSceneScriptingFloatChannel}
         ]
        }]
      }
      {
       type=MovieSceneFloatTrack
       display_name=Render Opacity
       sections [{
         type=MovieSceneFloatSection
         end_frame=6
         start_frame=0
         channels [None {type=MovieSceneScriptingFloatChannel, keys [[0,0],[1,0],[3,1],[6,1]]}]
        }]
      }
     ]
    }]}
 }
 ConfirmationOpened {
  binding_count=1
  display_label=""
  length_seconds=0.3
  object=&self:ConfirmationOpened
  sequence {bindings [Border_Confirmation {
     id=#2
     display_name=VerticalBox_Confirmation
     possessed_class=Border
     tracks [
      {
       type=MovieScene2DTransformTrack
       sections [{
         type=MovieScene2DTransformSection
         channels [
          Translation.X {type=MovieSceneScriptingFloatChannel}
          Translation.Y {type=MovieSceneScriptingFloatChannel, keys [[0,0],[1,-72],[3,0],[4,0]]}
          Angle {type=MovieSceneScriptingFloatChannel}
          Scale.X {type=MovieSceneScriptingFloatChannel}
          Scale.Y {type=MovieSceneScriptingFloatChannel}
          Shear.X {type=MovieSceneScriptingFloatChannel}
          Shear.Y {type=MovieSceneScriptingFloatChannel}
         ]
        }]
      }
      {
       type=MovieSceneFloatTrack
       display_name=Render Opacity
       sections [{
         type=MovieSceneFloatSection
         end_frame=6
         start_frame=0
         channels [None {type=MovieSceneScriptingFloatChannel, keys [[0,0],[1,0],[3,1],[6,1]]}]
        }]
      }
     ]
    }]}
 }
]
graphs [EventGraph {nodes [
   n1 = Event OnToggleConfirmation!() -> (OutputDelegate, then, bIsVisible:bool) {is=title K2Node_Event}
   > Branch(Condition=@n1.bIsVisible)
   > PlayAnimation(InAnimation=Get ConfirmationOpened() -> (ConfirmationOpened:Widget Animation Object Reference))
   Event OnOpened() {is=title K2Node_Event}
   > PlayAnimation(InAnimation=Get Opened() -> (Opened:Widget Animation Object Reference))
  ]}]
variables []
widgets {
 info {inheritedWidgetCount=13, parentClass=/Script/FortniteUI.FortJoinablePartyActionMenu, rootWidgetClass=&Overlay, widgetCount=35}
 widgets [
  Overlay_141 {widgetClassPath=&Overlay, widget=Overlay_141}
  VerticalBox_1 {
   widgetClassPath=&VerticalBox
   parent=Overlay_141
   widget=VerticalBox_1
   slot {class=OverlaySlot, properties {horizontalAlignment=HAlign_Fill, verticalAlignment=VAlign_Fill}}
  }
  SizeBox_PartyIdentity {
   widgetClassPath=&SizeBox
   parent=VerticalBox_1
   widget=SizeBox_PartyIdentity
   properties {heightOverride=250, widthOverride=500}
   slot {class=VerticalBoxSlot}
  }
  VerticalBox {widgetClassPath=&VerticalBox, parent=SizeBox_PartyIdentity, widget=VerticalBox, slot {class=SizeBoxSlot}}
  Border_1 {
   widgetClassPath=&Border
   parent=VerticalBox
   widget=Border_1
   properties {
    background {
     bIsDynamicallyLoaded=false
     drawAs=Image
     imageType=NoImage
     mirroring=NoMirror
     resourceName=None
     resourceObject=/Game/UI/Sidebar/Materials/MI_UI_Sidebar_Header_Base
     tiling=NoTile
     imageSize=<32 32>
     margin=<ltrb 0 0 0 0>
     outlineSettings {
      bUseBrushTransparency=false
      roundingType=HalfHeightRadius
      width=0
      color {colorUseRule=UseColor_Specified, specifiedColor=<rgba 0 0 0 0>}
      cornerRadii=<0 0 0 0>
     }
     tintColor {colorUseRule=UseColor_Specified, specifiedColor=<rgba 1 1 1 1>}
     uVRegion {bIsValid=false, max=<0 0>, min=<0 0>}
    }
    padding=<ltrb 0 0 0 0>
   }
   slot {class=VerticalBoxSlot, properties {size {sizeRule=Fill, value=1}}}
  }
  Overlay {widgetClassPath=&Overlay, parent=Border_1, widget=Overlay, slot {class=BorderSlot, properties {padding=<ltrb 0 0 0 0>}}}
  VerticalBox_347 {
   widgetClassPath=&VerticalBox
   parent=Overlay
   widget=VerticalBox_347
   slot {class=OverlaySlot, properties {horizontalAlignment=HAlign_Fill, verticalAlignment=VAlign_Fill, padding=<ltrb 0 80 0 0>}}
  }
  SizeBox_2 {
   widgetClassPath=&SizeBox
   parent=VerticalBox_347
   widget=SizeBox_2
   properties {bOverride_HeightOverride=true, heightOverride=120}
   slot {class=VerticalBoxSlot, properties {horizontalAlignment=HAlign_Center}}
  }
  ScaleBox_0 {widgetClassPath=&ScaleBox, parent=SizeBox_2, widget=ScaleBox_0, properties {stretch=ScaleToFitY}, slot {class=SizeBoxSlot}}
  PartyPortraits {
   widgetClassPath=/Game/UI/Sidebar/SidebarJoinablePartyPortraits
   bInherited=true
   bIsVariable=true
   parent=ScaleBox_0
   widget=PartyPortraits
   slot {class=ScaleBoxSlot}
  }
  Border_179 {
   widgetClassPath=&Border
   parent=VerticalBox_347
   widget=Border_179
   properties {
    background {
     bIsDynamicallyLoaded=false
     drawAs=Image
     imageType=NoImage
     mirroring=NoMirror
     resourceName=None
     resourceObject=/Game/UI/Sidebar/Materials/M_UI_SkewBox_ProfileSubPanelIdentity
     tiling=NoTile
     imageSize=<32 32>
     margin=<ltrb 0 0 0 0>
     outlineSettings {
      bUseBrushTransparency=false
      roundingType=HalfHeightRadius
      width=0
      color {colorUseRule=UseColor_Specified, specifiedColor=<rgba 0 0 0 0>}
      cornerRadii=<0 0 0 0>
     }
     tintColor {colorUseRule=UseColor_Specified, specifiedColor=<rgba 0.004 0.0056 0.016 1>}
     uVRegion {bIsValid=false, max=<0 0>, min=<0 0>}
    }
    padding=<ltrb 0 0 0 64>
   }
   slot {class=VerticalBoxSlot, properties {verticalAlignment=VAlign_Top, padding=<ltrb 0 -32 0 0>}}
  }
  Border_181 {
   widgetClassPath=&Border
   parent=VerticalBox_347
   widget=Border_181
   properties {brushColor=<rgba 0.004 0.0056 0.016 1>, padding=<ltrb 10 -20 10 10>}
   slot {class=VerticalBoxSlot}
  }
  RichText_DisplayNames {
   widgetClassPath=&CommonRichTextBlock
   bInherited=true
   bIsVariable=true
   parent=Border_181
   widget=RichText_DisplayNames
   properties {
    autoWrapText=true
    bOverrideDefaultStyle=true
    bTintInlineIcon=true
    justification=Center
    lineHeightPercentage=1.2
    mobileTextBlockScale=1.34
    text=Mike
    defaultTextStyleOverride {
     overflowPolicy=Clip
     transformPolicy=None
     colorAndOpacity {colorUseRule=UseColor_Specified, specifiedColor=<rgba 1 1 1 1>}
     font {
      bForceMonospaced=false
      bMaterialIsStencil=false
      fontMaterial=null
      fontObject=&BurbankSmall
      letterSpacing=0
      monospacedWidth=1
      size=18
      skewAmount=0
      typefaceFontName=Black
      outlineSettings {
       bApplyOutlineToDropShadows=false
       bMiteredCorners=false
       bSeparateFillAlpha=false
       outlineMaterial=null
       outlineSize=0
       outlineColor=<rgba 0 0 0 1>
      }
     }
     highlightColor {colorUseRule=UseColor_Specified, specifiedColor=<rgba 0 0 0 1>}
     highlightShape {
      bIsDynamicallyLoaded=false
      drawAs=Image
      imageType=NoImage
      mirroring=NoMirror
      resourceName=None
      resourceObject=null
      tiling=NoTile
      imageSize=<32 32>
      margin=<ltrb 0 0 0 0>
      outlineSettings {
       bUseBrushTransparency=false
       roundingType=HalfHeightRadius
       width=0
       color {colorUseRule=UseColor_Specified, specifiedColor=<rgba 0 0 0 0>}
       cornerRadii=<0 0 0 0>
      }
      tintColor {colorUseRule=UseColor_Specified, specifiedColor=<rgba 1 1 1 1>}
      uVRegion {bIsValid=false, max=<0 0>, min=<0 0>}
     }
     selectedBackgroundColor {colorUseRule=UseColor_Foreground, specifiedColor=<rgba 1 0 1 1>}
     shadowColorAndOpacity=<rgba 0 0 0 1>
     shadowOffset=<0 0>
     strikeBrush {
      bIsDynamicallyLoaded=false
      drawAs=Image
      imageType=NoImage
      mirroring=NoMirror
      resourceName=None
      resourceObject=null
      tiling=NoTile
      imageSize=<32 32>
      margin=<ltrb 0 0 0 0>
      outlineSettings {
       bUseBrushTransparency=false
       roundingType=HalfHeightRadius
       width=0
       color {colorUseRule=UseColor_Specified, specifiedColor=<rgba 0 0 0 0>}
       cornerRadii=<0 0 0 0>
      }
      tintColor {colorUseRule=UseColor_Specified, specifiedColor=<rgba 1 1 1 1>}
      uVRegion {bIsValid=false, max=<0 0>, min=<0 0>}
     }
     underlineBrush {
      bIsDynamicallyLoaded=false
      drawAs=Image
      imageType=NoImage
      mirroring=NoMirror
      resourceName=None
      resourceObject=null
      tiling=NoTile
      imageSize=<32 32>
      margin=<ltrb 0 0 0 0>
      outlineSettings {
       bUseBrushTransparency=false
       roundingType=HalfHeightRadius
       width=0
       color {colorUseRule=UseColor_Specified, specifiedColor=<rgba 0 0 0 0>}
       cornerRadii=<0 0 0 0>
      }
      tintColor {colorUseRule=UseColor_Specified, specifiedColor=<rgba 1 1 1 1>}
      uVRegion {bIsValid=false, max=<0 0>, min=<0 0>}
     }
    }
   }
   slot {class=BorderSlot, properties {padding=<ltrb 10 -20 10 10>}}
  }
  Border {
   widgetClassPath=&Border
   parent=VerticalBox
   widget=Border
   properties {
    horizontalAlignment=HAlign_Center
    verticalAlignment=VAlign_Center
    background {
     bIsDynamicallyLoaded=false
     drawAs=Image
     imageType=NoImage
     mirroring=NoMirror
     resourceName=None
     resourceObject=null
     tiling=NoTile
     imageSize=<32 32>
     margin=<ltrb 0 0 0 0>
     outlineSettings {
      bUseBrushTransparency=false
      roundingType=HalfHeightRadius
      width=0
      color {colorUseRule=UseColor_Specified, specifiedColor=<rgba 0 0 0 0>}
      cornerRadii=<0 0 0 0>
     }
     tintColor {colorUseRule=UseColor_Specified, specifiedColor=<rgba 0.006 0.016 0.0595 0>}
     uVRegion {bIsValid=false, max=<0 0>, min=<0 0>}
    }
    padding=<ltrb 15 0 15 0>
   }
   slot {class=VerticalBoxSlot, properties {padding=<ltrb 0 30 0 0>}}
  }
  ScaleBox_1 {
   widgetClassPath=&ScaleBox
   parent=Border
   widget=ScaleBox_1
   slot {class=BorderSlot, properties {horizontalAlignment=HAlign_Center, verticalAlignment=VAlign_Center, padding=<ltrb 15 0 15 0>}}
  }
  Text_PartyPresence {
   widgetClassPath=&CommonRichTextBlock
   bInherited=true
   bIsVariable=true
   parent=ScaleBox_1
   widget=Text_PartyPresence
   properties {
    bOverrideDefaultStyle=true
    text=Game Mode
    defaultTextStyleOverride {
     overflowPolicy=Clip
     transformPolicy=None
     colorAndOpacity {colorUseRule=UseColor_Specified, specifiedColor=<rgba 0.521 0.521 0.521 1>}
     font {
      bForceMonospaced=false
      bMaterialIsStencil=false
      fontMaterial=null
      fontObject=&BurbankSmall
      letterSpacing=0
      monospacedWidth=1
      size=18
      skewAmount=0
      typefaceFontName=Bold
      outlineSettings {
       bApplyOutlineToDropShadows=false
       bMiteredCorners=false
       bSeparateFillAlpha=false
       outlineMaterial=null
       outlineSize=0
       outlineColor=<rgba 0 0 0 1>
      }
     }
     highlightColor {colorUseRule=UseColor_Specified, specifiedColor=<rgba 0 0 0 1>}
     highlightShape {
      bIsDynamicallyLoaded=false
      drawAs=Image
      imageType=NoImage
      mirroring=NoMirror
      resourceName=None
      resourceObject=null
      tiling=NoTile
      imageSize=<32 32>
      margin=<ltrb 0 0 0 0>
      outlineSettings {
       bUseBrushTransparency=false
       roundingType=HalfHeightRadius
       width=0
       color {colorUseRule=UseColor_Specified, specifiedColor=<rgba 0 0 0 0>}
       cornerRadii=<0 0 0 0>
      }
      tintColor {colorUseRule=UseColor_Specified, specifiedColor=<rgba 1 1 1 1>}
      uVRegion {bIsValid=false, max=<0 0>, min=<0 0>}
     }
     selectedBackgroundColor {colorUseRule=UseColor_Foreground, specifiedColor=<rgba 1 0 1 1>}
     shadowColorAndOpacity=<rgba 0 0 0 1>
     shadowOffset=<0 0>
     strikeBrush {
      bIsDynamicallyLoaded=false
      drawAs=Image
      imageType=NoImage
      mirroring=NoMirror
      resourceName=None
      resourceObject=null
      tiling=NoTile
      imageSize=<32 32>
      margin=<ltrb 0 0 0 0>
      outlineSettings {
       bUseBrushTransparency=false
       roundingType=HalfHeightRadius
       width=0
       color {colorUseRule=UseColor_Specified, specifiedColor=<rgba 0 0 0 0>}
       cornerRadii=<0 0 0 0>
      }
      tintColor {colorUseRule=UseColor_Specified, specifiedColor=<rgba 1 1 1 1>}
      uVRegion {bIsValid=false, max=<0 0>, min=<0 0>}
     }
     underlineBrush {
      bIsDynamicallyLoaded=false
      drawAs=Image
      imageType=NoImage
      mirroring=NoMirror
      resourceName=None
      resourceObject=null
      tiling=NoTile
      imageSize=<32 32>
      margin=<ltrb 0 0 0 0>
      outlineSettings {
       bUseBrushTransparency=false
       roundingType=HalfHeightRadius
       width=0
       color {colorUseRule=UseColor_Specified, specifiedColor=<rgba 0 0 0 0>}
       cornerRadii=<0 0 0 0>
      }
      tintColor {colorUseRule=UseColor_Specified, specifiedColor=<rgba 1 1 1 1>}
      uVRegion {bIsValid=false, max=<0 0>, min=<0 0>}
     }
    }
   }
   slot {class=ScaleBoxSlot}
  }
  VerticalBox_2 {
   widgetClassPath=&VerticalBox
   parent=VerticalBox_1
   widget=VerticalBox_2
   slot {class=VerticalBoxSlot, properties {verticalAlignment=VAlign_Center}}
  }
  Switcher_Confirmation {
   widgetClassPath=&CommonWidgetSwitcherLegacy
   bInherited=true
   bIsVariable=true
   parent=VerticalBox_2
   widget=Switcher_Confirmation
   properties {transitionDuration=0.2}
   slot {class=VerticalBoxSlot, properties {padding=<ltrb 0 40 0 40>}}
  }
  VerticalBox_Interactions {
   widgetClassPath=&VerticalBox
   parent=Switcher_Confirmation
   widget=VerticalBox_Interactions
   slot {class=WidgetSwitcherSlot, properties {padding=<ltrb 80 0 80 0>}}
  }
  EntryBox_PositiveInteractions {
   widgetClassPath=&DynamicEntryBox
   bInherited=true
   bIsVariable=true
   parent=VerticalBox_Interactions
   widget=EntryBox_PositiveInteractions
   properties {
    entryBoxType=Vertical
    entryWidgetClass=/Game/UI/Sidebar/SidebarInteraction_Profile.SidebarInteraction_Profile_C
    numDesignerPreviewEntries=1
    entrySpacing=<0 4>
   }
   slot {class=VerticalBoxSlot}
  }
  Spacer_InteractionSeparator {
   widgetClassPath=&Spacer
   bInherited=true
   parent=VerticalBox_Interactions
   widget=Spacer_InteractionSeparator
   properties {size=<1 20>}
   slot {class=VerticalBoxSlot}
  }
  EntryBox_NegativeInteractions {
   widgetClassPath=&DynamicEntryBox
   bInherited=true
   bIsVariable=true
   parent=VerticalBox_Interactions
   widget=EntryBox_NegativeInteractions
   properties {
    entryBoxType=Vertical
    entryWidgetClass=&SidebarInteraction_Profile_C
    numDesignerPreviewEntries=1
    entrySpacing=<0 4>
   }
   slot {class=VerticalBoxSlot, properties {padding=<ltrb 0 0 0 4>}}
  }
  VerticalBox_Confirmation {
   widgetClassPath=&VerticalBox
   parent=Switcher_Confirmation
   widget=VerticalBox_Confirmation
   slot {class=WidgetSwitcherSlot, properties {verticalAlignment=VAlign_Center, padding=<ltrb 0 8 0 8>}}
  }
  Text_ConfirmationLabel {
   widgetClassPath=&CommonTextBlock
   bInherited=true
   parent=VerticalBox_Confirmation
   widget=Text_ConfirmationLabel
   properties {
    justification=Center
    renderOpacity=0.3
    scrollStyle=&TextScrollStyle-Base_C
    text=CONFIRM ACTION?
    font {
     bForceMonospaced=false
     bMaterialIsStencil=false
     fontMaterial=null
     fontObject=&BurbankSmall
     letterSpacing=0
     monospacedWidth=1
     size=24
     skewAmount=0
     typefaceFontName=NotoSans-Regular
     outlineSettings {
      bApplyOutlineToDropShadows=false
      bMiteredCorners=false
      bSeparateFillAlpha=false
      outlineMaterial=null
      outlineSize=0
      outlineColor=<rgba 0 0 0 1>
     }
    }
   }
   slot {class=VerticalBoxSlot, properties {horizontalAlignment=HAlign_Center, padding=<ltrb 0 8 0 -6>}}
  }
  HorizontalBox_0 {
   widgetClassPath=&HorizontalBox
   parent=VerticalBox_Confirmation
   widget=HorizontalBox_0
   properties {flowDirectionPreference=Culture}
   slot {class=VerticalBoxSlot, properties {padding=<ltrb 0 16 0 0>}}
  }
  Button_Confirm {
   widgetClassPath=/Game/UI/SocialPanel/SocialConfirmationButton
   bInherited=true
   bIsVariable=true
   parent=HorizontalBox_0
   widget=Button_Confirm
   properties {
    bIsAccept=true
    displayText=YES
    navigation=&self:WidgetTree.Button_Confirm.WidgetNavigation_0
    onButtonBaseClicked=(&self:WidgetTree.Button_Confirm.HandleOnButtonClicked)
    onButtonBaseDoubleClicked=(&self:WidgetTree.Button_Confirm.HandleOnButtonDoubleClicked)
    onButtonBaseHovered=(&self:WidgetTree.Button_Confirm.HandleOnButtonHovered)
    onButtonBaseUnhovered=(&self:WidgetTree.Button_Confirm.HandleOnButtonUnhovered)
    onSelectedChangedBase=(&self:WidgetTree.Button_Confirm.HandleOnSelectedChanged)
   }
   slot {class=HorizontalBoxSlot, properties {padding=<ltrb 0 0 -4 0>, size {sizeRule=Fill, value=1}}}
  }
  Button_Decline {
   widgetClassPath=&SocialConfirmationButton
   bInherited=true
   bIsVariable=true
   parent=HorizontalBox_0
   widget=Button_Decline
   properties {
    displayText=NO
    navigation=&self:WidgetTree.Button_Decline.WidgetNavigation_0
    onButtonBaseClicked=(&self:WidgetTree.Button_Decline.HandleOnButtonClicked)
    onButtonBaseDoubleClicked=(&self:WidgetTree.Button_Decline.HandleOnButtonDoubleClicked)
    onButtonBaseHovered=(&self:WidgetTree.Button_Decline.HandleOnButtonHovered)
    onButtonBaseUnhovered=(&self:WidgetTree.Button_Decline.HandleOnButtonUnhovered)
    onSelectedChangedBase=(&self:WidgetTree.Button_Decline.HandleOnSelectedChanged)
   }
   slot {class=HorizontalBoxSlot, properties {padding=<ltrb -4 0 0 0>, size {sizeRule=Fill, value=1}}}
  }
  VerticalBox_PartyMembers {
   widgetClassPath=&VerticalBox
   parent=VerticalBox_1
   widget=VerticalBox_PartyMembers
   slot {class=VerticalBoxSlot, properties {size {sizeRule=Fill, value=0.6}}}
  }
  Overlay_2 {widgetClassPath=&Overlay, parent=VerticalBox_PartyMembers, widget=Overlay_2, slot {class=VerticalBoxSlot}}
  CommonBorder_0 {
   widgetClassPath=&CommonBorder
   parent=Overlay_2
   widget=CommonBorder_0
   properties {
    verticalAlignment=VAlign_Top
    background {
     bIsDynamicallyLoaded=false
     drawAs=Box
     imageType=NoImage
     mirroring=NoMirror
     resourceName=None
     resourceObject=null
     tiling=NoTile
     imageSize=<32 32>
     margin=<ltrb 0 0 0 0>
     outlineSettings {
      bUseBrushTransparency=false
      roundingType=HalfHeightRadius
      width=0
      color {colorUseRule=UseColor_Specified, specifiedColor=<rgba 0 0 0 0>}
      cornerRadii=<0 0 0 0>
     }
     tintColor {colorUseRule=UseColor_Specified, specifiedColor=<rgba 1 1 1 1>}
     uVRegion {bIsValid=false, max=<0 0>, min=<0 0>}
    }
    brushColor=<rgba 0.004 0.0056 0.016 1>
    padding=<ltrb 20 10 0 10>
   }
   slot {class=OverlaySlot, properties {horizontalAlignment=HAlign_Fill, verticalAlignment=VAlign_Fill}}
  }
  CommonRichTextBlock_PartyMembers {
   widgetClassPath=&CommonRichTextBlock
   bIsVariable=true
   parent=CommonBorder_0
   widget=CommonRichTextBlock_PartyMembers
   properties {
    bOverrideDefaultStyle=true
    text=Party Members
    textTransformPolicy=ToUpper
    defaultTextStyleOverride {
     overflowPolicy=Clip
     transformPolicy=None
     colorAndOpacity {colorUseRule=UseColor_Specified, specifiedColor=<rgba 0.5089 0.5583 0.7084 1>}
     font {
      bForceMonospaced=false
      bMaterialIsStencil=false
      fontMaterial=null
      fontObject=&BurbankSmall
      letterSpacing=0
      monospacedWidth=1
      size=17
      skewAmount=0
      typefaceFontName=Black
      outlineSettings {
       bApplyOutlineToDropShadows=false
       bMiteredCorners=false
       bSeparateFillAlpha=false
       outlineMaterial=null
       outlineSize=0
       outlineColor=<rgba 0 0 0 1>
      }
     }
     highlightColor {colorUseRule=UseColor_Specified, specifiedColor=<rgba 0 0 0 1>}
     highlightShape {
      bIsDynamicallyLoaded=false
      drawAs=Image
      imageType=NoImage
      mirroring=NoMirror
      resourceName=None
      resourceObject=null
      tiling=NoTile
      imageSize=<32 32>
      margin=<ltrb 0 0 0 0>
      outlineSettings {
       bUseBrushTransparency=false
       roundingType=HalfHeightRadius
       width=0
       color {colorUseRule=UseColor_Specified, specifiedColor=<rgba 0 0 0 0>}
       cornerRadii=<0 0 0 0>
      }
      tintColor {colorUseRule=UseColor_Specified, specifiedColor=<rgba 1 1 1 1>}
      uVRegion {bIsValid=false, max=<0 0>, min=<0 0>}
     }
     selectedBackgroundColor {colorUseRule=UseColor_Foreground, specifiedColor=<rgba 1 0 1 1>}
     shadowColorAndOpacity=<rgba 0 0 0 1>
     shadowOffset=<0 0>
     strikeBrush {
      bIsDynamicallyLoaded=false
      drawAs=Image
      imageType=NoImage
      mirroring=NoMirror
      resourceName=None
      resourceObject=null
      tiling=NoTile
      imageSize=<32 32>
      margin=<ltrb 0 0 0 0>
      outlineSettings {
       bUseBrushTransparency=false
       roundingType=HalfHeightRadius
       width=0
       color {colorUseRule=UseColor_Specified, specifiedColor=<rgba 0 0 0 0>}
       cornerRadii=<0 0 0 0>
      }
      tintColor {colorUseRule=UseColor_Specified, specifiedColor=<rgba 1 1 1 1>}
      uVRegion {bIsValid=false, max=<0 0>, min=<0 0>}
     }
     underlineBrush {
      bIsDynamicallyLoaded=false
      drawAs=Image
      imageType=NoImage
      mirroring=NoMirror
      resourceName=None
      resourceObject=null
      tiling=NoTile
      imageSize=<32 32>
      margin=<ltrb 0 0 0 0>
      outlineSettings {
       bUseBrushTransparency=false
       roundingType=HalfHeightRadius
       width=0
       color {colorUseRule=UseColor_Specified, specifiedColor=<rgba 0 0 0 0>}
       cornerRadii=<0 0 0 0>
      }
      tintColor {colorUseRule=UseColor_Specified, specifiedColor=<rgba 1 1 1 1>}
      uVRegion {bIsValid=false, max=<0 0>, min=<0 0>}
     }
    }
    renderTransform {angle=0, scale=<1 1>, shear=<-8 0>, translation=<0 0>}
   }
   slot {class=BorderSlot, properties {verticalAlignment=VAlign_Top, padding=<ltrb 20 10 0 10>}}
  }
  Border_0 {
   widgetClassPath=&Border
   parent=VerticalBox_PartyMembers
   widget=Border_0
   properties {
    background {
     bIsDynamicallyLoaded=false
     drawAs=NoDrawType
     imageType=NoImage
     mirroring=NoMirror
     resourceName=None
     resourceObject=null
     tiling=NoTile
     imageSize=<32 32>
     margin=<ltrb 0 0 0 0>
     outlineSettings {
      bUseBrushTransparency=false
      roundingType=HalfHeightRadius
      width=0
      color {colorUseRule=UseColor_Specified, specifiedColor=<rgba 0 0 0 0>}
      cornerRadii=<0 0 0 0>
     }
     tintColor {colorUseRule=UseColor_Specified, specifiedColor=<rgba 1 1 1 1>}
     uVRegion {bIsValid=false, max=<0 0>, min=<0 0>}
    }
    padding=<ltrb 0 0 0 0>
   }
   slot {class=VerticalBoxSlot, properties {size {sizeRule=Fill, value=1}}}
  }
  ScrollBox_PartyMembers {
   widgetClassPath=&ScrollBox
   bInherited=true
   parent=Border_0
   widget=ScrollBox_PartyMembers
   slot {class=BorderSlot, properties {padding=<ltrb 0 0 0 0>}}
  }
  EntryBox_PartyMembers {
   widgetClassPath=&DynamicEntryBox
   bInherited=true
   bIsVariable=true
   parent=ScrollBox_PartyMembers
   widget=EntryBox_PartyMembers
   properties {
    entryBoxType=Vertical
    entryWidgetClass=/FortniteCoreUI/UI/Sidebar/ListEntries/SidebarJoinRequestUserListEntry.SidebarJoinRequestUserListEntry_C
    numDesignerPreviewEntries=4
    visibility=HitTestInvisible
    entrySpacing=<0 4>
   }
   slot {class=ScrollBoxSlot, properties {padding=<ltrb 4 2 4 2>}}
  }
  SidebarBackButton {
   widgetClassPath=/FortUILibrary/Foundation/Buttons/BackButtonSmall
   bInherited=true
   parent=Overlay_141
   widget=SidebarBackButton
   properties {bIsFocusable=false}
   slot {class=OverlaySlot}
  }
 ]
}
