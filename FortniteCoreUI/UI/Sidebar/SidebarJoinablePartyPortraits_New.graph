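# Sidebar widget that displays a row of portraits for a joinable party.
# A right-aligned horizontal box holds an initially collapsed leader portrait, followed by a dynamic entry box of member portraits.
# Both use SidebarPartyPortrait, and the entry box uses Overlay layout with negative spacing to stagger the entries.
# The Blueprint supplies only layout, filling the two inherited widget bindings of FortJoinablePartyPortraitsDisplay, with no logic in its event graph.
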WidgetBlueprint /FortniteCoreUI/UI/Sidebar/SidebarJoinablePartyPortraits_New
graphs [EventGraph {nodes [
   Event PreConstruct()
   Event Construct()
   Event Tick()
  ]}]
variables []
widgets {
 info {inheritedWidgetCount=2, parentClass=/Script/FortniteUI.FortJoinablePartyPortraitsDisplay, rootWidgetClass=&Overlay, widgetCount=4}
 widgets [
  Overlay_PartyPortraits {widgetClassPath=&Overlay, widget=Overlay_PartyPortraits}
  HorizontalBox_0 {
   widgetClassPath=&HorizontalBox
   parent=Overlay_PartyPortraits
   widget=HorizontalBox_0
   slot {class=OverlaySlot, properties {horizontalAlignment=HAlign_Right}}
  }
  PartyPortrait_PartyLeader {
   widgetClassPath=/Game/UI/Sidebar/SidebarPartyPortrait
   bInherited=true
   bIsVariable=true
   parent=HorizontalBox_0
   widget=PartyPortrait_PartyLeader
   properties {isPartyLeader=true, visibility=Collapsed}
   slot {class=HorizontalBoxSlot, properties {horizontalAlignment=HAlign_Right}}
  }
  EntryBox_PartyPortraits {
   widgetClassPath=&DynamicEntryBox
   bInherited=true
   bIsVariable=true
   parent=HorizontalBox_0
   widget=EntryBox_PartyPortraits
   properties {
    entryBoxType=Overlay
    entryHorizontalAlignment=HAlign_Left
    entryVerticalAlignment=VAlign_Top
    entryWidgetClass=/Game/UI/Sidebar/SidebarPartyPortrait.SidebarPartyPortrait_C
    numDesignerPreviewEntries=2
    entrySpacing=<-24 -24>
    radialBoxSettings {angleBetweenItems=90, bClockwiseOrder=false, bDistributeItemsEvenly=true, marginSize=0, sectorCentralAngle=360, startingAngle=0}
   }
   slot {class=HorizontalBoxSlot, properties {horizontalAlignment=HAlign_Right, verticalAlignment=VAlign_Top}}
  }
 ]
}
