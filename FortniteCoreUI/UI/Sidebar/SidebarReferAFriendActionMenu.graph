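# Sidebar menu widget for Refer A Friend, built on FortReferFriendActionMenu, showing a header, reward text and a QR code to scan.
# OnRequestOriginChanged switches the QR image between Add Friend and Party Up variants via a widget switcher.
# OnOpened plays a short fade and slide-down of the content.
# OnKeyUp marks gamepad left face button releases as handled without further action; the ConfirmationOpened animation is empty and unused here.
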WidgetBlueprint /FortniteCoreUI/UI/Sidebar/SidebarReferAFriendActionMenu
animations [
 Opened {
  binding_count=1
  display_label=""
  length_seconds=0.2
  object=&self:Opened
  sequence {bindings [VerticalBox_144 {
     id=#1
     possessed_class=VerticalBox
     tracks [
      {
       type=MovieSceneFloatTrack
       display_name=Render Opacity
       sections [{type=MovieSceneFloatSection, channels [None {type=MovieSceneScriptingFloatChannel, keys [[0,0],[3,1]]}]}]
      }
      {
       type=MovieScene2DTransformTrack
       sections [{
         type=MovieScene2DTransformSection
         channels [
          Translation.X {type=MovieSceneScriptingFloatChannel}
          Translation.Y {type=MovieSceneScriptingFloatChannel, keys [[0,-72],[3,0]]}
          Angle {type=MovieSceneScriptingFloatChannel}
          Scale.X {type=MovieSceneScriptingFloatChannel}
          Scale.Y {type=MovieSceneScriptingFloatChannel}
          Shear.X {type=MovieSceneScriptingFloatChannel}
          Shear.Y {type=MovieSceneScriptingFloatChannel}
         ]
        }]
      }
     ]
    }]}
 }
 ConfirmationOpened {
  binding_count=0
  display_label=""
  length_seconds=0.3
  object=&self:ConfirmationOpened
  sequence {bindings []}
 }
]
graphs [
 EventGraph {nodes [
   n1 = Event OnRequestOriginChanged!() -> (OutputDelegate, then, RequestOrigin:EReferFriendOrigin Enum=AddFriend) {is=title K2Node_Event}
   > Set QRt Image!(execute, self, Origin:EReferFriendOrigin Enum=@n1.RequestOrigin) -> (then) {is=title K2Node_CallFunction}
   Event OnOpened() {is=title K2Node_Event}
   > PlayAnimation(InAnimation=Get Opened() -> (Opened:Widget Animation Object Reference))
  ]}
 OnKeyUp {nodes [
   n1 = OnKeyUp!() -> (then, MyGeometry:Geometry Structure, InKeyEvent:Key Event Structure) {is=title K2Node_FunctionEntry}
   > n2 = Branch(Condition="Equal (Key)"!(
     A:Key Structure=GetKey!(Input:"Key Event Structure (by ref)"=@n1.InKeyEvent) -> (ReturnValue:Key Structure)
     B:Key Structure=Gamepad_FaceButton_Left
    ) -> (ReturnValue:bool))
   > Return Node!(execute, ReturnValue:Event Reply Structure=Handled())
   Return Node(execute=@n2.else, ReturnValue=Unhandled())
  ]}
 Set QRt Image {nodes [
   n1 = Set QRt Image!() -> (then, Origin:EReferFriendOrigin Enum=AddFriend) {is=title K2Node_FunctionEntry}
   > n2 = Switch on EReferFriendOrigin!(execute, Selection:EReferFriendOrigin Enum=@n1) -> (AddFriend:Exec, PartyUp:Exec) {is=title K2Node_SwitchEnum}
   > SetActiveWidgetIndex(self=Get WidgetSwitcher_QR() -> (WidgetSwitcher_QR:Widget Switcher Object Reference), Index=0)
   SetActiveWidgetIndex(execute=@n2.PartyUp, self=Get WidgetSwitcher_QR(), Index=1)
  ]}
]
variables []
widgets {
 info {inheritedWidgetCount=1, parentClass=/Script/FortniteUI.FortReferFriendActionMenu, rootWidgetClass=&Overlay, widgetCount=18}
 widgets [
  Overlay_0 {widgetClassPath=&Overlay, widget=Overlay_0}
  VB_Content {
   widgetClassPath=&VerticalBox
   parent=Overlay_0
   widget=VB_Content
   slot {class=OverlaySlot, properties {horizontalAlignment=HAlign_Fill, verticalAlignment=VAlign_Fill}}
  }
  Overlay {widgetClassPath=&Overlay, parent=VB_Content, widget=Overlay, slot {class=VerticalBoxSlot}}
  Border_141 {
   widgetClassPath=&Border
   parent=Overlay
   widget=Border_141
   properties {
    background {
     bIsDynamicallyLoaded=false
     drawAs=Image
     imageType=NoImage
     mirroring=NoMirror
     resourceName=None
     resourceObject=/Game/UI/Sidebar/Materials/MI_UI_Sidebar_Header_BaseFade
     tiling=NoTile
     imageSize=<4096 4096>
     margin=<ltrb 0 0 0 0>
     outlineSettings {
      bUseBrushTransparency=false
      roundingType=HalfHeightRadius
      width=0
      color {colorUseRule=UseColor_Specified, specifiedColor=<rgba 0 0 0 0>}
      cornerRadii=<0 0 0 0>
     }
     tintColor {colorUseRule=UseColor_Specified, specifiedColor=<rgba 1 1 1 1>}
     uVRegion {bIsValid=false, max=<0 0>, min=<0 0>}
    }
    padding=<ltrb 0 0 0 0>
   }
   slot {class=OverlaySlot, properties {horizontalAlignment=HAlign_Fill, verticalAlignment=VAlign_Fill}}
  }
  VerticalBox_5 {
   widgetClassPath=&VerticalBox
   parent=Border_141
   widget=VerticalBox_5
   slot {class=BorderSlot, properties {padding=<ltrb 0 0 0 0>}}
  }
  ScaleBox_0 {
   widgetClassPath=&ScaleBox
   parent=VerticalBox_5
   widget=ScaleBox_0
   slot {class=VerticalBoxSlot, properties {horizontalAlignment=HAlign_Left, verticalAlignment=VAlign_Bottom, padding=<ltrb 20 60 20 20>}}
  }
  Text_PartyUpTitle {
   widgetClassPath=&CommonTextBlock
   parent=ScaleBox_0
   widget=Text_PartyUpTitle
   properties {
    text=Refer A Friend
    textTransformPolicy=ToUpper
    font {
     bForceMonospaced=false
     bMaterialIsStencil=false
     fontMaterial=null
     fontObject=&BurbankSmall
     letterSpacing=0
     monospacedWidth=1
     size=32
     skewAmount=0
     typefaceFontName=Black
     outlineSettings {
      bApplyOutlineToDropShadows=false
      bMiteredCorners=false
      bSeparateFillAlpha=false
      outlineMaterial=null
      outlineSize=0
      outlineColor=<rgba 0 0 0 1>
     }
    }
    renderTransform {angle=0, scale=<1 1>, shear=<-8 0>, translation=<0 0>}
    shadowOffset=<0 0>
   }
   slot {class=ScaleBoxSlot}
  }
  Button_Hide {
   widgetClassPath=/FortUILibrary/Foundation/Buttons/BackButtonSmall
   bInherited=true
   bIsVariable=true
   parent=Overlay
   widget=Button_Hide
   slot {class=OverlaySlot}
  }
  VerticalBox_144 {
   widgetClassPath=&VerticalBox
   parent=VB_Content
   widget=VerticalBox_144
   slot {class=VerticalBoxSlot, properties {verticalAlignment=VAlign_Center, padding=<ltrb 16 32 16 32>, size {sizeRule=Fill, value=1}}}
  }
  Title {
   widgetClassPath=&CommonTextBlock
   bIsVariable=true
   parent=VerticalBox_144
   widget=Title
   properties {
    autoWrapText=true
    bIsScrollingEnabled=false
    justification=Center
    style=/FortUILibrary/Styles/Typography/_TextStyle_SubH1._TextStyle_SubH1_C
    text=Unlock an outfit
    textOverflowPolicy=Ellipsis
    textTransformPolicy=ToUpper
    font {
     bForceMonospaced=false
     bMaterialIsStencil=false
     fontMaterial=null
     fontObject=&Burbank
     letterSpacing=0
     monospacedWidth=1
     size=42
     skewAmount=0
     typefaceFontName=BurbankRegular-Black
     outlineSettings {
      bApplyOutlineToDropShadows=false
      bMiteredCorners=false
      bSeparateFillAlpha=false
      outlineMaterial=null
      outlineSize=0
      outlineColor=<rgba 0 0 0 1>
     }
    }
    shadowOffset=<0 0>
    strikeBrush {
     bIsDynamicallyLoaded=false
     drawAs=NoDrawType
     imageType=NoImage
     mirroring=NoMirror
     resourceName=None
     resourceObject=null
     tiling=NoTile
     imageSize=<32 32>
     margin=<ltrb 0 0 0 0>
     outlineSettings {
      bUseBrushTransparency=false
      roundingType=HalfHeightRadius
      width=0
      color {colorUseRule=UseColor_Specified, specifiedColor=<rgba 0 0 0 0>}
      cornerRadii=<0 0 0 0>
     }
     tintColor {colorUseRule=UseColor_Specified, specifiedColor=<rgba 1 1 1 1>}
     uVRegion {bIsValid=false, max=<0 0>, min=<0 0>}
    }
   }
   slot {class=VerticalBoxSlot, properties {verticalAlignment=VAlign_Center}}
  }
  Border_2 {
   widgetClassPath=&Border
   parent=VerticalBox_144
   widget=Border_2
   properties {
    background {
     bIsDynamicallyLoaded=false
     drawAs=NoDrawType
     imageType=NoImage
     mirroring=NoMirror
     resourceName=None
     resourceObject=null
     tiling=NoTile
     imageSize=<32 32>
     margin=<ltrb 0 0 0 0>
     outlineSettings {
      bUseBrushTransparency=false
      roundingType=HalfHeightRadius
      width=0
      color {colorUseRule=UseColor_Specified, specifiedColor=<rgba 0 0 0 0>}
      cornerRadii=<0 0 0 0>
     }
     tintColor {colorUseRule=UseColor_Specified, specifiedColor=<rgba 1 1 1 1>}
     uVRegion {bIsValid=false, max=<0 0>, min=<0 0>}
    }
    contentColorAndOpacity=<rgba 0.6038 1 1 1>
    padding=<ltrb 0 0 0 0>
   }
   slot {class=VerticalBoxSlot, properties {verticalAlignment=VAlign_Center, padding=<ltrb 0 8 0 0>}}
  }
  Description {
   widgetClassPath=&CommonTextBlock
   bIsVariable=true
   parent=Border_2
   widget=Description
   properties {
    autoWrapText=true
    bIsScrollingEnabled=false
    justification=Center
    lineHeightPercentage=1.29
    mobileFontSizeMultiplier=1.3
    style=/FortUILibrary/Styles/Typography/_TextStyle_Body1._TextStyle_Body1_C
    text=Scan this QR code and complete tasks in refer a friend to earn in-game rewards
    textOverflowPolicy=Ellipsis
    font {
     bForceMonospaced=false
     bMaterialIsStencil=false
     fontMaterial=null
     fontObject=&BurbankSmall
     letterSpacing=0
     monospacedWidth=1
     size=18
     skewAmount=0
     typefaceFontName=Bold
     outlineSettings {
      bApplyOutlineToDropShadows=false
      bMiteredCorners=false
      bSeparateFillAlpha=false
      outlineMaterial=null
      outlineSize=0
      outlineColor=<rgba 0 0 0 1>
     }
    }
    shadowOffset=<0 0>
    strikeBrush {
     bIsDynamicallyLoaded=false
     drawAs=NoDrawType
     imageType=NoImage
     mirroring=NoMirror
     resourceName=None
     resourceObject=null
     tiling=NoTile
     imageSize=<32 32>
     margin=<ltrb 0 0 0 0>
     outlineSettings {
      bUseBrushTransparency=false
      roundingType=HalfHeightRadius
      width=0
      color {colorUseRule=UseColor_Specified, specifiedColor=<rgba 0 0 0 0>}
      cornerRadii=<0 0 0 0>
     }
     tintColor {colorUseRule=UseColor_Specified, specifiedColor=<rgba 1 1 1 1>}
     uVRegion {bIsValid=false, max=<0 0>, min=<0 0>}
    }
   }
   slot {class=BorderSlot, properties {padding=<ltrb 0 0 0 0>}}
  }
  QRCode {
   widgetClassPath=&Overlay
   parent=VerticalBox_144
   widget=QRCode
   slot {class=VerticalBoxSlot, properties {horizontalAlignment=HAlign_Center, verticalAlignment=VAlign_Top, padding=<ltrb 0 48 0 48>}}
  }
  Spacer_104 {
   widgetClassPath=&Spacer
   parent=QRCode
   widget=Spacer_104
   properties {size=<256 256>}
   slot {class=OverlaySlot, properties {horizontalAlignment=HAlign_Fill, verticalAlignment=VAlign_Fill}}
  }
  Lines {
   widgetClassPath=&Image
   bIsVariable=true
   parent=QRCode
   widget=Lines
   properties {brush {
     bIsDynamicallyLoaded=false
     drawAs=Image
     imageType=NoImage
     mirroring=NoMirror
     resourceName=None
     resourceObject=/Game/UI/Sidebar/Materials/MI_UI_SideBar_QR_Lines
     tiling=NoTile
     imageSize=<32 32>
     margin=<ltrb 0 0 0 0>
     outlineSettings {
      bUseBrushTransparency=false
      roundingType=HalfHeightRadius
      width=0
      color {colorUseRule=UseColor_Specified, specifiedColor=<rgba 0 0 0 0>}
      cornerRadii=<0 0 0 0>
     }
     tintColor {colorUseRule=UseColor_Specified, specifiedColor=<rgba 1 1 1 1>}
     uVRegion {bIsValid=false, max=<0 0>, min=<0 0>}
    }}
   slot {class=OverlaySlot, properties {horizontalAlignment=HAlign_Fill, verticalAlignment=VAlign_Fill, padding=<ltrb -24 -24 -24 -24>}}
  }
  WidgetSwitcher_QR {
   widgetClassPath=&WidgetSwitcher
   bIsVariable=true
   parent=QRCode
   widget=WidgetSwitcher_QR
   slot {class=OverlaySlot, properties {horizontalAlignment=HAlign_Fill, verticalAlignment=VAlign_Fill}}
  }
  Image_AddFriendsQR {
   widgetClassPath=&Image
   bIsVariable=true
   parent=WidgetSwitcher_QR
   widget=Image_AddFriendsQR
   properties {brush {
     bIsDynamicallyLoaded=false
     drawAs=Image
     imageType=NoImage
     mirroring=NoMirror
     resourceName=None
     resourceObject=/Game/UI/Sidebar/Materials/MI_UI_SideBar_QR_add
     tiling=NoTile
     imageSize=<256 256>
     margin=<ltrb 0 0 0 0>
     outlineSettings {
      bUseBrushTransparency=false
      roundingType=HalfHeightRadius
      width=0
      color {colorUseRule=UseColor_Specified, specifiedColor=<rgba 0 0 0 0>}
      cornerRadii=<0 0 0 0>
     }
     tintColor {colorUseRule=UseColor_Specified, specifiedColor=<rgba 1 1 1 1>}
     uVRegion {bIsValid=false, max=<0 0>, min=<0 0>}
    }}
   slot {class=WidgetSwitcherSlot}
  }
  Image_PartyUpQR {
   widgetClassPath=&Image
   bIsVariable=true
   parent=WidgetSwitcher_QR
   widget=Image_PartyUpQR
   properties {brush {
     bIsDynamicallyLoaded=false
     drawAs=Image
     imageType=NoImage
     mirroring=NoMirror
     resourceName=None
     resourceObject=/Game/UI/Sidebar/Materials/MI_UI_SideBar_QR_UP
     tiling=NoTile
     imageSize=<256 256>
     margin=<ltrb 0 0 0 0>
     outlineSettings {
      bUseBrushTransparency=false
      roundingType=HalfHeightRadius
      width=0
      color {colorUseRule=UseColor_Specified, specifiedColor=<rgba 0 0 0 0>}
      cornerRadii=<0 0 0 0>
     }
     tintColor {colorUseRule=UseColor_Specified, specifiedColor=<rgba 1 1 1 1>}
     uVRegion {bIsValid=false, max=<0 0>, min=<0 0>}
    }}
   slot {class=WidgetSwitcherSlot}
  }
 ]
}
