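# Sidebar list entry widget for "Refer a friend", built on FortReferFriendListEntry as a fixed-height row with a faint dark background.
# A non-hit-testable SidebarLfgEmptyButton supplies the visual, and the entry forwards its hover and unhover events to it.
# An inherited menu anchor targets UserActionMenu on the right, though its open-changed event is empty here.
# The inherited input action widget is kept collapsed as an unused binding, yet its icon material is still restyled on update.
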WidgetBlueprint /FortniteCoreUI/UI/Sidebar/SidebarReferAFriendEntry
graphs [EventGraph {nodes [
   "On Input Icon Updated (InputActionWidget)"() {is=title K2Node_ComponentBoundEvent}
   > n1 = GetIconDynamicMaterial!(execute, self:Common Action Widget Object Reference=Get InputActionWidget() -> (InputActionWidget:Common Action Widget Object Reference)) -> (then, ReturnValue:Material Instance Dynamic Object Reference)
   > SetScalarParameterValue(self=@n1, ParameterName=Use Dark, Value=1)
   > SetScalarParameterValue(self=@n1, ParameterName=Outline Opacity, Value=0)
   Event On Hovered()
   n2 = Get SidebarLfgEmptyButton() -> (SidebarLfgEmptyButton:Sidebar Lfg Empty Button Object Reference)
   > OnHover(self:Sidebar Lfg Empty Button Object Reference=@n2)
   Event On Unhovered()
   > OnUnhover(self:Sidebar Lfg Empty Button Object Reference=@n2)
   "On Menu Open Changed (MenuAnchor_Actions)"!() -> (OutputDelegate, then, bIsOpen:bool) {is=title K2Node_ComponentBoundEvent}
  ]}]
variables []
widgets {
 info {inheritedWidgetCount=2, parentClass=/Script/FortniteUI.FortReferFriendListEntry, rootWidgetClass=&SizeBox, widgetCount=7}
 widgets [
  SizeBox_0 {widgetClassPath=&SizeBox, widget=SizeBox_0, properties {bOverride_MinDesiredHeight=true, heightOverride=72, minDesiredHeight=72}}
  Overlay_1 {widgetClassPath=&Overlay, parent=SizeBox_0, widget=Overlay_1, slot {class=SizeBoxSlot}}
  MenuAnchor_Actions {
   widgetClassPath=/Script/UMG.MenuAnchor
   bInherited=true
   bIsVariable=true
   parent=Overlay_1
   widget=MenuAnchor_Actions
   properties {
    menuClass=/Game/UI/SocialPanel/UserActionMenu.UserActionMenu_C
    placement=MenuPlacement_MenuRight
    renderTransform {angle=0, scale=<1 1>, shear=<0 0>, translation=<8 0>}
   }
   slot {class=OverlaySlot, properties {horizontalAlignment=HAlign_Right}}
  }
  Bg {
   widgetClassPath=&Image
   bIsVariable=true
   parent=Overlay_1
   widget=Bg
   properties {
    renderOpacity=0.25
    visibility=SelfHitTestInvisible
    brush {
     bIsDynamicallyLoaded=false
     drawAs=Image
     imageType=NoImage
     mirroring=NoMirror
     resourceName=None
     resourceObject=null
     tiling=NoTile
     imageSize=<32 32>
     margin=<ltrb 0 0 0 0>
     outlineSettings {
      bUseBrushTransparency=false
      roundingType=HalfHeightRadius
      width=0
      color {colorUseRule=UseColor_Specified, specifiedColor=<rgba 0 0 0 0>}
      cornerRadii=<0 0 0 0>
     }
     tintColor {colorUseRule=UseColor_Specified, specifiedColor=<rgba 0 0 0 1>}
     uVRegion {bIsValid=false, max=<0 0>, min=<0 0>}
    }
   }
   slot {class=OverlaySlot, properties {horizontalAlignment=HAlign_Fill, verticalAlignment=VAlign_Fill}}
  }
  UnusedOldBoundWidgets {
   widgetClassPath=&Overlay
   parent=Overlay_1
   widget=UnusedOldBoundWidgets
   slot {class=OverlaySlot, properties {horizontalAlignment=HAlign_Center, verticalAlignment=VAlign_Center}}
  }
  InputActionWidget {
   widgetClassPath=&CommonActionWidget
   bInherited=true
   parent=UnusedOldBoundWidgets
   widget=InputActionWidget
   properties {renderOpacity=0, visibility=Collapsed, inputActions [{dataTable=&InputActionDataTable, rowName=Select}]}
   slot {class=OverlaySlot}
  }
  SidebarLfgEmptyButton {
   widgetClassPath=/FortniteCoreUI/UI/Sidebar/Tags/SidebarLfgEmptyButton
   bIsVariable=true
   parent=Overlay_1
   widget=SidebarLfgEmptyButton
   properties {
    buttonCopy=Refer a friend
    visibility=HitTestInvisible
    triggeringInputAction {dataTable=&InputActionDataTable, rowName=Select}
   }
   slot {class=OverlaySlot, properties {horizontalAlignment=HAlign_Center, verticalAlignment=VAlign_Center}}
  }
 ]
}
