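# Sidebar header button for a voice channel, built on FortVoiceChannelHeaderButton, showing channel title, status, join prompt and expansion arrow.
# OnUpdateVisuals feeds active, available, listening and restricted flags into setter functions for opacity, colors, icons and visibility.
# A runtime option selects either ON/OFF labels for simultaneous connections or Join text with the arrow.
# Hover swaps text variants and plays animations; list expansion rotates the arrow, with sound.
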WidgetBlueprint /FortniteCoreUI/UI/Sidebar/SidebarVoiceChannelHeaderButton
animations [
 ScaleUp {
  binding_count=4
  length_seconds=0.1
  object=&self:ScaleUp
  sequence {bindings [
    SizeBox_0 {
     id=#1
     possessed_class=SizeBox
     tracks [{
       type=MovieSceneFloatTrack
       display_name=Render Opacity
       sections [{type=MovieSceneFloatSection, end_frame=2, start_frame=0, channels [None {type=MovieSceneScriptingFloatChannel, keys [[0,0],[2,1]]}]}]
      }]
    }
    HorizontalBoxSlot_2 {
     id=#2
     parent_id=#1
     possessed_class=HorizontalBoxSlot
     tracks [{
       type=MovieSceneMarginTrack
       display_name=Padding
       sections [{
         type=MovieSceneMarginSection
         end_frame=2
         start_frame=0
         channels [
          Left {type=MovieSceneScriptingFloatChannel}
          Top {type=MovieSceneScriptingFloatChannel}
          Right {type=MovieSceneScriptingFloatChannel, keys [[0,-15],[2,0]]}
          Bottom {type=MovieSceneScriptingFloatChannel}
         ]
        }]
      }]
    }
    HorizontalBoxSlot_1 {
     id=#3
     parent_id=#4
     possessed_class=HorizontalBoxSlot
     tracks [{
       type=MovieSceneMarginTrack
       display_name=Padding
       sections [{
         type=MovieSceneMarginSection
         end_frame=2
         start_frame=0
         channels [
          Left {type=MovieSceneScriptingFloatChannel, keys [[0,20],[2,15]]}
          Top {type=MovieSceneScriptingFloatChannel}
          Right {type=MovieSceneScriptingFloatChannel}
          Bottom {type=MovieSceneScriptingFloatChannel}
         ]
        }]
      }]
    }
    VerticalBox_0 {id=#4, possessed_class=VerticalBox, tracks []}
   ]}
 }
 JoinSlideOut {
  binding_count=1
  length_seconds=0.15
  object=&self:JoinSlideOut
  sequence {bindings [Overlay_1 {
     id=#5
     possessed_class=Overlay
     tracks [{
       type=MovieScene2DTransformTrack
       sections [{
         type=MovieScene2DTransformSection
         end_frame=3
         start_frame=0
         channels [
          Translation.X {type=MovieSceneScriptingFloatChannel, const [0,2]}
          Translation.Y {type=MovieSceneScriptingFloatChannel, const [0,2]}
          Angle {type=MovieSceneScriptingFloatChannel, const [0,2]}
          Scale.X {type=MovieSceneScriptingFloatChannel, keys [[0,0],[3,1]]}
          Scale.Y {type=MovieSceneScriptingFloatChannel, const [1,2]}
          Shear.X {type=MovieSceneScriptingFloatChannel, const [0,2]}
          Shear.Y {type=MovieSceneScriptingFloatChannel, const [0,2]}
         ]
        }]
      }]
    }]}
 }
 OnExpansion {
  binding_count=1
  length_seconds=0.2
  object=&self:OnExpansion
  sequence {bindings [ExpandDirectionTick {
     id=#6
     possessed_class=FortMobileImage
     tracks [{
       type=MovieScene2DTransformTrack
       sections [{
         type=MovieScene2DTransformSection
         end_frame=4
         start_frame=0
         channels [
          Translation.X {type=MovieSceneScriptingFloatChannel, const [0,1]}
          Translation.Y {type=MovieSceneScriptingFloatChannel, const [0,1]}
          Angle {type=MovieSceneScriptingFloatChannel, keys [[0,0],[4,90]]}
          Scale.X {type=MovieSceneScriptingFloatChannel, keys [[0,1],[3,1.5],[4,1]]}
          Scale.Y {type=MovieSceneScriptingFloatChannel, keys [[0,1],[3,1.5],[4,1]]}
          Shear.X {type=MovieSceneScriptingFloatChannel, const [0,1]}
          Shear.Y {type=MovieSceneScriptingFloatChannel, const [0,1]}
         ]
        }]
      }]
    }]}
 }
 NewAnimation {binding_count=0, length_seconds=5, object=&self:NewAnimation, sequence {bindings []}}
]
graphs [
 EventGraph {nodes [
   Event Construct()
   > SetDefaultColorAndOpacity!(
    execute
    self:Rich Text Block Object Reference=[Get OnText() -> (OnText:Common Rich Text Block Object Reference), Get OffText() -> (OffText:Common Rich Text Block Object Reference)]
    InColorAndOpacity:Slate Color Structure=Get OnOffTextColor() -> (OnOffTextColor:Slate Color Structure)
   ) -> (then)
   n1 = Event On Item Expansion Changed!() -> (OutputDelegate, then, bIsExpanded:bool) {is=title K2Node_Event}
   > PlayAnimation(
    InAnimation=Get OnExpansion() -> (OnExpansion:Widget Animation Object Reference)
    PlayMode=Select(Option 0:EUMGSequencePlayMode Enum=Reverse, Option 1:EUMGSequencePlayMode Enum, Index=@n1.bIsExpanded) -> (ReturnValue:EUMGSequencePlayMode Enum)
    PlaybackSpeed=1.4
   )
   > PlaySound2D(Sound=/Game/Sounds/Fort_UI_Sounds/FrontEnd/New_FrontEnd/Fortnite_UI_MainMenu_Tab_Hover_02_Cue)
   n2 = Event OnUpdateVisuals!() -> (
    OutputDelegate
    then
    bIsActiveAndListening:bool
    bIsActiveOrAvailable:bool
    bIsAvailableAndNotActive:bool
    bIsListening:bool
    bShowingRestriction:bool
   ) {is=title K2Node_Event}
   > SetActiveAndListeningVisuals!(execute, self, IsActiveAndListening:bool=@n2.bIsActiveAndListening, IsRestricted:bool=@n2.bShowingRestriction) -> (then)
   > SetIsActiveOrAvailableVisuals!(execute, self, IsActiveOrAvailable:bool=@n2.bIsActiveOrAvailable) -> (then)
   > SetIsListeningVisuals!(execute, self, IsListening:bool=@n2.bIsListening) -> (then)
   > SetIsAvaliableAndNotActiveVisuals!(execute, self, IsAvaliableAndNotActive:bool=@n2.bIsAvailableAndNotActive) -> (then)
   > SetRestrictedVisuals!(execute, self, IsRestricted:bool=@n2.bShowingRestriction) -> (then)
   n3 = GetRuntimeOptions()
   > SetActiveWidgetIndex(
    self=Get SimultaneousChannelSwitcher() -> (SimultaneousChannelSwitcher:Widget Switcher Object Reference)
    Index=SelectInt(
     A=0
     B=1
     bPickA=Get bAllowSimultaneousChannelConnections!(self:Fort Runtime Options Object Reference=@n3) -> (bAllowSimultaneousChannelConnections:bool)
    )
   )
   Event On Hovered()
   > Set IsHoveredColor(IsHoveredColor:bool=true) -> (Output_Get:bool)
   > PlayAnimationForward(InAnimation=Get ScaleUp() -> (ScaleUp:Widget Animation Object Reference))
   > SetActiveWidget(
    self:Widget Switcher Object Reference=Get Switcher_ChannelTitle() -> (Switcher_ChannelTitle:Common Activatable Widget Switcher Object Reference)
    Widget=Get Text_ChannelTitle_Hover() -> (Text_ChannelTitle_Hover:Common Rich Text Block Object Reference)
   )
   > SetActiveWidget(
    self=Get Switcher_ChannelInfo() -> (Switcher_ChannelInfo:Common Activatable Widget Switcher Object Reference)
    Widget=Get Text_ChannelInfo_Hover() -> (Text_ChannelInfo_Hover:Common Text Object Reference)
   )
   > SetDefaultColorAndOpacity(self=[Get OnText(), Get OffText()], InColorAndOpacity=Get OnOffTextColor_Hover() -> (OnOffTextColor_Hover:Slate Color Structure))
   > n4 = Branch(Condition=IsActiveChannel!(self:Fort Voice Channel Header Button Object Reference) -> (ReturnValue:bool))
   Branch(execute=@n4.else, Condition=CanJoinChannel!(self:Fort Voice Channel Header Button Object Reference) -> (ReturnValue:bool))
   > PlayAnimationForward(InAnimation=Get JoinSlideOut() -> (JoinSlideOut:Widget Animation Object Reference))
   Event On Unhovered()
   > Set IsHoveredColor(IsHoveredColor=false)
   > PlayAnimationReverse(InAnimation=Get ScaleUp())
   > SetActiveWidget(self=Get Switcher_ChannelTitle(), Widget=Get Text_ChannelTitle() -> (Text_ChannelTitle:Common Rich Text Block Object Reference))
   > SetActiveWidget(self=Get Switcher_ChannelInfo(), Widget=Get Text_ChannelInfo() -> (Text_ChannelInfo:Common Text Object Reference))
   > SetDefaultColorAndOpacity(self=[Get OnText(), Get OffText()], InColorAndOpacity=Get OnOffTextColor())
   > n5 = Branch(Condition=IsActiveChannel())
   Branch(execute=@n5.else, Condition=CanJoinChannel())
   > PlayAnimationReverse(InAnimation=Get JoinSlideOut())
   K2Node_Event_6 = OnListItemObjectSet_1!() -> (OutputDelegate, then, ListItemObject:Object Reference) {is=title K2Node_CustomEvent}
   > Branch(Condition=IsListItemExpanded!(UserListEntry:User List Entry Interface) -> (ReturnValue:bool))
   n6 = Get OnExpansion()
   > PlayAnimation(InAnimation=@n6, StartAtTime=GetEndTime(self=@n6))
   Event On Enabled() {is=title K2Node_Event}
   > SetRenderScale!(execute, self:Widget Object Reference=Get Overlay_1() -> (Overlay_1:Overlay Object Reference), Scale:Vector 2D Structure=(X=1,Y=1)) -> (then)
   Event On Disabled() {is=title K2Node_Event}
   > SetRenderScale(self=Get Overlay_1(), Scale=(X=0,Y=1))
   Branch(Condition=Event PreConstruct().IsDesignTime)
   > n7 = Set IsRestricted(IsRestricted:bool=true) -> (Output_Get:bool)
   > Set IsHoveredColor(IsHoveredColor=true)
   > SetActiveAndListeningVisuals(IsActiveAndListening=false, IsRestricted=@n7)
   > SetIsActiveOrAvailableVisuals(IsActiveOrAvailable=false)
   > SetIsAvaliableAndNotActiveVisuals(IsAvaliableAndNotActive=false)
   > SetIsListeningVisuals(IsListening=false)
   > SetRestrictedVisuals(IsRestricted=@n7)
  ]}
 SetActiveAndListeningVisuals {nodes [
   n1 = SetActiveAndListeningVisuals!() -> (then, IsActiveAndListening:bool=~, IsRestricted:bool=~) {is=title K2Node_FunctionEntry}
   > SetRenderOpacity(self=Get Text_ChannelTitle(), InOpacity=SelectFloat(A=1, B=0.5, bPickA=@n1))
   > SetRenderOpacity(InOpacity=SelectFloat(A=1, B=0.5, bPickA=@n1))
   > SetVisibility(
    self=Get Switch_Transmitting() -> (Switch_Transmitting:Widget Switcher Object Reference)
    InVisibility=Select(Option 0:ESlateVisibility Enum=Collapsed, Option 1:ESlateVisibility Enum=SelfHitTestInvisible, Index=@n1) -> (ReturnValue:ESlateVisibility Enum)
   )
   > SetVisibility(
    self=Get ExpandDirectionTick() -> (ExpandDirectionTick:Fort Mobile Image Object Reference)
    InVisibility=Select(Option 0=Collapsed, Option 1=Visible, Index=@n1)
   )
   n2 = SelectColor(
    A=Get VoiceStatusActiveColor() -> (VoiceStatusActiveColor:Linear Color Structure)
    B=Get VoiceStatusInactiveColor() -> (VoiceStatusInactiveColor:Linear Color Structure)
    bPickA=@n1
   )
   n3 = Get VoiceStatusWarningColor() -> (VoiceStatusWarningColor:Linear Color Structure)
   > SetColorAndOpacity(self=Get Text_ChannelInfo(), InColorAndOpacity=Make Slate Color(SpecifiedColor=SelectColor(A=@n3, B=@n2, bPickA=@n1.IsRestricted)))
   > SetVisibility(self=Get Switch_Transmitting(), InVisibility=Select(Option 0=Collapsed, Option 1=SelfHitTestInvisible, Index=@n1))
   > SetVisibility(self=Get ExpandDirectionTick(), InVisibility=Select(Option 0=Collapsed, Option 1=Visible, Index=@n1))
  ]}
 SetIsActiveOrAvailableVisuals {nodes [
   n1 = SetIsActiveOrAvailableVisuals!() -> (then, IsActiveOrAvailable:bool=~) {is=title K2Node_FunctionEntry}
   > SetVisibility(
    self=Get HBox_Join() -> (HBox_Join:Horizontal Box Object Reference)
    InVisibility=Select(Option 0=Collapsed, Option 1=SelfHitTestInvisible, Index=@n1)
   )
   > SetRenderScale(
    self=Get Overlay_1()
    Scale=Select(Option 0:Vector 2D Structure=(X=0,Y=1), Option 1:Vector 2D Structure=(X=1,Y=1), Index=@n1) -> (ReturnValue:Vector 2D Structure)
   )
  ]}
 SetIsAvaliableAndNotActiveVisuals {nodes [
   n1 = SetIsAvaliableAndNotActiveVisuals!() -> (then, IsAvaliableAndNotActive:bool=~) {is=title K2Node_FunctionEntry}
   > SetVisibility(
    self=Get JoinText() -> (JoinText:Common Rich Text Block Object Reference)
    InVisibility=Select(Option 0:ESlateVisibility Enum=Collapsed, Option 1:ESlateVisibility Enum=Visible, Index=@n1) -> (ReturnValue:ESlateVisibility Enum)
   )
  ]}
 SetIsListeningVisuals {nodes [
   n1 = SetIsListeningVisuals!() -> (then, IsListening:bool=~) {is=title K2Node_FunctionEntry}
   > SetActiveWidgetIndex(self=Get OnOffSwitcher() -> (OnOffSwitcher:Widget Switcher Object Reference), Index=SelectInt(A=0, B=1, bPickA=@n1))
  ]}
 SetRestrictedVisuals {nodes [
   n1 = SetRestrictedVisuals!() -> (then, IsRestricted:bool=~) {is=title K2Node_FunctionEntry}
   n2 = Get Image_VoiceChatWarning() -> (Image_VoiceChatWarning:Image Object Reference)
   > SetVisibility(self=@n2, InVisibility=Select(Option 0=Collapsed, Option 1=Visible, Index=@n1))
   > SetColorAndOpacity(
    self:Image Object Reference=@n2
    InColorAndOpacity:Linear Color Structure=Select(
     Option 0:Linear Color Structure=(R=0.2051,G=0.2384,B=0.3564,A=1)
     Option 1:Linear Color Structure=(R=0.0513,G=0.0612,B=0.0723,A=1)
     Index=Get IsHoveredColor() -> (IsHoveredColor:bool)
    ) -> (ReturnValue:Linear Color Structure)
   )
   > SetColorAndOpacity!(
    execute
    self:Text Object Reference=Get Text_ChannelInfo_Hover()
    InColorAndOpacity_SpecifiedColor:Linear Color Structure=Select(Option 0=(R=0.2051,G=0.2384,B=0.3564,A=1), Option 1=(R=0.0513,G=0.0612,B=0.0723,A=1), Index=@n1)
    InColorAndOpacity_ColorUseRule:ESlateColorStylingMode Enum
   ) -> (then)
  ]}
]
variables [
 IsHoveredColor {type=boolean}
 IsRestricted {type=boolean}
 OnOffTextColor {type=SlateColor, default {colorUseRule=UseColor_Specified, specifiedColor=<rgba 0.0578 0.0578 0.0578 0>}}
 OnOffTextColor_Hover {type=SlateColor, default {colorUseRule=UseColor_Specified, specifiedColor=<rgba 0.0578 0.0578 0.0578 1>}}
 VoiceStatusActiveColor {type=LinearColor, default=<rgba 0.0314 0.8471 0.9922 1>}
 VoiceStatusInactiveColor {type=LinearColor, default=<rgba 0.7135 0.7135 0.7135 1>}
 VoiceStatusWarningColor {type=LinearColor, default=<rgba 0.4784 0.5216 0.5529 1>}
]
widgets {
 info {
  inheritedWidgetCount=7
  parentClass=/Script/FortniteUI.FortVoiceChannelHeaderButton
  rootWidgetClass=&HorizontalBox
  widgetCount=29
 }
 widgets [
  HorizontalBox_1 {widgetClassPath=&HorizontalBox, widget=HorizontalBox_1}
  VerticalBox_0 {
   widgetClassPath=&VerticalBox
   parent=HorizontalBox_1
   widget=VerticalBox_0
   slot {class=HorizontalBoxSlot, properties {verticalAlignment=VAlign_Center, padding=<ltrb 20 10 20 13>, size {sizeRule=Fill, value=1}}}
  }
  HorizontalBox_0 {widgetClassPath=&HorizontalBox, parent=VerticalBox_0, widget=HorizontalBox_0, slot {class=VerticalBoxSlot}}
  Switcher_ChannelTitle {
   widgetClassPath=&CommonActivatableWidgetSwitcher
   bIsVariable=true
   parent=HorizontalBox_0
   widget=Switcher_ChannelTitle
   properties {transitionCurveType=Linear, transitionDuration=0.1}
   slot {class=HorizontalBoxSlot}
  }
  Text_ChannelTitle {
   widgetClassPath=&CommonRichTextBlock
   bInherited=true
   bIsVariable=true
   parent=Switcher_ChannelTitle
   widget=Text_ChannelTitle
   properties {
    bOverrideDefaultStyle=true
    text=Party Channel
    textTransformPolicy=ToUpper
    defaultTextStyleOverride {
     overflowPolicy=Clip
     transformPolicy=None
     colorAndOpacity {colorUseRule=UseColor_Specified, specifiedColor=<rgba 1 1 1 1>}
     font {
      bForceMonospaced=false
      bMaterialIsStencil=false
      fontMaterial=null
      fontObject=&BurbankSmall
      letterSpacing=0
      monospacedWidth=1
      size=20
      skewAmount=0
      typefaceFontName=Black
      outlineSettings {
       bApplyOutlineToDropShadows=false
       bMiteredCorners=false
       bSeparateFillAlpha=false
       outlineMaterial=null
       outlineSize=0
       outlineColor=<rgba 0 0 0 1>
      }
     }
     highlightColor {colorUseRule=UseColor_Specified, specifiedColor=<rgba 1 1 1 1>}
     highlightShape {
      bIsDynamicallyLoaded=false
      drawAs=Image
      imageType=NoImage
      mirroring=NoMirror
      resourceName=None
      resourceObject=null
      tiling=NoTile
      imageSize=<32 32>
      margin=<ltrb 0 0 0 0>
      outlineSettings {
       bUseBrushTransparency=false
       roundingType=HalfHeightRadius
       width=0
       color {colorUseRule=UseColor_Specified, specifiedColor=<rgba 0 0 0 0>}
       cornerRadii=<0 0 0 0>
      }
      tintColor {colorUseRule=UseColor_Specified, specifiedColor=<rgba 1 1 1 1>}
      uVRegion {bIsValid=false, max=<0 0>, min=<0 0>}
     }
     selectedBackgroundColor {colorUseRule=UseColor_Foreground, specifiedColor=<rgba 1 0 1 1>}
     shadowColorAndOpacity=<rgba 0 0 0 1>
     shadowOffset=<0 0>
     strikeBrush {
      bIsDynamicallyLoaded=false
      drawAs=Image
      imageType=NoImage
      mirroring=NoMirror
      resourceName=None
      resourceObject=null
      tiling=NoTile
      imageSize=<32 32>
      margin=<ltrb 0 0 0 0>
      outlineSettings {
       bUseBrushTransparency=false
       roundingType=HalfHeightRadius
       width=0
       color {colorUseRule=UseColor_Specified, specifiedColor=<rgba 0 0 0 0>}
       cornerRadii=<0 0 0 0>
      }
      tintColor {colorUseRule=UseColor_Specified, specifiedColor=<rgba 1 1 1 1>}
      uVRegion {bIsValid=false, max=<0 0>, min=<0 0>}
     }
     underlineBrush {
      bIsDynamicallyLoaded=false
      drawAs=Image
      imageType=NoImage
      mirroring=NoMirror
      resourceName=None
      resourceObject=null
      tiling=NoTile
      imageSize=<32 32>
      margin=<ltrb 0 0 0 0>
      outlineSettings {
       bUseBrushTransparency=false
       roundingType=HalfHeightRadius
       width=0
       color {colorUseRule=UseColor_Specified, specifiedColor=<rgba 0 0 0 0>}
       cornerRadii=<0 0 0 0>
      }
      tintColor {colorUseRule=UseColor_Specified, specifiedColor=<rgba 1 1 1 1>}
      uVRegion {bIsValid=false, max=<0 0>, min=<0 0>}
     }
    }
   }
   slot {class=WidgetSwitcherSlot}
  }
  Text_ChannelTitle_Hover {
   widgetClassPath=&CommonRichTextBlock
   bInherited=true
   bIsVariable=true
   parent=Switcher_ChannelTitle
   widget=Text_ChannelTitle_Hover
   properties {
    bOverrideDefaultStyle=true
    text=Party Channel
    textTransformPolicy=ToUpper
    defaultTextStyleOverride {
     overflowPolicy=Clip
     transformPolicy=None
     colorAndOpacity {colorUseRule=UseColor_Specified, specifiedColor=<rgba 0.0578 0.0578 0.0578 1>}
     font {
      bForceMonospaced=false
      bMaterialIsStencil=false
      fontMaterial=null
      fontObject=&BurbankSmall
      letterSpacing=0
      monospacedWidth=1
      size=20
      skewAmount=0
      typefaceFontName=Black
      outlineSettings {
       bApplyOutlineToDropShadows=false
       bMiteredCorners=false
       bSeparateFillAlpha=false
       outlineMaterial=null
       outlineSize=0
       outlineColor=<rgba 0 0 0 1>
      }
     }
     highlightColor {colorUseRule=UseColor_Specified, specifiedColor=<rgba 1 1 1 1>}
     highlightShape {
      bIsDynamicallyLoaded=false
      drawAs=Image
      imageType=NoImage
      mirroring=NoMirror
      resourceName=None
      resourceObject=null
      tiling=NoTile
      imageSize=<32 32>
      margin=<ltrb 0 0 0 0>
      outlineSettings {
       bUseBrushTransparency=false
       roundingType=HalfHeightRadius
       width=0
       color {colorUseRule=UseColor_Specified, specifiedColor=<rgba 0 0 0 0>}
       cornerRadii=<0 0 0 0>
      }
      tintColor {colorUseRule=UseColor_Specified, specifiedColor=<rgba 1 1 1 1>}
      uVRegion {bIsValid=false, max=<0 0>, min=<0 0>}
     }
     selectedBackgroundColor {colorUseRule=UseColor_Foreground, specifiedColor=<rgba 1 0 1 1>}
     shadowColorAndOpacity=<rgba 0 0 0 1>
     shadowOffset=<0 0>
     strikeBrush {
      bIsDynamicallyLoaded=false
      drawAs=Image
      imageType=NoImage
      mirroring=NoMirror
      resourceName=None
      resourceObject=null
      tiling=NoTile
      imageSize=<32 32>
      margin=<ltrb 0 0 0 0>
      outlineSettings {
       bUseBrushTransparency=false
       roundingType=HalfHeightRadius
       width=0
       color {colorUseRule=UseColor_Specified, specifiedColor=<rgba 0 0 0 0>}
       cornerRadii=<0 0 0 0>
      }
      tintColor {colorUseRule=UseColor_Specified, specifiedColor=<rgba 1 1 1 1>}
      uVRegion {bIsValid=false, max=<0 0>, min=<0 0>}
     }
     underlineBrush {
      bIsDynamicallyLoaded=false
      drawAs=Image
      imageType=NoImage
      mirroring=NoMirror
      resourceName=None
      resourceObject=null
      tiling=NoTile
      imageSize=<32 32>
      margin=<ltrb 0 0 0 0>
      outlineSettings {
       bUseBrushTransparency=false
       roundingType=HalfHeightRadius
       width=0
       color {colorUseRule=UseColor_Specified, specifiedColor=<rgba 0 0 0 0>}
       cornerRadii=<0 0 0 0>
      }
      tintColor {colorUseRule=UseColor_Specified, specifiedColor=<rgba 1 1 1 1>}
      uVRegion {bIsValid=false, max=<0 0>, min=<0 0>}
     }
    }
   }
   slot {class=WidgetSwitcherSlot}
  }
  Switch_Transmitting {
   widgetClassPath=&WidgetSwitcher
   bIsVariable=true
   parent=HorizontalBox_0
   widget=Switch_Transmitting
   slot {class=HorizontalBoxSlot, properties {padding=<ltrb 10 0 0 0>}}
  }
  Transmitting {
   widgetClassPath=/Script/FortniteUI.FortMobileImage
   bIsVariable=true
   parent=Switch_Transmitting
   widget=Transmitting
   properties {
    mobileSizeMultiplier=1.2
    visibility=HitTestInvisible
    brush {
     bIsDynamicallyLoaded=false
     drawAs=Image
     imageType=NoImage
     mirroring=NoMirror
     resourceName=None
     resourceObject=/Game/UI/Foundation/Textures/Icons/Audio/T-Icon-Sound-On-M
     tiling=NoTile
     imageSize=<24 24>
     margin=<ltrb 0 0 0 0>
     outlineSettings {
      bUseBrushTransparency=false
      roundingType=HalfHeightRadius
      width=0
      color {colorUseRule=UseColor_Specified, specifiedColor=<rgba 0 0 0 0>}
      cornerRadii=<0 0 0 0>
     }
     tintColor {colorUseRule=UseColor_Specified, specifiedColor=<rgba 1 1 1 1>}
     uVRegion {bIsValid=false, max=<0 0>, min=<0 0>}
    }
   }
   slot {class=WidgetSwitcherSlot, properties {horizontalAlignment=HAlign_Center, verticalAlignment=VAlign_Center}}
  }
  NotTransmitting {
   widgetClassPath=&FortMobileImage
   bIsVariable=true
   parent=Switch_Transmitting
   widget=NotTransmitting
   properties {
    mobileSizeMultiplier=1.2
    visibility=HitTestInvisible
    brush {
     bIsDynamicallyLoaded=false
     drawAs=Image
     imageType=NoImage
     mirroring=NoMirror
     resourceName=None
     resourceObject=/Game/UI/Foundation/Textures/Icons/Audio/T-Icon-Sound-Mute-M
     tiling=NoTile
     imageSize=<24 24>
     margin=<ltrb 0 0 0 0>
     outlineSettings {
      bUseBrushTransparency=false
      roundingType=HalfHeightRadius
      width=0
      color {colorUseRule=UseColor_Specified, specifiedColor=<rgba 0 0 0 0>}
      cornerRadii=<0 0 0 0>
     }
     tintColor {colorUseRule=UseColor_Specified, specifiedColor=<rgba 1 1 1 1>}
     uVRegion {bIsValid=false, max=<0 0>, min=<0 0>}
    }
   }
   slot {class=WidgetSwitcherSlot, properties {horizontalAlignment=HAlign_Center, verticalAlignment=VAlign_Center}}
  }
  SizeBox_0 {
   widgetClassPath=&SizeBox
   parent=HorizontalBox_0
   widget=SizeBox_0
   properties {bOverride_MinDesiredWidth=true, minDesiredWidth=72}
   slot {class=HorizontalBoxSlot, properties {horizontalAlignment=HAlign_Right, size {sizeRule=Fill, value=1}}}
  }
  Overlay_1 {
   widgetClassPath=&Overlay
   bIsVariable=true
   parent=SizeBox_0
   widget=Overlay_1
   properties {renderTransformPivot=<1 0.5>}
   slot {class=SizeBoxSlot, properties {horizontalAlignment=HAlign_Right}}
  }
  HorizontalBox_2 {
   widgetClassPath=&HorizontalBox
   parent=Overlay_1
   widget=HorizontalBox_2
   slot {class=OverlaySlot, properties {horizontalAlignment=HAlign_Center, verticalAlignment=VAlign_Center}}
  }
  HBox_Join {
   widgetClassPath=&HorizontalBox
   bIsVariable=true
   parent=HorizontalBox_2
   widget=HBox_Join
   slot {class=HorizontalBoxSlot, properties {horizontalAlignment=HAlign_Right, verticalAlignment=VAlign_Center, padding=<ltrb 0 0 20 0>}}
  }
  SimultaneousChannelSwitcher {
   widgetClassPath=&WidgetSwitcher
   bIsVariable=true
   parent=HBox_Join
   widget=SimultaneousChannelSwitcher
   slot {class=HorizontalBoxSlot, properties {verticalAlignment=VAlign_Center, padding=<ltrb 0 0 10 0>}}
  }
  OnOffSwitcher {
   widgetClassPath=&WidgetSwitcher
   bIsVariable=true
   parent=SimultaneousChannelSwitcher
   widget=OnOffSwitcher
   slot {class=WidgetSwitcherSlot, properties {horizontalAlignment=HAlign_Right}}
  }
  OnText {
   widgetClassPath=&CommonRichTextBlock
   bIsVariable=true
   parent=OnOffSwitcher
   widget=OnText
   properties {
    bOverrideDefaultStyle=true
    mobileTextBlockScale=1.15
    text=ON
    textTransformPolicy=ToUpper
    defaultTextStyleOverride {
     overflowPolicy=Clip
     transformPolicy=None
     colorAndOpacity {colorUseRule=UseColor_Specified, specifiedColor=<rgba 0.0578 0.0578 0.0578 1>}
     font {
      bForceMonospaced=false
      bMaterialIsStencil=false
      fontMaterial=null
      fontObject=&BurbankSmall
      letterSpacing=0
      monospacedWidth=1
      size=13
      skewAmount=0
      typefaceFontName=Black
      outlineSettings {
       bApplyOutlineToDropShadows=false
       bMiteredCorners=false
       bSeparateFillAlpha=false
       outlineMaterial=null
       outlineSize=0
       outlineColor=<rgba 0 0 0 1>
      }
     }
     highlightColor {colorUseRule=UseColor_Specified, specifiedColor=<rgba 0 0 0 1>}
     highlightShape {
      bIsDynamicallyLoaded=false
      drawAs=Image
      imageType=NoImage
      mirroring=NoMirror
      resourceName=None
      resourceObject=null
      tiling=NoTile
      imageSize=<32 32>
      margin=<ltrb 0 0 0 0>
      outlineSettings {
       bUseBrushTransparency=false
       roundingType=HalfHeightRadius
       width=0
       color {colorUseRule=UseColor_Specified, specifiedColor=<rgba 0 0 0 0>}
       cornerRadii=<0 0 0 0>
      }
      tintColor {colorUseRule=UseColor_Specified, specifiedColor=<rgba 1 1 1 1>}
      uVRegion {bIsValid=false, max=<0 0>, min=<0 0>}
     }
     selectedBackgroundColor {colorUseRule=UseColor_Foreground, specifiedColor=<rgba 1 0 1 1>}
     shadowColorAndOpacity=<rgba 0 0 0 1>
     shadowOffset=<0 0>
     strikeBrush {
      bIsDynamicallyLoaded=false
      drawAs=Image
      imageType=NoImage
      mirroring=NoMirror
      resourceName=None
      resourceObject=null
      tiling=NoTile
      imageSize=<32 32>
      margin=<ltrb 0 0 0 0>
      outlineSettings {
       bUseBrushTransparency=false
       roundingType=HalfHeightRadius
       width=0
       color {colorUseRule=UseColor_Specified, specifiedColor=<rgba 0 0 0 0>}
       cornerRadii=<0 0 0 0>
      }
      tintColor {colorUseRule=UseColor_Specified, specifiedColor=<rgba 1 1 1 1>}
      uVRegion {bIsValid=false, max=<0 0>, min=<0 0>}
     }
     underlineBrush {
      bIsDynamicallyLoaded=false
      drawAs=Image
      imageType=NoImage
      mirroring=NoMirror
      resourceName=None
      resourceObject=null
      tiling=NoTile
      imageSize=<32 32>
      margin=<ltrb 0 0 0 0>
      outlineSettings {
       bUseBrushTransparency=false
       roundingType=HalfHeightRadius
       width=0
       color {colorUseRule=UseColor_Specified, specifiedColor=<rgba 0 0 0 0>}
       cornerRadii=<0 0 0 0>
      }
      tintColor {colorUseRule=UseColor_Specified, specifiedColor=<rgba 1 1 1 1>}
      uVRegion {bIsValid=false, max=<0 0>, min=<0 0>}
     }
    }
   }
   slot {class=WidgetSwitcherSlot, properties {horizontalAlignment=HAlign_Right}}
  }
  OffText {
   widgetClassPath=&CommonRichTextBlock
   bIsVariable=true
   parent=OnOffSwitcher
   widget=OffText
   properties {
    bOverrideDefaultStyle=true
    mobileTextBlockScale=1.15
    text=OFF
    textTransformPolicy=ToUpper
    defaultTextStyleOverride {
     overflowPolicy=Clip
     transformPolicy=None
     colorAndOpacity {colorUseRule=UseColor_Specified, specifiedColor=<rgba 0.0578 0.0578 0.0578 1>}
     font {
      bForceMonospaced=false
      bMaterialIsStencil=false
      fontMaterial=null
      fontObject=&BurbankSmall
      letterSpacing=0
      monospacedWidth=1
      size=13
      skewAmount=0
      typefaceFontName=Black
      outlineSettings {
       bApplyOutlineToDropShadows=false
       bMiteredCorners=false
       bSeparateFillAlpha=false
       outlineMaterial=null
       outlineSize=0
       outlineColor=<rgba 0 0 0 1>
      }
     }
     highlightColor {colorUseRule=UseColor_Specified, specifiedColor=<rgba 0 0 0 1>}
     highlightShape {
      bIsDynamicallyLoaded=false
      drawAs=Image
      imageType=NoImage
      mirroring=NoMirror
      resourceName=None
      resourceObject=null
      tiling=NoTile
      imageSize=<32 32>
      margin=<ltrb 0 0 0 0>
      outlineSettings {
       bUseBrushTransparency=false
       roundingType=HalfHeightRadius
       width=0
       color {colorUseRule=UseColor_Specified, specifiedColor=<rgba 0 0 0 0>}
       cornerRadii=<0 0 0 0>
      }
      tintColor {colorUseRule=UseColor_Specified, specifiedColor=<rgba 1 1 1 1>}
      uVRegion {bIsValid=false, max=<0 0>, min=<0 0>}
     }
     selectedBackgroundColor {colorUseRule=UseColor_Foreground, specifiedColor=<rgba 1 0 1 1>}
     shadowColorAndOpacity=<rgba 0 0 0 1>
     shadowOffset=<0 0>
     strikeBrush {
      bIsDynamicallyLoaded=false
      drawAs=Image
      imageType=NoImage
      mirroring=NoMirror
      resourceName=None
      resourceObject=null
      tiling=NoTile
      imageSize=<32 32>
      margin=<ltrb 0 0 0 0>
      outlineSettings {
       bUseBrushTransparency=false
       roundingType=HalfHeightRadius
       width=0
       color {colorUseRule=UseColor_Specified, specifiedColor=<rgba 0 0 0 0>}
       cornerRadii=<0 0 0 0>
      }
      tintColor {colorUseRule=UseColor_Specified, specifiedColor=<rgba 1 1 1 1>}
      uVRegion {bIsValid=false, max=<0 0>, min=<0 0>}
     }
     underlineBrush {
      bIsDynamicallyLoaded=false
      drawAs=Image
      imageType=NoImage
      mirroring=NoMirror
      resourceName=None
      resourceObject=null
      tiling=NoTile
      imageSize=<32 32>
      margin=<ltrb 0 0 0 0>
      outlineSettings {
       bUseBrushTransparency=false
       roundingType=HalfHeightRadius
       width=0
       color {colorUseRule=UseColor_Specified, specifiedColor=<rgba 0 0 0 0>}
       cornerRadii=<0 0 0 0>
      }
      tintColor {colorUseRule=UseColor_Specified, specifiedColor=<rgba 1 1 1 1>}
      uVRegion {bIsValid=false, max=<0 0>, min=<0 0>}
     }
    }
   }
   slot {class=WidgetSwitcherSlot, properties {horizontalAlignment=HAlign_Right}}
  }
  SingleChannelSwitcher {widgetClassPath=&HorizontalBox, parent=SimultaneousChannelSwitcher, widget=SingleChannelSwitcher, slot {class=WidgetSwitcherSlot}}
  JoinText {
   widgetClassPath=&CommonRichTextBlock
   bIsVariable=true
   parent=SingleChannelSwitcher
   widget=JoinText
   properties {
    bOverrideDefaultStyle=true
    mobileTextBlockScale=1.15
    text=Join
    textTransformPolicy=ToUpper
    defaultTextStyleOverride {
     overflowPolicy=Clip
     transformPolicy=None
     colorAndOpacity {colorUseRule=UseColor_Specified, specifiedColor=<rgba 1 1 1 1>}
     font {
      bForceMonospaced=false
      bMaterialIsStencil=false
      fontMaterial=null
      fontObject=&BurbankSmall
      letterSpacing=0
      monospacedWidth=1
      size=18
      skewAmount=0
      typefaceFontName=Bold
      outlineSettings {
       bApplyOutlineToDropShadows=false
       bMiteredCorners=false
       bSeparateFillAlpha=false
       outlineMaterial=null
       outlineSize=0
       outlineColor=<rgba 0 0 0 1>
      }
     }
     highlightColor {colorUseRule=UseColor_Specified, specifiedColor=<rgba 0 0 0 1>}
     highlightShape {
      bIsDynamicallyLoaded=false
      drawAs=Image
      imageType=NoImage
      mirroring=NoMirror
      resourceName=None
      resourceObject=null
      tiling=NoTile
      imageSize=<32 32>
      margin=<ltrb 0 0 0 0>
      outlineSettings {
       bUseBrushTransparency=false
       roundingType=HalfHeightRadius
       width=0
       color {colorUseRule=UseColor_Specified, specifiedColor=<rgba 0 0 0 0>}
       cornerRadii=<0 0 0 0>
      }
      tintColor {colorUseRule=UseColor_Specified, specifiedColor=<rgba 1 1 1 1>}
      uVRegion {bIsValid=false, max=<0 0>, min=<0 0>}
     }
     selectedBackgroundColor {colorUseRule=UseColor_Foreground, specifiedColor=<rgba 1 0 1 1>}
     shadowColorAndOpacity=<rgba 0 0 0 1>
     shadowOffset=<0 0>
     strikeBrush {
      bIsDynamicallyLoaded=false
      drawAs=Image
      imageType=NoImage
      mirroring=NoMirror
      resourceName=None
      resourceObject=null
      tiling=NoTile
      imageSize=<32 32>
      margin=<ltrb 0 0 0 0>
      outlineSettings {
       bUseBrushTransparency=false
       roundingType=HalfHeightRadius
       width=0
       color {colorUseRule=UseColor_Specified, specifiedColor=<rgba 0 0 0 0>}
       cornerRadii=<0 0 0 0>
      }
      tintColor {colorUseRule=UseColor_Specified, specifiedColor=<rgba 1 1 1 1>}
      uVRegion {bIsValid=false, max=<0 0>, min=<0 0>}
     }
     underlineBrush {
      bIsDynamicallyLoaded=false
      drawAs=Image
      imageType=NoImage
      mirroring=NoMirror
      resourceName=None
      resourceObject=null
      tiling=NoTile
      imageSize=<32 32>
      margin=<ltrb 0 0 0 0>
      outlineSettings {
       bUseBrushTransparency=false
       roundingType=HalfHeightRadius
       width=0
       color {colorUseRule=UseColor_Specified, specifiedColor=<rgba 0 0 0 0>}
       cornerRadii=<0 0 0 0>
      }
      tintColor {colorUseRule=UseColor_Specified, specifiedColor=<rgba 1 1 1 1>}
      uVRegion {bIsValid=false, max=<0 0>, min=<0 0>}
     }
    }
   }
   slot {class=HorizontalBoxSlot}
  }
  ExpandDirectionTick {
   widgetClassPath=&FortMobileImage
   bIsVariable=true
   parent=SingleChannelSwitcher
   widget=ExpandDirectionTick
   properties {
    mobileSizeMultiplier=1.2
    renderOpacity=0.3
    brush {
     bIsDynamicallyLoaded=false
     drawAs=Image
     imageType=NoImage
     mirroring=NoMirror
     resourceName=None
     resourceObject=/Game/UI/SocialPanel/Art/Textures/T_UI_Icon_TreeArrow_64
     tiling=NoTile
     imageSize=<28 28>
     margin=<ltrb 0 0 0 0>
     outlineSettings {
      bUseBrushTransparency=false
      roundingType=HalfHeightRadius
      width=0
      color {colorUseRule=UseColor_Specified, specifiedColor=<rgba 0 0 0 0>}
      cornerRadii=<0 0 0 0>
     }
     tintColor {colorUseRule=UseColor_Specified, specifiedColor=<rgba 1 1 1 1>}
     uVRegion {bIsValid=false, max=<0 0>, min=<0 0>}
    }
    renderTransform {angle=90, scale=<1 1>, shear=<0 0>, translation=<0 0>}
   }
   slot {class=HorizontalBoxSlot}
  }
  InputActionWidget {
   widgetClassPath=&CommonActionWidget
   bInherited=true
   bIsVariable=true
   parent=HBox_Join
   widget=InputActionWidget
   properties {inputActions [{dataTable=&InputActionDataTable, rowName=GenericForward}]}
   slot {class=HorizontalBoxSlot}
  }
  ChannelInfoPanel {
   widgetClassPath=&HorizontalBox
   parent=VerticalBox_0
   widget=ChannelInfoPanel
   slot {class=VerticalBoxSlot, properties {padding=<ltrb 0 2 0 0>}}
  }
  Image_VoiceChatWarning {
   widgetClassPath=&Image
   bIsVariable=true
   parent=ChannelInfoPanel
   widget=Image_VoiceChatWarning
   properties {
    visibility=Collapsed
    brush {
     bIsDynamicallyLoaded=false
     drawAs=Image
     imageType=NoImage
     mirroring=NoMirror
     resourceName=None
     resourceObject=/Game/UI/SocialPanel/Art/Textures/T_UI_Icon_MicLocked
     tiling=NoTile
     imageSize=<25 25>
     margin=<ltrb 0 0 0 0>
     outlineSettings {
      bUseBrushTransparency=false
      roundingType=HalfHeightRadius
      width=0
      color {colorUseRule=UseColor_Specified, specifiedColor=<rgba 0 0 0 0>}
      cornerRadii=<0 0 0 0>
     }
     tintColor {colorUseRule=UseColor_Specified, specifiedColor=<rgba 1 1 1 1>}
     uVRegion {bIsValid=false, max=<0 0>, min=<0 0>}
    }
    colorAndOpacity=<rgba 0.4793 0.521 0.552 1>
   }
   slot {class=HorizontalBoxSlot, properties {horizontalAlignment=HAlign_Left, verticalAlignment=VAlign_Center, padding=<ltrb 0 0 5 0>}}
  }
  Switcher_ChannelInfo {
   widgetClassPath=&CommonActivatableWidgetSwitcher
   bIsVariable=true
   parent=ChannelInfoPanel
   widget=Switcher_ChannelInfo
   properties {transitionCurveType=Linear, transitionDuration=0.1}
   slot {class=HorizontalBoxSlot, properties {size {sizeRule=Fill, value=1}}}
  }
  Text_ChannelInfo {
   widgetClassPath=&CommonTextBlock
   bInherited=true
   bIsVariable=true
   parent=Switcher_ChannelInfo
   widget=Text_ChannelInfo
   properties {
    autoWrapText=true
    text=You are in this channel
    textTransformPolicy=ToUpper
    colorAndOpacity {colorUseRule=UseColor_Specified, specifiedColor=<rgba 0.0319 0.8469 0.9911 1>}
    font {
     bForceMonospaced=false
     bMaterialIsStencil=false
     fontMaterial=null
     fontObject=&BurbankSmall
     letterSpacing=0
     monospacedWidth=1
     size=13
     skewAmount=0
     typefaceFontName=Black
     outlineSettings {
      bApplyOutlineToDropShadows=false
      bMiteredCorners=false
      bSeparateFillAlpha=false
      outlineMaterial=null
      outlineSize=0
      outlineColor=<rgba 0 0 0 1>
     }
    }
    renderTransform {angle=0, scale=<1 1>, shear=<-8 0>, translation=<0 0>}
    shadowOffset=<0 0>
   }
   slot {class=WidgetSwitcherSlot}
  }
  Text_ChannelInfoHover {
   widgetClassPath=&CommonTextBlock
   parent=Switcher_ChannelInfo
   widget=Text_ChannelInfoHover
   properties {
    autoWrapText=true
    text=You are in this channel
    textTransformPolicy=ToUpper
    colorAndOpacity {colorUseRule=UseColor_Specified, specifiedColor=<rgba 0.2039 0.2392 0.3569 1>}
    font {
     bForceMonospaced=false
     bMaterialIsStencil=false
     fontMaterial=null
     fontObject=&BurbankSmall
     letterSpacing=0
     monospacedWidth=1
     size=13
     skewAmount=0
     typefaceFontName=Black
     outlineSettings {
      bApplyOutlineToDropShadows=false
      bMiteredCorners=false
      bSeparateFillAlpha=false
      outlineMaterial=null
      outlineSize=0
      outlineColor=<rgba 0 0 0 1>
     }
    }
    renderTransform {angle=0, scale=<1 1>, shear=<-8 0>, translation=<0 0>}
    shadowOffset=<0 0>
   }
   slot {class=WidgetSwitcherSlot}
  }
  JoinLeaveButton {widgetClassPath=/Script/FortniteUI.FortMainMenuButton, bInherited=true, widget=null}
  ReportButton {widgetClassPath=&FortMainMenuButton, bInherited=true, widget=null}
  Text_ChannelInfo_Hover {widgetClassPath=&CommonTextBlock, bInherited=true, widget=null}
 ]
}
