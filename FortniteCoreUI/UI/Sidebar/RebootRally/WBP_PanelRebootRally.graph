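# Reboot Rally sidebar panel on FortSidebarPanelReboot, with a header, timer, back button, social user tree and onboarding tooltip.
# On activation it shows the Missing Friends row only when runtime option bEnableHiddenPlayersList is set; its info button opens the missing-friends modal.
# On init it picks canvas alignment from Discover-in-lobby state, targeting variable RebootRallyEligiblePlayersOnboardingTooltip (default null), not the RebootRallyTooltip widget.
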WidgetBlueprint /FortniteCoreUI/UI/Sidebar/RebootRally/WBP_PanelRebootRally
graphs [
 EventGraph {nodes [
   "On Clicked (MoreInfo_Ressources_Button_67)"!() -> (OutputDelegate, then, Button:Common Button Base Object Reference) {is=title K2Node_ComponentBoundEvent}
   > ShowMissingFriendsModal(self:Fort Sidebar Panel Reboot Object Reference)
   Event On Activated()
   n1 = GetRuntimeOptions()
   > SetVisibility(
    self=Get HB_MissingFriends() -> (HB_MissingFriends:Horizontal Box Object Reference)
    InVisibility=Select(
     Option 0=Collapsed
     Option 1=SelfHitTestInvisible
     Index=Get bEnableHiddenPlayersList!(self:Fort Runtime Options Object Reference=@n1) -> (bEnableHiddenPlayersList:bool)
    )
   )
   Event OnInitialized()
   > SetTooltipAlignment()
  ]}
 SetTooltipAlignment {nodes [
   SetTooltipAlignment() {is=title K2Node_FunctionEntry}
   n1 = GetIsDiscoverInLobbyEnabled!(self:Fort Runtime Options Object Reference=GetRuntimeOptions(), ContextWidget:Widget Object Reference=Self-Reference()) -> (ReturnValue:bool)
   > SetAlignment!(
    execute
    self:Canvas Panel Slot Object Reference=SlotAsCanvasSlot!(Widget:Widget Object Reference=Get RebootRallyEligiblePlayersOnboardingTooltip() -> (RebootRallyEligiblePlayersOnboardingTooltip:Reboot Rally Eligible Players Onboarding Tooltip Object Reference)) -> (ReturnValue:Canvas Panel Slot Object Reference)
    InAlignment:Vector 2D Structure=Select(Option 0:Vector 2D Structure=(X=0,Y=2.6), Option 1:Vector 2D Structure=(X=1.8,Y=2.6), Index=OR Boolean(A=@n1, B=false)) -> (ReturnValue:Vector 2D Structure)
   ) -> (then)
  ]}
]
variables [RebootRallyEligiblePlayersOnboardingTooltip {type=RebootRallyEligiblePlayersOnboardingTooltip_C, default=null}]
widgets {
 info {inheritedWidgetCount=5, parentClass=/Script/FortniteUI.FortSidebarPanelReboot, rootWidgetClass=&VerticalBox, widgetCount=13}
 widgets [
  VBox_Panel {widgetClassPath=&VerticalBox, widget=VBox_Panel}
  Overlay_0 {
   widgetClassPath=&Overlay
   parent=VBox_Panel
   widget=Overlay_0
   slot {class=VerticalBoxSlot, properties {verticalAlignment=VAlign_Top}}
  }
  Header {
   widgetClassPath=&/WBP_PanelRebootRally_Header
   bInherited=true
   bIsVariable=true
   parent=Overlay_0
   widget=Header
   slot {class=OverlaySlot, properties {horizontalAlignment=HAlign_Fill}}
  }
  SidebarBackButton {
   widgetClassPath=/FortUILibrary/Foundation/Buttons/BackButtonSmall
   bInherited=true
   parent=Overlay_0
   widget=SidebarBackButton
   properties {bIsFocusable=false}
   slot {class=OverlaySlot}
  }
  WBP_RebootRallyTimer {
   widgetClassPath=/Game/UI/Sidebar/RebootRally/WBP_RebootRallyTimer
   bInherited=true
   bIsVariable=true
   parent=Overlay_0
   widget=WBP_RebootRallyTimer
   slot {class=OverlaySlot, properties {horizontalAlignment=HAlign_Right, padding=<ltrb 20 20 20 20>}}
  }
  UserTree_RebootRally {
   widgetClassPath=/Script/FortniteUI.FortSocialUserTreeView
   bInherited=true
   bIsVariable=true
   parent=VBox_Panel
   widget=UserTree_RebootRally
   properties {
    entryWidgetClass=/FortniteCoreUI/UI/Sidebar/SidebarUserListHeaderEntry.SidebarUserListHeaderEntry_C
    inviteEntryWidgetClass=/FortniteCoreUI/UI/Sidebar/SidebarInviteListEntry_FriendInvite_XS.SidebarInviteListEntry_FriendInvite_XS_C
    joinRequestEntryWidgetClass=/FortniteCoreUI/UI/Sidebar/ListEntries/SidebarJoinRequestUserListEntry.SidebarJoinRequestUserListEntry_C
    joinablePartyEntryWidgetClass=/FortniteCoreUI/UI/Sidebar/SidebarJoinablePartyEntry.SidebarJoinablePartyEntry_C
    leavePartyEntryWidgetClass=/FortniteCoreUI/UI/Sidebar/SidebarLeavePartyListEntry.SidebarLeavePartyListEntry_C
    lfgFilterButtonEntryWidgetClass=/FortniteCoreUI/UI/Sidebar/Tags/SidebarLFGChangeTagsButtonListEntry.SidebarLFGChangeTagsButtonListEntry_C
    lfgFilterHeaderEntryWidgetClass=/FortniteCoreUI/UI/Sidebar/Tags/SidebarLfgFilterListHeaderEntry.SidebarLfgFilterListHeaderEntry_C
    lfgInviteEntryWidgetClass=/FortniteCoreUI/UI/Sidebar/ListEntries/SidebarUserListEntry_XS.SidebarUserListEntry_XS_C
    lfgState2EntryWidgetClass=/FortniteCoreUI/UI/Sidebar/Tags/SidebarLfgState2ListEntry.SidebarLfgState2ListEntry_C
    lfgTagCheckboxEntryWidgetClass=/FortniteCoreUI/UI/Sidebar/Tags/SidebarSocialTagFilterCheckboxListEntry.SidebarSocialTagFilterCheckboxListEntry_C
    lfgToggleEntryWidgetClass=/FortniteCoreUI/UI/Sidebar/Tags/SidebarLFGToggleListEntry.SidebarLFGToggleListEntry_C
    missedInviteEntryWidgetClass=/FortniteCoreUI/UI/SocialPanel/MissedInvitesEntry.MissedInvitesEntry_C
    navigation=&self:WidgetTree.UserTree_RebootRally.WidgetNavigation_0
    partyMemberEntryWidgetClass=/FortniteCoreUI/UI/Sidebar/ListEntries/SidebarPartyMemberListEntry.SidebarPartyMemberListEntry_C
    rebootPlayersHeaderEntryWidgetClass=&SidebarUserListHeaderEntry_C
    referFriendEntryWidgetClass=/FortniteCoreUI/UI/Sidebar/SidebarReferAFriendEntry.SidebarReferAFriendEntry_C
    shuffleLfgEntryWidgetClass=/FortniteCoreUI/UI/Sidebar/SidebarShuffleListEntry.SidebarShuffleListEntry_C
    suggestedFriendEntryWidgetClass=/FortniteCoreUI/UI/Sidebar/ListEntries/SidebarSuggestedFriendListEntry.SidebarSuggestedFriendListEntry_C
    teamMemberEntryWidgetClass=/FortniteCoreUI/UI/Sidebar/ListEntries/SidebarTeamMemberListEntry.SidebarTeamMemberListEntry_C
    userEntryWidgetClass=&SidebarUserListEntry_XS_C
    userSearchResultEntryWidgetClass=/FortniteCoreUI/UI/Sidebar/ListEntries/SidebarUserSearchResultListEntry_XS.SidebarUserSearchResultListEntry_XS_C
    voiceChannelHeaderEntryWidgetClass=/FortniteCoreUI/UI/Sidebar/SidebarVoiceChannelHeaderButton.SidebarVoiceChannelHeaderButton_C
    voiceChatMemberEntryWidgetClass=/FortniteCoreUI/UI/SocialPanel/ListEntries/VoiceChatMemberListEntry.VoiceChatMemberListEntry_C
   }
   slot {class=VerticalBoxSlot, properties {size {sizeRule=Fill, value=1}}}
  }
  HB_MissingFriends {
   widgetClassPath=&HorizontalBox
   bIsVariable=true
   parent=VBox_Panel
   widget=HB_MissingFriends
   slot {class=VerticalBoxSlot, properties {horizontalAlignment=HAlign_Center, padding=<ltrb 0 6 0 0>}}
  }
  Text_SeasonName {
   widgetClassPath=&CommonTextBlock
   bIsVariable=true
   parent=HB_MissingFriends
   widget=Text_SeasonName
   properties {
    bIsScrollingEnabled=false
    lineHeightPercentage=1.29
    mobileFontSizeMultiplier=1.1
    style=/FortUILibrary/Styles/Typography/_TextStyle_Body1._TextStyle_Body1_C
    text=Missing Friends
    textTransformPolicy=ToUpper
    wrapTextAt=350
    font {
     bForceMonospaced=false
     bMaterialIsStencil=false
     fontMaterial=null
     fontObject=&BurbankSmall
     letterSpacing=0
     monospacedWidth=1
     size=18
     skewAmount=0
     typefaceFontName=Bold
     outlineSettings {
      bApplyOutlineToDropShadows=false
      bMiteredCorners=false
      bSeparateFillAlpha=false
      outlineMaterial=null
      outlineSize=0
      outlineColor=<rgba 0 0 0 1>
     }
    }
    shadowOffset=<0 0>
    strikeBrush {
     bIsDynamicallyLoaded=false
     drawAs=NoDrawType
     imageType=NoImage
     mirroring=NoMirror
     resourceName=None
     resourceObject=null
     tiling=NoTile
     imageSize=<32 32>
     margin=<ltrb 0 0 0 0>
     outlineSettings {
      bUseBrushTransparency=false
      roundingType=HalfHeightRadius
      width=0
      color {colorUseRule=UseColor_Specified, specifiedColor=<rgba 0 0 0 0>}
      cornerRadii=<0 0 0 0>
     }
     tintColor {colorUseRule=UseColor_Specified, specifiedColor=<rgba 1 1 1 1>}
     uVRegion {bIsValid=false, max=<0 0>, min=<0 0>}
    }
   }
   slot {class=HorizontalBoxSlot}
  }
  MoreInfoOverlay {
   widgetClassPath=&Overlay
   bIsVariable=true
   parent=HB_MissingFriends
   widget=MoreInfoOverlay
   slot {class=HorizontalBoxSlot, properties {padding=<ltrb 0 -6 0 0>}}
  }
  MoreInfo_Ressources_Button_67 {
   widgetClassPath=/FortniteCoreUI/Athena/UI/Frontend/Lobby/MoreInfo_Ressources_Button
   bIsVariable=true
   parent=MoreInfoOverlay
   widget=MoreInfo_Ressources_Button_67
   properties {triggeringInputAction {dataTable=&InputActionDataTable, rowName=GenericRightStickClick}}
   slot {class=OverlaySlot}
  }
  Spacer_53 {widgetClassPath=&Spacer, parent=HB_MissingFriends, widget=Spacer_53, slot {class=HorizontalBoxSlot}}
  CanvasPanel_146 {
   widgetClassPath=&CanvasPanel
   parent=VBox_Panel
   widget=CanvasPanel_146
   slot {class=VerticalBoxSlot, properties {size {sizeRule=Fill, value=0}}}
  }
  RebootRallyTooltip {
   widgetClassPath=&/RebootRallyEligiblePlayersOnboardingTooltip
   bInherited=true
   bIsVariable=true
   parent=CanvasPanel_146
   widget=RebootRallyTooltip
   properties {flipHorizontal=true}
   slot {
    class=CanvasPanelSlot
    properties {bAutoSize=true, layoutData {alignment=<0 2.6>, anchors {maximum=<1 0>, minimum=<1 0>}, offsets=<ltrb -1200 -188 0 0>}}
   }
  }
 ]
}
