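# Translucent UI material for the sidebar presence indicator, drawing a procedural round dot with a colored fill and border.
# An inner exponential radial gradient, shrunk by the border thickness parameter, blends the border color into the fill color.
# Its density scales with the reduced radius.
# A second, full-radius gradient forms the outer circle mask, so the shape needs no texture.
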Material /Game/UI/Sidebar/Materials/M_PresenceIndicator
graph {
 properties {
  bUsedWithStaticMesh=false
  blendMode=BLEND_Translucent
  materialDomain=MD_UI
  pixelDepthOffsetMode=PDOM_Legacy
  usedShadingModels=Default Lit
 }
 nodes [
  n1 = TextureCoordinate() {coordinateIndex=4}
  n2 = Constant() {r=0.5}
  n3 = Subtract(A=@n2, B=Divide(A=ScalarParameter() {defaultValue=1, parameterName=Float_BorderThickness}) {constB=10})
  n4 = Constant() {r=15}
  n5 = MaterialFunctionCall(UVs=@n1, Radius=@n3, Density=Multiply(A=Divide(A=@n3, B=@n2), B=@n4)) {materialFunction=&RadialGradientExponential}
  Add(
   A=Multiply(
    A=VectorParameter() {parameterName=RGBA_Border, defaultValue=<rgba 0.0037 0.0168 0.0908 1>}.RGB
    B=OneMinus(None=@n5.RadialGradientExponential)
   )
   B=Multiply(A=@n5.RadialGradientExponential, B=VectorParameter() {parameterName=RGBA_Fill, defaultValue=<rgba 0.2159 1 0.1329 1>}.RGB)
  )
  Reroute(None=MaterialFunctionCall(UVs=@n1, Radius=@n2, Density=@n4) {materialFunction=&RadialGradientExponential}.RadialGradientExponential)
 ]
}
