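# Social sidebar button for the local user, showing their avatar inside a bang wrapper with an online-status indicator.
# Hover (skipped for touch) and selection events play animations and recolor the skewed background and indicator outline through dynamic material parameters.
# Online-status changes set the indicator's fill color.
# The new-avatars toggle shows either PlayerAvatar or the older social avatar icon with its status indicator.
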WidgetBlueprint /FortniteCoreUI/UI/Sidebar/SidebarLocalUserButton
animations [
 Selected {
  binding_count=0
  display_label=""
  length_seconds=0.15
  object=&self:Selected
  sequence {bindings []}
 }
 Hovered {
  binding_count=0
  display_label=""
  length_seconds=0.15
  object=&self:Hovered
  sequence {bindings []}
 }
 Unselected {
  binding_count=0
  display_label=""
  length_seconds=0.15
  object=&self:Unselected
  sequence {bindings []}
 }
]
graphs [
 EventGraph {nodes [
   Event On Hovered()
   n1 = NOT Boolean(A=IsUsingTouch(PlayerController=GetOwningPlayer()))
   > Branch(Condition=@n1)
   n2 = Get Hovered() -> (Hovered:Widget Animation Object Reference)
   > PlayAnimation(InAnimation=@n2)
   n3 = Get Color_SkewBox_Hover() -> (Color_SkewBox_Hover:Linear Color Structure)
   > SetSkewBoxColor!(execute, self, InputPin:Linear Color Structure=@n3) -> (then)
   > SetPresenceIndicatorOutlineColor!(execute, self, Value:Linear Color Structure=@n3) -> (then)
   Event On Unhovered()
   > Branch(Condition=@n1)
   > PlayAnimation(InAnimation=@n2, PlayMode=Reverse)
   n4 = GetSelected!(self:Common Button Base Object Reference) -> (ReturnValue:bool)
   > SetSkewBoxColor(InputPin=SelectColor(A=Get Color_SkewBox_Selected() -> (Color_SkewBox_Selected:Linear Color Structure), B=(R=0,G=0,B=0,A=0), bPickA=@n4))
   > SetPresenceIndicatorOutlineColor(Value=SelectColor(
     A=Get Color_SkewBox_Selected()
     B=Get Color_PresenceIndicatorOutline() -> (Color_PresenceIndicatorOutline:Linear Color Structure)
     bPickA=@n4
    ))
   Event On Selected()
   > PlayAnimation(InAnimation=Get Selected() -> (Selected:Widget Animation Object Reference))
   n5 = SelectColor(A=Get Color_SkewBox_Hover(), B=Get Color_SkewBox_Selected(), bPickA=IsHovered())
   > SetSkewBoxColor(InputPin=@n5)
   > SetPresenceIndicatorOutlineColor(Value=@n5)
   Event On Deselected()
   > PlayAnimation(InAnimation=Get Unselected() -> (Unselected:Widget Animation Object Reference))
   > SetSkewBoxColor(InputPin=(R=0,G=0,B=0,A=0))
   > SetPresenceIndicatorOutlineColor(Value=Get Color_PresenceIndicatorOutline())
   n6 = Event OnOnlineStatusChanged!() -> (OutputDelegate, then, OnlineStatus:ELocalUserOnlineStatus Enum=Online) {is=title K2Node_Event}
   > n7 = GetDynamicMaterial(self:Border Object Reference=Get Border_OnlineStatusIndicator() -> (Border_OnlineStatusIndicator:Border Object Reference))
   n8 = Get Online() -> (Online:Linear Color Structure)
   n9 = Get Away() -> (Away:Linear Color Structure)
   > SetVectorParameterValue(
    self=@n7
    ParameterName=RGBA_Fill
    Value=Select!(
     Online:Linear Color Structure=@n8
     Offline:Linear Color Structure=Get Offline() -> (Offline:Linear Color Structure)
     Away:Linear Color Structure=@n9
     ExtendedAway:Linear Color Structure=@n9
     DoNotDisturb:Linear Color Structure=@n9
     Chat:Linear Color Structure=@n8
     Index:ELocalUserOnlineStatus Enum=@n6.OnlineStatus
    ) -> (ReturnValue:Linear Color Structure)
   )
   Event PreConstruct()
   > SetSkewBoxColor(InputPin=SelectColor(A=Get Color_SkewBox_Selected(), B=(R=0,G=0,B=0,A=0), bPickA=GetSelected()))
   n10 = Event OnNewAvatarsEnabledChanged!() -> (OutputDelegate, then, bUsingNewAvatars:bool) {is=title K2Node_Event}
   > SetVisibility(
    self=Get PlayerAvatar() -> (PlayerAvatar:Fort Player Avatar Object Reference)
    InVisibility=Select!(Option 0:ESlateVisibility Enum=Collapsed, Option 1:ESlateVisibility Enum=HitTestInvisible, Index:bool=@n10.bUsingNewAvatars) -> (ReturnValue:ESlateVisibility Enum)
   )
   n11 = Select(Option 0=HitTestInvisible, Option 1=Collapsed, Index=@n10.bUsingNewAvatars)
   > SetVisibility(self=Get SocialAvatarIcon() -> (SocialAvatarIcon:Fort Social Avatar Icon Object Reference), InVisibility=@n11)
   > SetVisibility(self=Get Border_OnlineStatusIndicator(), InVisibility=@n11)
   AND Boolean(A=NOT Boolean(A=GetSelected()), B=@n1)
  ]}
 SetPresenceIndicatorOutlineColor {nodes [
   n1 = SetPresenceIndicatorOutlineColor!() -> (then, Value:Linear Color Structure) {is=title K2Node_FunctionEntry}
   > n2 = GetDynamicMaterial(self=Get Border_OnlineStatusIndicator())
   > SetVectorParameterValue(self=@n2, ParameterName=RGBA_Border, Value=@n1)
  ]}
 SetSkewBoxColor {nodes [
   n1 = SetSkewBoxColor!() -> (then, InputPin:Linear Color Structure) {is=title K2Node_FunctionEntry}
   > n2 = GetDynamicMaterial(self=Get Border_SkewBox() -> (Border_SkewBox:Border Object Reference))
   > SetVectorParameterValue(self=@n2, ParameterName=Color1, Value=@n1)
   > SetVectorParameterValue(self=@n2, ParameterName=Color2, Value=@n1)
  ]}
]
variables [
 Away {type=LinearColor, default=<rgba 0.9647 0.5457 0.1221 1>}
 Color_PresenceIndicatorOutline {type=LinearColor, default=<rgba 0 0 0 0>}
 Offline {type=LinearColor, default=<rgba 0.1413 0.3278 0.5149 1>}
 Online {type=LinearColor, default=<rgba 0.2159 1 0.1329 1>}
]
widgets {
 info {
  inheritedWidgetCount=6
  namedSlotCount=1
  parentClass=/Script/FortniteUI.FortSocialSidebarLocalUserButton
  rootWidgetClass=&Border
  widgetCount=12
 }
 widgets [
  Border_SkewBox {
   widgetClassPath=&Border
   bInherited=true
   widget=Border_SkewBox
   properties {
    background {
     bIsDynamicallyLoaded=false
     drawAs=Image
     imageType=NoImage
     mirroring=NoMirror
     resourceName=None
     resourceObject=/Game/UI/Sidebar/Materials/M_UI_SkewBox_SidebarButtonBackground
     tiling=NoTile
     imageSize=<32 32>
     margin=<ltrb 0 0 0 0>
     outlineSettings {
      bUseBrushTransparency=false
      roundingType=HalfHeightRadius
      width=0
      color {colorUseRule=UseColor_Specified, specifiedColor=<rgba 0 0 0 0>}
      cornerRadii=<0 0 0 0>
     }
     tintColor {colorUseRule=UseColor_Specified, specifiedColor=<rgba 1 1 1 1>}
     uVRegion {bIsValid=false, max=<0 0>, min=<0 0>}
    }
    padding=<ltrb 0 0 0 0>
   }
  }
  Overlay_0 {widgetClassPath=&Overlay, parent=Border_SkewBox, widget=Overlay_0, slot {class=BorderSlot, properties {padding=<ltrb 0 0 0 0>}}}
  Image_TabIcon {
   widgetClassPath=/Script/FortniteUI.FortMobileImage
   bIsVariable=true
   parent=Overlay_0
   widget=Image_TabIcon
   properties {
    mobileSizeMultiplier=1.1
    renderOpacity=0.5
    visibility=Collapsed
    brush {
     bIsDynamicallyLoaded=false
     drawAs=NoDrawType
     imageType=NoImage
     mirroring=NoMirror
     resourceName=None
     resourceObject=/Game/UI/SocialPanel/Art/Textures/T_UI_Icon_AddFriends_64
     tiling=NoTile
     imageSize=<48 48>
     margin=<ltrb 0 0 0 0>
     outlineSettings {
      bUseBrushTransparency=false
      roundingType=HalfHeightRadius
      width=0
      color {colorUseRule=UseColor_Specified, specifiedColor=<rgba 0 0 0 0>}
      cornerRadii=<0 0 0 0>
     }
     tintColor {colorUseRule=UseColor_Specified, specifiedColor=<rgba 1 1 1 1>}
     uVRegion {bIsValid=false, max=<0 0>, min=<0 0>}
    }
   }
   slot {class=OverlaySlot, properties {horizontalAlignment=HAlign_Center, verticalAlignment=VAlign_Center, padding=<ltrb 10 10 10 10>}}
  }
  SizeBox_43 {
   widgetClassPath=&SizeBox
   parent=Overlay_0
   widget=SizeBox_43
   properties {bOverride_HeightOverride=true, bOverride_WidthOverride=true, heightOverride=64, widthOverride=64}
   slot {class=OverlaySlot, properties {horizontalAlignment=HAlign_Center, verticalAlignment=VAlign_Center}}
  }
  BangWrapper {
   widgetClassPath=/Game/UI/Foundation/Bangs/NormalBangWrapper
   bInherited=true
   bIsVariable=true
   parent=SizeBox_43
   widget=BangWrapper
   properties {useSmallVersion=true}
   slot {class=SizeBoxSlot}
  }
  Overlay_2 {widgetClassPath=&Overlay, namedSlotHost=BangWrapper, widget=Overlay_2}
  SocialAvatarIcon {
   widgetClassPath=/Script/FortniteUI.FortSocialAvatarIcon
   bInherited=true
   bIsVariable=true
   parent=Overlay_2
   widget=SocialAvatarIcon
   properties {
    avatarMaterial=/Game/UI/SocialPanel/Art/Materials/UI_UserPortrait_Buffet
    materialTextureParamName=SmallPreview
    visibility=Collapsed
    brush {
     bIsDynamicallyLoaded=false
     drawAs=Image
     imageType=NoImage
     mirroring=NoMirror
     resourceName=None
     resourceObject=&UI_UserPortrait_Buffet
     tiling=NoTile
     imageSize=<60 60>
     margin=<ltrb 0 0 0 0>
     outlineSettings {
      bUseBrushTransparency=false
      roundingType=HalfHeightRadius
      width=0
      color {colorUseRule=UseColor_Specified, specifiedColor=<rgba 0 0 0 0>}
      cornerRadii=<0 0 0 0>
     }
     tintColor {colorUseRule=UseColor_Specified, specifiedColor=<rgba 1 1 1 1>}
     uVRegion {bIsValid=false, max=<0 0>, min=<0 0>}
    }
    loadingBackgroundBrush {
     bIsDynamicallyLoaded=false
     drawAs=Image
     imageType=NoImage
     mirroring=NoMirror
     resourceName=None
     resourceObject=/FortUILibrary/UIKit/Throbber/MI_UIKit_Monochrome_Throbber
     tiling=NoTile
     imageSize=<32 32>
     margin=<ltrb 0 0 0 0>
     outlineSettings {
      bUseBrushTransparency=false
      roundingType=HalfHeightRadius
      width=0
      color {colorUseRule=UseColor_Specified, specifiedColor=<rgba 0 0 0 0>}
      cornerRadii=<0 0 0 0>
     }
     tintColor {colorUseRule=UseColor_Specified, specifiedColor=<rgba 1 1 1 1>}
     uVRegion {bIsValid=false, max=<0 0>, min=<0 0>}
    }
   }
   slot {class=OverlaySlot, properties {horizontalAlignment=HAlign_Fill, verticalAlignment=VAlign_Fill}}
  }
  Border_OnlineStatusIndicator {
   widgetClassPath=&Border
   bIsVariable=true
   parent=Overlay_2
   widget=Border_OnlineStatusIndicator
   properties {
    visibility=Collapsed
    background {
     bIsDynamicallyLoaded=false
     drawAs=Image
     imageType=NoImage
     mirroring=NoMirror
     resourceName=None
     resourceObject=/Game/UI/Sidebar/Materials/M_PresenceIndicator
     tiling=NoTile
     imageSize=<16 16>
     margin=<ltrb 0 0 0 0>
     outlineSettings {
      bUseBrushTransparency=false
      roundingType=HalfHeightRadius
      width=0
      color {colorUseRule=UseColor_Specified, specifiedColor=<rgba 0 0 0 0>}
      cornerRadii=<0 0 0 0>
     }
     tintColor {colorUseRule=UseColor_Specified, specifiedColor=<rgba 1 1 1 1>}
     uVRegion {bIsValid=false, max=<0 0>, min=<0 0>}
    }
    padding=<ltrb 0 0 3 3>
    renderTransform {angle=0, scale=<1.1 1.1>, shear=<0 0>, translation=<1 1>}
   }
   slot {class=OverlaySlot, properties {horizontalAlignment=HAlign_Right, verticalAlignment=VAlign_Bottom}}
  }
  SizeBox_100 {
   widgetClassPath=&SizeBox
   parent=Border_OnlineStatusIndicator
   widget=SizeBox_100
   properties {bOverride_HeightOverride=true, bOverride_WidthOverride=true, heightOverride=16, widthOverride=16}
   slot {class=BorderSlot, properties {padding=<ltrb 0 0 3 3>}}
  }
  PlayerAvatar {
   widgetClassPath=/Game/UI/Foundation/Profile/WBP_PlayerAvatar
   bInherited=true
   bIsVariable=true
   parent=Overlay_2
   widget=PlayerAvatar
   slot {class=OverlaySlot}
  }
  Image_Crown {widgetClassPath=&Image, bInherited=true, widget=null}
  InputActionWidget {widgetClassPath=&CommonActionWidget, bInherited=true, widget=null}
 ]
}
