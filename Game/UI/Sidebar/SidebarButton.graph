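# Tab button for the social sidebar, based on FortSocialSidebarButton, with an icon, bang wrapper and optional number in a skewed box.
# Hover, select and deselect events recolor icon, number and box (via dynamic material parameters) and play animations that have no bindings; hover is skipped for touch.
# OnInitialized sizes it for mobile or other platforms.
# The press sound needs hover plus a flag enabled shortly after Construct and cleared by non-hovered selection.
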WidgetBlueprint /Game/UI/Sidebar/SidebarButton
animations [
 Selected {binding_count=0, display_label="", length_seconds=0.15, object=&self:Selected, sequence {bindings []}}
 Hovered {binding_count=0, display_label="", length_seconds=0.15, object=&self:Hovered, sequence {bindings []}}
 Unselected {binding_count=0, display_label="", length_seconds=0.2, object=&self:Unselected, sequence {bindings []}}
]
graphs [
 EventGraph {nodes [
   n1 = EnablePlaySound() {is=title K2Node_CustomEvent}
   > Set bPlaySound(bPlaySound:bool=true) -> (Output_Get:bool)
   Event On Hovered()
   n2 = NOT Boolean(A=IsUsingTouch(PlayerController=GetOwningPlayer()))
   > Branch(Condition=@n2)
   n3 = Get Hovered() -> (Hovered:Widget Animation Object Reference)
   > PlayAnimation(InAnimation=@n3)
   > SetSkewBoxColor!(execute, self, InputPin:Linear Color Structure=Get Color_SkewBox_Hover() -> (Color_SkewBox_Hover:Linear Color Structure)) -> (then)
   n4 = Get Color_ContentHovered() -> (Color_ContentHovered:Linear Color Structure)
   > SetColorAndOpacity!(
    execute
    self=Get Text_Number() -> (Text_Number:Common Text Object Reference)
    InColorAndOpacity_SpecifiedColor:Linear Color Structure=@n4
    InColorAndOpacity_ColorUseRule:ESlateColorStylingMode Enum
   ) -> (then)
   > SetColorAndOpacity!(
    execute
    self:Image Object Reference=Get Image_TabIcon() -> (Image_TabIcon:Image Object Reference)
    InColorAndOpacity:Linear Color Structure=@n4
   ) -> (then)
   Event On Unhovered()
   > Branch(Condition=@n2)
   > PlayAnimation(InAnimation=@n3, PlayMode=Reverse)
   n5 = GetSelected!(self:Common Button Base Object Reference) -> (ReturnValue:bool)
   > SetSkewBoxColor(InputPin=SelectColor(A=Get Color_SkewBox_Selected() -> (Color_SkewBox_Selected:Linear Color Structure), B=(R=0,G=0,B=0,A=0), bPickA=@n5))
   n6 = SelectColor(
    A=Get Color_ContentSelected() -> (Color_ContentSelected:Linear Color Structure)
    B=Get Color_ContentUnselected() -> (Color_ContentUnselected:Linear Color Structure)
    bPickA=@n5
   )
   > SetColorAndOpacity!(
    execute
    self:Text Object Reference=Get Text_Number()
    InColorAndOpacity_SpecifiedColor:Linear Color Structure=@n6
    InColorAndOpacity_ColorUseRule:ESlateColorStylingMode Enum
   ) -> (then)
   > SetColorAndOpacity!(execute, self:Image Object Reference=Get Image_TabIcon(), InColorAndOpacity:Linear Color Structure=@n6) -> (then)
   Event On Selected()
   > PlayAnimation(InAnimation=Get Selected() -> (Selected:Widget Animation Object Reference))
   n7 = IsHovered()
   > SetSkewBoxColor(InputPin=SelectColor(A=Get Color_SkewBox_Hover(), B=Get Color_SkewBox_Selected(), bPickA=@n7))
   n8 = SelectColor(A=Get Color_ContentHovered(), B=Get Color_ContentSelected(), bPickA=@n7)
   > SetColorAndOpacity!(
    execute
    self:Text Object Reference=Get Text_Number()
    InColorAndOpacity_SpecifiedColor:Linear Color Structure=@n8
    InColorAndOpacity_ColorUseRule:ESlateColorStylingMode Enum
   ) -> (then)
   > SetColorAndOpacity!(execute, self:Image Object Reference=Get Image_TabIcon(), InColorAndOpacity:Linear Color Structure=@n8) -> (then)
   > n9 = Branch(Condition=IsHovered())
   Branch(execute=[@n10.then, @n9.then], Condition=Get bPlaySound() -> (bPlaySound:bool))
   > PlaySound2D(Sound=Get PressedSound() -> (PressedSound:Sound Base Object Reference))
   Set bPlaySound(execute=@n9.else, bPlaySound=false)
   > n10 = Set Timer by Event(Delegate=Create Event(), Time=0.1, bLooping=false)
   Event On Deselected()
   > PlayAnimation(InAnimation=Get Unselected() -> (Unselected:Widget Animation Object Reference))
   > SetSkewBoxColor(InputPin=(R=0,G=0,B=0,A=0))
   n11 = Get Color_ContentUnselected()
   > SetColorAndOpacity!(
    execute
    self:Text Object Reference=Get Text_Number()
    InColorAndOpacity_SpecifiedColor:Linear Color Structure=@n11
    InColorAndOpacity_ColorUseRule:ESlateColorStylingMode Enum
   ) -> (then)
   > SetColorAndOpacity!(execute, self:Image Object Reference=Get Image_TabIcon(), InColorAndOpacity:Linear Color Structure=@n11) -> (then)
   Event OnInitialized()
   n12 = IsMobileGame()
   n13 = Get SizeBox_0() -> (SizeBox_0:Size Box Object Reference)
   > SetWidthOverride(self=@n13, InWidthOverride=Select(Option 0:double=104, Option 1:double=128, Index=@n12) -> (ReturnValue:double))
   > SetHeightOverride(self=@n13, InHeightOverride=Select(Option 0=100, Option 1=135, Index=@n12))
   Event PreConstruct()
   n14 = GetSelected()
   > SetSkewBoxColor(InputPin=SelectColor(A=Get Color_SkewBox_Selected(), B=(R=0,G=0,B=0,A=0), bPickA=@n14))
   n15 = SelectColor(A=Get Color_ContentSelected(), B=Get Color_ContentUnselected(), bPickA=@n14)
   > SetColorAndOpacity!(
    execute
    self:Text Object Reference=Get Text_Number()
    InColorAndOpacity_SpecifiedColor:Linear Color Structure=@n15
    InColorAndOpacity_ColorUseRule:ESlateColorStylingMode Enum
   ) -> (then)
   > SetColorAndOpacity!(execute, self:Image Object Reference=Get Image_TabIcon(), InColorAndOpacity:Linear Color Structure=@n15) -> (then)
   Event Construct()
   > Set Timer by Event(Delegate=@n1, Time=0.1, bLooping=false)
  ]}
 SetSkewBoxColor {nodes [
   n1 = SetSkewBoxColor!() -> (then, InputPin:Linear Color Structure) {is=title K2Node_FunctionEntry}
   > n2 = GetDynamicMaterial(self:Border Object Reference=Get Border_SkewBox() -> (Border_SkewBox:Border Object Reference))
   > SetVectorParameterValue(self=@n2, ParameterName=Color1, Value=@n1)
   > SetVectorParameterValue(self=@n2, ParameterName=Color2, Value=@n1)
  ]}
]
variables [
 bPlaySound {type=boolean}
 Condition {type=boolean, default=true}
 PressedSound {type=SoundBase, default=null}
]
widgets {
 info {
  inheritedWidgetCount=8
  namedSlotCount=1
  parentClass=/Script/FortniteUI.FortSocialSidebarButton
  rootWidgetClass=&SizeBox
  widgetCount=14
 }
 widgets [
  SizeBox_0 {
   widgetClassPath=&SizeBox
   bIsVariable=true
   widget=SizeBox_0
   properties {bOverride_HeightOverride=true, bOverride_WidthOverride=true, heightOverride=100, widthOverride=104}
  }
  Border_SkewBox {
   widgetClassPath=&Border
   bInherited=true
   bIsVariable=true
   parent=SizeBox_0
   widget=Border_SkewBox
   properties {
    background {
     bIsDynamicallyLoaded=false
     drawAs=Image
     imageType=NoImage
     mirroring=NoMirror
     resourceName=None
     resourceObject=/Game/UI/Sidebar/Materials/M_UI_SkewBox_SidebarButtonBackground
     tiling=NoTile
     imageSize=<32 32>
     margin=<ltrb 0 0 0 0>
     outlineSettings {
      bUseBrushTransparency=false
      roundingType=HalfHeightRadius
      width=0
      color {colorUseRule=UseColor_Specified, specifiedColor=<rgba 0 0 0 0>}
      cornerRadii=<0 0 0 0>
     }
     tintColor {colorUseRule=UseColor_Specified, specifiedColor=<rgba 1 1 1 1>}
     uVRegion {bIsValid=false, max=<0 0>, min=<0 0>}
    }
    padding=<ltrb 6 6 6 6>
   }
   slot {class=SizeBoxSlot}
  }
  ContentContainer {
   widgetClassPath=&Overlay
   parent=Border_SkewBox
   widget=ContentContainer
   slot {class=BorderSlot, properties {padding=<ltrb 6 6 6 6>}}
  }
  HorizontalBox_5 {
   widgetClassPath=&HorizontalBox
   parent=ContentContainer
   widget=HorizontalBox_5
   slot {class=OverlaySlot, properties {horizontalAlignment=HAlign_Center, verticalAlignment=VAlign_Center}}
  }
  BangWrapper {
   widgetClassPath=/Game/UI/Foundation/Bangs/NormalBangWrapper
   bInherited=true
   bIsVariable=true
   parent=HorizontalBox_5
   widget=BangWrapper
   properties {useSmallVersion=true}
   slot {class=HorizontalBoxSlot}
  }
  Border_2 {
   widgetClassPath=&Border
   namedSlotHost=BangWrapper
   widget=Border_2
   properties {brushColor=<rgba 0 0 0 0>, padding=<ltrb 10 10 10 10>}
  }
  Overlay_2 {widgetClassPath=&Overlay, parent=Border_2, widget=Overlay_2, slot {class=BorderSlot, properties {padding=<ltrb 10 10 10 10>}}}
  Image_TabIcon {
   widgetClassPath=/Script/FortniteUI.FortMobileImage
   bInherited=true
   bIsVariable=true
   parent=Overlay_2
   widget=Image_TabIcon
   properties {
    mobileSizeMultiplier=1.4
    brush {
     bIsDynamicallyLoaded=false
     drawAs=Image
     imageType=NoImage
     mirroring=NoMirror
     resourceName=None
     resourceObject=/Game/UI/Sidebar/Textures/T_UI_Icon_Sidebar_AddFriend
     tiling=NoTile
     imageSize=<36 36>
     margin=<ltrb 0 0 0 0>
     outlineSettings {
      bUseBrushTransparency=false
      roundingType=HalfHeightRadius
      width=0
      color {colorUseRule=UseColor_Specified, specifiedColor=<rgba 0 0 0 0>}
      cornerRadii=<0 0 0 0>
     }
     tintColor {colorUseRule=UseColor_Specified, specifiedColor=<rgba 1 1 1 1>}
     uVRegion {bIsValid=false, max=<0 0>, min=<0 0>}
    }
    colorAndOpacity=<rgba 0.0331 0.402 1 0>
   }
   slot {class=OverlaySlot, properties {horizontalAlignment=HAlign_Fill, verticalAlignment=VAlign_Fill}}
  }
  Image_TabIcon_Selected {
   widgetClassPath=&FortMobileImage
   bInherited=true
   bIsVariable=true
   parent=Overlay_2
   widget=Image_TabIcon_Selected
   properties {
    mobileSizeMultiplier=1.4
    brush {
     bIsDynamicallyLoaded=false
     drawAs=Image
     imageType=NoImage
     mirroring=NoMirror
     resourceName=None
     resourceObject=/Game/UI/Sidebar/Textures/T_UI_Icon_Sidebar_PartyUp
     tiling=NoTile
     imageSize=<64 64>
     margin=<ltrb 0 0 0 0>
     outlineSettings {
      bUseBrushTransparency=false
      roundingType=HalfHeightRadius
      width=0
      color {colorUseRule=UseColor_Specified, specifiedColor=<rgba 0 0 0 0>}
      cornerRadii=<0 0 0 0>
     }
     tintColor {colorUseRule=UseColor_Specified, specifiedColor=<rgba 1 1 1 1>}
     uVRegion {bIsValid=false, max=<0 0>, min=<0 0>}
    }
    colorAndOpacity=<rgba 1 1 1 0>
   }
   slot {class=OverlaySlot, properties {horizontalAlignment=HAlign_Fill, verticalAlignment=VAlign_Fill}}
  }
  ScaleBox_Number {
   widgetClassPath=&ScaleBox
   bInherited=true
   bIsVariable=true
   parent=HorizontalBox_5
   widget=ScaleBox_Number
   properties {stretchDirection=DownOnly, visibility=Collapsed}
   slot {class=HorizontalBoxSlot, properties {padding=<ltrb 0 0 10 0>}}
  }
  Overlay_3 {
   widgetClassPath=&Overlay
   parent=ScaleBox_Number
   widget=Overlay_3
   slot {class=ScaleBoxSlot, properties {horizontalAlignment=HAlign_Fill}}
  }
  Text_Number {
   widgetClassPath=&CommonTextBlock
   bInherited=true
   bIsVariable=true
   parent=Overlay_3
   widget=Text_Number
   properties {
    justification=Right
    mobileFontSizeMultiplier=1.35
    text=99
    visibility=Collapsed
    colorAndOpacity {colorUseRule=UseColor_Specified, specifiedColor=<rgba 0.0331 0.402 1 0>}
    font {
     bForceMonospaced=false
     bMaterialIsStencil=false
     fontMaterial=null
     fontObject=&BurbankSmall
     letterSpacing=0
     monospacedWidth=1
     size=16
     skewAmount=0
     typefaceFontName=Black
     outlineSettings {
      bApplyOutlineToDropShadows=false
      bMiteredCorners=false
      bSeparateFillAlpha=false
      outlineMaterial=null
      outlineSize=0
      outlineColor=<rgba 0 0 0 1>
     }
    }
    shadowOffset=<0 0>
   }
   slot {class=OverlaySlot, properties {horizontalAlignment=HAlign_Fill, verticalAlignment=VAlign_Center}}
  }
  Text_Number_Selected {
   widgetClassPath=&CommonTextBlock
   bInherited=true
   bIsVariable=true
   parent=Overlay_3
   widget=Text_Number_Selected
   properties {
    justification=Right
    mobileFontSizeMultiplier=1.35
    text=99
    visibility=Collapsed
    font {
     bForceMonospaced=false
     bMaterialIsStencil=false
     fontMaterial=null
     fontObject=&BurbankSmall
     letterSpacing=0
     monospacedWidth=1
     size=16
     skewAmount=0
     typefaceFontName=Black
     outlineSettings {
      bApplyOutlineToDropShadows=false
      bMiteredCorners=false
      bSeparateFillAlpha=false
      outlineMaterial=null
      outlineSize=0
      outlineColor=<rgba 0 0 0 1>
     }
    }
    shadowOffset=<0 0>
   }
   slot {class=OverlaySlot, properties {horizontalAlignment=HAlign_Fill, verticalAlignment=VAlign_Center}}
  }
  InputActionWidget {widgetClassPath=&CommonActionWidget, bInherited=true, widget=null}
 ]
}
