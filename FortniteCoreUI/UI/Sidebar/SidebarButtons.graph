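# Sidebar button column widget on FortSidebarButtons, stacking profile, social, voice chat, Settings, More and Exit buttons between tab-left/right input prompts.
# PreConstruct binds input-method changes and immediately runs the handler.
# The handler ignores input type, using ShouldSeparateSettingsAndMoreMenus to show or collapse Settings and swap More's icon between menu and settings-menu.
# On initialization, runtime option bEnableTopSafeZone drives both top and bottom safe-zone padding; sides stay unpadded.
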WidgetBlueprint /FortniteCoreUI/UI/Sidebar/SidebarButtons
graphs [
 EventGraph {nodes [
   Event PreConstruct()
   n1 = Get CommonInputSubsystem!() -> (ReturnValue:Common Input Object Reference) {is=title K2Node_GetBlueprintContextEx}
   > Bind Event to On Input Method Changed(self=@n1, Delegate=Create Event())
   > Handle Input Method Changed!(execute, self, bNewInputType:ECommonInputType Enum=GetCurrentInputType(self=@n1)) -> (then)
   Event OnInitialized()
   n2 = Get bEnableTopSafeZone!(self:Fort Runtime Options Object Reference=GetRuntimeOptions()) -> (bEnableTopSafeZone:bool)
   > SetSidesToPad!(
    execute
    self:Safe Zone Object Reference=Get SafeZone() -> (SafeZone:Safe Zone Object Reference)
    InPadLeft:bool=false
    InPadRight:bool=false
    InPadTop:bool=@n2
    InPadBottom:bool=@n2
   ) -> (then)
  ]}
 Handle Input Method Changed {nodes [
   Handle Input Method Changed!() -> (then, bNewInputType:ECommonInputType Enum) {is=title K2Node_FunctionEntry}
   n1 = ShouldSeparateSettingsAndMoreMenus!(self:Fort Settings Context Object Reference=Get FortSettingsContext!() -> (ReturnValue:Fort Settings Context Object Reference) {is=title K2Node_GetBlueprintContextEx}) -> (ReturnValue:bool)
   > SetVisibility(
    self=Get Button_Settings() -> (Button_Settings:Fort Social Sidebar Button Object Reference)
    InVisibility=Select(Option 0=Collapsed, Option 1=SelfHitTestInvisible, Index=@n1)
   )
   n2 = Get Image_TabIcon!(self:Fort Social Sidebar Button Object Reference=Get Button_More() -> (Button_More:Fort Social Sidebar Button Object Reference)) -> (Image_TabIcon:Image Object Reference)
   n3 = Get!(self:Image Object Reference=@n2) -> (
    Brush_TintColor:Slate Color Structure
    Brush_DrawAs:ESlateBrushDrawType Enum
    Brush_Tiling:ESlateBrushTileType Enum
    Brush_Mirroring:ESlateBrushMirrorType Enum
    Brush_ImageSize:Vector 2D Structure
    Brush_Margin:Margin Structure
    Brush_ResourceObject:Object Reference
    Brush_OutlineSettings:Slate Brush Outline Settings Structure
   )
   > SetBrush!(
    execute
    self=@n2
    InBrush_TintColor:Slate Color Structure=@n3
    InBrush_DrawAs:ESlateBrushDrawType Enum=@n3.Brush_DrawAs
    InBrush_Tiling:ESlateBrushTileType Enum=@n3.Brush_Tiling
    InBrush_Mirroring:ESlateBrushMirrorType Enum=@n3.Brush_Mirroring
    InBrush_ImageSize:Vector 2D Structure=@n3.Brush_ImageSize
    InBrush_Margin:Margin Structure=@n3.Brush_Margin
    InBrush_ResourceObject:Object Reference=Select(
     Option 0:Object Reference=/FortUILibrary/Textures/MaterialInstances/MI_UIKit_Icon_SettingsMenu
     Option 1:Object Reference=/FortUILibrary/Textures/MaterialInstances/MI_UIKit_Icon_Menu
     Index=@n1
    ) -> (ReturnValue:Object Reference)
    InBrush_OutlineSettings:Slate Brush Outline Settings Structure=@n3.Brush_OutlineSettings
   ) -> (then)
  ]}
]
variables []
widgets {
 info {inheritedWidgetCount=8, parentClass=/Script/FortniteUI.FortSidebarButtons, rootWidgetClass=&Overlay, widgetCount=14}
 widgets [
  Overlay_3 {widgetClassPath=&Overlay, widget=Overlay_3}
  SafeZone {
   widgetClassPath=&SafeZone
   bIsVariable=true
   parent=Overlay_3
   widget=SafeZone
   properties {padLeft=false, padRight=false}
   slot {class=OverlaySlot, properties {horizontalAlignment=HAlign_Fill, verticalAlignment=VAlign_Fill, padding=<ltrb 0 16 0 16>}}
  }
  VerticalBox_34 {widgetClassPath=&VerticalBox, parent=SafeZone, widget=VerticalBox_34, slot {class=SafeZoneSlot}}
  InputActionWidget_Left {
   widgetClassPath=&CommonActionWidget
   parent=VerticalBox_34
   widget=InputActionWidget_Left
   properties {visibility=SelfHitTestInvisible, inputActions [{dataTable=&InputActionDataTable, rowName=MainTabLeft}]}
   slot {class=VerticalBoxSlot, properties {padding=<ltrb 0 8 0 8>}}
  }
  Button_UserProfile {
   widgetClassPath=&/SidebarLocalUserButton
   bInherited=true
   bIsVariable=true
   parent=VerticalBox_34
   widget=Button_UserProfile
   properties {
    bSelectable=true
    minHeight=100
    minWidth=104
    navigation=&self:WidgetTree.Button_UserProfile.WidgetNavigation_0
    onButtonBaseClicked=(&self:WidgetTree.Button_UserProfile.HandleOnButtonClicked)
    onButtonBaseDoubleClicked=(&self:WidgetTree.Button_UserProfile.HandleOnButtonDoubleClicked)
    onButtonBaseHovered=(&self:WidgetTree.Button_UserProfile.HandleOnButtonHovered)
    onButtonBaseUnhovered=(&self:WidgetTree.Button_UserProfile.HandleOnButtonUnhovered)
    onSelectedChangedBase=(&self:WidgetTree.Button_UserProfile.HandleOnSelectedChanged)
    color_SkewBox_Selected=<rgba 0.0024 0.0513 0.3864 1>
   }
   slot {class=VerticalBoxSlot, properties {padding=<ltrb 0 6 0 6>}}
  }
  Button_PartyUp {
   widgetClassPath=/Game/UI/Sidebar/SidebarButton
   bInherited=true
   bIsVariable=true
   parent=VerticalBox_34
   widget=Button_PartyUp
   properties {
    bShowBang=false
    minHeight=100
    minWidth=104
    onButtonBaseClicked=(&self:WidgetTree.Button_PartyUp.HandleOnButtonClicked)
    onButtonBaseDoubleClicked=(&self:WidgetTree.Button_PartyUp.HandleOnButtonDoubleClicked)
    onButtonBaseHovered=(&self:WidgetTree.Button_PartyUp.HandleOnButtonHovered)
    onButtonBaseUnhovered=(&self:WidgetTree.Button_PartyUp.HandleOnButtonUnhovered)
    onSelectedChangedBase=(&self:WidgetTree.Button_PartyUp.HandleOnSelectedChanged)
    color_ContentHovered=<rgba 0 0.0578 0.5972 1>
    color_SkewBox_Selected=<rgba 0.0024 0.0513 0.3864 1>
   }
   slot {class=VerticalBoxSlot, properties {padding=<ltrb 0 0 0 6>}}
  }
  Button_FindFriends {
   widgetClassPath=&SidebarButton
   bInherited=true
   bIsVariable=true
   parent=VerticalBox_34
   widget=Button_FindFriends
   properties {
    bShowBang=false
    buttonType=FindFriends
    material_Icon=/FortUILibrary/Textures/MaterialInstances/MI_UIKit_Icon_AddFriend
    minHeight=100
    minWidth=104
    numberToShow=0
    onButtonBaseClicked=(&self:WidgetTree.Button_FindFriends.HandleOnButtonClicked)
    onButtonBaseDoubleClicked=(&self:WidgetTree.Button_FindFriends.HandleOnButtonDoubleClicked)
    onButtonBaseHovered=(&self:WidgetTree.Button_FindFriends.HandleOnButtonHovered)
    onButtonBaseUnhovered=(&self:WidgetTree.Button_FindFriends.HandleOnButtonUnhovered)
    onSelectedChangedBase=(&self:WidgetTree.Button_FindFriends.HandleOnSelectedChanged)
    color_ContentHovered=<rgba 0 0.0578 0.5972 1>
    color_SkewBox_Selected=<rgba 0.0024 0.0513 0.3864 1>
   }
   slot {class=VerticalBoxSlot, properties {padding=<ltrb 0 0 0 6>}}
  }
  Button_Tags {
   widgetClassPath=&SidebarButton
   bInherited=true
   bIsVariable=true
   parent=VerticalBox_34
   widget=Button_Tags
   properties {
    bShowBang=false
    buttonType=Tags
    material_Icon=/FortUILibrary/Textures/MaterialInstances/MI_UIKit_Icon_Tags
    minHeight=100
    minWidth=104
    numberToShow=0
    onButtonBaseClicked=(&self:WidgetTree.Button_Tags.HandleOnButtonClicked)
    onButtonBaseDoubleClicked=(&self:WidgetTree.Button_Tags.HandleOnButtonDoubleClicked)
    onButtonBaseHovered=(&self:WidgetTree.Button_Tags.HandleOnButtonHovered)
    onButtonBaseUnhovered=(&self:WidgetTree.Button_Tags.HandleOnButtonUnhovered)
    onSelectedChangedBase=(&self:WidgetTree.Button_Tags.HandleOnSelectedChanged)
    color_ContentHovered=<rgba 0 0.0578 0.5972 1>
    color_SkewBox_Selected=<rgba 0.0024 0.0513 0.3864 1>
   }
   slot {class=VerticalBoxSlot, properties {padding=<ltrb 0 0 0 6>}}
  }
  Button_VoiceChat {
   widgetClassPath=&SidebarButton
   bInherited=true
   bIsVariable=true
   parent=VerticalBox_34
   widget=Button_VoiceChat
   properties {
    bShowBang=false
    buttonType=VoiceChat
    material_Icon=/FortUILibrary/Textures/MaterialInstances/MI_UIKit_Icon_Mic
    material_Icon_FeatureDisabled=/FortUILibrary/Textures/MaterialInstances/MI_UIKit_Icon_MicLocked
    minHeight=100
    minWidth=104
    numberToShow=0
    onButtonBaseClicked=(&self:WidgetTree.Button_VoiceChat.HandleOnButtonClicked)
    onButtonBaseDoubleClicked=(&self:WidgetTree.Button_VoiceChat.HandleOnButtonDoubleClicked)
    onButtonBaseHovered=(&self:WidgetTree.Button_VoiceChat.HandleOnButtonHovered)
    onButtonBaseUnhovered=(&self:WidgetTree.Button_VoiceChat.HandleOnButtonUnhovered)
    onSelectedChangedBase=(&self:WidgetTree.Button_VoiceChat.HandleOnSelectedChanged)
    color_ContentHovered=<rgba 0 0.0578 0.5972 1>
    color_SkewBox_Selected=<rgba 0.0024 0.0513 0.3864 1>
   }
   slot {class=VerticalBoxSlot, properties {padding=<ltrb 0 0 0 6>}}
  }
  Spacer_79 {
   widgetClassPath=&Spacer
   parent=VerticalBox_34
   widget=Spacer_79
   slot {class=VerticalBoxSlot, properties {size {sizeRule=Fill, value=1}}}
  }
  Button_Settings {
   widgetClassPath=&SidebarButton
   bInherited=true
   bIsVariable=true
   parent=VerticalBox_34
   widget=Button_Settings
   properties {
    buttonType=Settings
    material_Icon=/FortUILibrary/Textures/MaterialInstances/MI_UIKit_Icon_GeneralSettings
    minHeight=100
    minWidth=104
    numberToShow=0
    onButtonBaseClicked=(&self:WidgetTree.Button_Settings.HandleOnButtonClicked)
    onButtonBaseDoubleClicked=(&self:WidgetTree.Button_Settings.HandleOnButtonDoubleClicked)
    onButtonBaseHovered=(&self:WidgetTree.Button_Settings.HandleOnButtonHovered)
    onButtonBaseUnhovered=(&self:WidgetTree.Button_Settings.HandleOnButtonUnhovered)
    onSelectedChangedBase=(&self:WidgetTree.Button_Settings.HandleOnSelectedChanged)
    visibility=Collapsed
    color_ContentHovered=<rgba 0 0.0578 0.5972 1>
    color_SkewBox_Selected=<rgba 0.0024 0.0513 0.3864 1>
   }
   slot {class=VerticalBoxSlot, properties {verticalAlignment=VAlign_Bottom, padding=<ltrb 0 0 0 6>}}
  }
  Button_More {
   widgetClassPath=&SidebarButton
   bInherited=true
   bIsVariable=true
   parent=VerticalBox_34
   widget=Button_More
   properties {
    buttonType=More
    material_Icon=&MI_UIKit_Icon_Menu
    minHeight=100
    minWidth=104
    numberToShow=0
    onButtonBaseClicked=()
    onButtonBaseDoubleClicked=()
    onButtonBaseHovered=()
    onButtonBaseUnhovered=()
    onSelectedChangedBase=()
    color_ContentHovered=<rgba 0 0.0578 0.5972 1>
    color_SkewBox_Selected=<rgba 0.0024 0.0513 0.3864 1>
   }
   slot {class=VerticalBoxSlot, properties {verticalAlignment=VAlign_Bottom, padding=<ltrb 0 0 0 6>}}
  }
  Button_Exit {
   widgetClassPath=&SidebarButton
   bInherited=true
   bIsVariable=true
   parent=VerticalBox_34
   widget=Button_Exit
   properties {
    buttonType=Exit
    material_Icon=/FortUILibrary/Textures/MaterialInstances/MI_UIKit_Icon_Exit
    minHeight=100
    minWidth=104
    navigation=&self:WidgetTree.Button_Exit.WidgetNavigation_0
    numberToShow=0
    onButtonBaseClicked=(&self:WidgetTree.Button_Exit.HandleOnButtonClicked)
    onButtonBaseDoubleClicked=(&self:WidgetTree.Button_Exit.HandleOnButtonDoubleClicked)
    onButtonBaseHovered=(&self:WidgetTree.Button_Exit.HandleOnButtonHovered)
    onButtonBaseUnhovered=(&self:WidgetTree.Button_Exit.HandleOnButtonUnhovered)
    onSelectedChangedBase=(&self:WidgetTree.Button_Exit.HandleOnSelectedChanged)
    color_ContentHovered=<rgba 0.3 0.0059 0.0064 1>
    color_ContentUnselected=<rgba 1 0.1714 0.1329 1>
    color_SkewBox_Selected=<rgba 0.1812 0.016 0.0194 1>
    lockedPressedSlateSoundOverride {resourceObject=/Game/Sounds/Fort_UI_Sounds/FrontEnd/DefaultSounds/PressedButtons/MSS_UI_Button_Pressed_Exit}
    selectedPressedSlateSoundOverride {resourceObject=&MSS_UI_Button_Pressed_Exit}
   }
   slot {class=VerticalBoxSlot, properties {verticalAlignment=VAlign_Bottom}}
  }
  InputActionWidget_Right {
   widgetClassPath=&CommonActionWidget
   parent=VerticalBox_34
   widget=InputActionWidget_Right
   properties {visibility=SelfHitTestInvisible, inputActions [{dataTable=&InputActionDataTable, rowName=MainTabRight}]}
   slot {class=VerticalBoxSlot, properties {padding=<ltrb 0 8 0 8>}}
  }
 ]
}
