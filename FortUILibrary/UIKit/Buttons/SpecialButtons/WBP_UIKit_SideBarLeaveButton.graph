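# UI Kit leave button for the main menu sidebar, a FortMainMenuLeaveButtonWrapper child that wraps a regular UI Kit button.
# PreConstruct casts the inherited ButtonInternal to the CTA base, caches it and applies the Text variable as its label.
# Parent events drive the button: text changes update the variable and label, while timer text becomes secondary text.
# Timer visibility changes act only on Collapsed, which clears the secondary text.
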WidgetBlueprint /FortUILibrary/UIKit/Buttons/SpecialButtons/WBP_UIKit_SideBarLeaveButton
graphs [EventGraph {nodes [
   n1 = Event OnTimerTextChanged!() -> (OutputDelegate, then, Text:"Text (by ref)") {is=title K2Node_Event}
   > SetSecondaryText!(
    execute
    self:Fort CTAButton Object Reference=Get CommonButtonInternal() -> (CommonButtonInternal:WBP UIKit Button CTA Base Object Reference)
    ButtonSecondaryText:"Text (by ref)"=@n1.Text
   ) -> (then)
   Event PreConstruct()
   > n2 = Cast To WBP_UIKit_ButtonCTA_Base!(execute, Object=Get ButtonInternal() -> (ButtonInternal:Common Button Base Object Reference)) -> (then, CastFailed, AsWBP UIKit Button CTA Base:WBP UIKit Button CTA Base Object Reference) {is=title K2Node_DynamicCast}
   > n3 = Set CommonButtonInternal(CommonButtonInternal:WBP UIKit Button CTA Base Object Reference=@n2) -> (Output_Get:WBP UIKit Button CTA Base Object Reference)
   > SetText!(execute, self:Fort CTAButton Object Reference=@n3, ButtonText:"Text (by ref)"=Get Text!() -> (Text)) -> (then)
   n4 = Event OnTimerVisibilityChanged!() -> (OutputDelegate, then, InVisibility:ESlateVisibility Enum=Visible) {is=title K2Node_Event}
   > n5 = Switch on ESlateVisibility!(execute, Selection:ESlateVisibility Enum=@n4.InVisibility) -> (Visible:Exec, Collapsed:Exec, Hidden:Exec, HitTestInvisible:Exec, SelfHitTestInvisible:Exec) {is=title K2Node_SwitchEnum}
   SetSecondaryText(execute=@n5.Collapsed, self=Get CommonButtonInternal(), ButtonSecondaryText=MakeLiteralText())
   n6 = Event OnTextChanged!() -> (OutputDelegate, then, Text:"Text (by ref)") {is=title K2Node_Event}
   > n7 = Set Text!(execute, Text=@n6.Text) -> (then, Output_Get)
   > SetText(self=Get CommonButtonInternal(), ButtonText=@n7)
  ]}]
variables [
 CommonButtonInternal {type=WBP_UIKit_ButtonCTA_Base_C, default=null}
 Text {type=string, default=SideBarLeaveButton}
]
widgets {
 info {
  inheritedWidgetCount=1
  parentClass=/Script/FortniteUI.FortMainMenuLeaveButtonWrapper
  rootWidgetClass=/FortUILibrary/UIKit/Buttons/WBP_UIKit_Button_Regular.WBP_UIKit_Button_Regular_C
  widgetCount=1
 }
 widgets [ButtonInternal {
   widgetClassPath=&WBP_UIKit_Button_Regular
   bInherited=true
   bIsVariable=true
   widget=ButtonInternal
   properties {text=SideBarLeaveButton}
  }]
}
