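# Creative UI row widget with Favorite, Recommend and Not interested buttons plus a coach mark, for expressing support.
# Clicks forward to parent FortExpressYourSupportRow handlers; Recommend and Not interested each toggle the other's enabled state first.
# Parent events re-enable both buttons on reset and set Favorite's displayed and selected state.
# Custom left/right navigation rules, bound on Construct, route focus around whichever button is disabled.
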WidgetBlueprint /Game/Athena/UI/InGame/Creative/WBP_ExpressYourSupportRow
graphs [
 EventGraph {nodes [
   "On Clicked (Button_Favorite)"!() -> (OutputDelegate, then, Button:Common Button Base Object Reference) {is=title K2Node_ComponentBoundEvent}
   > HandleFavoriteButtonClicked(self:Fort Express Your Support Row Object Reference)
   "On Clicked (Button_Recommend)"!() -> (OutputDelegate, then, Button:Common Button Base Object Reference) {is=title K2Node_ComponentBoundEvent}
   n1 = Get Button_NotInterested() -> (Button_NotInterested:WBP Express Your Support Button Object Reference)
   > SetIsEnabled(self=@n1, bInIsEnabled=NOT Boolean(A=GetIsEnabled!(self:Widget Object Reference=@n1) -> (ReturnValue:bool)))
   > HandleLikeButtonClicked(self:Fort Express Your Support Row Object Reference)
   "On Clicked (Button_NotInterested)"!() -> (OutputDelegate, then, Button:Common Button Base Object Reference) {is=title K2Node_ComponentBoundEvent}
   n2 = Get Button_Recommend() -> (Button_Recommend:WBP Express Your Support Button Object Reference)
   > SetIsEnabled(self=@n2, bInIsEnabled=NOT Boolean(A=GetIsEnabled(self=@n2)))
   > HandleNotInterestedButtonClicked(self:Fort Express Your Support Row Object Reference)
   Event ResetSupportPlayerFacingState() {is=title K2Node_Event}
   > SetIsEnabled(self=[Get Button_Recommend(), Get Button_NotInterested()], bInIsEnabled=true)
   Event Construct()
   n3 = Get Button_Recommend()
   > SetNavigationRuleCustom!(execute, self:Widget Object Reference=@n3, Direction:EUINavigation Enum=Right, InCustomDelegate:Delegate=Create Event()) -> (then)
   > SetNavigationRuleCustom(self=@n3, Direction=Left, InCustomDelegate=Create Event())
   > SetNavigationRuleCustom(self=Get Button_NotInterested(), Direction=Left, InCustomDelegate=Create Event())
   > SetNavigationRuleCustom(
    self=Get Button_Favorite() -> (Button_Favorite:WBP Express Your Support Button Object Reference)
    Direction=Right
    InCustomDelegate=Create Event()
   )
   n4 = Event SetPlayerFacingFavoriteState!() -> (OutputDelegate, then, bInIsFavorite:bool) {is=title K2Node_Event}
   n5 = Get Button_Favorite()
   > SetIsFavorite!(execute, self:WBP Express Your Support Button Object Reference=@n5, IsFavorite:bool=@n4.bInIsFavorite) -> (then)
   > SetIsSelected!(execute, self:Common Button Base Object Reference=@n5, InSelected:bool=@n4.bInIsFavorite, bGiveClickFeedback:bool=false) -> (then)
  ]}
 FromFavoriteButton {nodes [
   FromFavoriteButton!() -> (then, Navigation:EUINavigation Enum) {is=title K2Node_FunctionEntry}
   > n1 = Branch(Condition=GetIsEnabled(self=Get Button_Recommend()))
   > Return Node!(execute, ReturnValue:Widget Object Reference=Get Button_Recommend())
   Return Node(execute=@n1.else, ReturnValue=Get Button_NotInterested())
  ]}
 FromNotInterestedButton {nodes [
   FromNotInterestedButton!() -> (then, Navigation:EUINavigation Enum) {is=title K2Node_FunctionEntry}
   > n1 = Branch(Condition=GetIsEnabled(self=Get Button_Recommend()))
   > Return Node(ReturnValue=Get Button_Recommend())
   Return Node(execute=@n1.else, ReturnValue=Get Button_Favorite())
  ]}
 FromRecommendButtonLeft {nodes [
   FromRecommendButtonLeft!() -> (then, Navigation:EUINavigation Enum) {is=title K2Node_FunctionEntry}
   > Return Node(ReturnValue=Get Button_Favorite())
  ]}
 FromRecommendButtonRight {nodes [
   FromRecommendButtonRight!() -> (then, Navigation:EUINavigation Enum) {is=title K2Node_FunctionEntry}
   > Branch(Condition=GetIsEnabled(self=Get Button_NotInterested()))
   > Return Node(ReturnValue=Get Button_NotInterested())
  ]}
]
variables []
widgets {
 info {inheritedWidgetCount=0, parentClass=/Script/FortniteUI.FortExpressYourSupportRow, rootWidgetClass=&SizeBox, widgetCount=7}
 widgets [
  SizeBox_0 {widgetClassPath=&SizeBox, widget=SizeBox_0, properties {bOverride_MinDesiredHeight=true, heightOverride=72, minDesiredHeight=127}}
  GP_Content {
   widgetClassPath=&GridPanel
   parent=SizeBox_0
   widget=GP_Content
   properties {bOverride_Cursor=true, columnFill [1,1,1], rowFill [1]}
   slot {class=SizeBoxSlot}
  }
  Button_Favorite {
   widgetClassPath=&/WBP_ExpressYourSupportButton
   bIsVariable=true
   parent=GP_Content
   widget=Button_Favorite
   properties {holdData=&CommonUI_HoldData_Base_C}
   slot {class=GridSlot}
  }
  Button_Recommend {
   widgetClassPath=&WBP_ExpressYourSupportButton
   bIsVariable=true
   parent=GP_Content
   widget=Button_Recommend
   properties {
    holdData=&CommonUI_HoldData_Base_C
    icon=/FortUILibrary/Foundation/Buttons/Materials/Favorite/MI_UI_ThumbsUp_ExpressYourSupport
    pulseIcon=/FortUILibrary/Foundation/Buttons/Materials/Favorite/MI_UI_ThumbsUp_Pulse
    text=Recommend
   }
   slot {class=GridSlot, properties {column=1}}
  }
  Button_NotInterested {
   widgetClassPath=&WBP_ExpressYourSupportButton
   bIsVariable=true
   parent=GP_Content
   widget=Button_NotInterested
   properties {
    holdData=&CommonUI_HoldData_Base_C
    icon=&MI_UI_ThumbsUp_ExpressYourSupport
    isIconsInverted=true
    pulseIcon=&MI_UI_ThumbsUp_Pulse
    text=Not interested
   }
   slot {class=GridSlot, properties {column=2}}
  }
  FortVisualAttachment_40 {
   widgetClassPath=&FortVisualAttachment
   parent=GP_Content
   widget=FortVisualAttachment_40
   properties {contentAnchor=<1 0.5>}
   slot {class=GridSlot, properties {horizontalAlignment=HAlign_Left, layer=5, verticalAlignment=VAlign_Center, nudge=<-25 0>}}
  }
  WBP_Sidebar_ExpressYourSupport_CoachMark {
   widgetClassPath=&/WBP_Sidebar_ExpressYourSupport_CoachMark
   bIsVariable=true
   parent=FortVisualAttachment_40
   widget=WBP_Sidebar_ExpressYourSupport_CoachMark
   slot {class=SizeBoxSlot}
  }
 ]
}
