WidgetBlueprint /Game/Athena/UI/InGame/Creative/WBP_ExpressYourSupportThumbsUpButtonRight
animations [
 HoveredToggleOn {
  binding_count=3
  length_seconds=0.75
  object=&self:HoveredToggleOn
  sequence {bindings [
    Image_FavoriteIcon {
     id=#1
     possessed_class=Image
     tracks [{
       type=MovieSceneWidgetMaterialTrack
       sections [{
         type=MovieSceneParameterSection
         channels [
          bIsInactive {type=MovieSceneScriptingFloatChannel, keys [[0,1],[4,0]]}
          bOnHover_RGBA {type=MovieSceneScriptingFloatChannel, const [1,1]}
         ]
        }]
      }]
    }
    Image_Pulse {
     id=#2
     possessed_class=Image
     tracks [
      {
       type=MovieSceneFloatTrack
       display_name=Render Opacity
       sections [{type=MovieSceneFloatSection, channels [None {type=MovieSceneScriptingFloatChannel, keys [[1,0],[4,1],[5.6,1],[9.6,0]]}]}]
      }
      {
       type=MovieScene2DTransformTrack
       sections [{
         type=MovieScene2DTransformSection
         channels [
          Scale.X {type=MovieSceneScriptingFloatChannel, keys [[2,1],[7.6,2.5]]}
          Scale.Y {type=MovieSceneScriptingFloatChannel, keys [[2,1],[7.6,2.5]]}
         ]
        }]
      }
     ]
    }
    TextSwitcher {
     id=#3
     possessed_class=WidgetSwitcher
     tracks [{
       type=MovieSceneIntegerTrack
       display_name=Active Widget Index
       sections [{type=MovieSceneIntegerSection, end_frame=14, start_frame=0, channels [None {type=MovieSceneScriptingIntegerChannel, keys [[0,1],[14,0]]}]}]
      }]
    }
   ]}
 }
 HoveredToggleOff {
  binding_count=2
  length_seconds=0.25
  object=&self:HoveredToggleOff
  sequence {bindings [
    Image_FavoriteIcon {
     id=#4
     possessed_class=Image
     tracks [
      {
       type=MovieSceneWidgetMaterialTrack
       sections [{
         type=MovieSceneParameterSection
         channels [
          bIsInactive {type=MovieSceneScriptingFloatChannel, keys [[0,0],[5,1]]}
          bOnHover_RGBA {type=MovieSceneScriptingFloatChannel, const [1,1]}
         ]
        }]
      }
      {
       type=MovieScene2DTransformTrack
       sections [{
         type=MovieScene2DTransformSection
         channels [
          Translation.X {type=MovieSceneScriptingFloatChannel}
          Translation.Y {type=MovieSceneScriptingFloatChannel}
          Angle {type=MovieSceneScriptingFloatChannel}
          Scale.X {type=MovieSceneScriptingFloatChannel, keys [[0,1],[2.4,0.875],[5,1]]}
          Scale.Y {type=MovieSceneScriptingFloatChannel, keys [[0,1],[2.4,0.875],[5,1]]}
          Shear.X {type=MovieSceneScriptingFloatChannel}
          Shear.Y {type=MovieSceneScriptingFloatChannel}
         ]
        }]
      }
     ]
    }
    TextSwitcher {
     id=#5
     possessed_class=WidgetSwitcher
     tracks [{
       type=MovieSceneIntegerTrack
       display_name=Active Widget Index
       sections [{type=MovieSceneIntegerSection, end_frame=0, start_frame=0, channels [None {type=MovieSceneScriptingIntegerChannel, const [0,1]}]}]
      }]
    }
   ]}
 }
 HoverToggledOn {
  binding_count=3
  length_seconds=0.2
  object=&self:HoverToggledOn
  sequence {bindings [
    Border_Background {
     id=#6
     possessed_class=Border
     tracks [{
       type=MovieSceneWidgetMaterialTrack
       sections [{
         type=MovieSceneParameterSection
         end_frame=4
         start_frame=0
         channels [AnimateHighlight {type=MovieSceneScriptingFloatChannel, keys [[0,0],[3,1],[4,1]]}]
        }]
      }]
    }
    Text_NormalText {
     id=#7
     possessed_class=CommonTextBlock
     tracks [{
       type=MovieSceneColorTrack
       display_name=Color and Opacity
       sections [{
         type=MovieSceneColorSection
         end_frame=4
         start_frame=0
         channels [
          Color.R {type=MovieSceneScriptingFloatChannel, keys [[0,1],[4,0.0012]]}
          Color.G {type=MovieSceneScriptingFloatChannel, keys [[0,1],[4,0.004]]}
          Color.B {type=MovieSceneScriptingFloatChannel, keys [[0,1],[4,0.0467]]}
          Color.A {type=MovieSceneScriptingFloatChannel, const [1,2]}
         ]
        }]
      }]
    }
    Image_FavoriteIcon {
     id=#8
     possessed_class=Image
     tracks [{
       type=MovieSceneWidgetMaterialTrack
       sections [{
         type=MovieSceneParameterSection
         end_frame=4
         start_frame=0
         channels [
          bIsInactive {type=MovieSceneScriptingFloatChannel, const [0,1]}
          bOnHover_RGBA {type=MovieSceneScriptingFloatChannel, keys [[0,0],[4,1]]}
         ]
        }]
      }]
    }
   ]}
 }
 UnhoverToggledOn {
  binding_count=3
  length_seconds=0.2
  object=&self:UnhoverToggledOn
  sequence {bindings [
    Text_NormalText {
     id=#9
     possessed_class=CommonTextBlock
     tracks [{
       type=MovieSceneColorTrack
       display_name=Color and Opacity
       sections [{
         type=MovieSceneColorSection
         end_frame=4
         start_frame=0
         channels [
          Color.R {type=MovieSceneScriptingFloatChannel, keys [[0,0.0012],[4,1]]}
          Color.G {type=MovieSceneScriptingFloatChannel, keys [[0,0.004],[4,1]]}
          Color.B {type=MovieSceneScriptingFloatChannel, keys [[0,0.0467],[4,1]]}
          Color.A {type=MovieSceneScriptingFloatChannel, const [1,2]}
         ]
        }]
      }]
    }
    Border_Background {
     id=#10
     possessed_class=Border
     tracks [{
       type=MovieSceneWidgetMaterialTrack
       sections [{
         type=MovieSceneParameterSection
         end_frame=4
         start_frame=0
         channels [AnimateHighlight {type=MovieSceneScriptingFloatChannel, keys [[0,1],[1,1],[4,0]]}]
        }]
      }]
    }
    Image_FavoriteIcon {
     id=#11
     possessed_class=Image
     tracks [{
       type=MovieSceneWidgetMaterialTrack
       sections [{
         type=MovieSceneParameterSection
         end_frame=4
         start_frame=0
         channels [
          bIsInactive {type=MovieSceneScriptingFloatChannel, const [0,1]}
          bOnHover_RGBA {type=MovieSceneScriptingFloatChannel, keys [[0,1],[4,0]]}
         ]
        }]
      }]
    }
   ]}
 }
 HoverToggledOff {
  binding_count=3
  length_seconds=0.2
  object=&self:HoverToggledOff
  sequence {bindings [
    Border_Background {
     id=#12
     possessed_class=Border
     tracks [{
       type=MovieSceneWidgetMaterialTrack
       sections [{
         type=MovieSceneParameterSection
         end_frame=4
         start_frame=0
         channels [AnimateHighlight {type=MovieSceneScriptingFloatChannel, keys [[0,0],[3,1],[4,1]]}]
        }]
      }]
    }
    Text_NormalText {
     id=#13
     possessed_class=CommonTextBlock
     tracks [{
       type=MovieSceneColorTrack
       display_name=Color and Opacity
       sections [{
         type=MovieSceneColorSection
         end_frame=4
         start_frame=0
         channels [
          Color.R {type=MovieSceneScriptingFloatChannel, keys [[0,1],[4,0.0012]]}
          Color.G {type=MovieSceneScriptingFloatChannel, keys [[0,1],[4,0.004]]}
          Color.B {type=MovieSceneScriptingFloatChannel, keys [[0,1],[4,0.0467]]}
          Color.A {type=MovieSceneScriptingFloatChannel, const [1,2]}
         ]
        }]
      }]
    }
    Image_FavoriteIcon {
     id=#14
     possessed_class=Image
     tracks [{
       type=MovieSceneWidgetMaterialTrack
       sections [{type=MovieSceneParameterSection, channels [bOnHover_RGBA {type=MovieSceneScriptingFloatChannel, keys [[0,0],[3,1]]}]}]
      }]
    }
   ]}
 }
 UnhoverToggledOff {
  binding_count=3
  length_seconds=0.2
  object=&self:UnhoverToggledOff
  sequence {bindings [
    Border_Background {
     id=#15
     possessed_class=Border
     tracks [{
       type=MovieSceneWidgetMaterialTrack
       sections [{
         type=MovieSceneParameterSection
         end_frame=4
         start_frame=0
         channels [AnimateHighlight {type=MovieSceneScriptingFloatChannel, keys [[0,1],[1,1],[4,0]]}]
        }]
      }]
    }
    Text_NormalText {
     id=#16
     possessed_class=CommonTextBlock
     tracks [{
       type=MovieSceneColorTrack
       display_name=Color and Opacity
       sections [{
         type=MovieSceneColorSection
         end_frame=4
         start_frame=0
         channels [
          Color.R {type=MovieSceneScriptingFloatChannel, keys [[0,0.0012],[4,1]]}
          Color.G {type=MovieSceneScriptingFloatChannel, keys [[0,0.004],[4,1]]}
          Color.B {type=MovieSceneScriptingFloatChannel, keys [[0,0.0467],[4,1]]}
          Color.A {type=MovieSceneScriptingFloatChannel, const [1,2]}
         ]
        }]
      }]
    }
    Image_FavoriteIcon {
     id=#17
     possessed_class=Image
     tracks [{
       type=MovieSceneWidgetMaterialTrack
       sections [{type=MovieSceneParameterSection, channels [bOnHover_RGBA {type=MovieSceneScriptingFloatChannel, keys [[0,1],[3,0]]}]}]
      }]
    }
   ]}
 }
 UnhoveredToggleOn {
  binding_count=3
  length_seconds=0.75
  object=&self:UnhoveredToggleOn
  sequence {bindings [
    TextSwitcher {
     id=#18
     possessed_class=WidgetSwitcher
     tracks [{
       type=MovieSceneIntegerTrack
       display_name=Active Widget Index
       sections [{type=MovieSceneIntegerSection, end_frame=14, start_frame=0, channels [None {type=MovieSceneScriptingIntegerChannel, keys [[0,1],[14,0]]}]}]
      }]
    }
    Image_FavoriteIcon {
     id=#1
     possessed_class=Image
     tracks [{
       type=MovieSceneWidgetMaterialTrack
       sections [{
         type=MovieSceneParameterSection
         channels [
          bIsInactive {type=MovieSceneScriptingFloatChannel, keys [[0,1],[4,0]]}
          bOnHover_RGBA {type=MovieSceneScriptingFloatChannel, const [0,1]}
         ]
        }]
      }]
    }
    Image_Pulse {
     id=#2
     possessed_class=Image
     tracks [
      {
       type=MovieSceneFloatTrack
       display_name=Render Opacity
       sections [{type=MovieSceneFloatSection, channels [None {type=MovieSceneScriptingFloatChannel, keys [[1,0],[4,1],[5.6,1],[9.6,0]]}]}]
      }
      {
       type=MovieScene2DTransformTrack
       sections [{
         type=MovieScene2DTransformSection
         channels [
          Scale.X {type=MovieSceneScriptingFloatChannel, keys [[2,1],[7.6,2.5]]}
          Scale.Y {type=MovieSceneScriptingFloatChannel, keys [[2,1],[7.6,2.5]]}
         ]
        }]
      }
     ]
    }
   ]}
 }
 UnhoveredToggleOff {
  binding_count=2
  length_seconds=0.25
  object=&self:UnhoveredToggleOff
  sequence {bindings [
    Image_FavoriteIcon {
     id=#4
     possessed_class=Image
     tracks [
      {
       type=MovieSceneWidgetMaterialTrack
       sections [{
         type=MovieSceneParameterSection
         channels [
          bIsInactive {type=MovieSceneScriptingFloatChannel, keys [[0,0],[5,1]]}
          bOnHover_RGBA {type=MovieSceneScriptingFloatChannel, const [0,1]}
         ]
        }]
      }
      {
       type=MovieScene2DTransformTrack
       sections [{
         type=MovieScene2DTransformSection
         channels [
          Translation.X {type=MovieSceneScriptingFloatChannel}
          Translation.Y {type=MovieSceneScriptingFloatChannel}
          Angle {type=MovieSceneScriptingFloatChannel}
          Scale.X {type=MovieSceneScriptingFloatChannel, keys [[0,1],[2.4,0.875],[5,1]]}
          Scale.Y {type=MovieSceneScriptingFloatChannel, keys [[0,1],[2.4,0.875],[5,1]]}
          Shear.X {type=MovieSceneScriptingFloatChannel}
          Shear.Y {type=MovieSceneScriptingFloatChannel}
         ]
        }]
      }
     ]
    }
    TextSwitcher {
     id=#19
     possessed_class=WidgetSwitcher
     tracks [{
       type=MovieSceneIntegerTrack
       display_name=Active Widget Index
       sections [{type=MovieSceneIntegerSection, end_frame=0, start_frame=0, channels [None {type=MovieSceneScriptingIntegerChannel, const [0,1]}]}]
      }]
    }
   ]}
 }
]
graphs [
 EventGraph {nodes [
   Event Construct()
   > SetButtonHeight!(execute, self, IsTouch:bool=IsUsingTouch(PlayerController=GetOwningPlayer())) -> (then)
   Event On Selected()
   n1 = IsHovered()
   > n2 = Branch(Condition=@n1)
   > PlayAnimation(InAnimation=Get HoveredToggleOn() -> (HoveredToggleOn:Widget Animation Object Reference))
   PlayAnimation(execute=@n2.else, InAnimation=Get UnhoveredToggleOn() -> (UnhoveredToggleOn:Widget Animation Object Reference))
   Event On Hovered()
   n3 = GetSelected!(self:Common Button Base Object Reference) -> (ReturnValue:bool)
   > n4 = Branch(Condition=@n3)
   > PlayAnimation(InAnimation=Get HoverToggledOn() -> (HoverToggledOn:Widget Animation Object Reference))
   PlayAnimation(execute=@n4.else, InAnimation=Get HoverToggledOff() -> (HoverToggledOff:Widget Animation Object Reference))
   Event On Unhovered()
   > StopHoveringAnimations()
   > n5 = Branch(Condition=@n3)
   > PlayAnimation(InAnimation=Get UnhoverToggledOn() -> (UnhoverToggledOn:Widget Animation Object Reference), PlaybackSpeed=2)
   PlayAnimation(execute=@n5.else, InAnimation=Get UnhoverToggledOff() -> (UnhoverToggledOff:Widget Animation Object Reference))
   Event On Deselected()
   > StopTogglingAnimations()
   > n6 = Branch(Condition=@n1)
   > PlayAnimation(InAnimation=Get HoveredToggleOff() -> (HoveredToggleOff:Widget Animation Object Reference))
   PlayAnimation(execute=@n6.else, InAnimation=Get UnhoveredToggleOff() -> (UnhoveredToggleOff:Widget Animation Object Reference))
   Cast To WBP_ExpressYourSupportOnboardingTooltipRight!(execute, Object:Wildcard) -> (then, CastFailed, AsWBP Express Your Support Onboarding Tooltip Right:WBP Express Your Support Onboarding Tooltip Right Object Reference) {is=title K2Node_DynamicCast}
  ]}
 SetButtonHeight {nodes [
   n1 = SetButtonHeight!() -> (then, IsTouch:bool=~) {is=title K2Node_FunctionEntry}
   > n2 = Branch(Condition=@n1)
   > SetHeightOverride(self=Get SizeBox_0() -> (SizeBox_0:Size Box Object Reference), InHeightOverride=90)
   SetHeightOverride(execute=@n2.else, self=Get SizeBox_0(), InHeightOverride=72)
  ]}
 StopHoveringAnimations {nodes [
   StopHoveringAnimations() {is=title K2Node_FunctionEntry}
   > StopAnimation(InAnimation=Get HoverToggledOn())
   > StopAnimation(InAnimation=Get UnhoverToggledOn())
   > StopAnimation(InAnimation=Get HoverToggledOff())
   > StopAnimation(InAnimation=Get UnhoverToggledOff())
   > Return Node()
  ]}
 StopTogglingAnimations {nodes [
   StopTogglingAnimations() {is=title K2Node_FunctionEntry}
   > StopAnimation(InAnimation=Get HoveredToggleOn())
   > StopAnimation(InAnimation=Get HoveredToggleOff())
   > StopAnimation(InAnimation=Get UnhoveredToggleOn())
   > StopAnimation(InAnimation=Get UnhoveredToggleOff())
   > Return Node()
  ]}
]
variables []
widgets {
 info {
  inheritedWidgetCount=3
  parentClass=/Script/FortniteUI.FortExpressYourSupportThumbsUpButton
  rootWidgetClass=&SizeBox
  widgetCount=11
 }
 widgets [
  SizeBox_0 {
   widgetClassPath=&SizeBox
   bIsVariable=true
   widget=SizeBox_0
   properties {bOverride_HeightOverride=true, heightOverride=72, minDesiredHeight=72}
  }
  Overlay_1 {widgetClassPath=&Overlay, parent=SizeBox_0, widget=Overlay_1, slot {class=SizeBoxSlot}}
  Border_Background {
   widgetClassPath=&Border
   bIsVariable=true
   parent=Overlay_1
   widget=Border_Background
   properties {
    horizontalAlignment=HAlign_Center
    verticalAlignment=VAlign_Top
    visibility=SelfHitTestInvisible
    background {
     bIsDynamicallyLoaded=false
     drawAs=Image
     imageType=NoImage
     mirroring=NoMirror
     resourceName=None
     resourceObject=/Game/UI/Sidebar/Materials/M_UI_sidebar_Button_Selection_Inst
     tiling=NoTile
     imageSize=<32 32>
     margin=<ltrb 0 0 0 0>
     outlineSettings {
      bUseBrushTransparency=false
      roundingType=HalfHeightRadius
      width=0
      color {colorUseRule=UseColor_Specified, specifiedColor=<rgba 0 0 0 0>}
      cornerRadii=<0 0 0 0>
     }
     tintColor {colorUseRule=UseColor_Specified, specifiedColor=<rgba 1 1 1 1>}
     uVRegion {bIsValid=false, max=<0 0>, min=<0 0>}
    }
    padding=<ltrb 0 0 0 0>
   }
   slot {class=OverlaySlot, properties {horizontalAlignment=HAlign_Fill, verticalAlignment=VAlign_Fill}}
  }
  Image_FavoriteIcon {
   widgetClassPath=&Image
   bIsVariable=true
   parent=Overlay_1
   widget=Image_FavoriteIcon
   properties {brush {
     bIsDynamicallyLoaded=false
     drawAs=Image
     imageType=NoImage
     mirroring=NoMirror
     resourceName=None
     resourceObject=/FortUILibrary/Foundation/Buttons/Materials/Favorite/MI_UI_ThumbsUp_ExpressYourSupport
     tiling=NoTile
     imageSize=<48 48>
     margin=<ltrb 0 0 0 0>
     outlineSettings {
      bUseBrushTransparency=false
      roundingType=HalfHeightRadius
      width=0
      color {colorUseRule=UseColor_Specified, specifiedColor=<rgba 0 0 0 0>}
      cornerRadii=<0 0 0 0>
     }
     tintColor {colorUseRule=UseColor_Specified, specifiedColor=<rgba 1 1 1 1>}
     uVRegion {bIsValid=false, max=<0 0>, min=<0 0>}
    }}
   slot {class=OverlaySlot, properties {verticalAlignment=VAlign_Center, padding=<ltrb 20 0 0 0>}}
  }
  Image_Pulse {
   widgetClassPath=&Image
   bIsVariable=true
   parent=Overlay_1
   widget=Image_Pulse
   properties {
    renderOpacity=0
    brush {
     bIsDynamicallyLoaded=false
     drawAs=Image
     imageType=NoImage
     mirroring=NoMirror
     resourceName=None
     resourceObject=/FortUILibrary/Foundation/Buttons/Materials/Favorite/MI_UI_ThumbsUp_Pulse
     tiling=NoTile
     imageSize=<48 48>
     margin=<ltrb 0 0 0 0>
     outlineSettings {
      bUseBrushTransparency=false
      roundingType=HalfHeightRadius
      width=0
      color {colorUseRule=UseColor_Specified, specifiedColor=<rgba 0 0 0 0>}
      cornerRadii=<0 0 0 0>
     }
     tintColor {colorUseRule=UseColor_Specified, specifiedColor=<rgba 1 1 1 1>}
     uVRegion {bIsValid=false, max=<0 0>, min=<0 0>}
    }
   }
   slot {class=OverlaySlot, properties {verticalAlignment=VAlign_Center, padding=<ltrb 20 0 0 0>}}
  }
  TextSwitcher {
   widgetClassPath=&WidgetSwitcher
   bIsVariable=true
   parent=Overlay_1
   widget=TextSwitcher
   slot {class=OverlaySlot, properties {verticalAlignment=VAlign_Center, padding=<ltrb 100 0 0 0>}}
  }
  Text_NormalText {
   widgetClassPath=&CommonTextBlock
   bInherited=true
   parent=TextSwitcher
   widget=Text_NormalText
   properties {
    mobileFontSizeMultiplier=1.3
    text=RECOMMEND
    font {
     bForceMonospaced=false
     bMaterialIsStencil=false
     fontMaterial=null
     fontObject=&BurbankSmall
     letterSpacing=0
     monospacedWidth=1
     size=16
     skewAmount=0
     typefaceFontName=Black
     outlineSettings {
      bApplyOutlineToDropShadows=false
      bMiteredCorners=false
      bSeparateFillAlpha=false
      outlineMaterial=null
      outlineSize=0
      outlineColor=<rgba 0 0 0 1>
     }
    }
    shadowOffset=<0 0>
   }
   slot {class=WidgetSwitcherSlot}
  }
  Text_ThankYou {
   widgetClassPath=&CommonTextBlock
   parent=TextSwitcher
   widget=Text_ThankYou
   properties {
    mobileFontSizeMultiplier=1.3
    text="THANK YOU!"
    colorAndOpacity {colorUseRule=UseColor_Specified, specifiedColor=<rgba 0 0.3185 1 1>}
    font {
     bForceMonospaced=false
     bMaterialIsStencil=false
     fontMaterial=null
     fontObject=&BurbankSmall
     letterSpacing=0
     monospacedWidth=1
     size=18
     skewAmount=0
     typefaceFontName=Black
     outlineSettings {
      bApplyOutlineToDropShadows=false
      bMiteredCorners=false
      bSeparateFillAlpha=false
      outlineMaterial=null
      outlineSize=0
      outlineColor=<rgba 0 0 0 1>
     }
    }
    shadowOffset=<0 0>
   }
   slot {class=WidgetSwitcherSlot}
  }
  InputActionWidget {
   widgetClassPath=&CommonActionWidget
   bInherited=true
   bIsVariable=true
   parent=Overlay_1
   widget=InputActionWidget
   properties {inputActions [{dataTable=&InputActionDataTable, rowName=Select}]}
   slot {class=OverlaySlot, properties {horizontalAlignment=HAlign_Right, verticalAlignment=VAlign_Center, padding=<ltrb 0 0 24 0>}}
  }
  CanvasPanel_98 {
   widgetClassPath=&CanvasPanel
   parent=Overlay_1
   widget=CanvasPanel_98
   slot {class=OverlaySlot, properties {horizontalAlignment=HAlign_Right, verticalAlignment=VAlign_Center}}
  }
  SidebarOnboardingLikeIsland {
   widgetClassPath=&/WBP_ExpressYourSupportOnboardingTooltipRight
   bInherited=true
   bIsVariable=true
   parent=CanvasPanel_98
   widget=SidebarOnboardingLikeIsland
   slot {
    class=CanvasPanelSlot
    properties {bAutoSize=true, layoutData {alignment=<1 0.5>, anchors {maximum=<0.9 0.7>, minimum=<0.9 0.7>}, offsets=<ltrb 30 -4 0 0>}}
   }
  }
 ]
}
