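# M_UI_Font_SimpleEmojiAlpha is a translucent UI-domain material in the sidebar folder, named as a font material for emoji with adjustable alpha.
# Its graph holds only a constant of one and an Alpha scalar parameter that defaults to fully on.
# It samples no texture, and the file does not show which material outputs the two nodes drive.
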Material /Game/UI/Sidebar/Materials/M_UI_Font_SimpleEmojiAlpha
graph {
 properties {
  bUsedWithStaticMesh=false
  blendMode=BLEND_Translucent
  materialDomain=MD_UI
  pixelDepthOffsetMode=PDOM_Legacy
  usedShadingModels=Default Lit
 }
 nodes [
  Constant() {r=1}
  ScalarParameter() {defaultValue=1, parameterName=Alpha, sliderMax=1}
 ]
}
