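# Sidebar exit popup (FortSidebarExitPopup child) offering return to lobby, logout, main-menu leave and close Fortnite, plus favorite and support-creator buttons.
# Action buttons set prompt text, play a whoosh and open the switcher's confirmation view.
# Yes calls QuitGame or Logout where selected, then always ConfirmSelectedAction; No cancels.
# It also adjusts layout for touch, shows a support coach mark on activation and defines fade/slide show/hide animations.
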WidgetBlueprint /FortniteCoreUI/UI/Sidebar/SidebarExitPopupRight
animations [
 ShowPopupAnim {
  binding_count=3
  length_seconds=0.1
  object=&self:ShowPopupAnim
  sequence {bindings [
    Border_1 {
     id=#1
     possessed_class=Border
     tracks [{
       type=MovieSceneFloatTrack
       display_name=Render Opacity
       sections [{type=MovieSceneFloatSection, end_frame=2, start_frame=0, channels [None {type=MovieSceneScriptingFloatChannel, keys [[0,0],[2,1]]}]}]
      }]
    }
    VerticalBox_0 {
     id=#2
     possessed_class=VerticalBox
     tracks [{
       type=MovieScene2DTransformTrack
       sections [{
         type=MovieScene2DTransformSection
         end_frame=2
         start_frame=0
         channels [
          Translation.X {type=MovieSceneScriptingFloatChannel}
          Translation.Y {type=MovieSceneScriptingFloatChannel, keys [[0,50],[2,0]]}
          Angle {type=MovieSceneScriptingFloatChannel}
          Scale.X {type=MovieSceneScriptingFloatChannel}
          Scale.Y {type=MovieSceneScriptingFloatChannel}
          Shear.X {type=MovieSceneScriptingFloatChannel}
          Shear.Y {type=MovieSceneScriptingFloatChannel}
         ]
        }]
      }]
    }
    Text_ActivityName {
     id=#3
     possessed_class=CommonTextBlock
     tracks [{
       type=MovieSceneWidgetMaterialTrack
       display_name=Font.FontMaterial
       sections [{type=MovieSceneParameterSection, end_frame=2, start_frame=0, channels [Alpha {type=MovieSceneScriptingFloatChannel, keys [[0,0],[2,1]]}]}]
      }]
    }
   ]}
 }
 HidePopupAnim {
  binding_count=3
  length_seconds=0.1
  object=&self:HidePopupAnim
  sequence {bindings [
    Border_1 {
     id=#1
     possessed_class=Border
     tracks [{
       type=MovieSceneFloatTrack
       display_name=Render Opacity
       sections [{type=MovieSceneFloatSection, end_frame=2, start_frame=0, channels [None {type=MovieSceneScriptingFloatChannel, keys [[0,1],[2,0]]}]}]
      }]
    }
    VerticalBox_0 {
     id=#2
     possessed_class=VerticalBox
     tracks [{
       type=MovieScene2DTransformTrack
       sections [{
         type=MovieScene2DTransformSection
         end_frame=2
         start_frame=0
         channels [
          Translation.X {type=MovieSceneScriptingFloatChannel}
          Translation.Y {type=MovieSceneScriptingFloatChannel, keys [[0,0],[2,50]]}
          Angle {type=MovieSceneScriptingFloatChannel}
          Scale.X {type=MovieSceneScriptingFloatChannel}
          Scale.Y {type=MovieSceneScriptingFloatChannel}
          Shear.X {type=MovieSceneScriptingFloatChannel}
          Shear.Y {type=MovieSceneScriptingFloatChannel}
         ]
        }]
      }]
    }
    Text_ActivityName {
     id=#4
     possessed_class=CommonTextBlock
     tracks [{
       type=MovieSceneWidgetMaterialTrack
       display_name=Font.FontMaterial
       sections [{type=MovieSceneParameterSection, end_frame=2, start_frame=0, channels [Alpha {type=MovieSceneScriptingFloatChannel, keys [[0,1],[2,0]]}]}]
      }]
    }
   ]}
 }
]
graphs [
 EventGraph {nodes [
   n1 = "On Clicked (Button_LeaveExperience)"!() -> (OutputDelegate, then, Button:Common Button Base Object Reference) {is=title K2Node_ComponentBoundEvent}
   > n2 = Cast To WBP_UIKit_ButtonCTA_Base!(execute, Object=@n1.Button) -> (then, CastFailed, AsWBP UIKit Button CTA Base:WBP UIKit Button CTA Base Object Reference) {is=title K2Node_DynamicCast}
   > ResetButtonState(self:WBP UIKit Button CTA Base Object Reference=@n2)
   > "SetText (Text)"(
    self=Get Text_AreYouSure() -> (Text_AreYouSure:Common Text Object Reference)
    InText=Get LeaveExperienceConfirmationMessageText() -> (LeaveExperienceConfirmationMessageText:Text)
   )
   > SpawnSound2D!(
    execute
    Sound:Sound Base Object Reference=Get AreYouSureWhoosh() -> (AreYouSureWhoosh:Sound Base Object Reference)
    VolumeMultiplier:float=1
    PitchMultiplier:float=1
    StartTime:float=0
    ConcurrencySettings:Sound Concurrency Object Reference
    bPersistAcrossLevelTransition:bool=true
    bAutoDestroy:bool=true
   ) -> (then, ReturnValue:Audio Component Object Reference)
   > ShowConfirmationView!(execute, self:Fort Sidebar Exit Popup Object Reference, ActionToConfirm:ESidebarExitActions Enum=LeaveExperience) -> (then)
   n3 = "On Clicked (Button_Logout)"!() -> (OutputDelegate, then, Button:Common Button Base Object Reference) {is=title K2Node_ComponentBoundEvent}
   > n4 = Cast To WBP_UIKit_ButtonCTA_Base(Object=@n3.Button)
   > ResetButtonState(self=@n4)
   > "SetText (Text)"(self=Get Text_AreYouSure(), InText=Get LogoutConfirmationMessageText() -> (LogoutConfirmationMessageText:Text))
   > SpawnSound2D(Sound=Get AreYouSureWhoosh(), VolumeMultiplier=1, PitchMultiplier=1, StartTime=0, bPersistAcrossLevelTransition=true, bAutoDestroy=true)
   > ShowConfirmationView(ActionToConfirm=LogOut)
   n5 = "On Clicked (Button_Yes)"!() -> (OutputDelegate, then, Button:Common Button Base Object Reference) {is=title K2Node_ComponentBoundEvent}
   > n6 = Cast To WBP_UIKit_ButtonCTA_Base(Object=@n5.Button)
   > ResetButtonState(self=@n6)
   > n7 = Sequence()
   > n8 = Switch on ESidebarExitActions!(execute, Selection:ESidebarExitActions Enum=Get SelectedAction() -> (SelectedAction:ESidebarExitActions Enum=None)) -> (None:Exec, LeaveExperience:Exec, AthenaLeaveAction:Exec, LogOut:Exec, QuitGame:Exec) {is=title K2Node_SwitchEnum}
   QuitGame(
    execute=@n8.QuitGame
    self:Fort Global UIContext Object Reference=Get FortGlobalUIContext!() -> (ReturnValue:Fort Global UIContext Object Reference) {is=title K2Node_GetBlueprintContextEx}
   )
   Logout(execute=@n8.LogOut, self:Fort Global UIContext Object Reference=Get FortGlobalUIContext())
   ConfirmSelectedAction(execute=@n7.then_1, self:Fort Sidebar Exit Popup Object Reference)
   n9 = "On Hovered (Button_Yes)"!() -> (OutputDelegate, then, Button:Common Button Base Object Reference) {is=title K2Node_ComponentBoundEvent}
   > SetFocus(self=@n9.Button)
   n10 = "On Hovered (Button_No)"!() -> (OutputDelegate, then, Button:Common Button Base Object Reference) {is=title K2Node_ComponentBoundEvent}
   > SetFocus(self=@n10.Button)
   n11 = "On Clicked (Button_ExitGame)"!() -> (OutputDelegate, then, Button:Common Button Base Object Reference) {is=title K2Node_ComponentBoundEvent}
   > n12 = Cast To WBP_UIKit_ButtonCTA_Base(Object=@n11.Button)
   > ResetButtonState(self=@n12)
   > "SetText (Text)"(self=Get Text_AreYouSure(), InText=Get ExitConfirmationMessageText() -> (ExitConfirmationMessageText:Text))
   > SpawnSound2D(Sound=Get AreYouSureWhoosh(), VolumeMultiplier=1, PitchMultiplier=1, StartTime=0, bPersistAcrossLevelTransition=true, bAutoDestroy=true)
   > ShowConfirmationView(ActionToConfirm=QuitGame)
   n13 = "On Clicked (Button_No)"!() -> (OutputDelegate, then, Button:Common Button Base Object Reference) {is=title K2Node_ComponentBoundEvent}
   > n14 = Cast To WBP_UIKit_ButtonCTA_Base(Object=@n13.Button)
   > ResetButtonState(self=@n14)
   > CancelConfirmationView(self:Fort Sidebar Exit Popup Object Reference)
   n15 = OnButtonBaseClicked_Event!() -> (OutputDelegate, then, Button:Common Button Base Object Reference) {is=title K2Node_CustomEvent}
   > n16 = Cast To WBP_UIKit_ButtonCTA_Base(Object=@n15.Button)
   > ResetButtonState(self=@n16)
   > "SetText (Text)"(self=Get Text_AreYouSure(), InText=Get AthenaConfirmationMessageText() -> (AthenaConfirmationMessageText:Text))
   > SpawnSound2D(Sound=Get AreYouSureWhoosh(), VolumeMultiplier=1, PitchMultiplier=1, StartTime=0, bPersistAcrossLevelTransition=false, bAutoDestroy=true)
   > ShowConfirmationView(ActionToConfirm=AthenaLeaveAction)
   Event Construct()
   > Assign On Clicked!(
    execute:Exec
    self:Common Button Base Object Reference=Get ButtonInternal!(self:Fort Main Menu Leave Button Wrapper Object Reference=Get Button_AthenaMainMenuLeave() -> (Button_AthenaMainMenuLeave:Fort Main Menu Leave Button Wrapper Object Reference)) -> (ButtonInternal:Common Button Base Object Reference)
    Delegate:"Delegate (by ref)"=@n15
   ) -> (then:Exec) {is=title K2Node_AssignDelegate}
   > Bind Event to On Input Method Changed(self=Get CommonInputSubsystem(), Delegate=Create Event())
   > OnInputMethodChanged!(execute, self, NewInputType:ECommonInputType Enum=GetCurrentInputType(self=Get CommonInputSubsystem())) -> (then)
   > SetVisibility(self=Get Overlay_Favorite() -> (Overlay_Favorite:Overlay Object Reference), InVisibility=Visible)
   > SetVisibility(self=Get Overlay_SAC() -> (Overlay_SAC:Overlay Object Reference), InVisibility=Visible)
   Event On Activated()
   n17 = Get WBP_ExpressYourSupportRow() -> (WBP_ExpressYourSupportRow:WBP Express Your Support Row Object Reference)
   n18 = Get WBP_Sidebar_ExpressYourSupport_CoachMark!(self:WBP Express Your Support Row Object Reference=@n17) -> (WBP_Sidebar_ExpressYourSupport_CoachMark:WBP Sidebar Express Your Support Coach Mark Object Reference)
   > Branch(Condition=AND Boolean(A=IsVisible(self=@n17), B=ShouldShowOnboarding!(self:Fort Sidebar Onboard Tooltip Widget Object Reference=@n18) -> (ReturnValue:bool)))
   > Delay(Duration=0.2)
   > ActivateWidget(self=@n18)
   n19 = Event UpdateNoEnabledButtonsText!() -> (OutputDelegate, then, bAllowLogout:bool, bAllowQuit:bool) {is=title K2Node_Event}
   n20 = Get LogoutAllowedNoButtonsText() -> (LogoutAllowedNoButtonsText:Text)
   n21 = Get LogoutAndQuitNotAllowedNoButtonsText() -> (LogoutAndQuitNotAllowedNoButtonsText:Text)
   > SetText(
    self=Get Text_NoEnabledButtons() -> (Text_NoEnabledButtons:Common Rich Text Block Object Reference)
    InText=Select(
     Option 0:Text=Select(Option 0=@n21, Option 1=@n20, Index=@n19.bAllowLogout)
     Option 1:Text=Get QuitAllowedNoButtonsText() -> (QuitAllowedNoButtonsText:Text)
     Index=@n19.bAllowQuit
    ) -> (ReturnValue:Text)
   )
  ]}
 OnInputMethodChanged {nodes [
   n1 = OnInputMethodChanged!() -> (then, NewInputType:ECommonInputType Enum) {is=title K2Node_FunctionEntry}
   > RefreshButtonsVisual!(
    execute
    self
    IsTouch:bool=Select!(MouseAndKeyboard:bool=false, Gamepad:bool=false, Touch:bool=true, Count:bool=false, Index:ECommonInputType Enum=@n1) -> (ReturnValue:bool)
   ) -> (then)
  ]}
 RefreshButtonsVisual {nodes [
   n1 = RefreshButtonsVisual!() -> (then, IsTouch:bool=~) {is=title K2Node_FunctionEntry}
   > n2 = Branch(Condition=@n1)
   > SetVisibility(self=Get Border_SpacingMobile() -> (Border_SpacingMobile:Border Object Reference), InVisibility=Visible)
   n3 = Get Button_SupportCreator() -> (Button_SupportCreator:Fort Express Your Support Thumbs Up Button Object Reference)
   > n4 = Cast To WBP_ExpressYourSupportThumbsUpButtonRight!(execute, Object=@n3) -> (then, CastFailed, AsWBP Express Your Support Thumbs Up Button Right:WBP Express Your Support Thumbs Up Button Right Object Reference) {is=title K2Node_DynamicCast}
   > n5 = SetButtonHeight!(execute, self:WBP Express Your Support Thumbs Up Button Right Object Reference=@n4, IsTouch:bool=Get IsTouch() -> (IsTouch:bool)) -> (then)
   n6 = Get Button_Favorite() -> (Button_Favorite:Fort Express Your Support Favorite Button Object Reference)
   n7 = Cast To WBP_ExpressYourSupportFavoriteButton!(execute=[@n5.then, @n4.CastFailed], Object=@n6) -> (then, CastFailed, AsWBP Express Your Support Favorite Button:WBP Express Your Support Favorite Button Object Reference) {is=title K2Node_DynamicCast}
   > SetButtonHeight(self:WBP Express Your Support Favorite Button Object Reference=@n7, IsTouch=Get IsTouch())
   SetVisibility(execute=@n2.else, self=Get Border_SpacingMobile(), InVisibility=Collapsed)
   > n8 = Cast To WBP_ExpressYourSupportThumbsUpButtonRight(Object=@n3)
   > n9 = SetButtonHeight(self:WBP Express Your Support Thumbs Up Button Right Object Reference=@n8, IsTouch=Get IsTouch())
   n10 = Cast To WBP_ExpressYourSupportFavoriteButton(execute=[@n9.then, @n8.CastFailed], Object=@n6)
   > SetButtonHeight(self:WBP Express Your Support Favorite Button Object Reference=@n10, IsTouch=Get IsTouch())
   Get PanelWidget_ExpressYourSupportButtons() -> (PanelWidget_ExpressYourSupportButtons:Panel Widget Object Reference)
  ]}
]
variables [
 AreYouSureWhoosh {type=SoundBase, default=/Game/Sounds/Fort_UI_Sounds/Gameplay/PostGame/Fortnite_UI_PostGame_AreYouSure_Whoosh_Cue}
 ExitConfirmationMessageText {type=string, default=Are you sure you want to exit Fortnite?}
 Is A {type=boolean}
 LeaveExperienceConfirmationMessageText {type=string, default=Are you sure you want to return to lobby?}
 LogoutAllowedNoButtonsText {type=string, default="You can use this menu to return to the lobby, leave your party, and log out"}
 LogoutAndQuitNotAllowedNoButtonsText {type=string, default=You can use this menu to return to the Lobby and leave your party}
 LogoutConfirmationMessageText {type=string, default=Are you sure you want to log out of Fortnite?}
 QuitAllowedNoButtonsText {type=string, default="You can use this menu to return to the lobby, leave your party, and exit Fortnite"}
]
widgets {
 info {inheritedWidgetCount=17, parentClass=/Script/FortniteUI.FortSidebarExitPopup, rootWidgetClass=&Border, widgetCount=35}
 widgets [
  Border_1 {
   widgetClassPath=&Border
   widget=Border_1
   properties {
    background {
     bIsDynamicallyLoaded=false
     drawAs=Box
     imageType=NoImage
     mirroring=NoMirror
     resourceName=None
     resourceObject=null
     tiling=NoTile
     imageSize=<32 32>
     margin=<ltrb 0 0 0 0>
     outlineSettings {
      bUseBrushTransparency=false
      roundingType=HalfHeightRadius
      width=0
      color {colorUseRule=UseColor_Specified, specifiedColor=<rgba 0 0 0 0>}
      cornerRadii=<0 0 0 0>
     }
     tintColor {colorUseRule=UseColor_Specified, specifiedColor=<rgba 1 1 1 1>}
     uVRegion {bIsValid=false, max=<0 0>, min=<0 0>}
    }
    brushColor=<rgba 0 0 0 0.73>
    padding=<ltrb 0 0 0 0>
   }
  }
  VerticalBox_0 {widgetClassPath=&VerticalBox, parent=Border_1, widget=VerticalBox_0, slot {class=BorderSlot, properties {padding=<ltrb 0 0 0 0>}}}
  Button_BackBoard {
   widgetClassPath=/Game/Athena/UI/Matchmaking/SimpleCommonButton
   bInherited=true
   bIsVariable=true
   parent=VerticalBox_0
   widget=Button_BackBoard
   properties {
    clickMethod=MouseDown
    onButtonBaseClicked=(&self:WidgetTree.Button_BackBoard.HandleOnButtonClicked)
    onButtonBaseDoubleClicked=(&self:WidgetTree.Button_BackBoard.HandleOnButtonDoubleClicked)
    onButtonBaseHovered=(&self:WidgetTree.Button_BackBoard.HandleOnButtonHovered)
    onButtonBaseUnhovered=(&self:WidgetTree.Button_BackBoard.HandleOnButtonUnhovered)
    onSelectedChangedBase=(&self:WidgetTree.Button_BackBoard.HandleOnSelectedChanged)
    pressMethod=ButtonPress
    style=/Game/UI/Sidebar/Sidebar_ButtonStyle_TransparentList.Sidebar_ButtonStyle_TransparentList_C
    touchMethod=Down
    pressedSlateSoundOverride {resourceObject=/Game/Sounds/Fort_UI_Sounds/FrontEnd/New_FrontEnd/Fortnite_UI_MainMenu_Generic_01_Cue}
   }
   slot {class=VerticalBoxSlot, properties {size {sizeRule=Fill, value=1}}}
  }
  Border_241 {
   widgetClassPath=&Border
   parent=VerticalBox_0
   widget=Border_241
   properties {
    visibility=SelfHitTestInvisible
    background {
     bIsDynamicallyLoaded=false
     drawAs=Box
     imageType=NoImage
     mirroring=NoMirror
     resourceName=None
     resourceObject=null
     tiling=NoTile
     imageSize=<32 32>
     margin=<ltrb 0 0 0 0>
     outlineSettings {
      bUseBrushTransparency=false
      roundingType=HalfHeightRadius
      width=0
      color {colorUseRule=UseColor_Specified, specifiedColor=<rgba 0 0 0 0>}
      cornerRadii=<0 0 0 0>
     }
     tintColor {colorUseRule=UseColor_Specified, specifiedColor=<rgba 1 1 1 1>}
     uVRegion {bIsValid=false, max=<0 0>, min=<0 0>}
    }
    brushColor=<rgba 0.1812 0.016 0.0194 1>
    padding=<ltrb 27 12 27 12>
   }
   slot {class=VerticalBoxSlot}
  }
  Text_Header {
   widgetClassPath=&CommonTextBlock
   parent=Border_241
   widget=Text_Header
   properties {
    autoWrapText=true
    mobileFontSizeMultiplier=1.3
    text=EXIT
    textTransformPolicy=ToUpper
    font {
     bForceMonospaced=false
     bMaterialIsStencil=false
     fontMaterial=null
     fontObject=&BurbankSmall
     letterSpacing=0
     monospacedWidth=1
     size=17
     skewAmount=0
     typefaceFontName=Black
     outlineSettings {
      bApplyOutlineToDropShadows=false
      bMiteredCorners=false
      bSeparateFillAlpha=false
      outlineMaterial=null
      outlineSize=0
      outlineColor=<rgba 0 0 0 1>
     }
    }
    renderTransform {angle=0, scale=<1 1>, shear=<-8 0>, translation=<0 0>}
    shadowOffset=<0 0>
   }
   slot {class=BorderSlot, properties {padding=<ltrb 27 12 27 12>}}
  }
  Border_0 {
   widgetClassPath=&Border
   parent=VerticalBox_0
   widget=Border_0
   properties {
    visibility=SelfHitTestInvisible
    background {
     bIsDynamicallyLoaded=false
     drawAs=Image
     imageType=NoImage
     mirroring=NoMirror
     resourceName=None
     resourceObject=/Game/UI/Sidebar/Materials/MI_UI_BPToast_ExitPanel
     tiling=NoTile
     imageSize=<32 32>
     margin=<ltrb 0 0 0 0>
     outlineSettings {
      bUseBrushTransparency=false
      roundingType=HalfHeightRadius
      width=0
      color {colorUseRule=UseColor_Specified, specifiedColor=<rgba 0 0 0 0>}
      cornerRadii=<0 0 0 0>
     }
     tintColor {colorUseRule=UseColor_Specified, specifiedColor=<rgba 1 1 1 1>}
     uVRegion {bIsValid=false, max=<0 0>, min=<0 0>}
    }
    padding=<ltrb 0 0 0 0>
   }
   slot {class=VerticalBoxSlot}
  }
  Overlay_0 {widgetClassPath=&Overlay, parent=Border_0, widget=Overlay_0, slot {class=BorderSlot, properties {padding=<ltrb 0 0 0 0>}}}
  WBP_CaptureForPostBufferUpdate {
   widgetClassPath=&WBP_CaptureForPostBufferUpdate
   bIsVariable=true
   parent=Overlay_0
   slot {class=OverlaySlot, properties {horizontalAlignment=HAlign_Fill, verticalAlignment=VAlign_Fill}}
  }
  WidgetSwitcher {
   widgetClassPath=&CommonActivatableWidgetSwitcher
   bInherited=true
   bIsVariable=true
   parent=Overlay_0
   widget=WidgetSwitcher
   properties {transitionDuration=0.15}
   slot {class=OverlaySlot, properties {horizontalAlignment=HAlign_Fill, verticalAlignment=VAlign_Fill}}
  }
  MenuButtons {widgetClassPath=&VerticalBox, parent=WidgetSwitcher, widget=MenuButtons, slot {class=WidgetSwitcherSlot}}
  PanelWidget_ExpressYourSupportButtons {
   widgetClassPath=&VerticalBox
   bInherited=true
   bIsVariable=true
   parent=MenuButtons
   widget=PanelWidget_ExpressYourSupportButtons
   properties {visibility=Collapsed}
   slot {class=VerticalBoxSlot}
  }
  SizeBox_0 {
   widgetClassPath=&SizeBox
   parent=PanelWidget_ExpressYourSupportButtons
   widget=SizeBox_0
   properties {bOverride_HeightOverride=true, heightOverride=72}
   slot {class=VerticalBoxSlot, properties {padding=<ltrb 4 2 4 2>}}
  }
  Text_ActivityName {
   widgetClassPath=&CommonTextBlock
   bInherited=true
   bIsVariable=true
   parent=SizeBox_0
   widget=Text_ActivityName
   properties {
    bAutoCollapseWithEmptyText=true
    clipping=ClipToBounds
    justification=Center
    mobileFontSizeMultiplier=1.2
    scrollStyle=&TextScrollStyle-Base_C
    text="{Game Mode}"
    textOverflowPolicy=Ellipsis
    textTransformPolicy=ToUpper
    font {
     bForceMonospaced=false
     bMaterialIsStencil=false
     fontMaterial=/Game/UI/Sidebar/Materials/M_UI_Font_SimpleEmojiAlpha
     fontObject=&BurbankSmall
     letterSpacing=0
     monospacedWidth=1
     size=18
     skewAmount=0
     typefaceFontName=Black
     outlineSettings {
      bApplyOutlineToDropShadows=false
      bMiteredCorners=false
      bSeparateFillAlpha=false
      outlineMaterial=null
      outlineSize=0
      outlineColor=<rgba 0 0 0 1>
     }
    }
    shadowOffset=<0 0>
   }
   slot {class=SizeBoxSlot, properties {verticalAlignment=VAlign_Center, padding=<ltrb 20 0 20 0>}}
  }
  Overlay_Favorite {
   widgetClassPath=&Overlay
   bIsVariable=true
   parent=PanelWidget_ExpressYourSupportButtons
   widget=Overlay_Favorite
   slot {class=VerticalBoxSlot}
  }
  Button_Favorite {
   widgetClassPath=/Game/Athena/UI/InGame/Creative/WBP_ExpressYourSupportFavoriteButton
   bInherited=true
   bIsVariable=true
   parent=Overlay_Favorite
   widget=Button_Favorite
   slot {class=OverlaySlot, properties {horizontalAlignment=HAlign_Fill, verticalAlignment=VAlign_Fill}}
  }
  Border_SpacingMobile {
   widgetClassPath=&Border
   bIsVariable=true
   parent=PanelWidget_ExpressYourSupportButtons
   widget=Border_SpacingMobile
   properties {
    visibility=Collapsed
    background {
     bIsDynamicallyLoaded=false
     drawAs=NoDrawType
     imageType=NoImage
     mirroring=NoMirror
     resourceName=None
     resourceObject=null
     tiling=NoTile
     imageSize=<32 32>
     margin=<ltrb 0 0 0 0>
     outlineSettings {
      bUseBrushTransparency=false
      roundingType=HalfHeightRadius
      width=0
      color {colorUseRule=UseColor_Specified, specifiedColor=<rgba 0 0 0 0>}
      cornerRadii=<0 0 0 0>
     }
     tintColor {colorUseRule=UseColor_Specified, specifiedColor=<rgba 1 1 1 1>}
     uVRegion {bIsValid=false, max=<0 0>, min=<0 0>}
    }
    padding=<ltrb 0 0 0 15>
   }
   slot {class=VerticalBoxSlot}
  }
  Overlay_SAC {
   widgetClassPath=&Overlay
   bIsVariable=true
   parent=PanelWidget_ExpressYourSupportButtons
   widget=Overlay_SAC
   slot {class=VerticalBoxSlot}
  }
  Button_SupportCreator {
   widgetClassPath=/Game/Athena/UI/InGame/Creative/WBP_ExpressYourSupportThumbsUpButtonRight
   bInherited=true
   bIsVariable=true
   parent=Overlay_SAC
   widget=Button_SupportCreator
   properties {holdData=&CommonUI_HoldData_Base_C}
   slot {class=OverlaySlot, properties {horizontalAlignment=HAlign_Fill, verticalAlignment=VAlign_Fill}}
  }
  Overlay_ExpressSupportRow {
   widgetClassPath=&Overlay
   bIsVariable=true
   parent=PanelWidget_ExpressYourSupportButtons
   widget=Overlay_ExpressSupportRow
   slot {class=VerticalBoxSlot}
  }
  WBP_ExpressYourSupportRow {
   widgetClassPath=/Game/Athena/UI/InGame/Creative/WBP_ExpressYourSupportRow
   bIsVariable=true
   parent=Overlay_ExpressSupportRow
   widget=WBP_ExpressYourSupportRow
   properties {bIsFocusable=true}
   slot {class=OverlaySlot, properties {horizontalAlignment=HAlign_Fill, verticalAlignment=VAlign_Fill}}
  }
  PanelWidget_MainMenuButtons {
   widgetClassPath=&VerticalBox
   bInherited=true
   bIsVariable=true
   parent=MenuButtons
   widget=PanelWidget_MainMenuButtons
   properties {flowDirectionPreference=Culture}
   slot {class=VerticalBoxSlot, properties {verticalAlignment=VAlign_Bottom, padding=<ltrb 36 27 36 27>}}
  }
  Text_NoEnabledButtons {
   widgetClassPath=&CommonRichTextBlock
   bInherited=true
   bIsVariable=true
   parent=PanelWidget_MainMenuButtons
   widget=Text_NoEnabledButtons
   properties {
    autoWrapText=true
    bOverrideDefaultStyle=true
    justification=Center
    mobileTextBlockScale=1.3
    text="You can use this menu to return to the Lobby, leave your party, and exit Fortnite"
    visibility=Collapsed
    defaultTextStyleOverride {
     overflowPolicy=Clip
     transformPolicy=None
     colorAndOpacity {colorUseRule=UseColor_Specified, specifiedColor=<rgba 1 1 1 1>}
     font {
      bForceMonospaced=false
      bMaterialIsStencil=false
      fontMaterial=null
      fontObject=&BurbankSmall
      letterSpacing=0
      monospacedWidth=1
      size=16
      skewAmount=0
      typefaceFontName=Black
      outlineSettings {
       bApplyOutlineToDropShadows=false
       bMiteredCorners=false
       bSeparateFillAlpha=false
       outlineMaterial=null
       outlineSize=0
       outlineColor=<rgba 0 0 0 1>
      }
     }
     highlightColor {colorUseRule=UseColor_Specified, specifiedColor=<rgba 0 0 0 1>}
     highlightShape {
      bIsDynamicallyLoaded=false
      drawAs=Image
      imageType=NoImage
      mirroring=NoMirror
      resourceName=None
      resourceObject=null
      tiling=NoTile
      imageSize=<32 32>
      margin=<ltrb 0 0 0 0>
      outlineSettings {
       bUseBrushTransparency=false
       roundingType=HalfHeightRadius
       width=0
       color {colorUseRule=UseColor_Specified, specifiedColor=<rgba 0 0 0 0>}
       cornerRadii=<0 0 0 0>
      }
      tintColor {colorUseRule=UseColor_Specified, specifiedColor=<rgba 1 1 1 1>}
      uVRegion {bIsValid=false, max=<0 0>, min=<0 0>}
     }
     selectedBackgroundColor {colorUseRule=UseColor_Foreground, specifiedColor=<rgba 1 0 1 1>}
     shadowColorAndOpacity=<rgba 0 0 0 1>
     shadowOffset=<0 0>
     strikeBrush {
      bIsDynamicallyLoaded=false
      drawAs=Image
      imageType=NoImage
      mirroring=NoMirror
      resourceName=None
      resourceObject=null
      tiling=NoTile
      imageSize=<32 32>
      margin=<ltrb 0 0 0 0>
      outlineSettings {
       bUseBrushTransparency=false
       roundingType=HalfHeightRadius
       width=0
       color {colorUseRule=UseColor_Specified, specifiedColor=<rgba 0 0 0 0>}
       cornerRadii=<0 0 0 0>
      }
      tintColor {colorUseRule=UseColor_Specified, specifiedColor=<rgba 1 1 1 1>}
      uVRegion {bIsValid=false, max=<0 0>, min=<0 0>}
     }
     underlineBrush {
      bIsDynamicallyLoaded=false
      drawAs=Image
      imageType=NoImage
      mirroring=NoMirror
      resourceName=None
      resourceObject=null
      tiling=NoTile
      imageSize=<32 32>
      margin=<ltrb 0 0 0 0>
      outlineSettings {
       bUseBrushTransparency=false
       roundingType=HalfHeightRadius
       width=0
       color {colorUseRule=UseColor_Specified, specifiedColor=<rgba 0 0 0 0>}
       cornerRadii=<0 0 0 0>
      }
      tintColor {colorUseRule=UseColor_Specified, specifiedColor=<rgba 1 1 1 1>}
      uVRegion {bIsValid=false, max=<0 0>, min=<0 0>}
     }
    }
    renderTransform {angle=0, scale=<1 1>, shear=<-6 0>, translation=<0 0>}
   }
   slot {class=VerticalBoxSlot, properties {padding=<ltrb 0 0 0 16>}}
  }
  Button_LeaveExperience {
   widgetClassPath=&WBP_UIKit_Button_Regular
   bInherited=true
   bIsVariable=true
   parent=PanelWidget_MainMenuButtons
   widget=Button_LeaveExperience
   properties {text=RETURN TO LOBBY}
   slot {class=VerticalBoxSlot, properties {size {sizeRule=Fill, value=1}}}
  }
  Button_Parsley {
   widgetClassPath=&WBP_UIKit_Button_Regular
   bInherited=true
   bIsVariable=true
   parent=PanelWidget_MainMenuButtons
   widget=Button_Parsley
   properties {text=Sleep Mode, visibility=Collapsed}
   slot {class=VerticalBoxSlot, properties {padding=<ltrb 0 16 0 0>, size {sizeRule=Fill, value=1}}}
  }
  Button_Logout {
   widgetClassPath=&WBP_UIKit_Button_Regular
   bInherited=true
   bIsVariable=true
   parent=PanelWidget_MainMenuButtons
   widget=Button_Logout
   properties {text=LOG OUT}
   slot {class=VerticalBoxSlot, properties {padding=<ltrb 0 16 0 0>, size {sizeRule=Fill, value=1}}}
  }
  Button_AthenaMainMenuLeave {
   widgetClassPath=/FortUILibrary/UIKit/Buttons/SpecialButtons/WBP_UIKit_SideBarLeaveButton
   bInherited=true
   bIsVariable=true
   parent=PanelWidget_MainMenuButtons
   widget=Button_AthenaMainMenuLeave
   slot {class=VerticalBoxSlot, properties {padding=<ltrb 0 16 0 0>, size {sizeRule=Fill, value=1}}}
  }
  Button_ExitGame {
   widgetClassPath=&WBP_UIKit_Button_Regular
   bInherited=true
   bIsVariable=true
   parent=PanelWidget_MainMenuButtons
   widget=Button_ExitGame
   properties {text=CLOSE FORTNITE}
   slot {class=VerticalBoxSlot, properties {padding=<ltrb 0 16 0 0>, size {sizeRule=Fill, value=1}}}
  }
  ConfirmationPrompt {
   widgetClassPath=&VerticalBox
   parent=WidgetSwitcher
   widget=ConfirmationPrompt
   slot {class=WidgetSwitcherSlot, properties {verticalAlignment=VAlign_Bottom, padding=<ltrb 36 27 36 27>}}
  }
  Widget_RankedExitWarning {
   widgetClassPath=&VerticalBox
   bInherited=true
   bIsVariable=true
   parent=ConfirmationPrompt
   widget=Widget_RankedExitWarning
   properties {visibility=Collapsed}
   slot {class=VerticalBoxSlot, properties {padding=<ltrb 0 30 0 30>}}
  }
  Text_FeralRankedAlertTitle {
   widgetClassPath=&CommonTextBlock
   parent=Widget_RankedExitWarning
   widget=Text_FeralRankedAlertTitle
   properties {
    autoWrapText=true
    justification=Center
    mobileFontSizeMultiplier=1.3
    text=Exit penalty
    textTransformPolicy=ToUpper
    font {
     bForceMonospaced=false
     bMaterialIsStencil=false
     fontMaterial=null
     fontObject=&BurbankSmall
     letterSpacing=0
     monospacedWidth=1
     size=24
     skewAmount=0
     typefaceFontName=Black
     outlineSettings {
      bApplyOutlineToDropShadows=false
      bMiteredCorners=false
      bSeparateFillAlpha=false
      outlineMaterial=null
      outlineSize=0
      outlineColor=<rgba 0 0 0 1>
     }
    }
    renderTransform {angle=0, scale=<1 1>, shear=<-6 0>, translation=<0 0>}
    shadowOffset=<0 0>
   }
   slot {class=VerticalBoxSlot, properties {padding=<ltrb 0 20 0 20>}}
  }
  Text_FeralRankedAlertDescription {
   widgetClassPath=&CommonTextBlock
   parent=Widget_RankedExitWarning
   widget=Text_FeralRankedAlertDescription
   properties {
    autoWrapText=true
    justification=Center
    mobileFontSizeMultiplier=1.3
    text=Leaving the match early will result in a Matchmaking restriction for Ranked in this game mode
    textTransformPolicy=ToUpper
    font {
     bForceMonospaced=false
     bMaterialIsStencil=false
     fontMaterial=null
     fontObject=&BurbankSmall
     letterSpacing=0
     monospacedWidth=1
     size=16
     skewAmount=0
     typefaceFontName=Black
     outlineSettings {
      bApplyOutlineToDropShadows=false
      bMiteredCorners=false
      bSeparateFillAlpha=false
      outlineMaterial=null
      outlineSize=0
      outlineColor=<rgba 0 0 0 1>
     }
    }
    shadowOffset=<0 0>
   }
   slot {class=VerticalBoxSlot, properties {padding=<ltrb 20 20 20 20>}}
  }
  Text_AreYouSure {
   widgetClassPath=&CommonTextBlock
   bIsVariable=true
   parent=ConfirmationPrompt
   widget=Text_AreYouSure
   properties {
    autoWrapText=true
    justification=Center
    mobileFontSizeMultiplier=1.3
    text=Are you sure you want to?
    font {
     bForceMonospaced=false
     bMaterialIsStencil=false
     fontMaterial=null
     fontObject=&BurbankSmall
     letterSpacing=0
     monospacedWidth=1
     size=16
     skewAmount=0
     typefaceFontName=Black
     outlineSettings {
      bApplyOutlineToDropShadows=false
      bMiteredCorners=false
      bSeparateFillAlpha=false
      outlineMaterial=null
      outlineSize=0
      outlineColor=<rgba 0 0 0 1>
     }
    }
    renderTransform {angle=0, scale=<1 1>, shear=<-6 0>, translation=<0 0>}
    shadowOffset=<0 0>
   }
   slot {class=VerticalBoxSlot, properties {padding=<ltrb 0 0 0 16>}}
  }
  ConfirmationButtons {
   widgetClassPath=&HorizontalBox
   bInherited=true
   bIsVariable=true
   parent=ConfirmationPrompt
   widget=ConfirmationButtons
   properties {navigation=&self:WidgetTree.ConfirmationButtons.WidgetNavigation_0}
   slot {class=VerticalBoxSlot}
  }
  Button_Yes {
   widgetClassPath=&WBP_UIKit_Button_Regular
   bInherited=true
   bIsVariable=true
   parent=ConfirmationButtons
   widget=Button_Yes
   properties {text=YES}
   slot {class=HorizontalBoxSlot, properties {padding=<ltrb 0 0 16 0>, size {sizeRule=Fill, value=1}}}
  }
  Button_No {
   widgetClassPath=&WBP_UIKit_Button_Regular
   bInherited=true
   bIsVariable=true
   parent=ConfirmationButtons
   widget=Button_No
   properties {text=NO}
   slot {class=HorizontalBoxSlot, properties {padding=<ltrb 16 0 0 0>, size {sizeRule=Fill, value=1}}}
  }
 ]
}
