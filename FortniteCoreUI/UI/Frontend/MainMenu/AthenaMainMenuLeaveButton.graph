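# Leave button widget for the Athena main menu, built on FortMainMenuLeaveButtonWrapper with a WBP_MainMenuButton as its inherited ButtonInternal.
# PreConstruct casts and caches that button, then applies the stored Text label.
# Text change events update the stored label and the button.
# Timer text changes show label and timer together, and a Collapsed timer visibility restores the plain label.
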WidgetBlueprint /FortniteCoreUI/UI/Frontend/MainMenu/AthenaMainMenuLeaveButton
graphs [EventGraph {nodes [
   n1 = Event OnTimerTextChanged!() -> (OutputDelegate, then, Text:"Text (by ref)") {is=title K2Node_Event}
   > Set Text!(
    execute
    self:WBP Main Menu Button Object Reference=Get CommonButtonInternal() -> (CommonButtonInternal:WBP Main Menu Button Object Reference)
    Text:Text=Format Text!(Format="LOCTEXT(\"{Button} {Timer}\")", Button:Text=Get Text!() -> (Text), Timer:"Text (by ref)"=@n1.Text) -> (Result)
   ) -> (then) {is=title K2Node_CallFunction}
   Event PreConstruct()
   > n2 = Cast To WBP_MainMenuButton!(execute, Object=Get ButtonInternal() -> (ButtonInternal:Common Button Base Object Reference)) -> (then, CastFailed, AsWBP Main Menu Button:WBP Main Menu Button Object Reference) {is=title K2Node_DynamicCast}
   > n3 = Set CommonButtonInternal(CommonButtonInternal:WBP Main Menu Button Object Reference=@n2) -> (Output_Get:WBP Main Menu Button Object Reference)
   > Set Text(self=@n3, Text=Get Text())
   n4 = Event OnTimerVisibilityChanged!() -> (OutputDelegate, then, InVisibility:ESlateVisibility Enum=Visible) {is=title K2Node_Event}
   > n5 = Switch on ESlateVisibility!(execute, Selection:ESlateVisibility Enum=@n4.InVisibility) -> (Visible:Exec, Collapsed:Exec, Hidden:Exec, HitTestInvisible:Exec, SelfHitTestInvisible:Exec) {is=title K2Node_SwitchEnum}
   Set Text(execute=@n5.Collapsed, self=Get CommonButtonInternal(), Text=Get Text())
   n6 = Event OnTextChanged!() -> (OutputDelegate, then, Text:"Text (by ref)") {is=title K2Node_Event}
   > n7 = Set Text!(execute, Text=@n6.Text) -> (then, Output_Get) {is=title K2Node_VariableSet}
   > Set Text(self=Get CommonButtonInternal(), Text=@n7) {is=title K2Node_CallFunction}
  ]}]
variables [
 CommonButtonInternal {type=WBP_MainMenuButton_C, default=null}
 Text {type=string, default=AthenaMainMenuLeaveButton}
]
widgets {
 info {
  inheritedWidgetCount=1
  parentClass=/Script/FortniteUI.FortMainMenuLeaveButtonWrapper
  rootWidgetClass=/Game/UI/Frontend/MainMenu/WBP_MainMenuButton.WBP_MainMenuButton_C
  widgetCount=1
 }
 widgets [ButtonInternal {widgetClassPath=/Game/UI/Frontend/MainMenu/WBP_MainMenuButton, bInherited=true, bIsVariable=true, widget=ButtonInternal}]
}
