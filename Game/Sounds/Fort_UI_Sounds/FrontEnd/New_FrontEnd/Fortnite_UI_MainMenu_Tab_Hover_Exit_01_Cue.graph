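# Sound cue for the hover-exit sound on front-end main menu tabs, in the Fort_UI sound class.
# A single button hover-click wave plays through a Modulator that slightly randomizes pitch, with the cue lowering overall pitch and boosting volume.
# A front-end hover-tick concurrency set is assigned, and its name and the cue's override values both indicate one voice, stopping the oldest.
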SoundCue /Game/Sounds/Fort_UI_Sounds/FrontEnd/New_FrontEnd/Fortnite_UI_MainMenu_Tab_Hover_Exit_01_Cue
root=Modulator_0
nodes [
 {id=Modulator_0, children [WavePlayer_2], properties {pitchMax=1.03, pitchMin=0.97, volumeMax=1, volumeMin=1}}
 {id=WavePlayer_2, properties {soundWaveAssetPtr=&/UI_FN_Button_Hover_Click_01}}
]
properties {
 maxDistance=2097152
 pitchMultiplier=0.85
 soundClassObject=&Fort_UI
 volumeMultiplier=1.4
 concurrencyOverrides {maxCount=1, resolutionRule=StopOldest}
 concurrencySet [/Game/Sounds/Concurrency_Rules/FrontEnd_UI_Menu_Max_01_Stop_Oldest_Hover_Tick]
}
