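# Sidebar Exit panel on FortSidebarPanelExit, with Return to Lobby, Athena leave, Log Out and Close Fortnite buttons.
# Log Out and Close Fortnite open async confirmation dialogs, then call Logout or QuitGame on FortGlobalUIContext when confirmed.
# UpdateNoEnabledButtonsText fills the initially collapsed no-buttons message from three string variables, with quit permission taking priority over logout.
# Leave button handling and message visibility have no logic in this graph.
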WidgetBlueprint /FortniteCoreUI/UI/Sidebar/SidebarPanelExit
graphs [EventGraph {nodes [
   "On Button Clicked (Button_ExitGame)"!() -> (OutputDelegate, then, Button:Common Button Legacy Object Reference) {is=title K2Node_ComponentBoundEvent}
   > n1 = ShowSimpleConfirmationDialog_NUI!(
    execute:Exec
    Icon:Slate Brush Structure
    Title:Text="LOCTEXT(\"Quit\")"
    Message:Text="LOCTEXT(\"Are you sure you want to exit Fortnite?\")"
    bShowConfirm:bool=true
    bShowDecline:bool=true
    AdditionalContent:Widget Object Reference
    LeftAdditionalContent:Widget Object Reference
    TargetPlayer:Local Player Object Reference
   ) -> (then:Exec, DialogResult:Exec, Result:EFortDialogResult Enum, ResultName:Name) {is=title K2Node_AsyncAction}
   Switch on EFortDialogResult!(execute=@n1.DialogResult, Selection:EFortDialogResult Enum=@n1) -> (Confirmed:Exec, Declined:Exec, Ignored:Exec, Killed:Exec, TimedOut:Exec) {is=title K2Node_SwitchEnum}
   > QuitGame(self:Fort Global UIContext Object Reference=Get FortGlobalUIContext!() -> (ReturnValue:Fort Global UIContext Object Reference) {is=title K2Node_GetBlueprintContextEx})
   "On Button Clicked (Button_Logout)"!() -> (OutputDelegate, then, Button:Common Button Legacy Object Reference) {is=title K2Node_ComponentBoundEvent}
   > n2 = ShowSimpleConfirmationDialog_NUI(Title="LOCTEXT(\"Log Out\")", Message="LOCTEXT(\"Are you sure you want to log out of Fortnite?\")", bShowConfirm=true, bShowDecline=true)
   Switch on EFortDialogResult(execute=@n2.DialogResult, Selection=@n2)
   > Logout(self:Fort Global UIContext Object Reference=Get FortGlobalUIContext())
   n3 = Event UpdateNoEnabledButtonsText!() -> (OutputDelegate, then, bAllowLogout:bool, bAllowQuit:bool) {is=title K2Node_Event}
   n4 = Get LogoutAndQuitNotAllowedNoButtonsText() -> (LogoutAndQuitNotAllowedNoButtonsText:Text)
   n5 = Get LogoutAllowedNoButtonsText() -> (LogoutAllowedNoButtonsText:Text)
   > SetText(
    self=Get Text_NoEnabledButtons() -> (Text_NoEnabledButtons:Common Rich Text Block Object Reference)
    InText=Select(
     Option 0:Text=Select(Option 0=@n4, Option 1=@n5, Index=@n3.bAllowLogout)
     Option 1:Text=Get QuitAllowedNoButtonsText() -> (QuitAllowedNoButtonsText:Text)
     Index=@n3.bAllowQuit
    ) -> (ReturnValue:Text)
   )
  ]}]
variables [
 LogoutAllowedNoButtonsText {type=string, default="You can use this menu to return to the lobby, leave your party, and log out"}
 LogoutAndQuitNotAllowedNoButtonsText {type=string, default=You can use this menu to return to the Lobby and leave your party}
 QuitAllowedNoButtonsText {type=string, default="You can use this menu to return to the lobby, leave your party, and exit Fortnite"}
]
widgets {
 info {inheritedWidgetCount=6, parentClass=/Script/FortniteUI.FortSidebarPanelExit, rootWidgetClass=&Overlay, widgetCount=12}
 widgets [
  Overlay_0 {widgetClassPath=&Overlay, widget=Overlay_0}
  VerticalBox_3 {
   widgetClassPath=&VerticalBox
   parent=Overlay_0
   widget=VerticalBox_3
   slot {class=OverlaySlot, properties {horizontalAlignment=HAlign_Fill, verticalAlignment=VAlign_Fill}}
  }
  Border_1 {
   widgetClassPath=&Border
   parent=VerticalBox_3
   widget=Border_1
   properties {
    background {
     bIsDynamicallyLoaded=false
     drawAs=Image
     imageType=NoImage
     mirroring=NoMirror
     resourceName=None
     resourceObject=/Game/UI/Sidebar/Materials/MI_UI_Sidebar_Header_BaseFade
     tiling=NoTile
     imageSize=<4096 4096>
     margin=<ltrb 0 0 0 0>
     outlineSettings {
      bUseBrushTransparency=false
      roundingType=HalfHeightRadius
      width=0
      color {colorUseRule=UseColor_Specified, specifiedColor=<rgba 0 0 0 0>}
      cornerRadii=<0 0 0 0>
     }
     tintColor {colorUseRule=UseColor_Specified, specifiedColor=<rgba 1 1 1 1>}
     uVRegion {bIsValid=false, max=<0 0>, min=<0 0>}
    }
    padding=<ltrb 0 0 0 0>
   }
   slot {class=VerticalBoxSlot, properties {verticalAlignment=VAlign_Top}}
  }
  VerticalBox_6 {widgetClassPath=&VerticalBox, parent=Border_1, widget=VerticalBox_6, slot {class=BorderSlot, properties {padding=<ltrb 0 0 0 0>}}}
  Text_VoiceChatTitle {
   widgetClassPath=&CommonTextBlock
   parent=VerticalBox_6
   widget=Text_VoiceChatTitle
   properties {
    text=Exit
    textTransformPolicy=ToUpper
    font {
     bForceMonospaced=false
     bMaterialIsStencil=false
     fontMaterial=null
     fontObject=&BurbankSmall
     letterSpacing=0
     monospacedWidth=1
     size=32
     skewAmount=0
     typefaceFontName=Black
     outlineSettings {
      bApplyOutlineToDropShadows=false
      bMiteredCorners=false
      bSeparateFillAlpha=false
      outlineMaterial=null
      outlineSize=0
      outlineColor=<rgba 0 0 0 1>
     }
    }
    renderTransform {angle=0, scale=<1 1>, shear=<-8 0>, translation=<0 0>}
    shadowOffset=<0 0>
   }
   slot {class=VerticalBoxSlot, properties {horizontalAlignment=HAlign_Left, verticalAlignment=VAlign_Bottom, padding=<ltrb 20 60 20 20>}}
  }
  CommonHierarchicalScrollBox_0 {
   widgetClassPath=/Script/CommonUI.CommonHierarchicalScrollBox
   parent=VerticalBox_3
   widget=CommonHierarchicalScrollBox_0
   properties {navigation=&self:WidgetTree.CommonHierarchicalScrollBox_0.WidgetNavigation_1}
   slot {class=VerticalBoxSlot, properties {padding=<ltrb 20 12 20 20>}}
  }
  PanelWidget_MainMenuButtons {
   widgetClassPath=&VerticalBox
   bInherited=true
   bIsVariable=true
   parent=CommonHierarchicalScrollBox_0
   widget=PanelWidget_MainMenuButtons
   properties {flowDirectionPreference=Culture}
   slot {class=ScrollBoxSlot}
  }
  Button_LeaveExperience {
   widgetClassPath=/Game/UI/Frontend/MainMenu/MainMenuButton
   bInherited=true
   bIsVariable=true
   parent=PanelWidget_MainMenuButtons
   widget=Button_LeaveExperience
   properties {
    defaultText=Return to Lobby
    mobileButtonHeight=80
    onButtonBaseClicked=(&self:WidgetTree.Button_LeaveExperience.HandleOnButtonClicked)
    onButtonBaseDoubleClicked=(&self:WidgetTree.Button_LeaveExperience.HandleOnButtonDoubleClicked)
    onButtonBaseHovered=(&self:WidgetTree.Button_LeaveExperience.HandleOnButtonHovered)
    onButtonBaseUnhovered=(&self:WidgetTree.Button_LeaveExperience.HandleOnButtonUnhovered)
    onSelectedChangedBase=(&self:WidgetTree.Button_LeaveExperience.HandleOnSelectedChanged)
    style=/Game/UI/Sidebar/Sidebar_ButtonStyle-Tab-Main-Recolor-Red.Sidebar_ButtonStyle-Tab-Main-Recolor-Red_C
   }
   slot {class=VerticalBoxSlot, properties {padding=<ltrb 0 8 0 0>}}
  }
  Button_AthenaMainMenuLeave {
   widgetClassPath=/FortniteCoreUI/UI/Frontend/MainMenu/AthenaMainMenuLeaveButton
   bInherited=true
   bIsVariable=true
   parent=PanelWidget_MainMenuButtons
   widget=Button_AthenaMainMenuLeave
   slot {class=VerticalBoxSlot, properties {padding=<ltrb 0 8 0 0>}}
  }
  Button_Logout {
   widgetClassPath=&MainMenuButton
   bInherited=true
   bIsVariable=true
   parent=PanelWidget_MainMenuButtons
   widget=Button_Logout
   properties {
    defaultText=Log Out
    mobileButtonHeight=80
    onButtonBaseClicked=(&self:WidgetTree.Button_Logout.HandleOnButtonClicked)
    onButtonBaseDoubleClicked=(&self:WidgetTree.Button_Logout.HandleOnButtonDoubleClicked)
    onButtonBaseHovered=(&self:WidgetTree.Button_Logout.HandleOnButtonHovered)
    onButtonBaseUnhovered=(&self:WidgetTree.Button_Logout.HandleOnButtonUnhovered)
    onSelectedChangedBase=(&self:WidgetTree.Button_Logout.HandleOnSelectedChanged)
    style=&Sidebar_ButtonStyle-Tab-Main-Recolor-Red_C
   }
   slot {class=VerticalBoxSlot, properties {padding=<ltrb 0 8 0 0>}}
  }
  Button_ExitGame {
   widgetClassPath=&MainMenuButton
   bInherited=true
   bIsVariable=true
   parent=PanelWidget_MainMenuButtons
   widget=Button_ExitGame
   properties {
    defaultText=Close Fortnite
    mobileButtonHeight=80
    onButtonBaseClicked=(&self:WidgetTree.Button_ExitGame.HandleOnButtonClicked)
    onButtonBaseDoubleClicked=(&self:WidgetTree.Button_ExitGame.HandleOnButtonDoubleClicked)
    onButtonBaseHovered=(&self:WidgetTree.Button_ExitGame.HandleOnButtonHovered)
    onButtonBaseUnhovered=(&self:WidgetTree.Button_ExitGame.HandleOnButtonUnhovered)
    onSelectedChangedBase=(&self:WidgetTree.Button_ExitGame.HandleOnSelectedChanged)
    style=&Sidebar_ButtonStyle-Tab-Main-Recolor-Red_C
    hoveredSlateSoundOverride {resourceObject=/Game/Sounds/Fort_UI_Sounds/FrontEnd/New_FrontEnd/Fortnite_UI_MainMenu_Tab_Hover_Exit_01_Cue}
   }
   slot {class=VerticalBoxSlot, properties {padding=<ltrb 0 8 0 0>}}
  }
  Text_NoEnabledButtons {
   widgetClassPath=&CommonRichTextBlock
   bInherited=true
   bIsVariable=true
   parent=Overlay_0
   widget=Text_NoEnabledButtons
   properties {
    autoWrapText=true
    bOverrideDefaultStyle=true
    justification=Center
    text="You can use this menu to return to the Lobby, leave your party, and exit Fortnite"
    visibility=Collapsed
    defaultTextStyleOverride {
     overflowPolicy=Clip
     transformPolicy=None
     colorAndOpacity {colorUseRule=UseColor_Specified, specifiedColor=<rgba 1 1 1 1>}
     font {
      bForceMonospaced=false
      bMaterialIsStencil=false
      fontMaterial=null
      fontObject=&BurbankSmall
      letterSpacing=0
      monospacedWidth=1
      size=16
      skewAmount=0
      typefaceFontName=Black
      outlineSettings {
       bApplyOutlineToDropShadows=false
       bMiteredCorners=false
       bSeparateFillAlpha=false
       outlineMaterial=null
       outlineSize=0
       outlineColor=<rgba 0 0 0 1>
      }
     }
     highlightColor {colorUseRule=UseColor_Specified, specifiedColor=<rgba 0 0 0 1>}
     highlightShape {
      bIsDynamicallyLoaded=false
      drawAs=Image
      imageType=NoImage
      mirroring=NoMirror
      resourceName=None
      resourceObject=null
      tiling=NoTile
      imageSize=<32 32>
      margin=<ltrb 0 0 0 0>
      outlineSettings {
       bUseBrushTransparency=false
       roundingType=HalfHeightRadius
       width=0
       color {colorUseRule=UseColor_Specified, specifiedColor=<rgba 0 0 0 0>}
       cornerRadii=<0 0 0 0>
      }
      tintColor {colorUseRule=UseColor_Specified, specifiedColor=<rgba 1 1 1 1>}
      uVRegion {bIsValid=false, max=<0 0>, min=<0 0>}
     }
     selectedBackgroundColor {colorUseRule=UseColor_Foreground, specifiedColor=<rgba 1 0 1 1>}
     shadowColorAndOpacity=<rgba 0 0 0 1>
     shadowOffset=<0 0>
     strikeBrush {
      bIsDynamicallyLoaded=false
      drawAs=Image
      imageType=NoImage
      mirroring=NoMirror
      resourceName=None
      resourceObject=null
      tiling=NoTile
      imageSize=<32 32>
      margin=<ltrb 0 0 0 0>
      outlineSettings {
       bUseBrushTransparency=false
       roundingType=HalfHeightRadius
       width=0
       color {colorUseRule=UseColor_Specified, specifiedColor=<rgba 0 0 0 0>}
       cornerRadii=<0 0 0 0>
      }
      tintColor {colorUseRule=UseColor_Specified, specifiedColor=<rgba 1 1 1 1>}
      uVRegion {bIsValid=false, max=<0 0>, min=<0 0>}
     }
     underlineBrush {
      bIsDynamicallyLoaded=false
      drawAs=Image
      imageType=NoImage
      mirroring=NoMirror
      resourceName=None
      resourceObject=null
      tiling=NoTile
      imageSize=<32 32>
      margin=<ltrb 0 0 0 0>
      outlineSettings {
       bUseBrushTransparency=false
       roundingType=HalfHeightRadius
       width=0
       color {colorUseRule=UseColor_Specified, specifiedColor=<rgba 0 0 0 0>}
       cornerRadii=<0 0 0 0>
      }
      tintColor {colorUseRule=UseColor_Specified, specifiedColor=<rgba 1 1 1 1>}
      uVRegion {bIsValid=false, max=<0 0>, min=<0 0>}
     }
    }
    renderTransform {angle=0, scale=<1 1>, shear=<-6 0>, translation=<0 0>}
   }
   slot {class=OverlaySlot, properties {horizontalAlignment=HAlign_Fill, verticalAlignment=VAlign_Center, padding=<ltrb 45 0 45 0>}}
  }
 ]
}
