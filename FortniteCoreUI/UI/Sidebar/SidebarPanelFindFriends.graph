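# Sidebar "Add Friends" panel widget, built on FortSidebarPanelFindFriends, for searching users and inviting friends.
# A header holds the title and friend search bar, above an invite button.
# A visibility switcher holds the default user tree, search-failed text, searching text with throbber, and search-results tree.
# On Construct, a config flag can relabel the invite button "Share invite"; the blueprint graph contains no search logic.
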WidgetBlueprint /FortniteCoreUI/UI/Sidebar/SidebarPanelFindFriends
graphs [EventGraph {nodes [
   Event Construct()
   > n1 = Cast To WBP_UIKit_Button_Loud!(execute, Object=Get Button_Invite() -> (Button_Invite:Common Button Base Object Reference)) -> (then, CastFailed, AsWBP UIKit Button Loud:WBP UIKit Button Loud Object Reference) {is=title K2Node_DynamicCast}
   > Branch(Condition=GetConfigDataBoolWithDefault!(EntryName:String=social_friend_invite_alt_text_shareinvite, bDefaultValue:bool=false) -> (ReturnValue:bool))
   > SetText!(execute, self:Fort CTAButton Object Reference=@n1, ButtonText:"Text (by ref)"=MakeLiteralText(Value="LOCTEXT(\"Share invite\")")) -> (then)
  ]}]
variables []
widgets {
 info {inheritedWidgetCount=7, parentClass=/Script/FortniteUI.FortSidebarPanelFindFriends, rootWidgetClass=&VerticalBox, widgetCount=14}
 widgets [
  VerticalBox_2 {widgetClassPath=&VerticalBox, widget=VerticalBox_2}
  SizeBox_0 {
   widgetClassPath=&SizeBox
   parent=VerticalBox_2
   widget=SizeBox_0
   properties {bOverride_HeightOverride=true, heightOverride=188}
   slot {class=VerticalBoxSlot}
  }
  Border {
   widgetClassPath=&Border
   parent=SizeBox_0
   widget=Border
   properties {
    background {
     bIsDynamicallyLoaded=false
     drawAs=Image
     imageType=NoImage
     mirroring=NoMirror
     resourceName=None
     resourceObject=/Game/UI/Sidebar/Materials/MI_UI_Sidebar_Header_Base
     tiling=NoTile
     imageSize=<4096 4096>
     margin=<ltrb 0 0 0 0>
     outlineSettings {
      bUseBrushTransparency=false
      roundingType=HalfHeightRadius
      width=0
      color {colorUseRule=UseColor_Specified, specifiedColor=<rgba 0 0 0 0>}
      cornerRadii=<0 0 0 0>
     }
     tintColor {colorUseRule=UseColor_Specified, specifiedColor=<rgba 1 1 1 1>}
     uVRegion {bIsValid=false, max=<0 0>, min=<0 0>}
    }
    padding=<ltrb 0 0 0 50>
   }
   slot {class=SizeBoxSlot}
  }
  VerticalBox {widgetClassPath=&VerticalBox, parent=Border, widget=VerticalBox, slot {class=BorderSlot, properties {padding=<ltrb 0 0 0 50>}}}
  Text_FindFriendsTitle {
   widgetClassPath=&CommonTextBlock
   parent=VerticalBox
   widget=Text_FindFriendsTitle
   properties {
    text=Add Friends
    textTransformPolicy=ToUpper
    font {
     bForceMonospaced=false
     bMaterialIsStencil=false
     fontMaterial=null
     fontObject=&BurbankSmall
     letterSpacing=0
     monospacedWidth=1
     size=32
     skewAmount=0
     typefaceFontName=Black
     outlineSettings {
      bApplyOutlineToDropShadows=false
      bMiteredCorners=false
      bSeparateFillAlpha=false
      outlineMaterial=null
      outlineSize=0
      outlineColor=<rgba 0 0 0 1>
     }
    }
    renderTransform {angle=0, scale=<1 1>, shear=<-8 0>, translation=<0 0>}
    shadowOffset=<0 0>
    shapedTextOptions {bOverride_TextFlowDirection=true, bOverride_TextShapingMethod=false, textFlowDirection=Culture, textShapingMethod=Auto}
   }
   slot {class=VerticalBoxSlot, properties {verticalAlignment=VAlign_Bottom, padding=<ltrb 20 60 20 0>}}
  }
  Button_Search_FindFriends {
   widgetClassPath=&/Sidebar_FriendSearchBar
   bInherited=true
   bIsVariable=true
   parent=VerticalBox
   widget=Button_Search_FindFriends
   properties {
    bIsFocusable=true
    onButtonBaseClicked=(&self:WidgetTree.Button_Search_FindFriends.HandleOnButtonClicked)
    onButtonBaseDoubleClicked=(&self:WidgetTree.Button_Search_FindFriends.HandleOnButtonDoubleClicked)
    onButtonBaseHovered=(&self:WidgetTree.Button_Search_FindFriends.HandleOnButtonHovered)
    onButtonBaseUnhovered=(&self:WidgetTree.Button_Search_FindFriends.HandleOnButtonUnhovered)
    onSelectedChangedBase=(&self:WidgetTree.Button_Search_FindFriends.HandleOnSelectedChanged)
   }
   slot {class=VerticalBoxSlot}
  }
  Button_Invite {
   widgetClassPath=/FortUILibrary/UIKit/Buttons/WBP_UIKit_Button_Loud
   bInherited=true
   bIsVariable=true
   parent=VerticalBox_2
   widget=Button_Invite
   properties {bSimulateHoverOnTouchInput=false, buttonSize=Small, minHeight=90, text=Invite a Friend, touchMethod=PreciseTap}
   slot {class=VerticalBoxSlot, properties {padding=<ltrb 16 16 16 16>}}
  }
  Switcher_FindFriends {
   widgetClassPath=&CommonVisibilitySwitcher
   bInherited=true
   parent=VerticalBox_2
   widget=Switcher_FindFriends
   slot {class=VerticalBoxSlot, properties {size {sizeRule=Fill, value=1}}}
  }
  UserTree_FindFriends {
   widgetClassPath=/Script/FortniteUI.FortSocialUserTreeView
   bInherited=true
   parent=Switcher_FindFriends
   widget=UserTree_FindFriends
   properties {
    bEnableScrollAnimation=false
    entryWidgetClass=&/SidebarUserListHeaderEntry.SidebarUserListHeaderEntry_C
    inviteEntryWidgetClass=&/SidebarInviteListEntry_FriendInvite_XS.SidebarInviteListEntry_FriendInvite_XS_C
    joinRequestEntryWidgetClass=/FortniteCoreUI/UI/Sidebar/ListEntries/SidebarJoinRequestUserListEntry.SidebarJoinRequestUserListEntry_C
    joinablePartyEntryWidgetClass=&/SidebarJoinablePartyEntry.SidebarJoinablePartyEntry_C
    joinablePartyEntryWidgetClass_2=/FortniteCoreUI/UI/Sidebar/ListEntries/JoinableParty_UserListEntry.JoinableParty_UserListEntry_C
    leavePartyEntryWidgetClass=&/SidebarLeavePartyListEntry.SidebarLeavePartyListEntry_C
    missedInviteEntryWidgetClass=/FortniteCoreUI/UI/SocialPanel/MissedInvitesEntry.MissedInvitesEntry_C
    multiFriendActionEntryWidgetClass=/FortniteCoreUI/UI/Sidebar/ListEntries/MultiFriendActionListEntry.MultiFriendActionListEntry_C
    partyMemberEntryWidgetClass=/FortniteCoreUI/UI/Sidebar/ListEntries/SidebarPartyMemberListEntry.SidebarPartyMemberListEntry_C
    recentPlayerEntryWidgetClass=/FortniteCoreUI/UI/Sidebar/ListEntries/SidebarRecentPlayerListEntry.SidebarRecentPlayerListEntry_C
    referFriendEntryWidgetClass=&/SidebarReferAFriendEntry.SidebarReferAFriendEntry_C
    suggestedFriendEntryWidgetClass=/FortniteCoreUI/UI/Sidebar/ListEntries/SidebarSuggestedFriendListEntry.SidebarSuggestedFriendListEntry_C
    teamMemberEntryWidgetClass=/FortniteCoreUI/UI/Sidebar/ListEntries/SidebarTeamMemberListEntry.SidebarTeamMemberListEntry_C
    userEntryWidgetClass=/FortniteCoreUI/UI/Sidebar/ListEntries/SidebarUserListEntry_XS.SidebarUserListEntry_XS_C
    userSearchResultEntryWidgetClass=/FortniteCoreUI/UI/Sidebar/ListEntries/SidebarUserSearchResultListEntry_XS.SidebarUserSearchResultListEntry_XS_C
    voiceChannelHeaderEntryWidgetClass=&/SidebarVoiceChannelHeaderButton.SidebarVoiceChannelHeaderButton_C
    voiceChatMemberEntryWidgetClass=/FortniteCoreUI/UI/SocialPanel/ListEntries/VoiceChatMemberListEntry.VoiceChatMemberListEntry_C
   }
   slot {class=CommonVisibilitySwitcherSlot}
  }
  Text_SearchFailed {
   widgetClassPath=&CommonTextBlock
   bInherited=true
   bIsVariable=true
   parent=Switcher_FindFriends
   widget=Text_SearchFailed
   properties {
    autoWrapText=true
    justification=Center
    text="Search Failed!"
    textTransformPolicy=ToUpper
    colorAndOpacity {colorUseRule=UseColor_Specified, specifiedColor=<rgba 1 0.1714 0.1329 1>}
    font {
     bForceMonospaced=false
     bMaterialIsStencil=false
     fontMaterial=null
     fontObject=&BurbankSmall
     letterSpacing=0
     monospacedWidth=1
     size=24
     skewAmount=0
     typefaceFontName=Bold
     outlineSettings {
      bApplyOutlineToDropShadows=false
      bMiteredCorners=false
      bSeparateFillAlpha=false
      outlineMaterial=null
      outlineSize=0
      outlineColor=<rgba 0 0 0 1>
     }
    }
    shadowOffset=<0 0>
   }
   slot {class=CommonVisibilitySwitcherSlot}
  }
  VerticalBox_8 {widgetClassPath=&VerticalBox, parent=Switcher_FindFriends, widget=VerticalBox_8, slot {class=CommonVisibilitySwitcherSlot}}
  Text_SearchingForUsers {
   widgetClassPath=&CommonTextBlock
   bInherited=true
   bIsVariable=true
   parent=VerticalBox_8
   widget=Text_SearchingForUsers
   properties {
    autoWrapText=true
    justification=Center
    text=Searching for users...
    textTransformPolicy=ToUpper
    colorAndOpacity {colorUseRule=UseColor_Specified, specifiedColor=<rgba 0.0723 0.8714 1 1>}
    font {
     bForceMonospaced=false
     bMaterialIsStencil=false
     fontMaterial=null
     fontObject=&BurbankSmall
     letterSpacing=0
     monospacedWidth=1
     size=24
     skewAmount=0
     typefaceFontName=Bold
     outlineSettings {
      bApplyOutlineToDropShadows=false
      bMiteredCorners=false
      bSeparateFillAlpha=false
      outlineMaterial=null
      outlineSize=0
      outlineColor=<rgba 0 0 0 1>
     }
    }
    shadowOffset=<0 0>
   }
   slot {class=VerticalBoxSlot}
  }
  CircularThrobber_UserSearch {
   widgetClassPath=&Image
   bIsVariable=true
   parent=VerticalBox_8
   widget=CircularThrobber_UserSearch
   properties {brush {
     bIsDynamicallyLoaded=false
     drawAs=Image
     imageType=NoImage
     mirroring=NoMirror
     resourceName=None
     resourceObject=/FortUILibrary/UIKit/Throbber/MI_UIKit_Color_Throbber
     tiling=NoTile
     imageSize=<32 32>
     margin=<ltrb 0 0 0 0>
     outlineSettings {
      bUseBrushTransparency=false
      roundingType=HalfHeightRadius
      width=0
      color {colorUseRule=UseColor_Specified, specifiedColor=<rgba 0 0 0 0>}
      cornerRadii=<0 0 0 0>
     }
     tintColor {colorUseRule=UseColor_Specified, specifiedColor=<rgba 1 1 1 1>}
     uVRegion {bIsValid=false, max=<0 0>, min=<0 0>}
    }}
   slot {class=VerticalBoxSlot, properties {horizontalAlignment=HAlign_Center, verticalAlignment=VAlign_Center, padding=<ltrb 16 12 16 16>}}
  }
  UserTree_FindFriends_SearchResults {
   widgetClassPath=&FortSocialUserTreeView
   bInherited=true
   bIsVariable=true
   parent=Switcher_FindFriends
   widget=UserTree_FindFriends_SearchResults
   properties {
    bEnableScrollAnimation=false
    entryWidgetClass=&SidebarUserListHeaderEntry_C
    inviteEntryWidgetClass=&SidebarInviteListEntry_FriendInvite_XS_C
    joinRequestEntryWidgetClass=&SidebarJoinRequestUserListEntry_C
    joinablePartyEntryWidgetClass=&SidebarJoinablePartyEntry_C
    joinablePartyEntryWidgetClass_2=&JoinableParty_UserListEntry_C
    leavePartyEntryWidgetClass=&SidebarLeavePartyListEntry_C
    missedInviteEntryWidgetClass=&MissedInvitesEntry_C
    multiFriendActionEntryWidgetClass=/Script/FortniteUI.FortMultiFriendActionListEntry
    partyMemberEntryWidgetClass=&SidebarPartyMemberListEntry_C
    recentPlayerEntryWidgetClass=&SidebarRecentPlayerListEntry_C
    referFriendEntryWidgetClass=/FortniteCoreUI/UI/SocialPanel/ReferAFriendEntry.ReferAFriendEntry_C
    suggestedFriendEntryWidgetClass=&SidebarSuggestedFriendListEntry_C
    teamMemberEntryWidgetClass=&SidebarTeamMemberListEntry_C
    userEntryWidgetClass=&SidebarUserListEntry_XS_C
    userSearchResultEntryWidgetClass=&SidebarUserSearchResultListEntry_XS_C
    voiceChannelHeaderEntryWidgetClass=&SidebarVoiceChannelHeaderButton_C
    voiceChatMemberEntryWidgetClass=&VoiceChatMemberListEntry_C
   }
   slot {class=CommonVisibilitySwitcherSlot}
  }
 ]
}
