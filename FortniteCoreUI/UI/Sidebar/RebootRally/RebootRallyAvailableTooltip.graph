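# Sidebar onboarding tooltip for Reboot Rally, an arrowed bubble telling players to play with friends for cosmetic rewards.
# While shown, it listens for input method changes and sets click, press or tap wording to match.
# When FlipHorizontal or Discover-in-lobby is on, it mirrors the bubble and counter-mirrors the text to keep it readable.
# On end, it unbinds and immediately reports the close animation finished, without playing one.
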WidgetBlueprint /FortniteCoreUI/UI/Sidebar/RebootRally/RebootRallyAvailableTooltip
graphs [
 EventGraph {nodes [
   n1 = HandleInputMethodChanged!() -> (OutputDelegate, then, bNewInputType:ECommonInputType Enum) {is=title K2Node_CustomEvent}
   > UpdateText()
   n2 = Event PreConstruct()
   SetTooltipAlignment(execute=[@n2.then, @n3.then])
   n3 = Event OnInitialized()
   Event OnShowTooltip() {is=title K2Node_Event}
   > Bind Event to On Input Method Changed(self=Get CommonInputSubsystem(), Delegate=@n1)
   > UpdateText()
   Event OnEndTooltip() {is=title K2Node_Event}
   > Unbind Event from On Input Method Changed(self=Get CommonInputSubsystem(), Delegate=@n1)
   > NotifyCloseAnimationFinished(self:Fort Sidebar Onboard Tooltip Widget Object Reference)
  ]}
 SetTooltipAlignment {nodes [
   SetTooltipAlignment() {is=title K2Node_FunctionEntry}
   n1 = GetIsDiscoverInLobbyEnabled!(self:Fort Runtime Options Object Reference=GetRuntimeOptions(), ContextWidget:Widget Object Reference=Self-Reference()) -> (ReturnValue:bool)
   n2 = Get FlipHorizontal() -> (FlipHorizontal:bool)
   n3 = Make Widget Transform!(
    Translation:Vector 2D Structure=(X=0,Y=0)
    Scale:Vector 2D Structure=Select(Option 0:Vector 2D Structure=(X=1,Y=1), Option 1:Vector 2D Structure=(X=-1,Y=1), Index=OR Boolean(A=@n2, B=@n1)) -> (ReturnValue:Vector 2D Structure)
    Shear:Vector 2D Structure=(X=0,Y=0)
    Angle:float=0
   ) -> (WidgetTransform:Widget Transform Structure) {is=title K2Node_MakeStruct}
   > SetRenderTransform!(execute, self:Widget Object Reference=Get SizeBox_9() -> (SizeBox_9:Size Box Object Reference), InTransform:Widget Transform Structure=@n3) -> (then)
   > SetRenderTransform(self=Get CommonRichTextBlock_140() -> (CommonRichTextBlock_140:Common Rich Text Block Object Reference), InTransform=@n3)
  ]}
 UpdateText {nodes [
   UpdateText() {is=title K2Node_FunctionEntry}
   n1 = MakeLiteralText(Value="LOCTEXT(\"Play with friends to earn cosmetic rewards. Click to get started!\")")
   n2 = Get CommonInputSubsystem()
   > SetText(
    self=Get CommonRichTextBlock_140()
    InText=Select!(
     MouseAndKeyboard:"Text (by ref)"=@n1
     Gamepad:"Text (by ref)"=MakeLiteralText(Value="LOCTEXT(\"Play with friends to earn cosmetic rewards. Press to get started!\")")
     Touch:"Text (by ref)"=MakeLiteralText(Value="LOCTEXT(\"Play with friends to earn cosmetic rewards. Tap to get started!\")")
     Count:"Text (by ref)"=@n1
     Index:ECommonInputType Enum=GetCurrentInputType(self=@n2)
    ) -> (ReturnValue:"Text (by ref)")
   )
  ]}
]
variables [FlipHorizontal {type=boolean}]
widgets {
 info {inheritedWidgetCount=2, parentClass=/Script/FortniteUI.FortSidebarOnboardTooltipWidget, rootWidgetClass=&Overlay, widgetCount=17}
 widgets [
  Overlay_1 {widgetClassPath=&Overlay, widget=Overlay_1}
  SizeBox_187 {
   widgetClassPath=&SizeBox
   parent=Overlay_1
   widget=SizeBox_187
   properties {bOverride_MaxDesiredWidth=true, maxDesiredWidth=400}
   slot {class=OverlaySlot, properties {horizontalAlignment=HAlign_Fill, verticalAlignment=VAlign_Fill}}
  }
  HorizontalBox_0 {
   widgetClassPath=&HorizontalBox
   parent=SizeBox_187
   widget=HorizontalBox_0
   slot {class=SizeBoxSlot, properties {horizontalAlignment=HAlign_Left}}
  }
  SizeBox_0 {
   widgetClassPath=&SizeBox
   parent=HorizontalBox_0
   widget=SizeBox_0
   properties {bOverride_WidthOverride=true, widthOverride=550}
   slot {class=HorizontalBoxSlot, properties {horizontalAlignment=HAlign_Left}}
  }
  Overlay_79 {widgetClassPath=&Overlay, parent=SizeBox_0, widget=Overlay_79, properties {visibility=Visible}, slot {class=SizeBoxSlot}}
  SizeBox_170 {
   widgetClassPath=&SizeBox
   parent=HorizontalBox_0
   widget=SizeBox_170
   properties {bOverride_WidthOverride=true, minDesiredWidth=77, widthOverride=96}
   slot {class=HorizontalBoxSlot, properties {horizontalAlignment=HAlign_Right}}
  }
  Overlay_120 {widgetClassPath=&Overlay, parent=SizeBox_170, widget=Overlay_120, slot {class=SizeBoxSlot}}
  SizeBox_9 {
   widgetClassPath=&SizeBox
   bIsVariable=true
   parent=Overlay_1
   widget=SizeBox_9
   properties {bOverride_MinDesiredWidth=true, bOverride_WidthOverride=true, heightOverride=76, minDesiredWidth=200, widthOverride=550}
   slot {class=OverlaySlot, properties {horizontalAlignment=HAlign_Fill, verticalAlignment=VAlign_Center, padding=<ltrb 77 0 0 0>}}
  }
  Overlay_166 {widgetClassPath=&Overlay, parent=SizeBox_9, widget=Overlay_166, slot {class=SizeBoxSlot}}
  ImageBurst {
   widgetClassPath=&Image
   bIsVariable=true
   parent=Overlay_166
   widget=ImageBurst
   properties {
    visibility=SelfHitTestInvisible
    brush {
     bIsDynamicallyLoaded=false
     drawAs=Box
     imageType=NoImage
     mirroring=NoMirror
     resourceName=None
     resourceObject=/FortUILibrary/Foundation/Art_ToolKit/AngledBoxes/M_UI_AngledBox_Tuto_burst
     tiling=NoTile
     imageSize=<90 90>
     margin=<ltrb 0.5 0 0.5 0>
     outlineSettings {
      bUseBrushTransparency=false
      roundingType=HalfHeightRadius
      width=0
      color {colorUseRule=UseColor_Specified, specifiedColor=<rgba 0 0 0 0>}
      cornerRadii=<0 0 0 0>
     }
     tintColor {colorUseRule=UseColor_Specified, specifiedColor=<rgba 1 1 1 1>}
     uVRegion {bIsValid=false, max=<0 0>, min=<0 0>}
    }
   }
   slot {class=OverlaySlot, properties {horizontalAlignment=HAlign_Fill, verticalAlignment=VAlign_Fill}}
  }
  Border_Shadow {
   widgetClassPath=&Border
   parent=Overlay_166
   widget=Border_Shadow
   properties {
    verticalAlignment=VAlign_Center
    background {
     bIsDynamicallyLoaded=false
     drawAs=Box
     imageType=NoImage
     mirroring=NoMirror
     resourceName=None
     resourceObject=/FortUILibrary/Foundation/Art_ToolKit/AngledBoxes/M_UI_AngledBox_Tuto
     tiling=NoTile
     imageSize=<90 90>
     margin=<ltrb 0.5 0 0.5 0>
     outlineSettings {
      bUseBrushTransparency=false
      roundingType=HalfHeightRadius
      width=0
      color {colorUseRule=UseColor_Specified, specifiedColor=<rgba 0 0 0 0>}
      cornerRadii=<0 0 0 0>
     }
     tintColor {colorUseRule=UseColor_Specified, specifiedColor=<rgba 1 1 1 1>}
     uVRegion {bIsValid=false, max=<0 0>, min=<0 0>}
    }
    brushColor=<rgba 0 0 0 0.5>
    padding=<ltrb 0 0 0 0>
    renderTransform {angle=0, scale=<1 1>, shear=<0 0>, translation=<8 10>}
   }
   slot {class=OverlaySlot, properties {horizontalAlignment=HAlign_Fill, verticalAlignment=VAlign_Fill}}
  }
  Border {
   widgetClassPath=&Border
   parent=Overlay_166
   widget=Border
   properties {
    verticalAlignment=VAlign_Center
    background {
     bIsDynamicallyLoaded=false
     drawAs=Box
     imageType=NoImage
     mirroring=NoMirror
     resourceName=None
     resourceObject=/FortUILibrary/Foundation/Art_ToolKit/AngledBoxes/M_UI_AngledBox_Tuto_Sidebar
     tiling=NoTile
     imageSize=<90 90>
     margin=<ltrb 0.5 0 0.5 0>
     outlineSettings {
      bUseBrushTransparency=false
      roundingType=HalfHeightRadius
      width=0
      color {colorUseRule=UseColor_Specified, specifiedColor=<rgba 0 0 0 0>}
      cornerRadii=<0 0 0 0>
     }
     tintColor {colorUseRule=UseColor_Specified, specifiedColor=<rgba 1 1 1 1>}
     uVRegion {bIsValid=false, max=<0 0>, min=<0 0>}
    }
    padding=<ltrb 0 0 0 0>
   }
   slot {class=OverlaySlot, properties {horizontalAlignment=HAlign_Fill, verticalAlignment=VAlign_Fill}}
  }
  VerticalBox_85 {
   widgetClassPath=&VerticalBox
   parent=Overlay_166
   widget=VerticalBox_85
   slot {class=OverlaySlot, properties {horizontalAlignment=HAlign_Fill, verticalAlignment=VAlign_Fill, padding=<ltrb 48 40 40 30>}}
  }
  CommonRichTextBlock_140 {
   widgetClassPath=&CommonRichTextBlock
   bIsVariable=true
   parent=VerticalBox_85
   widget=CommonRichTextBlock_140
   properties {
    autoWrapText=true
    bOverrideDefaultStyle=true
    mobileTextBlockScale=1.1
    text=Play with friends to earn cosmetic rewards.
    defaultTextStyleOverride {
     overflowPolicy=Clip
     transformPolicy=None
     colorAndOpacity {colorUseRule=UseColor_Specified, specifiedColor=<rgba 0.4233 0.7529 1 1>}
     font {
      bForceMonospaced=false
      bMaterialIsStencil=false
      fontMaterial=null
      fontObject=&BurbankSmall
      letterSpacing=0
      monospacedWidth=1
      size=16
      skewAmount=0
      typefaceFontName=Bold
      outlineSettings {
       bApplyOutlineToDropShadows=false
       bMiteredCorners=false
       bSeparateFillAlpha=false
       outlineMaterial=null
       outlineSize=0
       outlineColor=<rgba 0 0 0 1>
      }
     }
     highlightColor {colorUseRule=UseColor_Specified, specifiedColor=<rgba 0 0 0 1>}
     highlightShape {
      bIsDynamicallyLoaded=false
      drawAs=Image
      imageType=NoImage
      mirroring=NoMirror
      resourceName=None
      resourceObject=null
      tiling=NoTile
      imageSize=<32 32>
      margin=<ltrb 0 0 0 0>
      outlineSettings {
       bUseBrushTransparency=false
       roundingType=HalfHeightRadius
       width=0
       color {colorUseRule=UseColor_Specified, specifiedColor=<rgba 0 0 0 0>}
       cornerRadii=<0 0 0 0>
      }
      tintColor {colorUseRule=UseColor_Specified, specifiedColor=<rgba 1 1 1 1>}
      uVRegion {bIsValid=false, max=<0 0>, min=<0 0>}
     }
     selectedBackgroundColor {colorUseRule=UseColor_Foreground, specifiedColor=<rgba 1 0 1 1>}
     shadowColorAndOpacity=<rgba 0 0 0 1>
     shadowOffset=<0 0>
     strikeBrush {
      bIsDynamicallyLoaded=false
      drawAs=Image
      imageType=NoImage
      mirroring=NoMirror
      resourceName=None
      resourceObject=null
      tiling=NoTile
      imageSize=<32 32>
      margin=<ltrb 0 0 0 0>
      outlineSettings {
       bUseBrushTransparency=false
       roundingType=HalfHeightRadius
       width=0
       color {colorUseRule=UseColor_Specified, specifiedColor=<rgba 0 0 0 0>}
       cornerRadii=<0 0 0 0>
      }
      tintColor {colorUseRule=UseColor_Specified, specifiedColor=<rgba 1 1 1 1>}
      uVRegion {bIsValid=false, max=<0 0>, min=<0 0>}
     }
     underlineBrush {
      bIsDynamicallyLoaded=false
      drawAs=Image
      imageType=NoImage
      mirroring=NoMirror
      resourceName=None
      resourceObject=null
      tiling=NoTile
      imageSize=<32 32>
      margin=<ltrb 0 0 0 0>
      outlineSettings {
       bUseBrushTransparency=false
       roundingType=HalfHeightRadius
       width=0
       color {colorUseRule=UseColor_Specified, specifiedColor=<rgba 0 0 0 0>}
       cornerRadii=<0 0 0 0>
      }
      tintColor {colorUseRule=UseColor_Specified, specifiedColor=<rgba 1 1 1 1>}
      uVRegion {bIsValid=false, max=<0 0>, min=<0 0>}
     }
    }
   }
   slot {class=VerticalBoxSlot, properties {verticalAlignment=VAlign_Center}}
  }
  Image_Arrow {
   widgetClassPath=&Image
   bIsVariable=true
   parent=Overlay_166
   widget=Image_Arrow
   properties {
    brush {
     bIsDynamicallyLoaded=false
     drawAs=Image
     imageType=NoImage
     mirroring=NoMirror
     resourceName=None
     resourceObject=/Game/Athena/HUD/Quests/Art/T_UI_Quest_Arrow
     tiling=NoTile
     imageSize=<60 60>
     margin=<ltrb 0 0 0 0>
     outlineSettings {
      bUseBrushTransparency=false
      roundingType=HalfHeightRadius
      width=0
      color {colorUseRule=UseColor_Specified, specifiedColor=<rgba 0 0 0 0>}
      cornerRadii=<0 0 0 0>
     }
     tintColor {colorUseRule=UseColor_Specified, specifiedColor=<rgba 1 1 1 1>}
     uVRegion {bIsValid=false, max=<0 0>, min=<0 0>}
    }
    renderTransform {angle=-90, scale=<1 1>, shear=<0 0>, translation=<-47 0>}
   }
   slot {class=OverlaySlot, properties {verticalAlignment=VAlign_Center, padding=<ltrb 0 0 0 2>}}
  }
  Dismiss_Button {widgetClassPath=&CommonButtonLegacy, bInherited=true, widget=null}
  OtherAction_Button {widgetClassPath=&CommonButtonLegacy, bInherited=true, widget=null}
 ]
}
