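# Reboot Rally selection header widget for the sidebar, showing the logo banner, rewards image, timer and Select action prompt over a cloud material border.
# PreConstruct caches the border material's gradient colors.
# Hovered and Unhovered animations shift prompt, timer and rewards, and animate InterpolationValue, whose setter lerps those colors toward brighter linear and more transparent radial variants.
# Input method changes show the prompt's fade backing only for gamepad.
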WidgetBlueprint /FortniteCoreUI/UI/Sidebar/RebootRally/WBP_RebootRallySelectionHeader
animations [
 Default {
  binding_count=0
  length_seconds=5
  object=&self:Default
  sequence {bindings []}
 }
 Hovered {
  binding_count=7
  length_seconds=0.1667
  object=&self:Hovered
  sequence {
   bindings [
    CommonActionWidget {id=#1, possessed_class=CommonActionWidget, tracks []}
    Image_LinearFade {
     id=#2
     possessed_class=Image
     tracks [{
       type=MovieScene2DTransformTrack
       sections [{
         type=MovieScene2DTransformSection
         end_frame=5
         start_frame=0
         channels [
          Translation.X {type=MovieSceneScriptingFloatChannel, thinned_from=3, keys [[0,80],[5,0]]}
          Translation.Y {type=MovieSceneScriptingFloatChannel, const [0,1]}
          Angle {type=MovieSceneScriptingFloatChannel}
          Scale.X {type=MovieSceneScriptingFloatChannel}
          Scale.Y {type=MovieSceneScriptingFloatChannel}
          Shear.X {type=MovieSceneScriptingFloatChannel}
          Shear.Y {type=MovieSceneScriptingFloatChannel}
         ]
        }]
      }]
    }
    WBP_RebootRallySelectionHeader_Hover {
     id=#3
     possessed_class=WBP_RebootRallySelectionHeader_C
     tracks [{
       type=MovieSceneDoubleTrack
       display_name=Interpolation Value
       sections [{type=MovieSceneDoubleSection, end_frame=5, start_frame=0, channels [None {type=MovieSceneScriptingDoubleChannel, keys [[0,0],[5,0.2]]}]}]
      }]
    }
    Border {
     id=#4
     possessed_class=CommonBorder
     tracks [{
       type=MovieSceneWidgetMaterialTrack
       sections [{
         type=MovieSceneParameterSection
         end_frame=5
         start_frame=0
         channels [Left Bar Animation Transition {type=MovieSceneScriptingFloatChannel, keys [[1,0],[5,1]]}]
        }]
      }]
    }
    WBP_RebootRallyTimer {
     id=#5
     possessed_class=WBP_RebootRallyTimer_C
     tracks [{
       type=MovieScene2DTransformTrack
       sections [{
         type=MovieScene2DTransformSection
         end_frame=4
         start_frame=0
         channels [
          Translation.X {type=MovieSceneScriptingFloatChannel}
          Translation.Y {type=MovieSceneScriptingFloatChannel, keys [[0.8645,0],[3.4585,-2]]}
          Angle {type=MovieSceneScriptingFloatChannel}
          Scale.X {type=MovieSceneScriptingFloatChannel}
          Scale.Y {type=MovieSceneScriptingFloatChannel}
          Shear.X {type=MovieSceneScriptingFloatChannel}
          Shear.Y {type=MovieSceneScriptingFloatChannel}
         ]
        }]
      }]
    }
    GridSlot_7 {
     id=#6
     parent_id=#1
     possessed_class=GridSlot
     tracks [{
       type=MovieSceneMarginTrack
       display_name=Padding
       sections [{
         type=MovieSceneMarginSection
         end_frame=5
         start_frame=0
         channels [
          Left {type=MovieSceneScriptingFloatChannel, const [0,1]}
          Top {type=MovieSceneScriptingFloatChannel, const [0,1]}
          Right {type=MovieSceneScriptingFloatChannel, keys [[0,-32],[5,12]]}
          Bottom {type=MovieSceneScriptingFloatChannel, const [0,1]}
         ]
        }]
      }]
    }
    Image_Rewards {
     id=#7
     possessed_class=Image
     tracks [{
       type=MovieScene2DTransformTrack
       sections [{
         type=MovieScene2DTransformSection
         end_frame=3
         start_frame=0
         channels [
          Translation.X {type=MovieSceneScriptingFloatChannel, const [0,1]}
          Translation.Y {type=MovieSceneScriptingFloatChannel, const [0,1]}
          Angle {type=MovieSceneScriptingFloatChannel, const [0,1]}
          Scale.X {type=MovieSceneScriptingFloatChannel, keys [[0,1],[3,1.1]]}
          Scale.Y {type=MovieSceneScriptingFloatChannel, keys [[0,1],[3,1.1]]}
          Shear.X {type=MovieSceneScriptingFloatChannel, const [0,1]}
          Shear.Y {type=MovieSceneScriptingFloatChannel, const [0,1]}
         ]
        }]
      }]
    }
   ]
   display_rate {numerator=30}
  }
 }
 Unhovered {
  binding_count=7
  length_seconds=0.1
  object=&self:Unhovered
  sequence {
   bindings [
    CommonActionWidget {id=#1, possessed_class=CommonActionWidget, tracks []}
    Image_LinearFade {
     id=#2
     possessed_class=Image
     tracks [{
       type=MovieScene2DTransformTrack
       sections [{
         type=MovieScene2DTransformSection
         end_frame=3
         start_frame=0
         channels [
          Translation.X {type=MovieSceneScriptingFloatChannel, thinned_from=3, keys [[0,0],[3,80]]}
          Translation.Y {type=MovieSceneScriptingFloatChannel}
          Angle {type=MovieSceneScriptingFloatChannel}
          Scale.X {type=MovieSceneScriptingFloatChannel}
          Scale.Y {type=MovieSceneScriptingFloatChannel}
          Shear.X {type=MovieSceneScriptingFloatChannel}
          Shear.Y {type=MovieSceneScriptingFloatChannel}
         ]
        }]
      }]
    }
    WBP_RebootRallySelectionHeader_C_97 {
     id=#8
     possessed_class=WBP_RebootRallySelectionHeader_C
     tracks [{
       type=MovieSceneDoubleTrack
       display_name=Interpolation Value
       sections [{type=MovieSceneDoubleSection, end_frame=3, start_frame=0, channels [None {type=MovieSceneScriptingDoubleChannel, keys [[0,0.2],[3,0]]}]}]
      }]
    }
    Border {
     id=#4
     possessed_class=CommonBorder
     tracks [{
       type=MovieSceneWidgetMaterialTrack
       sections [{
         type=MovieSceneParameterSection
         end_frame=3
         start_frame=0
         channels [Left Bar Animation Transition {type=MovieSceneScriptingFloatChannel, keys [[0,1],[2.9995,0]]}]
        }]
      }]
    }
    Image_Rewards {
     id=#9
     possessed_class=Image
     tracks [{
       type=MovieScene2DTransformTrack
       sections [{
         type=MovieScene2DTransformSection
         end_frame=3
         start_frame=0
         channels [
          Translation.X {type=MovieSceneScriptingFloatChannel, const [0,1]}
          Translation.Y {type=MovieSceneScriptingFloatChannel, const [0,1]}
          Angle {type=MovieSceneScriptingFloatChannel, const [0,1]}
          Scale.X {type=MovieSceneScriptingFloatChannel, keys [[0,1.1],[3,1]]}
          Scale.Y {type=MovieSceneScriptingFloatChannel, keys [[0,1.1],[3,1]]}
          Shear.X {type=MovieSceneScriptingFloatChannel, const [0,1]}
          Shear.Y {type=MovieSceneScriptingFloatChannel, const [0,1]}
         ]
        }]
      }]
    }
    WBP_RebootRallyTimer {
     id=#5
     possessed_class=WBP_RebootRallyTimer_C
     tracks [{
       type=MovieScene2DTransformTrack
       sections [{
         type=MovieScene2DTransformSection
         end_frame=3
         start_frame=0
         channels [
          Translation.X {type=MovieSceneScriptingFloatChannel}
          Translation.Y {type=MovieSceneScriptingFloatChannel, keys [[0,-2],[3,0]]}
          Angle {type=MovieSceneScriptingFloatChannel}
          Scale.X {type=MovieSceneScriptingFloatChannel}
          Scale.Y {type=MovieSceneScriptingFloatChannel}
          Shear.X {type=MovieSceneScriptingFloatChannel}
          Shear.Y {type=MovieSceneScriptingFloatChannel}
         ]
        }]
      }]
    }
    GridSlot_7 {
     id=#6
     parent_id=#1
     possessed_class=GridSlot
     tracks [{
       type=MovieSceneMarginTrack
       display_name=Padding
       sections [{
         type=MovieSceneMarginSection
         end_frame=3
         start_frame=0
         channels [
          Left {type=MovieSceneScriptingFloatChannel, const [0,1]}
          Top {type=MovieSceneScriptingFloatChannel, const [0,1]}
          Right {type=MovieSceneScriptingFloatChannel, keys [[0,12],[3,-32]]}
          Bottom {type=MovieSceneScriptingFloatChannel, const [0,1]}
         ]
        }]
      }]
    }
   ]
   display_rate {numerator=30}
  }
 }
]
graphs [
 EventGraph {nodes [
   Event PreConstruct()
   > n1 = GetDynamicMaterial(self:Border Object Reference=Get Border() -> (Border:Common Border Object Reference))
   > n2 = Get Vector Parameter Value!(execute, self:Material Instance Dynamic Object Reference=@n1, ParameterName:Name=Get ColorA_Name() -> (ColorA_Name:Name=None)) -> (then, ReturnValue:Linear Color Structure) {is=title K2Node_CallFunction}
   > Set ColorA(ColorA:Linear Color Structure=@n2) -> (Output_Get:Linear Color Structure)
   > n3 = Get Vector Parameter Value(self=@n1, ParameterName=Get ColorB_Name() -> (ColorB_Name:Name=None))
   > Set ColorB(ColorB:Linear Color Structure=@n3) -> (Output_Get:Linear Color Structure)
   > n4 = Get Vector Parameter Value(self=@n1, ParameterName=Get RadialGradientA_Name() -> (RadialGradientA_Name:Name=None))
   > Set RadialGradientA(RadialGradientA:Linear Color Structure=@n4) -> (Output_Get:Linear Color Structure)
   > n5 = Get Vector Parameter Value(self=@n1, ParameterName=Get RadialGradientB_Name() -> (RadialGradientB_Name:Name=None))
   > Set RadialGradientB(RadialGradientB:Linear Color Structure=@n5) -> (Output_Get:Linear Color Structure)
   n6 = Event Construct()
   PlayAnimation(execute=[@n7.Finished, @n6.then], InAnimation=Get Default() -> (Default:Widget Animation Object Reference), NumLoopsToPlay=0)
   Event OnInitialized()
   > Assign On Input Method Changed!(execute:Exec, self:Common Input Object Reference=Get CommonInputSubsystem(), Delegate:"Delegate (by ref)"=Create Event()) -> (then:Exec) {is=title K2Node_AssignDelegate}
   Event On Hovered()
   > StopAllAnimations(self:User Widget Object Reference)
   > PlayAnimation(InAnimation=Get Hovered() -> (Hovered:Widget Animation Object Reference), PlaybackSpeed=1.5)
   Event On Unhovered()
   > n7 = "Play Animation with Finished event (legacy)"!(
    execute:Exec
    Widget:User Widget Object Reference=Self-Reference()
    InAnimation:Widget Animation Object Reference=Get Unhovered() -> (Unhovered:Widget Animation Object Reference)
    StartAtTime:float=0
    NumLoopsToPlay:int=1
    PlayMode:EUMGSequencePlayMode Enum
    PlaybackSpeed:float=1.5
   ) -> (then:Exec, Result:UMGSequence Player Object Reference, Finished:Exec) {is=title K2Node_PlayAnimation}
  ]}
 InputMethodChanged {nodes [
   n1 = InputMethodChanged!() -> (then, bNewInputType:ECommonInputType Enum) {is=title K2Node_FunctionEntry}
   > SetVisibility(
    self=Get Image_LinearFade() -> (Image_LinearFade:Image Object Reference)
    InVisibility=Select!(
     MouseAndKeyboard:ESlateVisibility Enum=Collapsed
     Gamepad:ESlateVisibility Enum=SelfHitTestInvisible
     Touch:ESlateVisibility Enum=Collapsed
     Count:ESlateVisibility Enum=Collapsed
     Index:ECommonInputType Enum=@n1
    ) -> (ReturnValue)
   )
  ]}
 SetInterpolationValue {nodes [
   n1 = SetInterpolationValue!() -> (then, NewParam:double=~) {is=title K2Node_FunctionEntry}
   > n2 = GetDynamicMaterial(self=Get Border())
   n3 = Get ColorA() -> (ColorA:Linear Color Structure)
   n4 = RGB to HSV!(InColor:Linear Color Structure=@n3) -> (H:float, S:float, V:float, A:float)
   n5 = float + float(A=@n4.V, B=0.55)
   > SetVectorParameterValue(
    self=@n2
    ParameterName=Get ColorA_Name() -> (ColorA_Name=None)
    Value="Lerp (LinearColor)"(A=@n3, B=HSV to RGB!(H:float=@n4, S:float=@n4.S, V:float=@n5, A:float=@n4.A) -> (ReturnValue:Linear Color Structure), Alpha=@n1)
   )
   n6 = Get ColorB() -> (ColorB:Linear Color Structure)
   n7 = RGB to HSV(InColor=@n6)
   n8 = float + float(A=@n7.V, B=0.55)
   > SetVectorParameterValue(
    self=@n2
    ParameterName=Get ColorB_Name() -> (ColorB_Name=None)
    Value="Lerp (LinearColor)"(A=@n6, B=HSV to RGB(H=@n7, S=@n7.S, V=@n8, A=@n7.A), Alpha=@n1)
   )
   n9 = Get RadialGradientA() -> (RadialGradientA:Linear Color Structure)
   n10 = RGB to HSV(InColor=@n9)
   n11 = float - float(A=@n10.A, B=0.1)
   > SetVectorParameterValue(
    self=@n2
    ParameterName=Get RadialGradientA_Name() -> (RadialGradientA_Name=None)
    Value="Lerp (LinearColor)"(A=@n9, B=HSV to RGB(H=@n10, S=@n10.S, V=@n10.V, A=@n11), Alpha=@n1)
   )
   n12 = Get RadialGradientB() -> (RadialGradientB:Linear Color Structure)
   n13 = RGB to HSV(InColor=@n12)
   n14 = float - float(A=@n13.A, B=0.1)
   > SetVectorParameterValue(
    self=@n2
    ParameterName=Get RadialGradientB_Name() -> (RadialGradientB_Name=None)
    Value="Lerp (LinearColor)"(A=@n12, B=HSV to RGB(H=@n13, S=@n13.S, V=@n13.V, A=@n14), Alpha=@n1)
   )
  ]}
]
variables [
 ColorA {type=LinearColor, default=<rgba 0 0 0 0>}
 ColorA_Name {type=string, default=LinearGradientColorA}
 ColorB {type=LinearColor, default=<rgba 0 0 0 0>}
 ColorB_Name {type=string, default=LinearGradientColorB}
 InterpolationValue {type=number}
 RadialGradientA {type=LinearColor, default=<rgba 0 0 0 0>}
 RadialGradientA_Name {type=string, default=RadialGradientOneColor}
 RadialGradientB {type=LinearColor, default=<rgba 0 0 0 0>}
 RadialGradientB_Name {type=string, default=RadialGradientTwoColor}
]
widgets {
 info {
  inheritedWidgetCount=2
  parentClass=/Script/FortniteUI.RebootRallySelectionHeaderBase
  rootWidgetClass=&CommonBorder
  widgetCount=9
 }
 widgets [
  Border {
   widgetClassPath=&CommonBorder
   bIsVariable=true
   widget=Border
   properties {
    clipping=ClipToBounds
    background {
     bIsDynamicallyLoaded=false
     drawAs=Image
     imageType=FullColor
     mirroring=NoMirror
     resourceName=None
     resourceObject=/Game/UI/Sidebar/RebootRally/Materials/MI_UI_RebootRally_BGClouds_S32_SelectHeader
     tiling=NoTile
     imageSize=<0 0>
     margin=<ltrb 0 0 0 0>
     outlineSettings {
      bUseBrushTransparency=false
      roundingType=HalfHeightRadius
      width=0
      color {colorUseRule=UseColor_Specified, specifiedColor=<rgba 0 0 0 0>}
      cornerRadii=<0 0 0 0>
     }
     tintColor {colorUseRule=UseColor_Specified, specifiedColor=<rgba 1 1 1 1>}
     uVRegion {bIsValid=false, max=<0 0>, min=<0 0>}
    }
    padding=<ltrb 0 0 0 0>
   }
  }
  GridPanel {
   widgetClassPath=&GridPanel
   parent=Border
   widget=GridPanel
   properties {columnFill [0,1], rowFill [1]}
   slot {class=BorderSlot, properties {padding=<ltrb 0 0 0 0>}}
  }
  Overlay_RebootRallyRewards {
   widgetClassPath=&Overlay
   parent=GridPanel
   widget=Overlay_RebootRallyRewards
   slot {class=GridSlot, properties {column=1, layer=1}}
  }
  Image_Rewards {
   widgetClassPath=&Image
   bIsVariable=true
   parent=Overlay_RebootRallyRewards
   widget=Image_Rewards
   properties {
    visibility=SelfHitTestInvisible
    brush {
     bIsDynamicallyLoaded=false
     drawAs=Image
     imageType=NoImage
     mirroring=NoMirror
     resourceName=None
     resourceObject=/Game/UI/Sidebar/RebootRally/Materials/MI_UI_Reward_Composition_SelectionHeader
     tiling=NoTile
     imageSize=<32 32>
     margin=<ltrb 0 0 0 0>
     outlineSettings {
      bUseBrushTransparency=false
      roundingType=HalfHeightRadius
      width=0
      color {colorUseRule=UseColor_Specified, specifiedColor=<rgba 0 0 0 0>}
      cornerRadii=<0 0 0 0>
     }
     tintColor {colorUseRule=UseColor_Specified, specifiedColor=<rgba 1 1 1 1>}
     uVRegion {bIsValid=false, max=<0 0>, min=<0 0>}
    }
   }
   slot {class=OverlaySlot, properties {horizontalAlignment=HAlign_Fill, verticalAlignment=VAlign_Fill}}
  }
  WBP_RebootRally_Logo_Banner {
   widgetClassPath=/Game/UI/Sidebar/RebootRally/WBP_RebootRally_Logo_Banner
   bIsVariable=true
   parent=GridPanel
   widget=WBP_RebootRally_Logo_Banner
   slot {class=GridSlot, properties {horizontalAlignment=HAlign_Left, layer=1, verticalAlignment=VAlign_Center, padding=<ltrb 24 0 0 4>}}
  }
  WBP_RebootRallyTimer {
   widgetClassPath=/Game/UI/Sidebar/RebootRally/WBP_RebootRallyTimer
   bInherited=true
   bIsVariable=true
   parent=GridPanel
   widget=WBP_RebootRallyTimer
   slot {class=GridSlot, properties {column=1, horizontalAlignment=HAlign_Right, layer=2, verticalAlignment=VAlign_Top, padding=<ltrb 0 4 4 0>}}
  }
  Image_LinearFade {
   widgetClassPath=&Image
   bIsVariable=true
   parent=GridPanel
   widget=Image_LinearFade
   properties {
    visibility=Collapsed
    brush {
     bIsDynamicallyLoaded=false
     drawAs=Image
     imageType=NoImage
     mirroring=NoMirror
     resourceName=None
     resourceObject=/Game/UI/Sidebar/RebootRally/Materials/MI_UI_RebootRallySelectionHeader_ActionLinearFadeBack
     tiling=NoTile
     imageSize=<80 32>
     margin=<ltrb 0 0 0 0>
     outlineSettings {
      bUseBrushTransparency=false
      roundingType=HalfHeightRadius
      width=0
      color {colorUseRule=UseColor_Specified, specifiedColor=<rgba 0 0 0 0>}
      cornerRadii=<0 0 0 0>
     }
     tintColor {colorUseRule=UseColor_Specified, specifiedColor=<rgba 1 1 1 1>}
     uVRegion {bIsValid=false, max=<0 0>, min=<0 0>}
    }
   }
   slot {class=GridSlot, properties {column=1, horizontalAlignment=HAlign_Right, layer=2}}
  }
  CommonActionWidget {
   widgetClassPath=&CommonActionWidget
   bIsVariable=true
   parent=GridPanel
   widget=CommonActionWidget
   properties {renderOpacity=17, visibility=SelfHitTestInvisible, inputActions [{dataTable=&InputActionDataTable, rowName=Select}]}
   slot {class=GridSlot, properties {column=1, horizontalAlignment=HAlign_Right, layer=3, verticalAlignment=VAlign_Center, padding=<ltrb 0 0 15 0>}}
  }
  InputActionWidget {widgetClassPath=&CommonActionWidget, bInherited=true, widget=null}
 ]
}
