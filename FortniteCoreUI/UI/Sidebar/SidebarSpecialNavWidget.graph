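# Sidebar special-nav widget that shows a Reboot Rally selection header button with an availability tooltip overlaid on it.
# The native FortSidebarSpecialNavWidgetBase parent supplies the button and tooltip bindings.
# On initialization, mobile builds attempt vertical padding via a VerticalBoxSlot cast, though Button sits in an Overlay slot.
# It then shifts the tooltip overlay horizontally by render translation, choosing sides by whether Discover-in-lobby is enabled.
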WidgetBlueprint /FortniteCoreUI/UI/Sidebar/SidebarSpecialNavWidget
graphs [
 EventGraph {nodes [
   Event OnInitialized()
   > n1 = Sequence()
   > Branch(Condition=IsMobileGame())
   > SetPadding!(
    execute
    self:Vertical Box Slot Object Reference=SlotAsVerticalBoxSlot(Widget=Get Button() -> (Button:Fort Sidebar Special Nav Button Base Object Reference))
    InPadding_Left:float=0
    InPadding_Top:float=15
    InPadding_Right:float=0
    InPadding_Bottom:float=15
   ) -> (then)
   SetToolTipAlignment(execute=@n1.then_1)
  ]}
 SetToolTipAlignment {nodes [
   SetToolTipAlignment() {is=title K2Node_FunctionEntry}
   n1 = GetRuntimeOptions()
   n2 = Self-Reference()
   > SetRenderTranslation(
    self=Get Overlay_1() -> (Overlay_1:Overlay Object Reference)
    Translation=Select(
     Option 0:Vector 2D Structure=(X=490,Y=0)
     Option 1:Vector 2D Structure=(X=-550,Y=0)
     Index=GetIsDiscoverInLobbyEnabled!(self:Fort Runtime Options Object Reference=@n1, ContextWidget:Widget Object Reference=@n2) -> (ReturnValue:bool)
    ) -> (ReturnValue:Vector 2D Structure)
   )
  ]}
]
variables []
widgets {
 info {
  inheritedWidgetCount=2
  parentClass=/Script/FortniteUI.FortSidebarSpecialNavWidgetBase
  rootWidgetClass=&VerticalBox
  widgetCount=5
 }
 widgets [
  VerticalBox_0 {widgetClassPath=&VerticalBox, widget=VerticalBox_0}
  Overlay_0 {widgetClassPath=&Overlay, parent=VerticalBox_0, widget=Overlay_0, slot {class=VerticalBoxSlot}}
  Button {
   widgetClassPath=/FortniteCoreUI/UI/Sidebar/RebootRally/WBP_RebootRallySelectionHeader
   bInherited=true
   bIsVariable=true
   parent=Overlay_0
   widget=Button
   slot {class=OverlaySlot, properties {horizontalAlignment=HAlign_Fill, verticalAlignment=VAlign_Fill}}
  }
  Overlay_1 {
   widgetClassPath=&Overlay
   bIsVariable=true
   parent=Overlay_0
   widget=Overlay_1
   properties {renderTransform {angle=0, scale=<1 1>, shear=<0 0>, translation=<-530 0>}}
   slot {class=OverlaySlot, properties {horizontalAlignment=HAlign_Right, verticalAlignment=VAlign_Center}}
  }
  AvailableTooltip {
   widgetClassPath=/FortniteCoreUI/UI/Sidebar/RebootRally/RebootRallyAvailableTooltip
   bInherited=true
   parent=Overlay_1
   widget=AvailableTooltip
   slot {class=OverlaySlot, properties {horizontalAlignment=HAlign_Fill, verticalAlignment=VAlign_Fill}}
  }
 ]
}
