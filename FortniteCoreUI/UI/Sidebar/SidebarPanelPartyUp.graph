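# Sidebar panel widget for the social Party Up view, built on the native FortSidebarPanelPartyUp class.
# A vertical layout stacks a fixed-height header with an uppercase, sheared "Social" title, an initially collapsed Reboot Rally nav widget, and a fill-sized social user tree.
# The inherited tree assigns entry widget classes for party members, invites, join requests, suggested friends, search results, voice chat and other row types.
# The Blueprint graph holds only unconnected stub events.
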WidgetBlueprint /FortniteCoreUI/UI/Sidebar/SidebarPanelPartyUp
graphs [EventGraph {nodes [
   Event PreConstruct()
   Event Construct()
   Event Tick()
  ]}]
variables []
widgets {
 info {inheritedWidgetCount=2, parentClass=/Script/FortniteUI.FortSidebarPanelPartyUp, rootWidgetClass=&VerticalBox, widgetCount=7}
 widgets [
  VerticalBox_1 {widgetClassPath=&VerticalBox, widget=VerticalBox_1}
  SizeBox_1 {
   widgetClassPath=&SizeBox
   parent=VerticalBox_1
   widget=SizeBox_1
   properties {bOverride_HeightOverride=true, heightOverride=120}
   slot {class=VerticalBoxSlot}
  }
  Border_141 {
   widgetClassPath=&Border
   parent=SizeBox_1
   widget=Border_141
   properties {
    flowDirectionPreference=Culture
    background {
     bIsDynamicallyLoaded=false
     drawAs=Image
     imageType=NoImage
     mirroring=NoMirror
     resourceName=None
     resourceObject=/Game/UI/Sidebar/Materials/MI_UI_Sidebar_Header_Base
     tiling=NoTile
     imageSize=<4096 4096>
     margin=<ltrb 0 0 0 0>
     outlineSettings {
      bUseBrushTransparency=false
      roundingType=HalfHeightRadius
      width=0
      color {colorUseRule=UseColor_Specified, specifiedColor=<rgba 0 0 0 0>}
      cornerRadii=<0 0 0 0>
     }
     tintColor {colorUseRule=UseColor_Specified, specifiedColor=<rgba 1 1 1 1>}
     uVRegion {bIsValid=false, max=<0 0>, min=<0 0>}
    }
    padding=<ltrb 0 0 0 50>
   }
   slot {class=SizeBoxSlot}
  }
  VerticalBox_5 {
   widgetClassPath=&VerticalBox
   parent=Border_141
   widget=VerticalBox_5
   slot {class=BorderSlot, properties {padding=<ltrb 0 0 0 50>}}
  }
  Text_PartyUpTitle {
   widgetClassPath=&CommonTextBlock
   parent=VerticalBox_5
   widget=Text_PartyUpTitle
   properties {
    text=Social
    textTransformPolicy=ToUpper
    font {
     bForceMonospaced=false
     bMaterialIsStencil=false
     fontMaterial=null
     fontObject=&BurbankSmall
     letterSpacing=0
     monospacedWidth=1
     size=32
     skewAmount=0
     typefaceFontName=Black
     outlineSettings {
      bApplyOutlineToDropShadows=false
      bMiteredCorners=false
      bSeparateFillAlpha=false
      outlineMaterial=null
      outlineSize=0
      outlineColor=<rgba 0 0 0 1>
     }
    }
    renderTransform {angle=0, scale=<1 1>, shear=<-8 0>, translation=<0 0>}
    shadowOffset=<0 0>
   }
   slot {class=VerticalBoxSlot, properties {horizontalAlignment=HAlign_Left, verticalAlignment=VAlign_Bottom, padding=<ltrb 20 60 20 0>}}
  }
  SpecialNavWidget {
   widgetClassPath=&/SidebarSpecialNavWidget
   bInherited=true
   bIsVariable=true
   parent=VerticalBox_1
   widget=SpecialNavWidget
   properties {bIsFocusable=true, navType=GoRebootRally, visibility=Collapsed}
   slot {class=VerticalBoxSlot, properties {verticalAlignment=VAlign_Top}}
  }
  UserTree_PartyUp {
   widgetClassPath=/Script/FortniteUI.FortSocialUserTreeView
   bInherited=true
   parent=VerticalBox_1
   widget=UserTree_PartyUp
   properties {
    allowOverscroll=false
    bAllowDragging=false
    bEnableRightClickScrolling=false
    bEnableScrollAnimation=false
    entryWidgetClass=&/SidebarUserListHeaderEntry.SidebarUserListHeaderEntry_C
    inviteEntryWidgetClass=&/SidebarInviteListEntry_FriendInvite_XS.SidebarInviteListEntry_FriendInvite_XS_C
    joinRequestEntryWidgetClass=/FortniteCoreUI/UI/Sidebar/ListEntries/SidebarJoinRequestUserListEntry.SidebarJoinRequestUserListEntry_C
    joinablePartyEntryWidgetClass=&/SidebarJoinablePartyEntry.SidebarJoinablePartyEntry_C
    joinablePartyEntryWidgetClass_2=/FortniteCoreUI/UI/Sidebar/ListEntries/JoinableParty_UserListEntry.JoinableParty_UserListEntry_C
    leavePartyEntryWidgetClass=&/SidebarLeavePartyListEntry.SidebarLeavePartyListEntry_C
    lfgState2EntryWidgetClass=/FortniteCoreUI/UI/Sidebar/Tags/SidebarLfgState2ListEntry.SidebarLfgState2ListEntry_C
    missedInviteEntryWidgetClass=/FortniteCoreUI/UI/SocialPanel/MissedInvitesEntry.MissedInvitesEntry_C
    partyMemberEntryWidgetClass=/FortniteCoreUI/UI/Sidebar/ListEntries/SidebarPartyMemberListEntry.SidebarPartyMemberListEntry_C
    referFriendEntryWidgetClass=&/SidebarReferAFriendEntry.SidebarReferAFriendEntry_C
    suggestedFriendEntryWidgetClass=/FortniteCoreUI/UI/Sidebar/ListEntries/SidebarSuggestedFriendListEntry.SidebarSuggestedFriendListEntry_C
    teamMemberEntryWidgetClass=/FortniteCoreUI/UI/Sidebar/ListEntries/SidebarTeamMemberListEntry.SidebarTeamMemberListEntry_C
    userEntryWidgetClass=/FortniteCoreUI/UI/Sidebar/ListEntries/SidebarUserListEntry_XS.SidebarUserListEntry_XS_C
    userSearchResultEntryWidgetClass=/FortniteCoreUI/UI/Sidebar/ListEntries/SidebarUserSearchResultListEntry_XS.SidebarUserSearchResultListEntry_XS_C
    voiceChannelHeaderEntryWidgetClass=&/SidebarVoiceChannelHeaderButton.SidebarVoiceChannelHeaderButton_C
    voiceChatMemberEntryWidgetClass=/FortniteCoreUI/UI/SocialPanel/ListEntries/VoiceChatMemberListEntry.VoiceChatMemberListEntry_C
   }
   slot {class=VerticalBoxSlot, properties {size {sizeRule=Fill, value=1}}}
  }
 ]
}
