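# Sidebar button widget for empty list panels, derived from SidebarListEmptyPanelButton, with default text "Go to Parental Controls".
# A rounded-material border wraps a row holding the inherited text block and input action icon.
# Hover events stop running animations, then play short mirrored animations that swap the border between blue and white and the text between light blue and dark.
# When the input icon updates, its dynamic material switches to dark with no outline.
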WidgetBlueprint /FortniteCoreUI/UI/Sidebar/SidebarEmptyPanelButton
animations [
 Hovered {
  binding_count=2
  length_seconds=0.1
  object=&self:Hovered
  sequence {bindings [
    ButtonText {
     id=#1
     display_name=CommonTextBlock_24
     possessed_class=CommonTextBlock
     tracks [{
       type=MovieSceneColorTrack
       display_name=Color and Opacity
       sections [{
         type=MovieSceneColorSection
         end_frame=2
         start_frame=0
         channels [
          Color.R {type=MovieSceneScriptingFloatChannel, keys [[0,0.3325],[2,0.0513]]}
          Color.G {type=MovieSceneScriptingFloatChannel, keys [[0,0.6724],[2,0.0513]]}
          Color.B {type=MovieSceneScriptingFloatChannel, keys [[0,1],[2,0.0513]]}
          Color.A {type=MovieSceneScriptingFloatChannel, const [1,2]}
         ]
        }]
      }]
    }
    EntryBorder {
     id=#2
     possessed_class=Border
     tracks [{
       type=MovieSceneColorTrack
       display_name=Brush Color
       sections [{
         type=MovieSceneColorSection
         end_frame=2
         start_frame=0
         channels [
          Color.R {type=MovieSceneScriptingFloatChannel, keys [[0,0.0044],[2,1]]}
          Color.G {type=MovieSceneScriptingFloatChannel, keys [[0,0.0723],[2,1]]}
          Color.B {type=MovieSceneScriptingFloatChannel, keys [[0,0.5149],[2,1]]}
          Color.A {type=MovieSceneScriptingFloatChannel, const [1,2]}
         ]
        }]
      }]
    }
   ]}
 }
 Unhovered {
  binding_count=2
  length_seconds=0.1
  object=&self:Unhovered
  sequence {bindings [
    ButtonText {
     id=#1
     display_name=CommonTextBlock_24
     possessed_class=CommonTextBlock
     tracks [{
       type=MovieSceneColorTrack
       display_name=Color and Opacity
       sections [{
         type=MovieSceneColorSection
         end_frame=2
         start_frame=0
         channels [
          Color.R {type=MovieSceneScriptingFloatChannel, keys [[0,0.0513],[2,0.3325]]}
          Color.G {type=MovieSceneScriptingFloatChannel, keys [[0,0.0513],[2,0.6724]]}
          Color.B {type=MovieSceneScriptingFloatChannel, keys [[0,0.0513],[2,1]]}
          Color.A {type=MovieSceneScriptingFloatChannel, const [1,2]}
         ]
        }]
      }]
    }
    EntryBorder {
     id=#2
     possessed_class=Border
     tracks [{
       type=MovieSceneColorTrack
       display_name=Brush Color
       sections [{
         type=MovieSceneColorSection
         end_frame=2
         start_frame=0
         channels [
          Color.R {type=MovieSceneScriptingFloatChannel, keys [[0,1],[2,0.0044]]}
          Color.G {type=MovieSceneScriptingFloatChannel, keys [[0,1],[2,0.0723]]}
          Color.B {type=MovieSceneScriptingFloatChannel, keys [[0,1],[2,0.5149]]}
          Color.A {type=MovieSceneScriptingFloatChannel, const [1,2]}
         ]
        }]
      }]
    }
   ]}
 }
]
graphs [EventGraph {nodes [
   "On Input Icon Updated (InputActionWidget)"() {is=title K2Node_ComponentBoundEvent}
   > n1 = GetIconDynamicMaterial!(execute, self:Common Action Widget Object Reference=Get InputActionWidget() -> (InputActionWidget:Common Action Widget Object Reference)) -> (then, ReturnValue:Material Instance Dynamic Object Reference)
   > SetScalarParameterValue(self=@n1, ParameterName=Use Dark, Value=1)
   > SetScalarParameterValue(self=@n1, ParameterName=Outline Opacity, Value=0)
   Event On Hovered()
   > StopAllAnimations(self:User Widget Object Reference)
   > PlayAnimation(InAnimation=Get Hovered() -> (Hovered:Widget Animation Object Reference))
   Event On Unhovered()
   > StopAllAnimations()
   > PlayAnimation(InAnimation=Get Unhovered() -> (Unhovered:Widget Animation Object Reference))
   Event PreConstruct()
   Event Construct()
   Event Tick()
   Event On Clicked()
  ]}]
variables []
widgets {
 info {inheritedWidgetCount=2, parentClass=/Script/FortniteUI.SidebarListEmptyPanelButton, rootWidgetClass=&Overlay, widgetCount=7}
 widgets [
  Overlay_0 {widgetClassPath=&Overlay, widget=Overlay_0}
  SizeBox_AddFriendsButton {
   widgetClassPath=&SizeBox
   parent=Overlay_0
   widget=SizeBox_AddFriendsButton
   properties {bOverride_MinDesiredHeight=true, bOverride_MinDesiredWidth=true, heightOverride=72, minDesiredHeight=56, minDesiredWidth=360}
   slot {class=OverlaySlot, properties {horizontalAlignment=HAlign_Center}}
  }
  EntryBorder {
   widgetClassPath=&Border
   bIsVariable=true
   parent=SizeBox_AddFriendsButton
   widget=EntryBorder
   properties {
    visibility=SelfHitTestInvisible
    background {
     bIsDynamicallyLoaded=false
     drawAs=Box
     imageType=NoImage
     mirroring=NoMirror
     resourceName=None
     resourceObject=/Game/UI/Sidebar/Materials/M-Button-RoundCorner_SidebarInteractionRound
     tiling=NoTile
     imageSize=<104 104>
     margin=<ltrb 0.5 0.5 0.5 0.5>
     outlineSettings {
      bUseBrushTransparency=false
      roundingType=HalfHeightRadius
      width=10
      color {colorUseRule=UseColor_Specified, specifiedColor=<rgba 0 0 0 1>}
      cornerRadii=<0.5 0.5 0.5 0.5>
     }
     tintColor {colorUseRule=UseColor_Specified, specifiedColor=<rgba 1 1 1 1>}
     uVRegion {bIsValid=false, max=<0 0>, min=<0 0>}
    }
    brushColor=<rgba 0.0044 0.0723 0.5149 1>
    padding=<ltrb 36 36 36 36>
   }
   slot {class=SizeBoxSlot, properties {padding=<ltrb -24 -24 -24 -24>}}
  }
  Overlay_1 {widgetClassPath=&Overlay, parent=EntryBorder, widget=Overlay_1, slot {class=BorderSlot, properties {padding=<ltrb 36 36 36 36>}}}
  HorizontalBox_0 {
   widgetClassPath=&HorizontalBox
   parent=Overlay_1
   widget=HorizontalBox_0
   slot {class=OverlaySlot, properties {horizontalAlignment=HAlign_Fill, verticalAlignment=VAlign_Fill}}
  }
  ButtonText {
   widgetClassPath=&CommonTextBlock
   bInherited=true
   parent=HorizontalBox_0
   widget=ButtonText
   properties {
    clipping=OnDemand
    scrollStyle=&TextScrollStyle-Base_C
    text=Go to Parental Controls
    textTransformPolicy=ToUpper
    colorAndOpacity {colorUseRule=UseColor_Specified, specifiedColor=<rgba 0.3325 0.6724 1 1>}
    font {
     bForceMonospaced=false
     bMaterialIsStencil=false
     fontMaterial=null
     fontObject=&BurbankSmall
     letterSpacing=0
     monospacedWidth=1
     size=18
     skewAmount=0
     typefaceFontName=Black
     outlineSettings {
      bApplyOutlineToDropShadows=false
      bMiteredCorners=false
      bSeparateFillAlpha=false
      outlineMaterial=null
      outlineSize=0
      outlineColor=<rgba 0 0 0 1>
     }
    }
    shadowOffset=<0 0>
   }
   slot {class=HorizontalBoxSlot, properties {size {sizeRule=Fill, value=1}}}
  }
  InputActionWidget {
   widgetClassPath=&CommonActionWidget
   bInherited=true
   bIsVariable=true
   parent=HorizontalBox_0
   widget=InputActionWidget
   slot {class=HorizontalBoxSlot}
  }
 ]
}
