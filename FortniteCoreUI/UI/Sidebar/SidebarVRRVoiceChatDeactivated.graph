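# Sidebar panel widget for deactivated voice chat, showing a voice reporting update notice with a header, headset icon and two buttons.
# Built on SocialListEntryEmptyPanelBase, it forwards both button clicks to the parent and fills title, description and button labels through UpdateText.
# OnSidebarPanelVoiceChatActivated waits briefly, then collapses the icon's ScaleBox in splitscreen or when its painted height is too small, using the delay before measuring geometry.
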WidgetBlueprint /FortniteCoreUI/UI/Sidebar/SidebarVRRVoiceChatDeactivated
graphs [EventGraph {nodes [
   "On Clicked (WBP_UIKit_Button_Regular)"!() -> (OutputDelegate, then, Button:Common Button Base Object Reference) {is=title K2Node_ComponentBoundEvent}
   > ButtonClicked(self:Social List Entry Empty Panel Base Object Reference)
   "On Clicked (WBP_UIKit_Button_Regular_Sec)"!() -> (OutputDelegate, then, Button:Common Button Base Object Reference) {is=title K2Node_ComponentBoundEvent}
   > ButtonSecClicked(self:Social List Entry Empty Panel Base Object Reference)
   OnSidebarPanelVoiceChatActivated() {is=title K2Node_CustomEvent}
   > Delay(Duration=0.1)
   n1 = GetFirstLocalFortPlayerController!(bIncludeNonSpectators:bool=true, bIncludeSpectators:bool=false) -> (ReturnValue:Fort Player Controller Object Reference)
   n2 = GetLocalSize!(Geometry:"Geometry Structure (by ref)"=GetPaintSpaceGeometry!(self:Widget Object Reference=Get ScaleBox() -> (ScaleBox:Scale Box Object Reference)) -> (ReturnValue:"Geometry Structure (by ref)")) -> (ReturnValue_X:double, ReturnValue_Y:double)
   > Branch(Condition=OR Boolean(A=IsInSplitscreen!(self:Fort Player Controller Object Reference=@n1) -> (ReturnValue:bool), B="float > float"(A=120, B=@n2.ReturnValue_Y)))
   > SetVisibility(self=Get ScaleBox(), InVisibility=Collapsed)
   n3 = Event UpdateText!() -> (OutputDelegate, then, Title:"Text (by ref)", Description:"Text (by ref)", ButtonText:"Text (by ref)", ButtonSecText:"Text (by ref)") {is=title K2Node_Event}
   > Set Text(Text=@n3.Title, self:Rich Text Block Object Reference=Get EmptyListTitle() -> (EmptyListTitle:Common Rich Text Block Object Reference))
   > Set Text(Text=@n3.Description, self=Get EmptyListDescription() -> (EmptyListDescription:Common Rich Text Block Object Reference))
   > SetText!(
    execute
    self:Fort CTAButton Object Reference=Get WBP_UIKit_Button_Regular() -> (WBP_UIKit_Button_Regular:Fort CTAButton Object Reference)
    ButtonText:"Text (by ref)"=@n3.ButtonText
   ) -> (then)
   > SetText(
    self=Get WBP_UIKit_Button_Regular_Sec() -> (WBP_UIKit_Button_Regular_Sec:WBP UIKit Button Regular Object Reference)
    ButtonText=@n3.ButtonSecText
   )
  ]}]
variables []
widgets {
 info {
  inheritedWidgetCount=7
  parentClass=/Script/FortniteUI.SocialListEntryEmptyPanelBase
  rootWidgetClass=&VerticalBox
  widgetCount=14
 }
 widgets [
  VerticalBox_0 {widgetClassPath=&VerticalBox, widget=VerticalBox_0}
  Border_1 {
   widgetClassPath=&Border
   parent=VerticalBox_0
   widget=Border_1
   properties {
    horizontalAlignment=HAlign_Left
    verticalAlignment=VAlign_Bottom
    background {
     bIsDynamicallyLoaded=false
     drawAs=Image
     imageType=NoImage
     mirroring=NoMirror
     resourceName=None
     resourceObject=/Game/UI/Sidebar/Materials/MI_UI_Sidebar_Header_Base
     tiling=NoTile
     imageSize=<4096 4096>
     margin=<ltrb 0 0 0 0>
     outlineSettings {
      bUseBrushTransparency=false
      roundingType=HalfHeightRadius
      width=0
      color {colorUseRule=UseColor_Specified, specifiedColor=<rgba 0 0 0 0>}
      cornerRadii=<0 0 0 0>
     }
     tintColor {colorUseRule=UseColor_Specified, specifiedColor=<rgba 1 1 1 1>}
     uVRegion {bIsValid=false, max=<0 0>, min=<0 0>}
    }
    padding=<ltrb 20 60 20 10>
   }
   slot {class=VerticalBoxSlot, properties {verticalAlignment=VAlign_Top, padding=<ltrb 0 0 0 32>}}
  }
  VerticalBox_6 {
   widgetClassPath=&VerticalBox
   parent=Border_1
   widget=VerticalBox_6
   slot {class=BorderSlot, properties {horizontalAlignment=HAlign_Left, verticalAlignment=VAlign_Bottom, padding=<ltrb 20 60 20 10>}}
  }
  Text_VoiceChatTitle {
   widgetClassPath=&CommonTextBlock
   parent=VerticalBox_6
   widget=Text_VoiceChatTitle
   properties {
    text=Voice Chat
    textTransformPolicy=ToUpper
    font {
     bForceMonospaced=false
     bMaterialIsStencil=false
     fontMaterial=null
     fontObject=&BurbankSmall
     letterSpacing=0
     monospacedWidth=1
     size=32
     skewAmount=0
     typefaceFontName=Black
     outlineSettings {
      bApplyOutlineToDropShadows=false
      bMiteredCorners=false
      bSeparateFillAlpha=false
      outlineMaterial=null
      outlineSize=0
      outlineColor=<rgba 0 0 0 1>
     }
    }
    renderTransform {angle=0, scale=<1 1>, shear=<-8 0>, translation=<0 0>}
    shadowOffset=<0 0>
   }
   slot {class=VerticalBoxSlot, properties {horizontalAlignment=HAlign_Left, verticalAlignment=VAlign_Top}}
  }
  ScaleBox {
   widgetClassPath=&ScaleBox
   bIsVariable=true
   parent=VerticalBox_0
   widget=ScaleBox
   slot {class=VerticalBoxSlot, properties {verticalAlignment=VAlign_Bottom, padding=<ltrb 0 0 0 32>, size {sizeRule=Fill, value=1}}}
  }
  EmptyListImage {
   widgetClassPath=&Image
   bInherited=true
   bIsVariable=true
   parent=ScaleBox
   widget=EmptyListImage
   properties {
    visibility=SelfHitTestInvisible
    brush {
     bIsDynamicallyLoaded=false
     drawAs=Image
     imageType=NoImage
     mirroring=NoMirror
     resourceName=None
     resourceObject=/Game/UI/Sidebar/Materials/MI_UI_Icon_Headset_Mic
     tiling=NoTile
     imageSize=<160 160>
     margin=<ltrb 0 0 0 0>
     outlineSettings {
      bUseBrushTransparency=false
      roundingType=HalfHeightRadius
      width=0
      color {colorUseRule=UseColor_Specified, specifiedColor=<rgba 0 0 0 0>}
      cornerRadii=<0 0 0 0>
     }
     tintColor {colorUseRule=UseColor_Specified, specifiedColor=<rgba 0.0331 0.402 1 1>}
     uVRegion {bIsValid=false, max=<0 0>, min=<0 0>}
    }
   }
   slot {class=ScaleBoxSlot}
  }
  EmptyListTitle {
   widgetClassPath=&CommonRichTextBlock
   bInherited=true
   bIsVariable=true
   parent=VerticalBox_0
   widget=EmptyListTitle
   properties {
    autoWrapText=true
    bOverrideDefaultStyle=true
    justification=Center
    text=Voice Reporting Update
    textTransformPolicy=ToUpper
    defaultTextStyleOverride {
     overflowPolicy=Clip
     transformPolicy=None
     colorAndOpacity {colorUseRule=UseColor_Specified, specifiedColor=<rgba 1 1 1 1>}
     font {
      bForceMonospaced=false
      bMaterialIsStencil=false
      fontMaterial=null
      fontObject=&BurbankSmall
      letterSpacing=0
      monospacedWidth=1
      size=24
      skewAmount=0
      typefaceFontName=Black
      outlineSettings {
       bApplyOutlineToDropShadows=false
       bMiteredCorners=false
       bSeparateFillAlpha=false
       outlineMaterial=null
       outlineSize=0
       outlineColor=<rgba 0 0 0 1>
      }
     }
     highlightColor {colorUseRule=UseColor_Specified, specifiedColor=<rgba 0 0 0 1>}
     highlightShape {
      bIsDynamicallyLoaded=false
      drawAs=Image
      imageType=NoImage
      mirroring=NoMirror
      resourceName=None
      resourceObject=null
      tiling=NoTile
      imageSize=<32 32>
      margin=<ltrb 0 0 0 0>
      outlineSettings {
       bUseBrushTransparency=false
       roundingType=HalfHeightRadius
       width=0
       color {colorUseRule=UseColor_Specified, specifiedColor=<rgba 0 0 0 0>}
       cornerRadii=<0 0 0 0>
      }
      tintColor {colorUseRule=UseColor_Specified, specifiedColor=<rgba 1 1 1 1>}
      uVRegion {bIsValid=false, max=<0 0>, min=<0 0>}
     }
     selectedBackgroundColor {colorUseRule=UseColor_Foreground, specifiedColor=<rgba 1 0 1 1>}
     shadowColorAndOpacity=<rgba 0 0 0 1>
     shadowOffset=<0 0>
     strikeBrush {
      bIsDynamicallyLoaded=false
      drawAs=Image
      imageType=NoImage
      mirroring=NoMirror
      resourceName=None
      resourceObject=null
      tiling=NoTile
      imageSize=<32 32>
      margin=<ltrb 0 0 0 0>
      outlineSettings {
       bUseBrushTransparency=false
       roundingType=HalfHeightRadius
       width=0
       color {colorUseRule=UseColor_Specified, specifiedColor=<rgba 0 0 0 0>}
       cornerRadii=<0 0 0 0>
      }
      tintColor {colorUseRule=UseColor_Specified, specifiedColor=<rgba 1 1 1 1>}
      uVRegion {bIsValid=false, max=<0 0>, min=<0 0>}
     }
     underlineBrush {
      bIsDynamicallyLoaded=false
      drawAs=Image
      imageType=NoImage
      mirroring=NoMirror
      resourceName=None
      resourceObject=null
      tiling=NoTile
      imageSize=<32 32>
      margin=<ltrb 0 0 0 0>
      outlineSettings {
       bUseBrushTransparency=false
       roundingType=HalfHeightRadius
       width=0
       color {colorUseRule=UseColor_Specified, specifiedColor=<rgba 0 0 0 0>}
       cornerRadii=<0 0 0 0>
      }
      tintColor {colorUseRule=UseColor_Specified, specifiedColor=<rgba 1 1 1 1>}
      uVRegion {bIsValid=false, max=<0 0>, min=<0 0>}
     }
    }
   }
   slot {class=VerticalBoxSlot, properties {verticalAlignment=VAlign_Top, padding=<ltrb 32 0 32 0>}}
  }
  EmptyListDescription {
   widgetClassPath=&CommonRichTextBlock
   bInherited=true
   bIsVariable=true
   parent=VerticalBox_0
   widget=EmptyListDescription
   properties {
    autoWrapText=true
    bIsScrollingEnabled=false
    bOverrideDefaultStyle=true
    lineHeightPercentage=0.85
    text="We've updated voice reporting to help us review reported violations of our Community Rules.\r\n- When voice chat is turned on, this feature stores voice chat audio on each participant's device on a rolling five-minute basis, which will be sent to Epic Games only if a participant reports a violation.\r\n- Voice reporting is usually on, except in parties formed among friends where all players prefer voice reporting off.\r\n- You can see if voice reporting is currently on or off in the Sidebar.\r\nSee your voice chat settings for more details and options.We've updated voice reporting to help us review reported violations of our Community Rules.\r\n"
    defaultTextStyleOverride {
     overflowPolicy=Clip
     transformPolicy=None
     colorAndOpacity {colorUseRule=UseColor_Specified, specifiedColor=<rgba 0.5089 0.5583 0.7084 1>}
     font {
      bForceMonospaced=false
      bMaterialIsStencil=false
      fontMaterial=null
      fontObject=&BurbankSmall
      letterSpacing=0
      monospacedWidth=1
      size=13.5
      skewAmount=0
      typefaceFontName=Bold
      outlineSettings {
       bApplyOutlineToDropShadows=false
       bMiteredCorners=false
       bSeparateFillAlpha=false
       outlineMaterial=null
       outlineSize=0
       outlineColor=<rgba 0 0 0 1>
      }
     }
     highlightColor {colorUseRule=UseColor_Specified, specifiedColor=<rgba 0 0 0 1>}
     highlightShape {
      bIsDynamicallyLoaded=false
      drawAs=Image
      imageType=NoImage
      mirroring=NoMirror
      resourceName=None
      resourceObject=null
      tiling=NoTile
      imageSize=<32 32>
      margin=<ltrb 0 0 0 0>
      outlineSettings {
       bUseBrushTransparency=false
       roundingType=HalfHeightRadius
       width=0
       color {colorUseRule=UseColor_Specified, specifiedColor=<rgba 0 0 0 0>}
       cornerRadii=<0 0 0 0>
      }
      tintColor {colorUseRule=UseColor_Specified, specifiedColor=<rgba 1 1 1 1>}
      uVRegion {bIsValid=false, max=<0 0>, min=<0 0>}
     }
     selectedBackgroundColor {colorUseRule=UseColor_Foreground, specifiedColor=<rgba 1 0 1 1>}
     shadowColorAndOpacity=<rgba 0 0 0 1>
     shadowOffset=<0 0>
     strikeBrush {
      bIsDynamicallyLoaded=false
      drawAs=Image
      imageType=NoImage
      mirroring=NoMirror
      resourceName=None
      resourceObject=null
      tiling=NoTile
      imageSize=<32 32>
      margin=<ltrb 0 0 0 0>
      outlineSettings {
       bUseBrushTransparency=false
       roundingType=HalfHeightRadius
       width=0
       color {colorUseRule=UseColor_Specified, specifiedColor=<rgba 0 0 0 0>}
       cornerRadii=<0 0 0 0>
      }
      tintColor {colorUseRule=UseColor_Specified, specifiedColor=<rgba 1 1 1 1>}
      uVRegion {bIsValid=false, max=<0 0>, min=<0 0>}
     }
     underlineBrush {
      bIsDynamicallyLoaded=false
      drawAs=Image
      imageType=NoImage
      mirroring=NoMirror
      resourceName=None
      resourceObject=null
      tiling=NoTile
      imageSize=<32 32>
      margin=<ltrb 0 0 0 0>
      outlineSettings {
       bUseBrushTransparency=false
       roundingType=HalfHeightRadius
       width=0
       color {colorUseRule=UseColor_Specified, specifiedColor=<rgba 0 0 0 0>}
       cornerRadii=<0 0 0 0>
      }
      tintColor {colorUseRule=UseColor_Specified, specifiedColor=<rgba 1 1 1 1>}
      uVRegion {bIsValid=false, max=<0 0>, min=<0 0>}
     }
    }
   }
   slot {class=VerticalBoxSlot, properties {verticalAlignment=VAlign_Top, padding=<ltrb 32 24 32 0>}}
  }
  Spacer_58 {
   widgetClassPath=&Spacer
   parent=VerticalBox_0
   widget=Spacer_58
   slot {class=VerticalBoxSlot, properties {size {sizeRule=Fill, value=0.5}}}
  }
  WBP_UIKit_Button_Regular {
   widgetClassPath=&WBP_UIKit_Button_Regular
   bInherited=true
   bIsVariable=true
   parent=VerticalBox_0
   widget=WBP_UIKit_Button_Regular
   properties {buttonSize=Small}
   slot {class=VerticalBoxSlot, properties {verticalAlignment=VAlign_Bottom, padding=<ltrb 32 32 32 16>}}
  }
  WBP_UIKit_Button_Regular_Sec {
   widgetClassPath=&WBP_UIKit_Button_Regular
   bIsVariable=true
   parent=VerticalBox_0
   widget=WBP_UIKit_Button_Regular_Sec
   properties {buttonSize=Small, buttonType=Quiet}
   slot {class=VerticalBoxSlot, properties {verticalAlignment=VAlign_Bottom, padding=<ltrb 32 4 32 64>}}
  }
  ButtonWidget {widgetClassPath=/Script/FortniteUI.SidebarListEmptyPanelButton, bInherited=true, widget=null}
  ButtonWidgetSec {widgetClassPath=&SidebarListEmptyPanelButton, bInherited=true, widget=null}
  InputActionWidget {widgetClassPath=&CommonActionWidget, bInherited=true, widget=null}
 ]
}
