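# Sidebar voice channel header widget (FortVoiceChannelHeaderButton child) showing channel status with expandable Join/Leave and Report options.
# Parent-class update events set transmit icon visibility, text colors, voice-permission and reporting switchers, and button availability.
# Options or header clicks toggle expansion, animating the arrow and retargeting down navigation to the first interactable button.
# Report requests open the voice reporting modal when recording, otherwise the player feedback modal.
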WidgetBlueprint /Game/UI/Sidebar/SidebarVoiceChannelHeaderOptions
animations [
 OnHover {
  binding_count=5
  display_label=""
  length_seconds=0.2
  object=&self:OnHover
  sequence {bindings [
    ExpandDirectionTick {
     id=#1
     possessed_class=Image
     tracks [{
       type=MovieSceneColorTrack
       display_name=Color and Opacity
       sections [{
         type=MovieSceneColorSection
         end_frame=3
         start_frame=0
         channels [
          Color.R {type=MovieSceneScriptingFloatChannel, keys [[0,0.5089],[3,1]]}
          Color.G {type=MovieSceneScriptingFloatChannel, keys [[0,0.5583],[3,1]]}
          Color.B {type=MovieSceneScriptingFloatChannel, keys [[0,0.7084],[3,1]]}
          Color.A {type=MovieSceneScriptingFloatChannel, const [1,1]}
         ]
        }]
      }]
    }
    AnimatedInputBorder {
     id=#2
     possessed_class=Border
     tracks [
      {
       type=MovieSceneWidgetMaterialTrack
       sections [{
         type=MovieSceneParameterSection
         channels [
          AnimateStateChange {type=MovieSceneScriptingFloatChannel, keys [[0,0],[3,1],[4,1]]}
          RGBR {type=MovieSceneScriptingFloatChannel, const [0.0021,1]}
          RGBG {type=MovieSceneScriptingFloatChannel, const [0.0497,1]}
          RGBB {type=MovieSceneScriptingFloatChannel, const [0.3712,1]}
          RGBA {type=MovieSceneScriptingFloatChannel, const [1,1]}
         ]
        }]
      }
      {
       type=MovieSceneColorTrack
       display_name=Brush Color
       sections [{
         type=MovieSceneColorSection
         channels [
          Color.R {type=MovieSceneScriptingFloatChannel}
          Color.G {type=MovieSceneScriptingFloatChannel}
          Color.B {type=MovieSceneScriptingFloatChannel}
          Color.A {type=MovieSceneScriptingFloatChannel, keys [[0,0],[2,1]]}
         ]
        }]
      }
     ]
    }
    Text_Options {
     id=#3
     possessed_class=CommonTextBlock
     tracks [{
       type=MovieSceneColorTrack
       display_name=Color and Opacity
       sections [{
         type=MovieSceneColorSection
         end_frame=4
         start_frame=0
         channels [
          Color.R {type=MovieSceneScriptingFloatChannel, keys [[0,0.5089],[4,1]]}
          Color.G {type=MovieSceneScriptingFloatChannel, keys [[0,0.5583],[4,1]]}
          Color.B {type=MovieSceneScriptingFloatChannel, keys [[0,0.7084],[4,1]]}
          Color.A {type=MovieSceneScriptingFloatChannel, const [1,1]}
         ]
        }]
      }]
    }
    EntryBorder {
     id=#4
     possessed_class=Border
     tracks [
      {
       type=MovieSceneWidgetMaterialTrack
       sections [{
         type=MovieSceneParameterSection
         end_frame=4
         start_frame=0
         channels [PrimaryAlpha {type=MovieSceneScriptingFloatChannel, keys [[0,0],[3,0.5],[4,0.5]]}]
        }]
      }
      {
       type=MovieScene2DTransformTrack
       sections [{
         type=MovieScene2DTransformSection
         end_frame=4
         start_frame=0
         channels [
          Angle {type=MovieSceneScriptingFloatChannel}
          Scale.X {type=MovieSceneScriptingFloatChannel}
          Scale.Y {type=MovieSceneScriptingFloatChannel}
          Shear.X {type=MovieSceneScriptingFloatChannel}
          Shear.Y {type=MovieSceneScriptingFloatChannel}
         ]
        }]
      }
      {
       type=MovieSceneColorTrack
       display_name=Brush Color
       sections [{
         type=MovieSceneColorSection
         end_frame=0
         start_frame=0
         channels [
          Color.R {type=MovieSceneScriptingFloatChannel, const [1,1]}
          Color.G {type=MovieSceneScriptingFloatChannel, const [1,1]}
          Color.B {type=MovieSceneScriptingFloatChannel, const [1,1]}
          Color.A {type=MovieSceneScriptingFloatChannel, const [1,1]}
         ]
        }]
      }
     ]
    }
    InputActionWidget {id=#5, possessed_class=CommonActionWidget, tracks []}
   ]}
 }
 OnExpansion {
  binding_count=1
  display_label=""
  length_seconds=0.2
  object=&self:OnExpansion
  sequence {bindings [ExpandDirectionTick {
     id=#6
     possessed_class=Image
     tracks [{
       type=MovieScene2DTransformTrack
       sections [{
         type=MovieScene2DTransformSection
         channels [
          Angle {type=MovieSceneScriptingFloatChannel, keys [[0,0],[4,90]]}
          Scale.X {type=MovieSceneScriptingFloatChannel, keys [[0,1],[3,1.5],[4,1]]}
          Scale.Y {type=MovieSceneScriptingFloatChannel, keys [[0,1],[3,1.5],[4,1]]}
          Shear.X {type=MovieSceneScriptingFloatChannel}
          Shear.Y {type=MovieSceneScriptingFloatChannel}
         ]
        }]
      }]
    }]}
 }
]
graphs [
 DoCustomNavigation_GoUp {nodes [
   DoCustomNavigation_GoUp!() -> (then, Navigation:EUINavigation Enum) {is=title K2Node_FunctionEntry}
   > SetFocus(self=Get Button_Options() -> (Button_Options:Button Object Reference))
   > Return Node!(execute, ReturnValue:Widget Object Reference)
  ]}
 EventGraph {nodes [
   "On Hovered (Button_Options)"() {is=title K2Node_ComponentBoundEvent}
   > SetVisibility(self=Get HB_Input() -> (HB_Input:Horizontal Box Object Reference), InVisibility=HitTestInvisible)
   > PlayAnimation(InAnimation=Get OnHover() -> (OnHover:Widget Animation Object Reference))
   > Set Hovered(Hovered:bool=true) -> (Output_Get:bool)
   "On Unhovered (Button_Options)"() {is=title K2Node_ComponentBoundEvent}
   > n1 = Branch(Condition=Get Is Expanded() -> (Is Expanded:bool))
   n2 = Get HB_Input()
   > n3 = SetVisibility(self=@n2, InVisibility=Hidden)
   PlayAnimation(execute=[@n4.then, @n3.then], InAnimation=Get OnHover(), PlayMode=Reverse)
   > Set Hovered(Hovered=false)
   n4 = SetVisibility(execute=@n1.else, self=@n2, InVisibility=HitTestInvisible)
   n5 = "On Clicked (Button_Options)"() {is=title K2Node_ComponentBoundEvent}
   n6 = Branch(execute=[@n5.then, @n23.then], Condition=Get Is Expanded())
   > On Options Expansion Changed!(execute, self, isExpanded:bool=false) -> (then)
   On Options Expansion Changed(execute=@n6.else, isExpanded=true)
   "On Button Clicked (JoinLeaveButton)"!() -> (OutputDelegate, then, Button:Common Button Legacy Object Reference) {is=title K2Node_ComponentBoundEvent}
   > HandleActionButtonClicked(self:Fort Voice Channel Header Button Object Reference)
   "On Button Clicked (ReportButton)"!() -> (OutputDelegate, then, Button:Common Button Legacy Object Reference) {is=title K2Node_ComponentBoundEvent}
   > HandleReportButtonClicked(self:Fort Voice Channel Header Button Object Reference)
   n7 = On Options Expansion Changed!() -> (OutputDelegate, then, isExpanded:bool=~) {is=title K2Node_CustomEvent}
   > n8 = Set Is Expanded(Is Expanded:bool=@n7.isExpanded) -> (Output_Get:bool)
   > n9 = Branch(Condition=@n8)
   > n10 = Branch(Condition=IsInteractionEnabled!(self:Common Button Base Object Reference=Get JoinLeaveButton() -> (JoinLeaveButton:Fort Main Menu Button Object Reference)) -> (ReturnValue:bool))
   > n11 = SetNavigationRuleExplicit!(
    execute
    self:Widget Object Reference=Get Button_Options()
    Direction:EUINavigation Enum=Down
    InWidget:Widget Object Reference=Get JoinLeaveButton()
   ) -> (then)
   SetVisibility(
    execute=[@n15.then, @n16.then, @n11.then]
    self=Get HorizontalBox_Buttons() -> (HorizontalBox_Buttons:Horizontal Box Object Reference)
    InVisibility=Visible
   )
   > PlayAnimation(InAnimation=Get OnExpansion() -> (OnExpansion:Widget Animation Object Reference))
   > n12 = Branch(Condition=Get Hovered() -> (Hovered:bool))
   SetVisibility(execute=@n12.else, self=Get HB_Input(), InVisibility=Hidden)
   n13 = Branch(execute=@n10.else, Condition=IsInteractionEnabled(self=Get ReportButton() -> (ReportButton:Fort Main Menu Button Object Reference)))
   n14 = Get Button_Options()
   > n15 = SetNavigationRuleExplicit(self=@n14, Direction=Down, InWidget=Get ReportButton())
   n16 = SetNavigationRuleBase!(execute=@n13.else, self:Widget Object Reference=@n14, Direction:EUINavigation Enum=Down, Rule:EUINavigationRule Enum=Escape) -> (then)
   SetNavigationRuleBase(execute=@n9.else, self=Get Button_Options(), Direction=Down, Rule=Escape)
   > SetVisibility(self=Get HorizontalBox_Buttons(), InVisibility=Collapsed)
   > PlayAnimation(InAnimation=Get OnExpansion(), PlayMode=Reverse)
   > SetVisibility(self=Get HB_Input(), InVisibility=HitTestInvisible)
   n17 = Set Header Status Visuals!() -> (OutputDelegate, then, Old:bool=~) {is=title K2Node_CustomEvent}
   > n18 = Set Old(Old:bool=@n17.Old) -> (Output_Get:bool)
   > SetVisibility(
    self=Get WrapBox_Header_Status() -> (WrapBox_Header_Status:Wrap Box Object Reference)
    InVisibility=Select(Option 0=HitTestInvisible, Option 1=Collapsed, Index=@n18)
   )
   n19 = Event DisplayReportModal!() -> (OutputDelegate, then, bIsRecording:bool) {is=title K2Node_Event}
   > n20 = Branch(Condition=@n19.bIsRecording)
   > ShowVoiceRepModal(self:Fort Voice Channel Header Button Object Reference)
   RequestToOpenPlayerFeedbackModalWithReportOrigin!(execute=@n20.else, Context:User Widget Object Reference=Self-Reference(), ReportOrigin:String) -> (then)
   n21 = Event On Item Expansion Changed!() -> (OutputDelegate, then, bIsExpanded:bool) {is=title K2Node_Event}
   > PlayAnimation(
    InAnimation=Get OnExpansion()
    PlayMode=Select(Option 0:EUMGSequencePlayMode Enum=Reverse, Option 1:EUMGSequencePlayMode Enum, Index=@n21.bIsExpanded) -> (ReturnValue:EUMGSequencePlayMode Enum)
    PlaybackSpeed=1.4
   )
   > PlaySound2D(Sound=/Game/Sounds/Fort_UI_Sounds/FrontEnd/New_FrontEnd/Fortnite_UI_MainMenu_Tab_Hover_02_Cue)
   n22 = Event OnUpdateVisuals!() -> (
    OutputDelegate
    then
    bIsActiveAndListening:bool
    bIsActiveOrAvailable:bool
    bIsAvailableAndNotActive:bool
    bIsListening:bool
    bShowingRestriction:bool
   ) {is=title K2Node_Event}
   > SetActiveAndListeningVisuals!(execute, self, IsActiveAndListening:bool=@n22.bIsActiveAndListening, IsRestricted:bool=@n22.bShowingRestriction) -> (then)
   > SetRestrictedVisuals!(execute, self, IsRestricted:bool=@n22.bShowingRestriction) -> (then)
   n23 = Event HandleHeaderClicked() {is=title K2Node_Event}
   K2Node_Event_6 = OnListItemObjectSet_1!() -> (OutputDelegate, then, ListItemObject:Object Reference) {is=title K2Node_CustomEvent}
   > Branch(Condition=IsListItemExpanded!(UserListEntry:User List Entry Interface) -> (ReturnValue:bool))
   n24 = Get OnExpansion()
   > PlayAnimation(InAnimation=@n24, StartAtTime=GetEndTime(self=@n24))
   n25 = Event SetExpansionState!() -> (OutputDelegate, then, bIsExpanded:bool) {is=title K2Node_Event}
   > On Options Expansion Changed(isExpanded=@n25.bIsExpanded) {is=title K2Node_CallFunction}
   n26 = Event OnUpdateOptionsVisuals!() -> (
    OutputDelegate
    then
    VoicePermission:ESocialCommsPermission Enum=Nobody
    bVoiceReportingIsAvailable:bool
    bIsRecording:bool
    bCanJoinOrLeave:bool
    bMightBeAbleToReport:bool
   ) {is=title K2Node_Event}
   > n27 = Sequence!(execute) -> (then_0, then_1, then_2:Exec, then_3:Exec)
   > n28 = Switch on ESocialCommsPermission!(execute, Selection:ESocialCommsPermission Enum=@n26.VoicePermission) -> (Nobody:Exec, FriendsOnly:Exec, FriendsAndTeammates:Exec, FriendsTeamConnections:Exec, Everybody:Exec) {is=title K2Node_SwitchEnum}
   SetActiveWidget(
    execute=@n28.Everybody
    self:Widget Switcher Object Reference=Get Switcher_Social() -> (Switcher_Social:Common Activatable Widget Switcher Object Reference)
    Widget=Get Text_Social_None() -> (Text_Social_None:Common Text Object Reference)
   )
   n29 = SetActiveWidget(execute=@n28.FriendsAndTeammates, self=Get Switcher_Social(), Widget=Get hb_teammates() -> (hb_teammates:Horizontal Box Object Reference))
   n30 = Get WrapBox_Header_Status()
   n31 = Branch(
    execute=[@n29.then, @n32.then]
    Condition="float > float"(A=GetDesiredSize!(self:Widget Object Reference=@n30) -> (ReturnValue_X:double, ReturnValue_Y:double).ReturnValue_Y, B=30)
   )
   > SetVisibility(
    self=[Get Text_Bullet_F() -> (Text_Bullet_F:Common Text Object Reference), Get Text_Bullet_FT() -> (Text_Bullet_FT:Common Text Object Reference)]
    InVisibility=Hidden
   )
   SetVisibility(execute=@n31.else, self=[Get Text_Bullet_FT(), Get Text_Bullet_F()], InVisibility=Visible)
   n32 = SetActiveWidget(execute=@n28.FriendsOnly, self=Get Switcher_Social(), Widget=Get hb_friends() -> (hb_friends:Horizontal Box Object Reference))
   SetActiveWidget(execute=@n28.Nobody, self=Get Switcher_Social(), Widget=Get Text_Social_None())
   n33 = Branch(execute=@n27.then_1, Condition=@n26.bIsRecording)
   > SetActiveWidget(
    self=Get Switcher_VoiceReporting() -> (Switcher_VoiceReporting:Common Activatable Widget Switcher Object Reference)
    Widget=Get Text_VoiceReporting_On() -> (Text_VoiceReporting_On:Common Text Object Reference)
   )
   SetActiveWidget(
    execute=@n33.else
    self=Get Switcher_VoiceReporting()
    Widget=Get Text_VoiceReporting_Off() -> (Text_VoiceReporting_Off:Common Text Object Reference)
   )
   n34 = Branch(execute=@n27.then_2, Condition=@n26.bCanJoinOrLeave)
   n35 = Get JoinLeaveButton()
   > Set bIsEnabled!(execute, bIsEnabled:bool=true, self:Widget Object Reference=@n35) -> (then, Output_Get:bool)
   n36 = Get JoinLeaveButton()
   > SetRenderOpacity(self=@n36, InOpacity=1)
   Set bIsEnabled(execute=@n34.else, bIsEnabled=false, self=@n35)
   > SetRenderOpacity(self=@n36, InOpacity=0.6)
   > SetJoinOrLeaveButton!(execute, self, Join:bool=true) -> (then)
   n37 = Branch(execute=@n27.then_3, Condition=@n26.bMightBeAbleToReport)
   n38 = Get ReportButton()
   > Set bIsEnabled(bIsEnabled=true, self=@n38)
   n39 = Get ReportButton()
   > SetRenderOpacity(self=@n39, InOpacity=1)
   Set bIsEnabled(execute=@n37.else, bIsEnabled=false, self=@n38)
   > SetRenderOpacity(self=@n39, InOpacity=0.6)
   Event On Focused() {is=title K2Node_Event}
   > n40 = Branch(Condition=Get Is Expanded())
   > n41 = Branch(Condition=IsInteractionEnabled(self=Get JoinLeaveButton()))
   > SetFocus(self=Get JoinLeaveButton())
   Branch(execute=@n41.else, Condition=IsInteractionEnabled(self=Get ReportButton()))
   > SetFocus(self=Get ReportButton())
   SetFocus(execute=@n40.else, self=Get Button_Options())
  ]}
 SetActiveAndListeningVisuals {nodes [
   n1 = SetActiveAndListeningVisuals!() -> (then, IsActiveAndListening:bool=~, IsRestricted:bool=~) {is=title K2Node_FunctionEntry}
   > SetRenderOpacity(self=Get Text_ChannelTitle() -> (Text_ChannelTitle:Common Rich Text Block Object Reference), InOpacity=SelectFloat(A=1, B=0.5, bPickA=@n1))
   > SetVisibility(
    self=Get Switch_Transmitting() -> (Switch_Transmitting:Widget Switcher Object Reference)
    InVisibility=Select(Option 0:ESlateVisibility Enum=Collapsed, Option 1:ESlateVisibility Enum=SelfHitTestInvisible, Index=@n1) -> (ReturnValue:ESlateVisibility Enum)
   )
   > n2 = Branch(Condition=Get Is Recording() -> (Is Recording:bool))
   > n3 = SetColorAndOpacity!(
    execute
    self=Get Text_ChannelInfo() -> (Text_ChannelInfo:Common Text Object Reference)
    InColorAndOpacity_SpecifiedColor:Linear Color Structure=(R=0.7084,G=0.7379,B=0.7157,A=1)
    InColorAndOpacity_ColorUseRule:ESlateColorStylingMode Enum
   ) -> (then)
   SetVisibility(
    execute=[@n8.then, @n3.then]
    self=[Get Switcher_Social(), Get Switch_Transmitting(), Get Switcher_VoiceReporting()]
    InVisibility=Select(Option 0=Collapsed, Option 1=SelfHitTestInvisible, Index=@n1)
   )
   n4 = NOT Boolean(A=Get Old() -> (Old:bool))
   > SetVisibility(self=Get Text_ChannelInfo(), InVisibility=Select(Option 0=SelfHitTestInvisible, Option 1=Collapsed, Index=AND Boolean(A=@n1, B=@n4)))
   > n5 = Branch(Condition=IsActiveChannel!(self:Fort Voice Channel Header Button Object Reference) -> (ReturnValue:bool))
   > SetJoinOrLeaveButton(Join=false)
   Branch(execute=@n5.else, Condition=CanJoinChannel!(self:Fort Voice Channel Header Button Object Reference) -> (ReturnValue:bool))
   > SetJoinOrLeaveButton(Join=true)
   n6 = SelectColor(
    A=Get VoiceStatusActiveColor() -> (VoiceStatusActiveColor:Linear Color Structure)
    B=Get VoiceStatusInactiveColor() -> (VoiceStatusInactiveColor:Linear Color Structure)
    bPickA=@n1
   )
   n7 = Get VoiceStatusWarningColor() -> (VoiceStatusWarningColor:Linear Color Structure)
   n8 = SetColorAndOpacity!(
    execute=@n2.else
    self=Get Text_ChannelInfo()
    InColorAndOpacity:Slate Color Structure=Make Slate Color(SpecifiedColor=SelectColor(A=@n7, B=@n6, bPickA=@n1.IsRestricted))
   ) -> (then)
  ]}
 SetJoinOrLeaveButton {nodes [
   n1 = SetJoinOrLeaveButton!() -> (then, Join:bool=~) {is=title K2Node_FunctionEntry}
   > n2 = Branch(Condition=@n1)
   > SetText(self:Fort Main Menu Button Object Reference=Get JoinLeaveButton(), InText=MakeLiteralText(Value="LOCTEXT(\"JOIN\")"))
   SetText(execute=@n2.else, self=Get JoinLeaveButton(), InText=MakeLiteralText(Value="LOCTEXT(\"LEAVE\")"))
  ]}
 SetRestrictedVisuals {nodes [
   n1 = SetRestrictedVisuals!() -> (then, IsRestricted:bool=~) {is=title K2Node_FunctionEntry}
   > SetColorAndOpacity!(
    execute
    self=Get Text_ChannelInfo_Hover() -> (Text_ChannelInfo_Hover:Common Text Object Reference)
    InColorAndOpacity_SpecifiedColor:Linear Color Structure=Select(
     Option 0:Linear Color Structure=(R=0.2051,G=0.2384,B=0.3564,A=1)
     Option 1:Linear Color Structure=(R=0.0513,G=0.0612,B=0.0723,A=1)
     Index=@n1
    ) -> (ReturnValue:Linear Color Structure)
    InColorAndOpacity_ColorUseRule:ESlateColorStylingMode Enum
   ) -> (then)
  ]}
]
variables [
 Hovered {type=boolean}
 Is Expanded {type=boolean}
 Is Recording {type=boolean}
 IsHoveredColor {type=boolean}
 IsRestricted {type=boolean}
 Old {type=boolean}
 OnOffTextColor {type=SlateColor, default {colorUseRule=UseColor_Specified, specifiedColor=<rgba 0.0578 0.0578 0.0578 0>}}
 OnOffTextColor_Hover {type=SlateColor, default {colorUseRule=UseColor_Specified, specifiedColor=<rgba 0.0578 0.0578 0.0578 1>}}
 VoiceStatusActiveColor {type=LinearColor, default=<rgba 0.0369 1 0.2232 1>}
 VoiceStatusInactiveColor {type=LinearColor, default=<rgba 0.7135 0.7135 0.7135 1>}
 VoiceStatusWarningColor {type=LinearColor, default=<rgba 0.4784 0.5216 0.5529 1>}
]
widgets {
 info {inheritedWidgetCount=7, parentClass=/Script/FortniteUI.FortVoiceChannelHeaderButton, rootWidgetClass=&SizeBox, widgetCount=42}
 widgets [
  SB_EntryContent {
   widgetClassPath=&SizeBox
   bIsVariable=true
   widget=SB_EntryContent
   properties {bOverride_MinDesiredHeight=true, heightOverride=40, maxDesiredWidth=500, minDesiredHeight=70}
  }
  EntryBorder {
   widgetClassPath=&Border
   bIsVariable=true
   parent=SB_EntryContent
   widget=EntryBorder
   properties {
    background {
     bIsDynamicallyLoaded=false
     drawAs=Image
     imageType=NoImage
     mirroring=NoMirror
     resourceName=None
     resourceObject=null
     tiling=NoTile
     imageSize=<32 32>
     margin=<ltrb 0 0 0 0>
     outlineSettings {
      bUseBrushTransparency=false
      roundingType=HalfHeightRadius
      width=0
      color {colorUseRule=UseColor_Specified, specifiedColor=<rgba 0 0 0 0>}
      cornerRadii=<0 0 0 0>
     }
     tintColor {colorUseRule=UseColor_Specified, specifiedColor=<rgba 0.0037 0.0052 0.016 0.7>}
     uVRegion {bIsValid=false, max=<0 0>, min=<0 0>}
    }
    padding=<ltrb 0 0 0 0>
   }
   slot {class=SizeBoxSlot}
  }
  VerticalBox_271 {
   widgetClassPath=&VerticalBox
   parent=EntryBorder
   widget=VerticalBox_271
   slot {class=BorderSlot, properties {padding=<ltrb 0 0 0 0>}}
  }
  HorizontalBox_281 {
   widgetClassPath=&HorizontalBox
   parent=VerticalBox_271
   widget=HorizontalBox_281
   properties {visibility=Visible}
   slot {class=VerticalBoxSlot}
  }
  VBHeader {
   widgetClassPath=&VerticalBox
   bIsVariable=true
   parent=HorizontalBox_281
   widget=VBHeader
   slot {class=HorizontalBoxSlot, properties {horizontalAlignment=HAlign_Left, padding=<ltrb 20 10 24 8>, size {sizeRule=Fill, value=1}}}
  }
  HorizontalBox_0 {widgetClassPath=&HorizontalBox, parent=VBHeader, widget=HorizontalBox_0, slot {class=VerticalBoxSlot}}
  Switcher_ChannelTitle {
   widgetClassPath=&CommonActivatableWidgetSwitcher
   bIsVariable=true
   parent=HorizontalBox_0
   widget=Switcher_ChannelTitle
   properties {transitionCurveType=Linear, transitionDuration=0.1}
   slot {class=HorizontalBoxSlot}
  }
  Text_ChannelTitle {
   widgetClassPath=&CommonRichTextBlock
   bInherited=true
   bIsVariable=true
   parent=Switcher_ChannelTitle
   widget=Text_ChannelTitle
   properties {
    bOverrideDefaultStyle=true
    text=Party Channel
    textTransformPolicy=ToUpper
    defaultTextStyleOverride {
     overflowPolicy=Clip
     transformPolicy=None
     colorAndOpacity {colorUseRule=UseColor_Specified, specifiedColor=<rgba 1 1 1 1>}
     font {
      bForceMonospaced=false
      bMaterialIsStencil=false
      fontMaterial=null
      fontObject=&BurbankSmall
      letterSpacing=0
      monospacedWidth=1
      size=20
      skewAmount=0
      typefaceFontName=Black
      outlineSettings {
       bApplyOutlineToDropShadows=false
       bMiteredCorners=false
       bSeparateFillAlpha=false
       outlineMaterial=null
       outlineSize=0
       outlineColor=<rgba 0 0 0 1>
      }
     }
     highlightColor {colorUseRule=UseColor_Specified, specifiedColor=<rgba 1 1 1 1>}
     highlightShape {
      bIsDynamicallyLoaded=false
      drawAs=Image
      imageType=NoImage
      mirroring=NoMirror
      resourceName=None
      resourceObject=null
      tiling=NoTile
      imageSize=<32 32>
      margin=<ltrb 0 0 0 0>
      outlineSettings {
       bUseBrushTransparency=false
       roundingType=HalfHeightRadius
       width=0
       color {colorUseRule=UseColor_Specified, specifiedColor=<rgba 0 0 0 0>}
       cornerRadii=<0 0 0 0>
      }
      tintColor {colorUseRule=UseColor_Specified, specifiedColor=<rgba 1 1 1 1>}
      uVRegion {bIsValid=false, max=<0 0>, min=<0 0>}
     }
     selectedBackgroundColor {colorUseRule=UseColor_Foreground, specifiedColor=<rgba 1 0 1 1>}
     shadowColorAndOpacity=<rgba 0 0 0 1>
     shadowOffset=<0 0>
     strikeBrush {
      bIsDynamicallyLoaded=false
      drawAs=Image
      imageType=NoImage
      mirroring=NoMirror
      resourceName=None
      resourceObject=null
      tiling=NoTile
      imageSize=<32 32>
      margin=<ltrb 0 0 0 0>
      outlineSettings {
       bUseBrushTransparency=false
       roundingType=HalfHeightRadius
       width=0
       color {colorUseRule=UseColor_Specified, specifiedColor=<rgba 0 0 0 0>}
       cornerRadii=<0 0 0 0>
      }
      tintColor {colorUseRule=UseColor_Specified, specifiedColor=<rgba 1 1 1 1>}
      uVRegion {bIsValid=false, max=<0 0>, min=<0 0>}
     }
     underlineBrush {
      bIsDynamicallyLoaded=false
      drawAs=Image
      imageType=NoImage
      mirroring=NoMirror
      resourceName=None
      resourceObject=null
      tiling=NoTile
      imageSize=<32 32>
      margin=<ltrb 0 0 0 0>
      outlineSettings {
       bUseBrushTransparency=false
       roundingType=HalfHeightRadius
       width=0
       color {colorUseRule=UseColor_Specified, specifiedColor=<rgba 0 0 0 0>}
       cornerRadii=<0 0 0 0>
      }
      tintColor {colorUseRule=UseColor_Specified, specifiedColor=<rgba 1 1 1 1>}
      uVRegion {bIsValid=false, max=<0 0>, min=<0 0>}
     }
    }
   }
   slot {class=WidgetSwitcherSlot}
  }
  Text_ChannelTitle_Hover {
   widgetClassPath=&CommonRichTextBlock
   bInherited=true
   bIsVariable=true
   parent=Switcher_ChannelTitle
   widget=Text_ChannelTitle_Hover
   properties {
    bOverrideDefaultStyle=true
    text=Party Channel
    textTransformPolicy=ToUpper
    defaultTextStyleOverride {
     overflowPolicy=Clip
     transformPolicy=None
     colorAndOpacity {colorUseRule=UseColor_Specified, specifiedColor=<rgba 0.0578 0.0578 0.0578 1>}
     font {
      bForceMonospaced=false
      bMaterialIsStencil=false
      fontMaterial=null
      fontObject=&BurbankSmall
      letterSpacing=0
      monospacedWidth=1
      size=20
      skewAmount=0
      typefaceFontName=Black
      outlineSettings {
       bApplyOutlineToDropShadows=false
       bMiteredCorners=false
       bSeparateFillAlpha=false
       outlineMaterial=null
       outlineSize=0
       outlineColor=<rgba 0 0 0 1>
      }
     }
     highlightColor {colorUseRule=UseColor_Specified, specifiedColor=<rgba 1 1 1 1>}
     highlightShape {
      bIsDynamicallyLoaded=false
      drawAs=Image
      imageType=NoImage
      mirroring=NoMirror
      resourceName=None
      resourceObject=null
      tiling=NoTile
      imageSize=<32 32>
      margin=<ltrb 0 0 0 0>
      outlineSettings {
       bUseBrushTransparency=false
       roundingType=HalfHeightRadius
       width=0
       color {colorUseRule=UseColor_Specified, specifiedColor=<rgba 0 0 0 0>}
       cornerRadii=<0 0 0 0>
      }
      tintColor {colorUseRule=UseColor_Specified, specifiedColor=<rgba 1 1 1 1>}
      uVRegion {bIsValid=false, max=<0 0>, min=<0 0>}
     }
     selectedBackgroundColor {colorUseRule=UseColor_Foreground, specifiedColor=<rgba 1 0 1 1>}
     shadowColorAndOpacity=<rgba 0 0 0 1>
     shadowOffset=<0 0>
     strikeBrush {
      bIsDynamicallyLoaded=false
      drawAs=Image
      imageType=NoImage
      mirroring=NoMirror
      resourceName=None
      resourceObject=null
      tiling=NoTile
      imageSize=<32 32>
      margin=<ltrb 0 0 0 0>
      outlineSettings {
       bUseBrushTransparency=false
       roundingType=HalfHeightRadius
       width=0
       color {colorUseRule=UseColor_Specified, specifiedColor=<rgba 0 0 0 0>}
       cornerRadii=<0 0 0 0>
      }
      tintColor {colorUseRule=UseColor_Specified, specifiedColor=<rgba 1 1 1 1>}
      uVRegion {bIsValid=false, max=<0 0>, min=<0 0>}
     }
     underlineBrush {
      bIsDynamicallyLoaded=false
      drawAs=Image
      imageType=NoImage
      mirroring=NoMirror
      resourceName=None
      resourceObject=null
      tiling=NoTile
      imageSize=<32 32>
      margin=<ltrb 0 0 0 0>
      outlineSettings {
       bUseBrushTransparency=false
       roundingType=HalfHeightRadius
       width=0
       color {colorUseRule=UseColor_Specified, specifiedColor=<rgba 0 0 0 0>}
       cornerRadii=<0 0 0 0>
      }
      tintColor {colorUseRule=UseColor_Specified, specifiedColor=<rgba 1 1 1 1>}
      uVRegion {bIsValid=false, max=<0 0>, min=<0 0>}
     }
    }
   }
   slot {class=WidgetSwitcherSlot}
  }
  Switch_Transmitting {
   widgetClassPath=&WidgetSwitcher
   bIsVariable=true
   parent=HorizontalBox_0
   widget=Switch_Transmitting
   slot {class=HorizontalBoxSlot, properties {padding=<ltrb 10 0 0 0>}}
  }
  Transmitting {
   widgetClassPath=/Script/FortniteUI.FortMobileImage
   bIsVariable=true
   parent=Switch_Transmitting
   widget=Transmitting
   properties {
    mobileSizeMultiplier=1.2
    visibility=HitTestInvisible
    brush {
     bIsDynamicallyLoaded=false
     drawAs=Image
     imageType=NoImage
     mirroring=NoMirror
     resourceName=None
     resourceObject=/Game/UI/Foundation/Textures/Icons/Audio/T-Icon-Sound-On-M
     tiling=NoTile
     imageSize=<24 24>
     margin=<ltrb 0 0 0 0>
     outlineSettings {
      bUseBrushTransparency=false
      roundingType=HalfHeightRadius
      width=0
      color {colorUseRule=UseColor_Specified, specifiedColor=<rgba 0 0 0 0>}
      cornerRadii=<0 0 0 0>
     }
     tintColor {colorUseRule=UseColor_Specified, specifiedColor=<rgba 1 1 1 1>}
     uVRegion {bIsValid=false, max=<0 0>, min=<0 0>}
    }
   }
   slot {class=WidgetSwitcherSlot, properties {horizontalAlignment=HAlign_Center, verticalAlignment=VAlign_Center}}
  }
  NotTransmitting {
   widgetClassPath=&FortMobileImage
   bIsVariable=true
   parent=Switch_Transmitting
   widget=NotTransmitting
   properties {
    mobileSizeMultiplier=1.2
    visibility=HitTestInvisible
    brush {
     bIsDynamicallyLoaded=false
     drawAs=Image
     imageType=NoImage
     mirroring=NoMirror
     resourceName=None
     resourceObject=/Game/UI/Foundation/Textures/Icons/Audio/T-Icon-Sound-Mute-M
     tiling=NoTile
     imageSize=<24 24>
     margin=<ltrb 0 0 0 0>
     outlineSettings {
      bUseBrushTransparency=false
      roundingType=HalfHeightRadius
      width=0
      color {colorUseRule=UseColor_Specified, specifiedColor=<rgba 0 0 0 0>}
      cornerRadii=<0 0 0 0>
     }
     tintColor {colorUseRule=UseColor_Specified, specifiedColor=<rgba 1 1 1 1>}
     uVRegion {bIsValid=false, max=<0 0>, min=<0 0>}
    }
   }
   slot {class=WidgetSwitcherSlot, properties {horizontalAlignment=HAlign_Center, verticalAlignment=VAlign_Center}}
  }
  ChannelInfoPanel {
   widgetClassPath=&HorizontalBox
   parent=VBHeader
   widget=ChannelInfoPanel
   slot {class=VerticalBoxSlot, properties {horizontalAlignment=HAlign_Left}}
  }
  Switcher_ChannelInfo {
   widgetClassPath=&CommonActivatableWidgetSwitcher
   bIsVariable=true
   parent=ChannelInfoPanel
   widget=Switcher_ChannelInfo
   properties {transitionCurveType=Linear, transitionDuration=0.1}
   slot {class=HorizontalBoxSlot, properties {horizontalAlignment=HAlign_Left, verticalAlignment=VAlign_Center, size {sizeRule=Fill, value=1}}}
  }
  Text_ChannelInfo {
   widgetClassPath=&CommonTextBlock
   bInherited=true
   bIsVariable=true
   parent=Switcher_ChannelInfo
   widget=Text_ChannelInfo
   properties {
    scrollStyle=&TextScrollStyle-Base_C
    text=You are in this channel
    textTransformPolicy=ToUpper
    wrapTextAt=500
    colorAndOpacity {colorUseRule=UseColor_Specified, specifiedColor=<rgba 0.0369 1 0.2232 1>}
    font {
     bForceMonospaced=false
     bMaterialIsStencil=false
     fontMaterial=null
     fontObject=&BurbankSmall
     letterSpacing=-2
     monospacedWidth=1
     size=12.75
     skewAmount=0
     typefaceFontName=Bold
     outlineSettings {
      bApplyOutlineToDropShadows=false
      bMiteredCorners=false
      bSeparateFillAlpha=false
      outlineMaterial=null
      outlineSize=0
      outlineColor=<rgba 0 0 0 1>
     }
    }
    shadowOffset=<0 0>
   }
   slot {class=WidgetSwitcherSlot, properties {horizontalAlignment=HAlign_Left, verticalAlignment=VAlign_Center}}
  }
  Text_ChannelInfoHover {
   widgetClassPath=&CommonTextBlock
   bIsVariable=true
   parent=Switcher_ChannelInfo
   widget=Text_ChannelInfoHover
   properties {
    scrollStyle=&TextScrollStyle-Base_C
    text=You are not in this channel
    textTransformPolicy=ToUpper
    wrapTextAt=500
    colorAndOpacity {colorUseRule=UseColor_Specified, specifiedColor=<rgba 0.4735 0.521 0.5583 1>}
    font {
     bForceMonospaced=false
     bMaterialIsStencil=false
     fontMaterial=null
     fontObject=&BurbankSmall
     letterSpacing=-2
     monospacedWidth=1
     size=12.75
     skewAmount=0
     typefaceFontName=Bold
     outlineSettings {
      bApplyOutlineToDropShadows=false
      bMiteredCorners=false
      bSeparateFillAlpha=false
      outlineMaterial=null
      outlineSize=0
      outlineColor=<rgba 0 0 0 1>
     }
    }
    shadowOffset=<0 0>
   }
   slot {class=WidgetSwitcherSlot, properties {horizontalAlignment=HAlign_Left, verticalAlignment=VAlign_Center}}
  }
  WrapBox_Header_Status {
   widgetClassPath=&WrapBox
   bIsVariable=true
   parent=VBHeader
   widget=WrapBox_Header_Status
   properties {bExplicitWrapSize=true, wrapSize=360, innerSlotPadding=<0 4>}
   slot {class=VerticalBoxSlot, properties {horizontalAlignment=HAlign_Left, verticalAlignment=VAlign_Top}}
  }
  Switcher_Social {
   widgetClassPath=&CommonActivatableWidgetSwitcher
   bIsVariable=true
   parent=WrapBox_Header_Status
   widget=Switcher_Social
   properties {transitionCurveType=Linear, transitionDuration=0.1, visibility=HitTestInvisible}
   slot {class=WrapBoxSlot}
  }
  hb_friends {widgetClassPath=&HorizontalBox, bIsVariable=true, parent=Switcher_Social, widget=hb_friends, slot {class=WidgetSwitcherSlot}}
  Image_VoiceWarning {
   widgetClassPath=&Image
   parent=hb_friends
   widget=Image_VoiceWarning
   properties {
    brush {
     bIsDynamicallyLoaded=false
     drawAs=Image
     imageType=NoImage
     mirroring=NoMirror
     resourceName=None
     resourceObject=/Game/UI/SocialPanel/Art/Textures/T_UI_Icon_MicLocked
     tiling=NoTile
     imageSize=<25 25>
     margin=<ltrb 0 0 0 0>
     outlineSettings {
      bUseBrushTransparency=false
      roundingType=HalfHeightRadius
      width=0
      color {colorUseRule=UseColor_Specified, specifiedColor=<rgba 0 0 0 0>}
      cornerRadii=<0 0 0 0>
     }
     tintColor {colorUseRule=UseColor_Specified, specifiedColor=<rgba 1 1 1 1>}
     uVRegion {bIsValid=false, max=<0 0>, min=<0 0>}
    }
    colorAndOpacity=<rgba 0.4735 0.521 0.5583 1>
   }
   slot {class=HorizontalBoxSlot, properties {horizontalAlignment=HAlign_Left, verticalAlignment=VAlign_Center, padding=<ltrb 0 0 4 0>}}
  }
  Text_Social_Friends_Only {
   widgetClassPath=&CommonTextBlock
   bIsVariable=true
   parent=hb_friends
   widget=Text_Social_Friends_Only
   properties {
    scrollStyle=&TextScrollStyle-Base_C
    text=Friends Only
    textTransformPolicy=ToUpper
    wrapTextAt=500
    colorAndOpacity {colorUseRule=UseColor_Specified, specifiedColor=<rgba 0.4735 0.521 0.5583 1>}
    font {
     bForceMonospaced=false
     bMaterialIsStencil=false
     fontMaterial=null
     fontObject=&BurbankSmall
     letterSpacing=0
     monospacedWidth=1
     size=12.75
     skewAmount=0
     typefaceFontName=Bold
     outlineSettings {
      bApplyOutlineToDropShadows=false
      bMiteredCorners=false
      bSeparateFillAlpha=false
      outlineMaterial=null
      outlineSize=0
      outlineColor=<rgba 0 0 0 1>
     }
    }
    shadowOffset=<0 0>
   }
   slot {class=HorizontalBoxSlot, properties {horizontalAlignment=HAlign_Left, verticalAlignment=VAlign_Center}}
  }
  Text_Bullet_F {
   widgetClassPath=&CommonTextBlock
   bIsVariable=true
   parent=hb_friends
   widget=Text_Bullet_F
   properties {
    scrollStyle=&TextScrollStyle-Base_C
    text="•"
    textTransformPolicy=ToUpper
    wrapTextAt=500
    colorAndOpacity {colorUseRule=UseColor_Specified, specifiedColor=<rgba 0.4735 0.521 0.5583 1>}
    font {
     bForceMonospaced=false
     bMaterialIsStencil=false
     fontMaterial=null
     fontObject=&BurbankSmall
     letterSpacing=0
     monospacedWidth=1
     size=12.75
     skewAmount=0
     typefaceFontName=Bold
     outlineSettings {
      bApplyOutlineToDropShadows=false
      bMiteredCorners=false
      bSeparateFillAlpha=false
      outlineMaterial=null
      outlineSize=0
      outlineColor=<rgba 0 0 0 1>
     }
    }
    shadowOffset=<0 0>
   }
   slot {class=HorizontalBoxSlot, properties {horizontalAlignment=HAlign_Left, verticalAlignment=VAlign_Center, padding=<ltrb 4 0 4 0>}}
  }
  hb_teammates {widgetClassPath=&HorizontalBox, bIsVariable=true, parent=Switcher_Social, widget=hb_teammates, slot {class=WidgetSwitcherSlot}}
  Image_VoiceWarning_1 {
   widgetClassPath=&Image
   parent=hb_teammates
   widget=Image_VoiceWarning_1
   properties {
    brush {
     bIsDynamicallyLoaded=false
     drawAs=Image
     imageType=NoImage
     mirroring=NoMirror
     resourceName=None
     resourceObject=&T_UI_Icon_MicLocked
     tiling=NoTile
     imageSize=<25 25>
     margin=<ltrb 0 0 0 0>
     outlineSettings {
      bUseBrushTransparency=false
      roundingType=HalfHeightRadius
      width=0
      color {colorUseRule=UseColor_Specified, specifiedColor=<rgba 0 0 0 0>}
      cornerRadii=<0 0 0 0>
     }
     tintColor {colorUseRule=UseColor_Specified, specifiedColor=<rgba 1 1 1 1>}
     uVRegion {bIsValid=false, max=<0 0>, min=<0 0>}
    }
    colorAndOpacity=<rgba 0.4735 0.521 0.5583 1>
   }
   slot {class=HorizontalBoxSlot, properties {horizontalAlignment=HAlign_Left, verticalAlignment=VAlign_Center, padding=<ltrb 0 0 4 0>}}
  }
  Text_Social_Friends_Only_1 {
   widgetClassPath=&CommonTextBlock
   bIsVariable=true
   parent=hb_teammates
   widget=Text_Social_Friends_Only_1
   properties {
    scrollStyle=&TextScrollStyle-Base_C
    text="Friends && Teammates"
    textTransformPolicy=ToUpper
    wrapTextAt=500
    colorAndOpacity {colorUseRule=UseColor_Specified, specifiedColor=<rgba 0.4735 0.521 0.5583 1>}
    font {
     bForceMonospaced=false
     bMaterialIsStencil=false
     fontMaterial=null
     fontObject=&BurbankSmall
     letterSpacing=0
     monospacedWidth=1
     size=12.75
     skewAmount=0
     typefaceFontName=Bold
     outlineSettings {
      bApplyOutlineToDropShadows=false
      bMiteredCorners=false
      bSeparateFillAlpha=false
      outlineMaterial=null
      outlineSize=0
      outlineColor=<rgba 0 0 0 1>
     }
    }
    shadowOffset=<0 0>
   }
   slot {class=HorizontalBoxSlot, properties {horizontalAlignment=HAlign_Left, verticalAlignment=VAlign_Center}}
  }
  Text_Bullet_FT {
   widgetClassPath=&CommonTextBlock
   bIsVariable=true
   parent=hb_teammates
   widget=Text_Bullet_FT
   properties {
    scrollStyle=&TextScrollStyle-Base_C
    text="•"
    textTransformPolicy=ToUpper
    wrapTextAt=500
    colorAndOpacity {colorUseRule=UseColor_Specified, specifiedColor=<rgba 0.4735 0.521 0.5583 1>}
    font {
     bForceMonospaced=false
     bMaterialIsStencil=false
     fontMaterial=null
     fontObject=&BurbankSmall
     letterSpacing=0
     monospacedWidth=1
     size=12.75
     skewAmount=0
     typefaceFontName=Bold
     outlineSettings {
      bApplyOutlineToDropShadows=false
      bMiteredCorners=false
      bSeparateFillAlpha=false
      outlineMaterial=null
      outlineSize=0
      outlineColor=<rgba 0 0 0 1>
     }
    }
    shadowOffset=<0 0>
   }
   slot {class=HorizontalBoxSlot, properties {horizontalAlignment=HAlign_Left, verticalAlignment=VAlign_Center, padding=<ltrb 4 0 4 0>}}
  }
  Text_Social_None {
   widgetClassPath=&CommonTextBlock
   bIsVariable=true
   parent=Switcher_Social
   widget=Text_Social_None
   properties {
    scrollStyle=&TextScrollStyle-Base_C
    textTransformPolicy=ToUpper
    wrapTextAt=500
    colorAndOpacity {colorUseRule=UseColor_Specified, specifiedColor=<rgba 0.2051 0.2384 0.3564 1>}
    font {
     bForceMonospaced=false
     bMaterialIsStencil=false
     fontMaterial=null
     fontObject=&BurbankSmall
     letterSpacing=0
     monospacedWidth=1
     size=13
     skewAmount=0
     typefaceFontName=Black
     outlineSettings {
      bApplyOutlineToDropShadows=false
      bMiteredCorners=false
      bSeparateFillAlpha=false
      outlineMaterial=null
      outlineSize=0
      outlineColor=<rgba 0 0 0 1>
     }
    }
    shadowOffset=<0 0>
   }
   slot {class=WidgetSwitcherSlot, properties {verticalAlignment=VAlign_Center}}
  }
  Switcher_VoiceReporting {
   widgetClassPath=&CommonActivatableWidgetSwitcher
   bIsVariable=true
   parent=WrapBox_Header_Status
   widget=Switcher_VoiceReporting
   properties {transitionCurveType=Linear, transitionDuration=0.1, visibility=HitTestInvisible}
   slot {class=WrapBoxSlot}
  }
  Text_VoiceReporting_On {
   widgetClassPath=&CommonTextBlock
   bIsVariable=true
   parent=Switcher_VoiceReporting
   widget=Text_VoiceReporting_On
   properties {
    scrollStyle=&TextScrollStyle-Base_C
    text=Voice Reporting On
    textTransformPolicy=ToUpper
    wrapTextAt=500
    colorAndOpacity {colorUseRule=UseColor_Specified, specifiedColor=<rgba 0.0369 1 0.2232 1>}
    font {
     bForceMonospaced=false
     bMaterialIsStencil=false
     fontMaterial=null
     fontObject=&BurbankSmall
     letterSpacing=0
     monospacedWidth=1
     size=12.75
     skewAmount=0
     typefaceFontName=Bold
     outlineSettings {
      bApplyOutlineToDropShadows=false
      bMiteredCorners=false
      bSeparateFillAlpha=false
      outlineMaterial=null
      outlineSize=0
      outlineColor=<rgba 0 0 0 1>
     }
    }
    shadowOffset=<0 0>
   }
   slot {class=WidgetSwitcherSlot, properties {horizontalAlignment=HAlign_Left, verticalAlignment=VAlign_Center}}
  }
  Text_VoiceReporting_Off {
   widgetClassPath=&CommonTextBlock
   bIsVariable=true
   parent=Switcher_VoiceReporting
   widget=Text_VoiceReporting_Off
   properties {
    scrollStyle=&TextScrollStyle-Base_C
    text=Voice Reporting Off
    textTransformPolicy=ToUpper
    wrapTextAt=500
    colorAndOpacity {colorUseRule=UseColor_Specified, specifiedColor=<rgba 0.4735 0.521 0.5583 1>}
    font {
     bForceMonospaced=false
     bMaterialIsStencil=false
     fontMaterial=null
     fontObject=&BurbankSmall
     letterSpacing=0
     monospacedWidth=1
     size=12.75
     skewAmount=0
     typefaceFontName=Bold
     outlineSettings {
      bApplyOutlineToDropShadows=false
      bMiteredCorners=false
      bSeparateFillAlpha=false
      outlineMaterial=null
      outlineSize=0
      outlineColor=<rgba 0 0 0 1>
     }
    }
    shadowOffset=<0 0>
   }
   slot {class=WidgetSwitcherSlot, properties {horizontalAlignment=HAlign_Left, verticalAlignment=VAlign_Center}}
  }
  Button_Options {
   widgetClassPath=&Button
   parent=HorizontalBox_281
   widget=Button_Options
   properties {
    backgroundColor=<rgba 1 1 1 0>
    widgetStyle {
     clickedSlateSound {resourceObject=null}
     disabled {
      bIsDynamicallyLoaded=false
      drawAs=NoDrawType
      imageType=NoImage
      mirroring=NoMirror
      resourceName=None
      resourceObject=null
      tiling=NoTile
      imageSize=<0 0>
      margin=<ltrb 0 0 0 0>
      outlineSettings {
       bUseBrushTransparency=false
       roundingType=HalfHeightRadius
       width=0
       color {colorUseRule=UseColor_Specified, specifiedColor=<rgba 0 0 0 0>}
       cornerRadii=<0 0 0 0>
      }
      tintColor {colorUseRule=UseColor_Specified, specifiedColor=<rgba 1 1 1 1>}
      uVRegion {bIsValid=false, max=<0 0>, min=<0 0>}
     }
     disabledForeground {colorUseRule=UseColor_Specified, specifiedColor=<rgba 0.5271 0.5271 0.5271 1>}
     hovered {
      bIsDynamicallyLoaded=false
      drawAs=RoundedBox
      imageType=NoImage
      mirroring=NoMirror
      resourceName=None
      resourceObject=null
      tiling=NoTile
      imageSize=<32 32>
      margin=<ltrb 0 0 0 0>
      outlineSettings {
       bUseBrushTransparency=true
       roundingType=FixedRadius
       width=1
       color {colorUseRule=UseColor_Specified, specifiedColor=<rgba 0.7243 0.7243 0.7243 1>}
       cornerRadii=<4 4 4 4>
      }
      tintColor {colorUseRule=UseColor_Specified, specifiedColor=<rgba 0.7243 0.7243 0.7243 1>}
      uVRegion {bIsValid=false, max=<0 0>, min=<0 0>}
     }
     hoveredForeground {colorUseRule=UseColor_Specified, specifiedColor=<rgba 0 1 1 1>}
     hoveredSlateSound {resourceObject=null}
     normal {
      bIsDynamicallyLoaded=false
      drawAs=RoundedBox
      imageType=NoImage
      mirroring=NoMirror
      resourceName=None
      resourceObject=null
      tiling=NoTile
      imageSize=<32 32>
      margin=<ltrb 0 0 0 0>
      outlineSettings {
       bUseBrushTransparency=true
       roundingType=FixedRadius
       width=1
       color {colorUseRule=UseColor_Specified, specifiedColor=<rgba 0.6951 0.6951 0.6951 1>}
       cornerRadii=<4 4 4 4>
      }
      tintColor {colorUseRule=UseColor_Specified, specifiedColor=<rgba 0.4955 0.4955 0.4955 1>}
      uVRegion {bIsValid=false, max=<0 0>, min=<0 0>}
     }
     normalForeground {colorUseRule=UseColor_Specified, specifiedColor=<rgba 0 1 1 1>}
     normalPadding=<ltrb 0 0 0 0>
     pressed {
      bIsDynamicallyLoaded=false
      drawAs=RoundedBox
      imageType=NoImage
      mirroring=NoMirror
      resourceName=None
      resourceObject=null
      tiling=NoTile
      imageSize=<32 32>
      margin=<ltrb 0 0 0 0>
      outlineSettings {
       bUseBrushTransparency=true
       roundingType=FixedRadius
       width=1
       color {colorUseRule=UseColor_Specified, specifiedColor=<rgba 0.7243 0.7243 0.7243 1>}
       cornerRadii=<4 4 4 4>
      }
      tintColor {colorUseRule=UseColor_Specified, specifiedColor=<rgba 0.3843 0.3843 0.3843 1>}
      uVRegion {bIsValid=false, max=<0 0>, min=<0 0>}
     }
     pressedForeground {colorUseRule=UseColor_Specified, specifiedColor=<rgba 0 1 1 1>}
     pressedPadding=<ltrb 0 0 0 0>
     pressedSlateSound {resourceObject=null}
    }
   }
   slot {class=HorizontalBoxSlot, properties {horizontalAlignment=HAlign_Right, size {sizeRule=Fill, value=1}}}
  }
  AnimatedInputBorder {
   widgetClassPath=&Border
   parent=Button_Options
   widget=AnimatedInputBorder
   properties {
    visibility=SelfHitTestInvisible
    background {
     bIsDynamicallyLoaded=false
     drawAs=Image
     imageType=NoImage
     mirroring=NoMirror
     resourceName=None
     resourceObject=/Game/UI/SocialPanel/Art/Materials/M_UI_UserHoverCategory_Inst
     tiling=NoTile
     imageSize=<64 64>
     margin=<ltrb 1 0 0 0>
     outlineSettings {
      bUseBrushTransparency=false
      roundingType=HalfHeightRadius
      width=0
      color {colorUseRule=UseColor_Specified, specifiedColor=<rgba 0 0 0 0>}
      cornerRadii=<0 0 0 0>
     }
     tintColor {colorUseRule=UseColor_Specified, specifiedColor=<rgba 1 1 1 1>}
     uVRegion {bIsValid=false, max=<0 0>, min=<0 0>}
    }
    brushColor=<rgba 1 1 1 0>
    padding=<ltrb 18 0 0 0>
   }
   slot {class=ButtonSlot, properties {horizontalAlignment=HAlign_Fill, verticalAlignment=VAlign_Fill, padding=<ltrb 0 0 0 0>}}
  }
  HorizontalBox_6 {
   widgetClassPath=&HorizontalBox
   parent=AnimatedInputBorder
   widget=HorizontalBox_6
   slot {class=BorderSlot, properties {padding=<ltrb 18 0 0 0>}}
  }
  HB_Input {widgetClassPath=&HorizontalBox, bIsVariable=true, parent=HorizontalBox_6, widget=HB_Input, slot {class=HorizontalBoxSlot}}
  InputActionWidget {
   widgetClassPath=&CommonActionWidget
   bInherited=true
   bIsVariable=true
   parent=HB_Input
   widget=InputActionWidget
   properties {
    visibility=SelfHitTestInvisible
    inputActions [{dataTable=&InputActionDataTable, rowName=Select}]
    progressMaterialBrush {
     bIsDynamicallyLoaded=false
     drawAs=Image
     imageType=NoImage
     mirroring=NoMirror
     resourceName=None
     resourceObject=null
     tiling=NoTile
     imageSize=<16 16>
     margin=<ltrb 0 0 0 0>
     outlineSettings {
      bUseBrushTransparency=false
      roundingType=HalfHeightRadius
      width=0
      color {colorUseRule=UseColor_Specified, specifiedColor=<rgba 0 0 0 0>}
      cornerRadii=<0 0 0 0>
     }
     tintColor {colorUseRule=UseColor_Specified, specifiedColor=<rgba 1 1 1 1>}
     uVRegion {bIsValid=false, max=<0 0>, min=<0 0>}
    }
   }
   slot {class=HorizontalBoxSlot, properties {horizontalAlignment=HAlign_Left, verticalAlignment=VAlign_Center, padding=<ltrb 4 0 4 0>}}
  }
  Text_Options {
   widgetClassPath=&CommonTextBlock
   bIsVariable=true
   parent=HB_Input
   widget=Text_Options
   properties {
    mobileFontSizeMultiplier=1.34
    text=Options
    textTransformPolicy=ToUpper
    colorAndOpacity {colorUseRule=UseColor_Specified, specifiedColor=<rgba 0.5089 0.5583 0.7084 1>}
    font {
     bForceMonospaced=false
     bMaterialIsStencil=false
     fontMaterial=null
     fontObject=&BurbankSmall
     letterSpacing=0
     monospacedWidth=1
     size=14
     skewAmount=0
     typefaceFontName=Black
     outlineSettings {
      bApplyOutlineToDropShadows=false
      bMiteredCorners=false
      bSeparateFillAlpha=false
      outlineMaterial=null
      outlineSize=0
      outlineColor=<rgba 0 0 0 1>
     }
    }
   }
   slot {class=HorizontalBoxSlot, properties {horizontalAlignment=HAlign_Left, verticalAlignment=VAlign_Center, padding=<ltrb 3 10 0 10>}}
  }
  ArrowSpacingSB {
   widgetClassPath=&SizeBox
   parent=HorizontalBox_6
   widget=ArrowSpacingSB
   properties {bOverride_MinDesiredHeight=true, heightOverride=72, minDesiredHeight=16, widthOverride=72}
   slot {class=HorizontalBoxSlot, properties {horizontalAlignment=HAlign_Left, verticalAlignment=VAlign_Center, padding=<ltrb 10 0 12 0>}}
  }
  ExpandDirectionTick {
   widgetClassPath=&FortMobileImage
   bIsVariable=true
   parent=ArrowSpacingSB
   widget=ExpandDirectionTick
   properties {
    mobileSizeMultiplier=1.2
    brush {
     bIsDynamicallyLoaded=false
     drawAs=Image
     imageType=NoImage
     mirroring=NoMirror
     resourceName=None
     resourceObject=/Game/UI/SocialPanel/Art/Textures/T_UI_Icon_TreeArrow_64
     tiling=NoTile
     imageSize=<16 16>
     margin=<ltrb 0 0 0 0>
     outlineSettings {
      bUseBrushTransparency=false
      roundingType=HalfHeightRadius
      width=0
      color {colorUseRule=UseColor_Specified, specifiedColor=<rgba 0 0 0 0>}
      cornerRadii=<0 0 0 0>
     }
     tintColor {colorUseRule=UseColor_Specified, specifiedColor=<rgba 1 1 1 1>}
     uVRegion {bIsValid=false, max=<0 0>, min=<0 0>}
    }
    colorAndOpacity=<rgba 0.5089 0.5583 0.7084 1>
   }
   slot {class=SizeBoxSlot, properties {horizontalAlignment=HAlign_Left, verticalAlignment=VAlign_Center}}
  }
  HorizontalBox_Buttons {
   widgetClassPath=&HorizontalBox
   bIsVariable=true
   parent=VerticalBox_271
   widget=HorizontalBox_Buttons
   properties {visibility=Collapsed}
   slot {class=VerticalBoxSlot, properties {padding=<ltrb 12 0 12 8>}}
  }
  JoinLeaveButton {
   widgetClassPath=&/SidebarRoundedRectangleButton
   bInherited=true
   bIsVariable=true
   parent=HorizontalBox_Buttons
   widget=JoinLeaveButton
   properties {
    bDisplayInputActionWhenNotInteractable=true
    bIsEnabled=false
    bIsFocusable=true
    buttonMaterial=null
    defaultText=JOIN
    flowDirectionPreference=Culture
    holdData=&CommonUI_HoldData_Base_C
    navigation=&self:WidgetTree.JoinLeaveButton.WidgetNavigation_0
    onButtonBaseClicked=(&self:WidgetTree.JoinLeaveButton.HandleOnButtonClicked)
    onButtonBaseDoubleClicked=(&self:WidgetTree.JoinLeaveButton.HandleOnButtonDoubleClicked)
    onButtonBaseHovered=(&self:WidgetTree.JoinLeaveButton.HandleOnButtonHovered)
    onButtonBaseUnhovered=(&self:WidgetTree.JoinLeaveButton.HandleOnButtonUnhovered)
    onSelectedChangedBase=(&self:WidgetTree.JoinLeaveButton.HandleOnSelectedChanged)
    style=/FortUILibrary/Foundation/Buttons/ButtonStyles/ButtonStyle-Empty.ButtonStyle-Empty_C
    visibility=Visible
   }
   slot {class=HorizontalBoxSlot, properties {size {sizeRule=Fill, value=0.5}}}
  }
  ReportButton {
   widgetClassPath=&SidebarRoundedRectangleButton
   bInherited=true
   bIsVariable=true
   parent=HorizontalBox_Buttons
   widget=ReportButton
   properties {
    bDisplayInputActionWhenNotInteractable=true
    bIsEnabled=false
    bIsFocusable=true
    buttonMaterial=null
    defaultText=Report Conversations
    flowDirectionPreference=Culture
    holdData=&CommonUI_HoldData_Base_C
    navigation=&self:WidgetTree.ReportButton.WidgetNavigation_0
    onButtonBaseClicked=(&self:WidgetTree.ReportButton.HandleOnButtonClicked)
    onButtonBaseDoubleClicked=(&self:WidgetTree.ReportButton.HandleOnButtonDoubleClicked)
    onButtonBaseHovered=(&self:WidgetTree.ReportButton.HandleOnButtonHovered)
    onButtonBaseUnhovered=(&self:WidgetTree.ReportButton.HandleOnButtonUnhovered)
    onSelectedChangedBase=(&self:WidgetTree.ReportButton.HandleOnSelectedChanged)
    style=&ButtonStyle-Empty_C
    visibility=Visible
   }
   slot {class=HorizontalBoxSlot, properties {size {sizeRule=Fill, value=0.5}}}
  }
  Text_ChannelInfo_Hover {widgetClassPath=&CommonTextBlock, bInherited=true, widget=null}
 ]
}
