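# Sidebar voice chat panel widget, derived from FortSidebarPanelVoiceChat, showing voice channel members, NPCs and channel settings.
# A titled header and NPC tree sit above a switcher between disabled and user-tree views, with settings and a supervised-settings tooltip below.
# A full-panel VRR voice-chat-deactivated overlay starts collapsed and, if visible on activation, is notified via OnSidebarPanelVoiceChatActivated.
# OpenParentalControls forwards to RequestToOpenParentalControls.
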WidgetBlueprint /FortniteCoreUI/UI/Sidebar/SidebarPanelVoiceChat
graphs [EventGraph {nodes [
   Event OpenParentalControls() {is=title K2Node_Event}
   > RequestToOpenParentalControls!(execute) -> (then)
   Event On Activated()
   > n1 = Switch on ESlateVisibility!(
    execute
    Selection:ESlateVisibility Enum=GetVisibility!(self:Widget Object Reference=Get SidebarVRRVoiceChatDeactivated() -> (SidebarVRRVoiceChatDeactivated:Social List Entry Empty Panel Base Object Reference)) -> (ReturnValue:ESlateVisibility Enum=Visible)
   ) -> (Visible:Exec, Collapsed:Exec, Hidden:Exec, HitTestInvisible:Exec, SelfHitTestInvisible:Exec) {is=title K2Node_SwitchEnum}
   n2 = Cast To SidebarVRRVoiceChatDeactivated!(execute=[@n1.HitTestInvisible, @n1.SelfHitTestInvisible, @n1.Visible], Object=Get SidebarVRRVoiceChatDeactivated()) -> (then, CastFailed, AsSidebar VRRVoice Chat Deactivated:Sidebar VRRVoice Chat Deactivated Object Reference) {is=title K2Node_DynamicCast}
   > OnSidebarPanelVoiceChatActivated(self:Sidebar VRRVoice Chat Deactivated Object Reference=@n2)
  ]}]
variables [ParentalControlsClass_New {type=Class, default=/FortniteCoreUI/UI/ContentControls/ParentalControlsScreen.ParentalControlsScreen_C}]
widgets {
 info {inheritedWidgetCount=11, parentClass=/Script/FortniteUI.FortSidebarPanelVoiceChat, rootWidgetClass=&Overlay, widgetCount=18}
 widgets [
  Overlay_0 {widgetClassPath=&Overlay, widget=Overlay_0}
  VerticalBoxPanelUI {
   widgetClassPath=&VerticalBox
   bInherited=true
   bIsVariable=true
   parent=Overlay_0
   widget=VerticalBoxPanelUI
   slot {class=OverlaySlot, properties {horizontalAlignment=HAlign_Fill, verticalAlignment=VAlign_Fill}}
  }
  Border_1 {
   widgetClassPath=&Border
   parent=VerticalBoxPanelUI
   widget=Border_1
   properties {
    verticalAlignment=VAlign_Bottom
    background {
     bIsDynamicallyLoaded=false
     drawAs=Image
     imageType=NoImage
     mirroring=NoMirror
     resourceName=None
     resourceObject=/Game/UI/Sidebar/Materials/MI_UI_Sidebar_Header_Base
     tiling=NoTile
     imageSize=<4096 4096>
     margin=<ltrb 0 0 0 0>
     outlineSettings {
      bUseBrushTransparency=false
      roundingType=HalfHeightRadius
      width=0
      color {colorUseRule=UseColor_Specified, specifiedColor=<rgba 0 0 0 0>}
      cornerRadii=<0 0 0 0>
     }
     tintColor {colorUseRule=UseColor_Specified, specifiedColor=<rgba 1 1 1 1>}
     uVRegion {bIsValid=false, max=<0 0>, min=<0 0>}
    }
    padding=<ltrb 20 60 20 10>
   }
   slot {class=VerticalBoxSlot, properties {verticalAlignment=VAlign_Top}}
  }
  VerticalBox_6 {
   widgetClassPath=&VerticalBox
   parent=Border_1
   widget=VerticalBox_6
   slot {class=BorderSlot, properties {verticalAlignment=VAlign_Bottom, padding=<ltrb 20 60 20 10>}}
  }
  Text_VoiceChatTitle {
   widgetClassPath=&CommonTextBlock
   parent=VerticalBox_6
   widget=Text_VoiceChatTitle
   properties {
    text=Voice Chat
    textTransformPolicy=ToUpper
    font {
     bForceMonospaced=false
     bMaterialIsStencil=false
     fontMaterial=null
     fontObject=&BurbankSmall
     letterSpacing=0
     monospacedWidth=1
     size=32
     skewAmount=0
     typefaceFontName=Black
     outlineSettings {
      bApplyOutlineToDropShadows=false
      bMiteredCorners=false
      bSeparateFillAlpha=false
      outlineMaterial=null
      outlineSize=0
      outlineColor=<rgba 0 0 0 1>
     }
    }
    renderTransform {angle=0, scale=<1 1>, shear=<-8 0>, translation=<0 0>}
    shadowOffset=<0 0>
    shapedTextOptions {bOverride_TextFlowDirection=true, bOverride_TextShapingMethod=false, textFlowDirection=Culture, textShapingMethod=Auto}
   }
   slot {class=VerticalBoxSlot, properties {verticalAlignment=VAlign_Top, size {sizeRule=Fill, value=1}}}
  }
  UserTree_Npcs {
   widgetClassPath=/Script/FortniteUI.FortSocialUserTreeView
   bInherited=true
   bIsVariable=true
   parent=VerticalBoxPanelUI
   widget=UserTree_Npcs
   properties {
    entryWidgetClass=&/SidebarUserListHeaderEntry.SidebarUserListHeaderEntry_C
    inviteEntryWidgetClass=&/SidebarInviteListEntry_FriendInvite_XS.SidebarInviteListEntry_FriendInvite_XS_C
    joinRequestEntryWidgetClass=/FortniteCoreUI/UI/Sidebar/ListEntries/SidebarJoinRequestUserListEntry.SidebarJoinRequestUserListEntry_C
    joinablePartyEntryWidgetClass=&/SidebarJoinablePartyEntry.SidebarJoinablePartyEntry_C
    joinablePartyEntryWidgetClass_2=/FortniteCoreUI/UI/Sidebar/ListEntries/JoinableParty_UserListEntry.JoinableParty_UserListEntry_C
    leavePartyEntryWidgetClass=&/SidebarLeavePartyListEntry.SidebarLeavePartyListEntry_C
    missedInviteEntryWidgetClass=/FortniteCoreUI/UI/SocialPanel/MissedInvitesEntry.MissedInvitesEntry_C
    npcEntryClass=/FortniteCoreUI/UI/Sidebar/ListEntries/SidebarVoiceChatNpcListEntry.SidebarVoiceChatNpcListEntry_C
    numDesignerPreviewEntries=20
    partyMemberEntryWidgetClass=/FortniteCoreUI/UI/Sidebar/ListEntries/SidebarPartyMemberListEntry.SidebarPartyMemberListEntry_C
    referFriendEntryWidgetClass=&/SidebarReferAFriendEntry.SidebarReferAFriendEntry_C
    suggestedFriendEntryWidgetClass=/FortniteCoreUI/UI/Sidebar/ListEntries/SidebarSuggestedFriendListEntry.SidebarSuggestedFriendListEntry_C
    teamMemberEntryWidgetClass=/FortniteCoreUI/UI/Sidebar/ListEntries/SidebarTeamMemberListEntry.SidebarTeamMemberListEntry_C
    userEntryWidgetClass=/FortniteCoreUI/UI/Sidebar/ListEntries/SidebarUserListEntry_XS.SidebarUserListEntry_XS_C
    userSearchResultEntryWidgetClass=/FortniteCoreUI/UI/Sidebar/ListEntries/SidebarUserSearchResultListEntry_XS.SidebarUserSearchResultListEntry_XS_C
    visibility=SelfHitTestInvisible
    voiceChannelHeaderEntryWidgetClass=&/SidebarVoiceChannelHeaderButton.SidebarVoiceChannelHeaderButton_C
    voiceChannelHeaderOptionEntryWidgetClass=/Game/UI/Sidebar/SidebarVoiceChannelHeaderOptions.SidebarVoiceChannelHeaderOptions_C
    voiceChatMemberEntryWidgetClass=/FortniteCoreUI/UI/Sidebar/ListEntries/SidebarVoiceChatMemberListEntry.SidebarVoiceChatMemberListEntry_C
   }
   slot {class=VerticalBoxSlot}
  }
  VoiceChannelPanelSwitcher {
   widgetClassPath=&CommonVisibilitySwitcher
   bInherited=true
   parent=VerticalBoxPanelUI
   widget=VoiceChannelPanelSwitcher
   properties {bAutoActivateSlot=false, shownVisibility=Visible}
   slot {class=VerticalBoxSlot, properties {size {sizeRule=Fill, value=1}}}
  }
  VoiceChatDisabledPanel {
   widgetClassPath=&/SidebarEmptyPanelWidget
   bInherited=true
   bIsVariable=true
   parent=VoiceChannelPanelSwitcher
   widget=VoiceChatDisabledPanel
   properties {
    onButtonBaseClicked=(&self:WidgetTree.VoiceChatDisabledPanel.HandleOnButtonClicked)
    onButtonBaseDoubleClicked=(&self:WidgetTree.VoiceChatDisabledPanel.HandleOnButtonDoubleClicked)
    onButtonBaseHovered=(&self:WidgetTree.VoiceChatDisabledPanel.HandleOnButtonHovered)
    onButtonBaseUnhovered=(&self:WidgetTree.VoiceChatDisabledPanel.HandleOnButtonUnhovered)
    onSelectedChangedBase=(&self:WidgetTree.VoiceChatDisabledPanel.HandleOnSelectedChanged)
    visibility=Visible
    hoveredSlateSoundOverride {resourceObject=/Game/Sounds/Fort_UI_Sounds/FrontEnd/DefaultSounds/PressedButtons/MSS_UI_Button_Hover_Default}
    pressedSlateSoundOverride {resourceObject=/Game/Sounds/Fort_UI_Sounds/FrontEnd/DefaultSounds/PressedButtons/MSS_UI_Button_Pressed_Default}
   }
   slot {class=CommonVisibilitySwitcherSlot}
  }
  VoiceChatEnabledPanel {
   widgetClassPath=&SizeBox
   bInherited=true
   parent=VoiceChannelPanelSwitcher
   widget=VoiceChatEnabledPanel
   properties {heightOverride=2000}
   slot {class=CommonVisibilitySwitcherSlot}
  }
  UserTree_VoiceChat {
   widgetClassPath=&FortSocialUserTreeView
   bInherited=true
   bIsVariable=true
   parent=VoiceChatEnabledPanel
   widget=UserTree_VoiceChat
   properties {
    entryWidgetClass=&SidebarUserListHeaderEntry_C
    inviteEntryWidgetClass=&SidebarInviteListEntry_FriendInvite_XS_C
    joinRequestEntryWidgetClass=&SidebarJoinRequestUserListEntry_C
    joinablePartyEntryWidgetClass=&SidebarJoinablePartyEntry_C
    joinablePartyEntryWidgetClass_2=&JoinableParty_UserListEntry_C
    leavePartyEntryWidgetClass=&SidebarLeavePartyListEntry_C
    missedInviteEntryWidgetClass=&MissedInvitesEntry_C
    npcEntryClass=&SidebarVoiceChatNpcListEntry_C
    numDesignerPreviewEntries=20
    partyMemberEntryWidgetClass=&SidebarPartyMemberListEntry_C
    referFriendEntryWidgetClass=&SidebarReferAFriendEntry_C
    suggestedFriendEntryWidgetClass=&SidebarSuggestedFriendListEntry_C
    teamMemberEntryWidgetClass=&SidebarTeamMemberListEntry_C
    userEntryWidgetClass=&SidebarUserListEntry_XS_C
    userSearchResultEntryWidgetClass=&SidebarUserSearchResultListEntry_XS_C
    visibility=SelfHitTestInvisible
    voiceChannelHeaderEntryWidgetClass=&SidebarVoiceChannelHeaderButton_C
    voiceChannelHeaderOptionEntryWidgetClass=&SidebarVoiceChannelHeaderOptions_C
    voiceChatMemberEntryWidgetClass=&SidebarVoiceChatMemberListEntry_C
   }
   slot {class=SizeBoxSlot}
  }
  Overlay_47 {widgetClassPath=&Overlay, parent=VerticalBoxPanelUI, widget=Overlay_47, slot {class=VerticalBoxSlot}}
  SidebarVoiceChannelSettingsPanel {
   widgetClassPath=&/SidebarVoiceChannelSettings
   bInherited=true
   bIsVariable=true
   parent=Overlay_47
   widget=SidebarVoiceChannelSettingsPanel
   slot {class=OverlaySlot, properties {horizontalAlignment=HAlign_Fill, verticalAlignment=VAlign_Bottom}}
  }
  Overlay_195 {
   widgetClassPath=&Overlay
   parent=Overlay_47
   widget=Overlay_195
   properties {renderTransform {angle=0, scale=<1 1>, shear=<0 0>, translation=<575 -65>}}
   slot {class=OverlaySlot}
  }
  SidebarVoiceChatPanelSupervisedSettingsTooltip {
   widgetClassPath=&/VoiceChatPanelSupervisedSettingsTooltip
   bInherited=true
   bIsVariable=true
   parent=Overlay_195
   widget=SidebarVoiceChatPanelSupervisedSettingsTooltip
   slot {class=OverlaySlot}
  }
  SidebarVRRVoiceChatDeactivated {
   widgetClassPath=&/SidebarVRRVoiceChatDeactivated
   bInherited=true
   bIsVariable=true
   parent=Overlay_0
   widget=SidebarVRRVoiceChatDeactivated
   properties {
    holdData=&CommonUI_HoldData_Base_C
    onButtonBaseClicked=(&self:WidgetTree.SidebarVRRVoiceChatDeactivated.HandleOnButtonClicked)
    onButtonBaseDoubleClicked=(&self:WidgetTree.SidebarVRRVoiceChatDeactivated.HandleOnButtonDoubleClicked)
    onButtonBaseHovered=(&self:WidgetTree.SidebarVRRVoiceChatDeactivated.HandleOnButtonHovered)
    onButtonBaseUnhovered=(&self:WidgetTree.SidebarVRRVoiceChatDeactivated.HandleOnButtonUnhovered)
    onSelectedChangedBase=(&self:WidgetTree.SidebarVRRVoiceChatDeactivated.HandleOnSelectedChanged)
    visibility=Collapsed
   }
   slot {class=OverlaySlot, properties {horizontalAlignment=HAlign_Fill, verticalAlignment=VAlign_Fill}}
  }
  Overlay {
   widgetClassPath=&Overlay
   parent=Overlay_0
   widget=Overlay
   properties {renderTransform {angle=0, scale=<1 1>, shear=<0 0>, translation=<500 50>}}
   slot {class=OverlaySlot}
  }
  VoiceTooltipSlot {widgetClassPath=&NamedSlot, bInherited=true, widget=null}
  VoiceTooltipWidget {widgetClassPath=/Script/FortniteUI.FortSidebarOnboardTooltipWidget, bInherited=true, widget=null}
 ]
}
