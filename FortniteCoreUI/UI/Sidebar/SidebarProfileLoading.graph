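# SidebarProfileLoading is the loading sub-panel of the sidebar social profile, a widget derived from FortSocialProfileSubPanel_Loading that shows a throbber while loading is in progress.
# An overlay holds a single image that draws the monochrome UIKit throbber material.
# The parent's OnLoadingStarted event makes the widget visible without taking hit tests, and OnLoadingFinished collapses it.
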WidgetBlueprint /FortniteCoreUI/UI/Sidebar/SidebarProfileLoading
graphs [EventGraph {nodes [
   Event OnLoadingStarted() {is=title K2Node_Event}
   > SetVisibility(InVisibility=SelfHitTestInvisible)
   Event OnLoadingFinished() {is=title K2Node_Event}
   > SetVisibility(InVisibility=Collapsed)
   Event PreConstruct()
   Event Construct()
   Event Tick()
  ]}]
variables []
widgets {
 info {inheritedWidgetCount=0, parentClass=/Script/FortniteUI.FortSocialProfileSubPanel_Loading, rootWidgetClass=&Overlay, widgetCount=2}
 widgets [
  Overlay_Spinner {widgetClassPath=&Overlay, widget=Overlay_Spinner}
  ProgressSpinner {
   widgetClassPath=&Image
   bIsVariable=true
   parent=Overlay_Spinner
   widget=ProgressSpinner
   properties {brush {
     bIsDynamicallyLoaded=false
     drawAs=Image
     imageType=NoImage
     mirroring=NoMirror
     resourceName=None
     resourceObject=/FortUILibrary/UIKit/Throbber/MI_UIKit_Monochrome_Throbber
     tiling=NoTile
     imageSize=<32 32>
     margin=<ltrb 0 0 0 0>
     outlineSettings {
      bUseBrushTransparency=false
      roundingType=HalfHeightRadius
      width=0
      color {colorUseRule=UseColor_Specified, specifiedColor=<rgba 0 0 0 0>}
      cornerRadii=<0 0 0 0>
     }
     tintColor {colorUseRule=UseColor_Specified, specifiedColor=<rgba 1 1 1 1>}
     uVRegion {bIsValid=false, max=<0 0>, min=<0 0>}
    }}
   slot {class=OverlaySlot}
  }
 ]
}
