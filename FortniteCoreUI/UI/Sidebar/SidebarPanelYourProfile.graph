WidgetBlueprint /FortniteCoreUI/UI/Sidebar/SidebarPanelYourProfile
animations [OnContentLoaded {
  binding_count=2
  length_seconds=0.2
  object=&self:OnContentLoaded
  sequence {
   bindings [
    Border_PresenceRegion {
     id=#1
     possessed_class=CommonBorder
     tracks [
      {
       type=MovieSceneDoubleVectorTrack
       display_name=Desired Size Scale
       sections [{
         type=MovieSceneDoubleVectorSection
         end_frame=12
         start_frame=0
         channels [
          Vector.X {type=MovieSceneScriptingDoubleChannel}
          Vector.Y {type=MovieSceneScriptingDoubleChannel, keys [[0,0],[12,1]]}
         ]
        }]
      }
      {
       type=MovieSceneFloatTrack
       display_name=Render Opacity
       sections [{type=MovieSceneFloatSection, end_frame=12, start_frame=0, channels [None {type=MovieSceneScriptingFloatChannel, keys [[6,0],[12,1]]}]}]
      }
      {
       type=MovieSceneByteTrack
       display_name=Visibility
       sections [{type=MovieSceneByteSection, end_frame=0, start_frame=0, channels [None {type=MovieSceneScriptingByteChannel, const [4,1]}]}]
      }
     ]
    }
    SidebarProfileLoading {
     id=#2
     possessed_class=Image
     tracks [
      {
       type=MovieSceneByteTrack
       display_name=Visibility
       sections [{type=MovieSceneByteSection, end_frame=12, start_frame=0, channels [None {type=MovieSceneScriptingByteChannel, keys [[0,4],[11,1]]}]}]
      }
      {
       type=MovieSceneFloatTrack
       display_name=Render Opacity
       sections [{type=MovieSceneFloatSection, end_frame=12, start_frame=0, channels [None {type=MovieSceneScriptingFloatChannel, keys [[0,1],[6,0]]}]}]
      }
     ]
    }
   ]
   display_rate {numerator=60}
  }
 }]
graphs [
 ChangeParentOnTouch {nodes [
   n1 = ChangeParentOnTouch!() -> (then, IsTouch:bool=~) {is=title K2Node_FunctionEntry}
   > n2 = Branch(Condition=@n1)
   > AddChild(
    self=Get ContainerTouch() -> (ContainerTouch:Border Object Reference)
    Content=Get VBox_Details() -> (VBox_Details:Vertical Box Object Reference)
   )
   AddChild(execute=@n2.else, self=Get ScrollBox_Details() -> (ScrollBox_Details:Scroll Box Object Reference), Content=Get VBox_Details())
  ]}
 ChangeParentOnTouchEvent {nodes [
   ChangeParentOnTouchEvent() {is=title K2Node_FunctionEntry}
   > ChangeParentOnTouch!(execute, self, IsTouch:bool=Get IsTouch() -> (IsTouch:bool)) -> (then) {is=title K2Node_CallFunction}
  ]}
 EventGraph {nodes [
   n1 = "On Entry Initialized (ListView_Settings)"!() -> (OutputDelegate, then, Item:Object Reference, Widget:User Widget Object Reference) {is=title K2Node_ComponentBoundEvent}
   > Branch(Condition=Get Should Show Onboarding() -> (Should Show Onboarding:bool))
   n2 = GetDisplayName(Object=@n1.Item)
   > Branch(Condition="integer > integer"(A=FindSubstring(SearchIn=@n2, Substring=ShareActivity)))
   > Set ShareActivityOption(ShareActivityOption:User Widget Object Reference=@n1.Widget) -> (Output_Get:User Widget Object Reference)
   "On User Scrolled (ScrollBox_Details)"!() -> (OutputDelegate, then, CurrentOffset:float) {is=title K2Node_ComponentBoundEvent}
   > n3 = Branch(Condition=Get Should Show Onboarding())
   n4 = Get ShareActivityOption() -> (ShareActivityOption:User Widget Object Reference)
   Cast To SidebarSettingsListEntry_Discrete_Narrow!(execute=[@n9.then, @n3.then], Object=@n4) -> (then, CastFailed, AsSidebar Settings List Entry Discrete Narrow:Sidebar Settings List Entry Discrete Narrow Object Reference) {is=title K2Node_DynamicCast}
   > n5 = Do Once Then Do!(execute:Exec, Reset:Exec, StartClosed:bool) -> (DoOnce:Exec, ThenDo:Exec) {is=title K2Node_MacroInstance}
   > n6 = ActivateWidget(self=Get WBP_Sidebar_ActivitySharing_CoachMark!(self:Sidebar Right Object Reference=Get Sidebar Right() -> (Sidebar Right:Sidebar Right Object Reference)) -> (WBP_Sidebar_ActivitySharing_CoachMark:WBP Sidebar Activity Sharing Coach Mark Object Reference))
   n7 = LocalToViewport!(
    Geometry:"Geometry Structure (by ref)"=GetCachedGeometry!(self:Widget Object Reference=@n4) -> (ReturnValue:"Geometry Structure (by ref)")
    LocalCoordinate:Vector 2D Structure
   ) -> (PixelPosition_X:double, PixelPosition_Y:double, ViewportPosition_X:double, ViewportPosition_Y:double)
   Set!(
    execute=[@n6.then, @n5.ThenDo]
    RenderTransform_Translation_X:double=0
    RenderTransform_Translation_Y:double=SelectFloat(
     A=@n7.ViewportPosition_Y
     B=float * float(A=@n7.ViewportPosition_Y, B=-1) {is=title K2Node_PromotableOperator}
     bPickA="float > float"(A=@n7.ViewportPosition_Y)
    )
    RenderTransform_Scale:Vector 2D Structure=(X=1,Y=1)
    RenderTransform_Shear:Vector 2D Structure=(X=0,Y=0)
    RenderTransform_Angle:float=0
    self:Widget Object Reference=Get WBP_Sidebar_ActivitySharing_CoachMark(self=Get Sidebar Right())
   ) -> (
    then
    Output_Get_Translation_X:double
    Output_Get_Translation_Y:double
    Output_Get_Scale:Vector 2D Structure
    Output_Get_Shear:Vector 2D Structure
    Output_Get_Angle:float=~
   )
   n8 = OnTooltipDeactivated() {is=title K2Node_CustomEvent}
   > RequestRefreshFocus(self:Common Activatable Widget Object Reference)
   n9 = On Content Loaded() {is=title K2Node_CustomEvent}
   Event On Deactivated()
   > DeactivateWidget(self=Get WBP_Sidebar_ActivitySharing_CoachMark(self=Get Sidebar Right()))
   > Set Should Show Onboarding(Should Show Onboarding:bool=false) -> (Output_Get:bool)
   Event OnStartProfileLoading() {is=title K2Node_Event}
   > SetVisibility(self=Get SidebarProfileLoading() -> (SidebarProfileLoading:Sidebar Profile Loading Object Reference), InVisibility=SelfHitTestInvisible)
   > SetVisibility(self=Get Border_PresenceRegion() -> (Border_PresenceRegion:Common Border Object Reference), InVisibility=Collapsed)
   Event OnFinishedProfileLoading() {is=title K2Node_Event}
   > PlayAnimation(InAnimation=Get OnContentLoaded() -> (OnContentLoaded:Widget Animation Object Reference))
   Event OnInitialized()
   > Bind Event to On Input Method Changed(self=Get CommonInputSubsystem(), Delegate=Create Event())
   > OnInputMethodChanged!(execute, self, NewInputType:ECommonInputType Enum=GetCurrentInputType(self=Get CommonInputSubsystem())) -> (then)
   Event Construct()
   > n10 = FindParentWidgetOfType!(execute, StartingWidget:Widget Object Reference=Self-Reference(), Type:Widget Class Reference=/Script/FortniteUI.FortSocialPanelView_Sidebar) -> (then, ReturnValue:Fort Social Panel View Sidebar Object Reference)
   > n11 = Cast To SidebarRight!(execute, Object=@n10) -> (then, CastFailed, AsSidebar Right:Sidebar Right Object Reference) {is=title K2Node_DynamicCast}
   > Set Sidebar Right(Sidebar Right:Sidebar Right Object Reference=@n11) -> (Output_Get:Sidebar Right Object Reference)
   n12 = Get Sidebar Right()
   > n13 = Set Should Show Onboarding(Should Show Onboarding=ShouldShowOnboarding!(self:Fort Sidebar Onboard Tooltip Widget Object Reference=Get WBP_Sidebar_ActivitySharing_CoachMark(self=@n12)) -> (ReturnValue:bool))
   > Branch(Condition=@n13)
   > Bind Event to On Widget Deactivated(self:Common Activatable Widget Object Reference=Get WBP_Sidebar_ActivitySharing_CoachMark(self=Get Sidebar Right()), Delegate=@n8) {is=title K2Node_AddDelegate}
   > BindToAnimationFinished(Animation=Get OnContentLoaded(), Delegate=@n9)
  ]}
 OnInputMethodChanged {nodes [
   n1 = OnInputMethodChanged!() -> (then, NewInputType:ECommonInputType Enum) {is=title K2Node_FunctionEntry}
   > Set IsTouch(IsTouch:bool=Select!(MouseAndKeyboard:bool=false, Gamepad:bool=false, Touch:bool=true, Count:bool=false, Index:ECommonInputType Enum=@n1) -> (ReturnValue:bool)) -> (Output_Get:bool)
   > Set Timer by Event(Delegate=Create Event(), Time=Get InputSwitchDelay() -> (InputSwitchDelay:double), bLooping=false, bMaxOncePerFrame=true)
  ]}
]
variables [
 InputSwitchDelay {type=number, default=0.25}
 IsTouch {type=boolean}
 ShareActivityOption {type=UserWidget, default=null}
 Should Show Onboarding {type=boolean}
 Sidebar Right {type=SidebarRight_C, default=null}
]
widgets {
 info {inheritedWidgetCount=11, parentClass=/Script/FortniteUI.FortSidebarPanelYourProfile, rootWidgetClass=&GridPanel, widgetCount=21}
 widgets [
  GP_Profile {widgetClassPath=&GridPanel, widget=GP_Profile, properties {bOverride_Cursor=true, columnFill [1], rowFill [0,1]}}
  Overlay_Header {
   widgetClassPath=&Overlay
   parent=GP_Profile
   widget=Overlay_Header
   slot {class=GridSlot, properties {layer=1, verticalAlignment=VAlign_Top}}
  }
  SubPanelIdentity {
   widgetClassPath=/Game/UI/Sidebar/SidebarProfileSubPanelIdentity
   bInherited=true
   bIsVariable=true
   parent=Overlay_Header
   widget=SubPanelIdentity
   slot {class=OverlaySlot, properties {horizontalAlignment=HAlign_Fill}}
  }
  SidebarBackButton {
   widgetClassPath=/FortUILibrary/Foundation/Buttons/BackButtonSmall
   bInherited=true
   bIsVariable=true
   parent=Overlay_Header
   widget=SidebarBackButton
   properties {visibility=Collapsed}
   slot {class=OverlaySlot}
  }
  Overlay_0 {widgetClassPath=&Overlay, parent=GP_Profile, widget=Overlay_0, slot {class=GridSlot, properties {row=1}}}
  ScrollBox_Details {
   widgetClassPath=&ScrollBox
   bIsVariable=true
   parent=Overlay_0
   widget=ScrollBox_Details
   properties {
    alwaysShowScrollbarTrack=true
    bAnimateWheelScrolling=true
    clipping=Inherit
    navigationScrollPadding=250
    scrollWhenFocusChanges=AnimatedScroll
    visibility=SelfHitTestInvisible
    wheelScrollMultiplier=2
    scrollbarThickness=<12 12>
   }
   slot {class=OverlaySlot, properties {horizontalAlignment=HAlign_Fill}}
  }
  VBox_Details {widgetClassPath=&VerticalBox, bIsVariable=true, parent=ScrollBox_Details, widget=VBox_Details, slot {class=ScrollBoxSlot}}
  SubPanelTags {
   widgetClassPath=/Game/UI/Sidebar/SidebarProfileSubPanelTags
   bInherited=true
   bIsVariable=true
   parent=VBox_Details
   widget=SubPanelTags
   properties {bAllowEditing=true, bShowAddingWidgetInSocialTagWhenSelected=false, visibility=Collapsed}
   slot {class=VerticalBoxSlot}
  }
  Overlay_PresenceRegion {
   widgetClassPath=&Overlay
   parent=VBox_Details
   widget=Overlay_PresenceRegion
   slot {class=VerticalBoxSlot, properties {padding=<ltrb 0 0 0 10>}}
  }
  SidebarProfileLoading {
   widgetClassPath=&/SidebarProfileLoading
   bIsVariable=true
   parent=Overlay_PresenceRegion
   widget=SidebarProfileLoading
   slot {class=OverlaySlot, properties {horizontalAlignment=HAlign_Center}}
  }
  Border_PresenceRegion {
   widgetClassPath=&CommonBorder
   bIsVariable=true
   parent=Overlay_PresenceRegion
   widget=Border_PresenceRegion
   properties {clipping=ClipToBounds, desiredSizeScale=<1 0>, padding=<ltrb 0 0 0 0>}
   slot {class=OverlaySlot, properties {horizontalAlignment=HAlign_Fill, verticalAlignment=VAlign_Fill}}
  }
  VBox_PresenceRegion {
   widgetClassPath=&VerticalBox
   parent=Border_PresenceRegion
   widget=VBox_PresenceRegion
   slot {class=BorderSlot, properties {padding=<ltrb 0 0 0 0>}}
  }
  SubPanelPresence {
   widgetClassPath=/Game/UI/Sidebar/SidebarProfileSubPanelPresence
   bInherited=true
   bIsVariable=true
   parent=VBox_PresenceRegion
   widget=SubPanelPresence
   properties {padding=<ltrb 0 0 0 12>}
   slot {class=VerticalBoxSlot}
  }
  SubPanelRegion {
   widgetClassPath=/Game/UI/Sidebar/SidebarProfileSubPanelRegion
   bInherited=true
   bIsVariable=true
   parent=VBox_PresenceRegion
   widget=SubPanelRegion
   slot {class=VerticalBoxSlot}
  }
  SubPanelRebootRallyText {
   widgetClassPath=/Game/UI/Sidebar/RebootRally/SidebarProfileSubPanelRebootRallyEligible
   bInherited=true
   bIsVariable=true
   parent=VBox_Details
   widget=SubPanelRebootRallyText
   properties {padding=<ltrb 0 0 0 0>}
   slot {class=VerticalBoxSlot, properties {padding=<ltrb 20 0 20 20>}}
  }
  ListView_Settings {
   widgetClassPath=/Script/FortniteUI.FortSettingsListView
   bInherited=true
   bIsVariable=true
   parent=VBox_Details
   widget=ListView_Settings
   properties {
    bEnableRightClickScrolling=false
    clipping=Inherit
    consumeMouseWheel=Never
    entryWidgetClass=/FortniteCoreUI/Settings/SettingsListEntry_Missing.SettingsListEntry_Missing_C
    horizontalEntrySpacing=12
    verticalEntrySpacing=12
    visibility=SelfHitTestInvisible
    visualData=&/SidebarVoiceChatTab_Visuals
    wheelScrollMultiplier=0
   }
   slot {class=VerticalBoxSlot, properties {padding=<ltrb 0 0 0 25>, size {sizeRule=Fill, value=1}}}
  }
  ContainerTouch {
   widgetClassPath=&Border
   bIsVariable=true
   parent=Overlay_0
   widget=ContainerTouch
   properties {
    visibility=SelfHitTestInvisible
    background {
     bIsDynamicallyLoaded=false
     drawAs=NoDrawType
     imageType=NoImage
     mirroring=NoMirror
     resourceName=None
     resourceObject=null
     tiling=NoTile
     imageSize=<32 32>
     margin=<ltrb 0 0 0 0>
     outlineSettings {
      bUseBrushTransparency=false
      roundingType=HalfHeightRadius
      width=0
      color {colorUseRule=UseColor_Specified, specifiedColor=<rgba 0 0 0 0>}
      cornerRadii=<0 0 0 0>
     }
     tintColor {colorUseRule=UseColor_Specified, specifiedColor=<rgba 1 1 1 1>}
     uVRegion {bIsValid=false, max=<0 0>, min=<0 0>}
    }
    padding=<ltrb 0 0 0 0>
   }
   slot {class=OverlaySlot, properties {horizontalAlignment=HAlign_Fill, verticalAlignment=VAlign_Fill, padding=<ltrb 0 0 0 40>}}
  }
  SubPanelActions {widgetClassPath=/Script/FortniteUI.FortSocialProfileSubPanel, bInherited=true, widget=null}
  SubPanelMutualFriends {widgetClassPath=&FortSocialProfileSubPanel, bInherited=true, widget=null}
  SubPanelMutualFriendsCount {widgetClassPath=&FortSocialProfileSubPanel, bInherited=true, widget=null}
  ScrollBox_ScrollablePanels {widgetClassPath=&ScrollBox, bInherited=true, widget=null}
 ]
}
