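# Right-hand social sidebar widget for Fortnite, a FortSocialPanelView_Sidebar subclass that plays an open sound on activation.
# A splitscreen-scaled button strip sits beside a fixed-width panel whose activatable switcher hosts party, friends, voice, settings, profile, tags, reboot rally and exit pages.
# A plugin injection slot adds a Cilantro page, and an exit popup starts collapsed over the panel.
# Onboarding tooltips and an activity-sharing coach mark overlay the sidebar.
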WidgetBlueprint /FortniteCoreUI/UI/Sidebar/SidebarRight
graphs [EventGraph {nodes [
   Event On Activated()
   > PlaySound2D(Sound=/Game/Sounds/Fort_UI_Sounds/FrontEnd/MSS_UI_Sidebar_Open, VolumeMultiplier=0.75)
  ]}]
variables [
 Text_FindFriends {type=string, default=Find Friends}
 Text_PartyUp {type=string, default=Party Up}
 Text_UserSettings {type=string, default=User Settings}
 Text_VoiceChat {type=string, default=Voice Chat}
]
widgets {
 info {inheritedWidgetCount=18, parentClass=/Script/FortniteUI.FortSocialPanelView_Sidebar, rootWidgetClass=&Overlay, widgetCount=38}
 widgets [
  Overlay_1 {widgetClassPath=&Overlay, widget=Overlay_1}
  SidePanelWidgetSwitcher {
   widgetClassPath=&CommonVisibilitySwitcher
   bIsVariable=true
   parent=Overlay_1
   widget=SidePanelWidgetSwitcher
   slot {class=OverlaySlot, properties {horizontalAlignment=HAlign_Right, verticalAlignment=VAlign_Fill}}
  }
  MainSideBar_Border {
   widgetClassPath=&Border
   bIsVariable=true
   parent=SidePanelWidgetSwitcher
   widget=MainSideBar_Border
   properties {
    background {
     bIsDynamicallyLoaded=false
     drawAs=Image
     imageType=FullColor
     mirroring=NoMirror
     resourceName=None
     resourceObject=/Game/UI/SocialPanel/Art/Materials/M_UI_SocialPanelBackground_New
     tiling=NoTile
     imageSize=<1920 1080>
     margin=<ltrb 0 0 0 0>
     outlineSettings {
      bUseBrushTransparency=false
      roundingType=HalfHeightRadius
      width=0
      color {colorUseRule=UseColor_Specified, specifiedColor=<rgba 0 0 0 0>}
      cornerRadii=<0 0 0 0>
     }
     tintColor {colorUseRule=UseColor_Specified, specifiedColor=<rgba 1 1 1 1>}
     uVRegion {bIsValid=false, max=<0 0>, min=<0 0>}
    }
    padding=<ltrb 0 0 0 0>
   }
   slot {class=CommonVisibilitySwitcherSlot}
  }
  SafeZone_0 {
   widgetClassPath=&SafeZone
   bIsVariable=true
   parent=MainSideBar_Border
   widget=SafeZone_0
   properties {padBottom=false, padLeft=false, padTop=false}
   slot {class=BorderSlot, properties {padding=<ltrb 0 0 0 0>}}
  }
  HorizontalBox_14 {widgetClassPath=&HorizontalBox, parent=SafeZone_0, widget=HorizontalBox_14, slot {class=SafeZoneSlot}}
  ScaleBox_FitIconsInSplitscreen {
   widgetClassPath=&ScaleBox
   parent=HorizontalBox_14
   widget=ScaleBox_FitIconsInSplitscreen
   properties {stretch=ScaleToFitY, stretchDirection=DownOnly}
   slot {class=HorizontalBoxSlot}
  }
  SizeBox_12 {
   widgetClassPath=&SizeBox
   parent=ScaleBox_FitIconsInSplitscreen
   widget=SizeBox_12
   properties {widthOverride=96}
   slot {class=ScaleBoxSlot, properties {horizontalAlignment=HAlign_Fill, verticalAlignment=VAlign_Fill}}
  }
  Border_12 {
   widgetClassPath=&Border
   parent=SizeBox_12
   widget=Border_12
   properties {horizontalAlignment=HAlign_Center, brushColor=<rgba 0.0044 0.0091 0.0343 1>, padding=<ltrb 0 0 0 0>}
   slot {class=SizeBoxSlot}
  }
  SafeZone_SideBar {
   widgetClassPath=&SafeZone
   bIsVariable=true
   parent=Border_12
   widget=SafeZone_SideBar
   properties {padBottom=false, padLeft=false, padRight=false, padTop=false}
   slot {class=BorderSlot, properties {horizontalAlignment=HAlign_Center, padding=<ltrb 0 0 0 0>}}
  }
  Panel_SidebarButtons {
   widgetClassPath=&/SidebarButtons
   bInherited=true
   bIsVariable=true
   parent=SafeZone_SideBar
   widget=Panel_SidebarButtons
   slot {class=SafeZoneSlot}
  }
  SizeBox_1 {
   widgetClassPath=&SizeBox
   parent=HorizontalBox_14
   widget=SizeBox_1
   properties {bOverride_WidthOverride=true, widthOverride=550}
   slot {class=HorizontalBoxSlot}
  }
  Overlay_0 {widgetClassPath=&Overlay, parent=SizeBox_1, widget=Overlay_0, slot {class=SizeBoxSlot}}
  Border_13 {
   widgetClassPath=&Border
   parent=Overlay_0
   widget=Border_13
   properties {
    background {
     bIsDynamicallyLoaded=false
     drawAs=Image
     imageType=FullColor
     mirroring=NoMirror
     resourceName=None
     resourceObject=&M_UI_SocialPanelBackground_New
     tiling=NoTile
     imageSize=<32 32>
     margin=<ltrb 0 0 0 0>
     outlineSettings {
      bUseBrushTransparency=false
      roundingType=HalfHeightRadius
      width=0
      color {colorUseRule=UseColor_Specified, specifiedColor=<rgba 0 0 0 0>}
      cornerRadii=<0 0 0 0>
     }
     tintColor {colorUseRule=UseColor_Specified, specifiedColor=<rgba 1 1 1 1>}
     uVRegion {bIsValid=false, max=<0 0>, min=<0 0>}
    }
    padding=<ltrb 0 0 0 0>
   }
   slot {class=OverlaySlot, properties {horizontalAlignment=HAlign_Fill, verticalAlignment=VAlign_Fill}}
  }
  Overlay_2 {widgetClassPath=&Overlay, parent=Border_13, widget=Overlay_2, slot {class=BorderSlot, properties {padding=<ltrb 0 0 0 0>}}}
  WBP_CaptureForPostBufferUpdate {
   widgetClassPath=&WBP_CaptureForPostBufferUpdate
   bIsVariable=true
   parent=Overlay_2
   slot {class=OverlaySlot, properties {horizontalAlignment=HAlign_Fill, verticalAlignment=VAlign_Fill}}
  }
  Switcher_UserLists {
   widgetClassPath=&CommonActivatableWidgetSwitcher
   bInherited=true
   bIsVariable=true
   parent=Overlay_2
   widget=Switcher_UserLists
   properties {transitionDuration=0}
   slot {class=OverlaySlot, properties {horizontalAlignment=HAlign_Fill, verticalAlignment=VAlign_Fill}}
  }
  Panel_PartyUp {
   widgetClassPath=&/SidebarPanelPartyUp
   bInherited=true
   bIsVariable=true
   parent=Switcher_UserLists
   widget=Panel_PartyUp
   properties {bIsFocusable=true}
   slot {class=WidgetSwitcherSlot}
  }
  Panel_FindFriends {
   widgetClassPath=&/SidebarPanelFindFriends
   bInherited=true
   bIsVariable=true
   parent=Switcher_UserLists
   widget=Panel_FindFriends
   properties {bIsFocusable=true}
   slot {class=WidgetSwitcherSlot}
  }
  Panel_VoiceChat {
   widgetClassPath=&/SidebarPanelVoiceChat
   bInherited=true
   bIsVariable=true
   parent=Switcher_UserLists
   widget=Panel_VoiceChat
   slot {class=WidgetSwitcherSlot}
  }
  Panel_More {
   widgetClassPath=&/SidebarPanelSettings
   bInherited=true
   bIsVariable=true
   parent=Switcher_UserLists
   widget=Panel_More
   properties {bIsFocusable=true}
   slot {class=WidgetSwitcherSlot}
  }
  Panel_YourProfile {
   widgetClassPath=&/SidebarPanelYourProfile
   bInherited=true
   bIsVariable=true
   parent=Switcher_UserLists
   widget=Panel_YourProfile
   properties {bIsFocusable=true}
   slot {class=WidgetSwitcherSlot}
  }
  Panel_Tags {
   widgetClassPath=/FortniteCoreUI/UI/Sidebar/Tags/SidebarPanelTags
   bInherited=true
   bIsVariable=true
   parent=Switcher_UserLists
   widget=Panel_Tags
   properties {bIsFocusable=true}
   slot {class=WidgetSwitcherSlot}
  }
  Panel_Profile {
   widgetClassPath=/Game/UI/Sidebar/SidebarProfile
   bIsVariable=true
   parent=Switcher_UserLists
   widget=Panel_Profile
   properties {bSetVisibilityOnActivated=false, bSetVisibilityOnDeactivated=false}
   slot {class=WidgetSwitcherSlot}
  }
  Panel_EditProfile {
   widgetClassPath=&/SidebarPanelEditProfile
   bInherited=true
   bIsVariable=true
   parent=Switcher_UserLists
   widget=Panel_EditProfile
   properties {bIsFocusable=true}
   slot {class=WidgetSwitcherSlot}
  }
  Panel_RebootRally {
   widgetClassPath=/FortniteCoreUI/UI/Sidebar/RebootRally/WBP_PanelRebootRally
   bInherited=true
   bIsVariable=true
   parent=Switcher_UserLists
   widget=Panel_RebootRally
   properties {bSetVisibilityOnActivated=false, bSetVisibilityOnDeactivated=false}
   slot {class=WidgetSwitcherSlot}
  }
  Panel_Exit {
   widgetClassPath=&/SidebarPanelExit
   bInherited=true
   bIsVariable=true
   parent=Switcher_UserLists
   widget=Panel_Exit
   properties {bIsFocusable=true}
   slot {class=WidgetSwitcherSlot}
  }
  Panel_CilantroInjection {
   widgetClassPath=/Game/UI/Plugin/FortPluginWidgetInjectionSlot
   bInherited=true
   bIsVariable=true
   parent=Switcher_UserLists
   widget=Panel_CilantroInjection
   properties {pluginId=Cilantro, pluginWidgetId=CilantroPanel}
   slot {class=WidgetSwitcherSlot}
  }
  Popup_Exit {
   widgetClassPath=&/SidebarExitPopupRight
   bInherited=true
   bIsVariable=true
   parent=Overlay_0
   widget=Popup_Exit
   properties {visibility=Collapsed}
   slot {class=OverlaySlot, properties {horizontalAlignment=HAlign_Fill, verticalAlignment=VAlign_Fill}}
  }
  Safezone_OnboardingTooltip {
   widgetClassPath=&SafeZone
   bIsVariable=true
   parent=Overlay_1
   widget=Safezone_OnboardingTooltip
   properties {padBottom=false, padLeft=false, padTop=false}
   slot {class=OverlaySlot, properties {horizontalAlignment=HAlign_Right, verticalAlignment=VAlign_Fill, padding=<ltrb 0 0 600 0>}}
  }
  OnboardingTooltip_InputBlocker {
   widgetClassPath=&Overlay
   bInherited=true
   parent=Safezone_OnboardingTooltip
   widget=OnboardingTooltip_InputBlocker
   properties {visibility=Visible}
   slot {class=SafeZoneSlot}
  }
  SidebarOnboardingTagsEditTooltipRight {
   widgetClassPath=/FortniteCoreUI/UI/Sidebar/Tags/SidebarOnboardingTagsEditTooltipRight
   bIsVariable=true
   parent=OnboardingTooltip_InputBlocker
   widget=SidebarOnboardingTagsEditTooltipRight
   properties {bIsBackHandler=false, bIsModal=true, positionController=<59 81>, positionKeyboardMouse=<59 32>}
   slot {class=OverlaySlot}
  }
  SidebarOnboardingLookingForPartyRight {
   widgetClassPath=/FortniteCoreUI/UI/Sidebar/Tags/SidebarOnboardingLookingForPartyRight
   bIsVariable=true
   parent=OnboardingTooltip_InputBlocker
   widget=SidebarOnboardingLookingForPartyRight
   properties {
    bIsModal=true
    positionController=<59 342>
    positionKeyboardMouse=<59 293>
    renderTransform {angle=0, scale=<1 1>, shear=<0 0>, translation=<0 110>}
   }
   slot {class=OverlaySlot}
  }
  SiderbarSupervisedSettingsTooltipRight {
   widgetClassPath=&/SiderbarSupervisedSettingsTooltipRight
   bIsVariable=true
   parent=OnboardingTooltip_InputBlocker
   widget=SiderbarSupervisedSettingsTooltipRight
   properties {positionController=<59 470>, positionKeyboardMouse=<59 470>, renderTransform {angle=0, scale=<1 1>, shear=<0 0>, translation=<0 500>}}
   slot {class=OverlaySlot}
  }
  SidebarCilantroTooltip {
   widgetClassPath=&/SidebarCilantroTooltip
   bInherited=true
   bIsVariable=true
   parent=OnboardingTooltip_InputBlocker
   widget=SidebarCilantroTooltip
   properties {
    bIsModal=true
    positionController=<0 775>
    positionKeyboardMouse=<0 825>
    renderTransform {angle=0, scale=<1 1>, shear=<0 0>, translation=<0 825>}
   }
   slot {class=OverlaySlot}
  }
  WBP_Sidebar_ActivitySharing_CoachMark {
   widgetClassPath=&/WBP_Sidebar_ActivitySharing_CoachMark
   bIsVariable=true
   parent=Overlay_1
   widget=WBP_Sidebar_ActivitySharing_CoachMark
   slot {class=OverlaySlot, properties {horizontalAlignment=HAlign_Right, padding=<ltrb 0 -200 570 0>}}
  }
  SidebarOnboardingTagsEditTooltip {widgetClassPath=/Script/FortniteUI.FortSidebarOnboardTooltipWidget, bInherited=true, widget=null}
  SidebarOnboardingLookingForParty {widgetClassPath=&FortSidebarOnboardTooltipWidget, bInherited=true, widget=null}
  SiderbarSupervisedSettingsTooltip {widgetClassPath=&FortSidebarOnboardTooltipWidget, bInherited=true, widget=null}
 ]
}
