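# Menu button widget, based on LiveStreamMenuToggleButton, for showing or hiding an in-game live stream.
# RefreshButton picks Show or Hide text from the enabled flag, naming Llamatron in Creative/Playground playlists and Party Royale otherwise.
# Creative/Playground playlists also collapse the icon box holding the stage art and LIVE badge.
# Hover and unhover play one Hover animation forward and in reverse, including a material-parameter shift on the button backing.
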WidgetBlueprint /FortniteCoreUI/Frontend/LiveStreamVIdeo/LiveStreamMenuButtonToggle
animations [Hover {
  binding_count=4
  length_seconds=0.25
  object=&self:Hover
  sequence {bindings [
    BackgroundImage {
     id=#1
     display_name="ButtonBacking (4)"
     possessed_class=Image
     tracks [{
       type=MovieScene2DTransformTrack
       sections [{
         type=MovieScene2DTransformSection
         end_frame=5
         start_frame=0
         channels [
          Translation.X {type=MovieSceneScriptingFloatChannel}
          Translation.Y {type=MovieSceneScriptingFloatChannel}
          Angle {type=MovieSceneScriptingFloatChannel}
          Scale.X {type=MovieSceneScriptingFloatChannel, keys [[0,1],[5,1.2]]}
          Scale.Y {type=MovieSceneScriptingFloatChannel, keys [[0,1],[5,1.2]]}
          Shear.X {type=MovieSceneScriptingFloatChannel}
          Shear.Y {type=MovieSceneScriptingFloatChannel}
         ]
        }]
      }]
    }
    ButtonBacking {
     id=#2
     possessed_class=Image
     tracks [{
       type=MovieSceneWidgetMaterialTrack
       sections [{
         type=MovieSceneParameterSection
         channels [
          OnHover {type=MovieSceneScriptingFloatChannel, keys [[0,-1],[5,-0.5]]}
          RGB_1R {type=MovieSceneScriptingFloatChannel, keys [[0,0.05],[3,0.25]]}
          RGB_1G {type=MovieSceneScriptingFloatChannel, keys [[0,0.6833],[3,0.9]]}
          RGB_1B {type=MovieSceneScriptingFloatChannel, const [1,2]}
          RGB_1A {type=MovieSceneScriptingFloatChannel, const [1,2]}
          RGB_2R {type=MovieSceneScriptingFloatChannel, keys [[0,0.025],[3,0]]}
          RGB_2G {type=MovieSceneScriptingFloatChannel, keys [[0,0.48],[3,0.333]]}
          RGB_2B {type=MovieSceneScriptingFloatChannel, const [1,2]}
          RGB_2A {type=MovieSceneScriptingFloatChannel, const [1,2]}
         ]
        }]
      }]
    }
    "ButtonBacking (2)" {
     id=#3
     possessed_class=CommonTextBlock
     tracks [{
       type=MovieSceneColorTrack
       display_name=Color and Opacity
       sections [{
         type=MovieSceneColorSection
         channels [
          Color.R {type=MovieSceneScriptingFloatChannel, keys [[0,0.5],[3,1]]}
          Color.G {type=MovieSceneScriptingFloatChannel, keys [[0,0.9167],[3,1]]}
          Color.B {type=MovieSceneScriptingFloatChannel, const [1,2]}
          Color.A {type=MovieSceneScriptingFloatChannel, const [1,2]}
         ]
        }]
      }]
    }
   ]}
 }]
graphs [EventGraph {nodes [
   n1 = Event RefreshButton!() -> (OutputDelegate, then, IsLiveStreamEnabled:bool) {is=title K2Node_Event}
   n2 = IsCreativeOrPlaygroundModePlaylist!() -> (ReturnValue:bool)
   > "SetText (Text)"(
    self=Get ButtonText() -> (ButtonText:Common Text Object Reference)
    InText=Select(
     Option 0:Text=Select(Option 0="LOCTEXT(\"Show Party Royale Stream\")", Option 1="LOCTEXT(\"Hide Party Royale Stream\")", Index=@n1.IsLiveStreamEnabled)
     Option 1:Text=Select(Option 0="LOCTEXT(\"Show Llamatron Stream\")", Option 1="LOCTEXT(\"Hide Llamatron Stream\")", Index=@n1.IsLiveStreamEnabled)
     Index=@n2
    ) -> (ReturnValue:Text)
   )
   > SetVisibility(
    self=Get IconBox() -> (IconBox:Overlay Object Reference)
    InVisibility=Select(Option 0:ESlateVisibility Enum=SelfHitTestInvisible, Option 1:ESlateVisibility Enum=Collapsed, Index=@n2) -> (ReturnValue:ESlateVisibility Enum)
   )
   Event On Hovered()
   n3 = Get Hover() -> (Hover:Widget Animation Object Reference)
   > PlayAnimationForward(InAnimation=@n3, PlaybackSpeed=1.5)
   Event On Unhovered()
   > PlayAnimationReverse(InAnimation=@n3)
   Event Construct()
   Get Art() -> (Art:Image Object Reference)
  ]}]
variables []
widgets {
 info {inheritedWidgetCount=1, parentClass=/Script/FortniteUI.LiveStreamMenuToggleButton, rootWidgetClass=&Overlay, widgetCount=16}
 widgets [
  Overlay_0 {widgetClassPath=&Overlay, widget=Overlay_0}
  SizeBox_0 {
   widgetClassPath=&SizeBox
   parent=Overlay_0
   widget=SizeBox_0
   properties {bOverride_HeightOverride=true, heightOverride=90, widthOverride=350}
   slot {class=OverlaySlot, properties {horizontalAlignment=HAlign_Fill}}
  }
  Overlay_1 {widgetClassPath=&Overlay, parent=SizeBox_0, widget=Overlay_1, slot {class=SizeBoxSlot}}
  ButtonBacking {
   widgetClassPath=&Image
   bIsVariable=true
   parent=Overlay_1
   widget=ButtonBacking
   properties {brush {
     bIsDynamicallyLoaded=false
     drawAs=Image
     imageType=NoImage
     mirroring=NoMirror
     resourceName=None
     resourceObject=/FortniteCoreUI/Athena/UI/Season/Art/Athena_Button_Solo_BattlePass_Purchase_Blue
     tiling=NoTile
     imageSize=<32 32>
     margin=<ltrb 0 0 0 0>
     outlineSettings {
      bUseBrushTransparency=false
      roundingType=HalfHeightRadius
      width=0
      color {colorUseRule=UseColor_Specified, specifiedColor=<rgba 0 0 0 0>}
      cornerRadii=<0 0 0 0>
     }
     tintColor {colorUseRule=UseColor_Specified, specifiedColor=<rgba 1 1 1 1>}
     uVRegion {bIsValid=false, max=<0 0>, min=<0 0>}
    }}
   slot {class=OverlaySlot, properties {horizontalAlignment=HAlign_Fill, verticalAlignment=VAlign_Fill}}
  }
  Spikes_1 {
   widgetClassPath=&Image
   bIsVariable=true
   parent=Overlay_1
   widget=Spikes_1
   properties {
    brush {
     bIsDynamicallyLoaded=false
     drawAs=Image
     imageType=NoImage
     mirroring=NoMirror
     resourceName=None
     resourceObject=/Game/UI/HUD/Materials/M_FX_TriangleSpiral
     tiling=NoTile
     imageSize=<201.6473 116.0626>
     margin=<ltrb 0 0 0 0>
     outlineSettings {
      bUseBrushTransparency=false
      roundingType=HalfHeightRadius
      width=0
      color {colorUseRule=UseColor_Specified, specifiedColor=<rgba 0 0 0 0>}
      cornerRadii=<0 0 0 0>
     }
     tintColor {colorUseRule=UseColor_Specified, specifiedColor=<rgba 1 1 1 1>}
     uVRegion {bIsValid=false, max=<0 0>, min=<0 0>}
    }
    colorAndOpacity=<rgba 1 1 1 0.15>
    renderTransform {angle=0, scale=<-1 1>, shear=<0 0>, translation=<0 0>}
   }
   slot {class=OverlaySlot, properties {horizontalAlignment=HAlign_Fill, verticalAlignment=VAlign_Fill, padding=<ltrb 0 0 8 0>}}
  }
  HorizontalBox_69 {
   widgetClassPath=&HorizontalBox
   parent=Overlay_1
   widget=HorizontalBox_69
   slot {class=OverlaySlot, properties {verticalAlignment=VAlign_Center}}
  }
  HorizontalBox_0 {
   widgetClassPath=&HorizontalBox
   parent=HorizontalBox_69
   widget=HorizontalBox_0
   slot {
    class=HorizontalBoxSlot
    properties {horizontalAlignment=HAlign_Left, verticalAlignment=VAlign_Center, padding=<ltrb 16 0 16 0>, size {sizeRule=Fill, value=0.8}}
   }
  }
  InputActionWidget {
   widgetClassPath=&CommonActionWidget
   bInherited=true
   bIsVariable=true
   parent=HorizontalBox_0
   widget=InputActionWidget
   properties {
    progressMaterialParam=Percentage
    progressMaterialBrush {
     bIsDynamicallyLoaded=false
     drawAs=Image
     imageType=NoImage
     mirroring=NoMirror
     resourceName=None
     resourceObject=/Game/UI/Materials/HollowCircularProgressBar
     tiling=NoTile
     imageSize=<36 36>
     margin=<ltrb 0 0 0 0>
     outlineSettings {
      bUseBrushTransparency=false
      roundingType=HalfHeightRadius
      width=0
      color {colorUseRule=UseColor_Specified, specifiedColor=<rgba 0 0 0 0>}
      cornerRadii=<0 0 0 0>
     }
     tintColor {colorUseRule=UseColor_Specified, specifiedColor=<rgba 1 1 0 1>}
     uVRegion {bIsValid=false, max=<0 0>, min=<0 0>}
    }
   }
   slot {class=HorizontalBoxSlot, properties {horizontalAlignment=HAlign_Left, verticalAlignment=VAlign_Center, padding=<ltrb 4 4 4 4>}}
  }
  ButtonText {
   widgetClassPath=&CommonTextBlock
   bIsVariable=true
   parent=HorizontalBox_0
   widget=ButtonText
   properties {
    scrollStyle=&TextScrollStyle-Base_C
    text=Show Party Royale Stream
    textTransformPolicy=ToUpper
    font {
     bForceMonospaced=false
     bMaterialIsStencil=false
     fontMaterial=null
     fontObject=&BurbankSmall
     letterSpacing=0
     monospacedWidth=1
     size=16
     skewAmount=0
     typefaceFontName=Black
     outlineSettings {
      bApplyOutlineToDropShadows=false
      bMiteredCorners=false
      bSeparateFillAlpha=false
      outlineMaterial=null
      outlineSize=1
      outlineColor=<rgba 0 0.2764 0.4115 0.7>
     }
    }
    shadowColorAndOpacity=<rgba 0 0.2764 0.4115 0.7>
   }
   slot {class=HorizontalBoxSlot, properties {verticalAlignment=VAlign_Center, size {sizeRule=Fill, value=1}}}
  }
  IconBox {
   widgetClassPath=&Overlay
   bIsVariable=true
   parent=HorizontalBox_69
   widget=IconBox
   slot {class=HorizontalBoxSlot, properties {horizontalAlignment=HAlign_Right, verticalAlignment=VAlign_Top}}
  }
  SizeBox_1 {
   widgetClassPath=&SizeBox
   parent=IconBox
   widget=SizeBox_1
   properties {
    bOverride_HeightOverride=true
    bOverride_WidthOverride=true
    heightOverride=96
    widthOverride=96
    renderTransform {angle=0, scale=<1 1>, shear=<0 0>, translation=<15 0>}
   }
   slot {class=OverlaySlot, properties {horizontalAlignment=HAlign_Right, verticalAlignment=VAlign_Center}}
  }
  ScaleBox_1 {
   widgetClassPath=&ScaleBox
   parent=SizeBox_1
   widget=ScaleBox_1
   slot {class=SizeBoxSlot, properties {padding=<ltrb -12 -12 -12 -12>}}
  }
  Overlay_2 {
   widgetClassPath=&Overlay
   parent=ScaleBox_1
   widget=Overlay_2
   properties {renderTransform {angle=0, scale=<1 1>, shear=<0 0>, translation=<-5 0>}}
   slot {class=ScaleBoxSlot, properties {horizontalAlignment=HAlign_Right}}
  }
  Art {
   widgetClassPath=&Image
   bIsVariable=true
   parent=Overlay_2
   widget=Art
   properties {
    visibility=SelfHitTestInvisible
    brush {
     bIsDynamicallyLoaded=false
     drawAs=Image
     imageType=NoImage
     mirroring=NoMirror
     resourceName=None
     resourceObject=/Game/Environments/Papaya/Textures/Payapa_Icons_Dance_Stage
     tiling=NoTile
     imageSize=<512 512>
     margin=<ltrb 0 0 0 0>
     outlineSettings {
      bUseBrushTransparency=false
      roundingType=HalfHeightRadius
      width=0
      color {colorUseRule=UseColor_Specified, specifiedColor=<rgba 0 0 0 0>}
      cornerRadii=<0 0 0 0>
     }
     tintColor {colorUseRule=UseColor_Specified, specifiedColor=<rgba 1 1 1 1>}
     uVRegion {bIsValid=false, max=<0 0>, min=<0 0>}
    }
    renderTransform {angle=0, scale=<1.2 1.2>, shear=<0 0>, translation=<-122 0>}
   }
   slot {class=OverlaySlot, properties {horizontalAlignment=HAlign_Right, verticalAlignment=VAlign_Fill, padding=<ltrb 12 12 12 12>}}
  }
  ScaleBox_0 {
   widgetClassPath=&ScaleBox
   parent=IconBox
   widget=ScaleBox_0
   properties {stretch=UserSpecified, userSpecifiedScale=0.4}
   slot {class=OverlaySlot, properties {horizontalAlignment=HAlign_Right, padding=<ltrb 0 -4 0 0>}}
  }
  LiveViolator {
   widgetClassPath=/Game/UI/Frontend/Showdown/LiveViolator
   bIsVariable=true
   parent=ScaleBox_0
   widget=LiveViolator
   properties {liveText=LIVE, renderTransformPivot=<1 0>}
   slot {class=ScaleBoxSlot}
  }
 ]
}
