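# Creative minigame widget, derived from FortMinigameButtons, that shows Start Game or End Game buttons.
# UpdateButtons derives a state from the player's minigame, and a state-to-configuration map drives visibility, enabled state and "Please Wait" text.
# A WidgetSwitcher selects one of two button pairs through configuration structs, so update logic is shared.
# Clicks broadcast dispatchers and change state, only the second pair forwards hover events, and CurrentButton serves focus and centering.
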WidgetBlueprint /FortniteCoreUI/InGame/Creative/MinigameButtons
graphs [
 CenterOnCurrentButton {nodes [
   CenterOnCurrentButton() {is=title K2Node_FunctionEntry}
   > CenterWidget!(execute, ToCenter:Widget Object Reference=Get CurrentButton() -> (CurrentButton:Icon Text Button Object Reference)) -> (then)
  ]}
 EndButtonClicked {nodes [EndButtonClicked() {is=title K2Node_FunctionEntry}]}
 EndButtonHovered {nodes [EndButtonHovered!() -> (then, Button:Common Button Legacy Object Reference) {is=title K2Node_FunctionEntry}]}
 EndButtonUnhovered {nodes [EndButtonUnhovered!() -> (then, Button:Common Button Legacy Object Reference) {is=title K2Node_FunctionEntry}]}
 EventGraph {nodes [
   n1 = "On Button Clicked (StartGameButton)"!() -> (OutputDelegate, then, Button:Common Button Legacy Object Reference) {is=title K2Node_ComponentBoundEvent}
   Call Start Button Clicked!(execute:Exec=[@n1.then, @n4.then], self:Minigame Buttons Object Reference) -> (then:Exec) {is=title K2Node_CallDelegate}
   > SetButtonState!(execute, self, Value:MinigameButtonsState Enum=NewEnumerator4) -> (then)
   n2 = "On Button Clicked (EndGameButton)"!() -> (OutputDelegate, then, Button:Common Button Legacy Object Reference) {is=title K2Node_ComponentBoundEvent}
   Call End Button Clicked!(execute:Exec=[@n2.then, @n3.then], self:Minigame Buttons Object Reference) -> (then:Exec) {is=title K2Node_CallDelegate}
   > SetButtonState(Value=NewEnumerator5)
   n3 = "On Button Clicked (EndGameButton2)"!() -> (OutputDelegate, then, Button:Common Button Legacy Object Reference) {is=title K2Node_ComponentBoundEvent}
   n4 = "On Button Clicked (StartGameButton2)"!() -> (OutputDelegate, then, Button:Common Button Legacy Object Reference) {is=title K2Node_ComponentBoundEvent}
   n5 = "On Button Hovered (StartGameButton2)"!() -> (OutputDelegate, then, Button:Common Button Legacy Object Reference) {is=title K2Node_ComponentBoundEvent}
   > Call Start Button Hovered!(execute:Exec, self:Minigame Buttons Object Reference, Button:Common Button Legacy Object Reference=@n5.Button) -> (then:Exec) {is=title K2Node_CallDelegate}
   n6 = "On Button Hovered (EndGameButton2)"!() -> (OutputDelegate, then, Button:Common Button Legacy Object Reference) {is=title K2Node_ComponentBoundEvent}
   > Call End Button Hovered!(execute:Exec, self:Minigame Buttons Object Reference, Button:Common Button Legacy Object Reference=@n6.Button) -> (then:Exec) {is=title K2Node_CallDelegate}
   n7 = "On Button Unhovered (StartGameButton2)"!() -> (OutputDelegate, then, Button:Common Button Legacy Object Reference) {is=title K2Node_ComponentBoundEvent}
   > Call Start Button Unhovered!(execute:Exec, self:Minigame Buttons Object Reference, Button:Common Button Legacy Object Reference=@n7.Button) -> (then:Exec) {is=title K2Node_CallDelegate}
   n8 = "On Button Unhovered (EndGameButton2)"!() -> (OutputDelegate, then, Button:Common Button Legacy Object Reference) {is=title K2Node_ComponentBoundEvent}
   > Call End Button Unhovered!(execute:Exec, self:Minigame Buttons Object Reference, Button:Common Button Legacy Object Reference=@n8.Button) -> (then:Exec) {is=title K2Node_CallDelegate}
   n9 = SetWidgetSwitcherIndex!() -> (OutputDelegate, then, Index:int) {is=title K2Node_CustomEvent}
   > SetActiveWidgetIndex(self=Get ButtonSwitcher() -> (ButtonSwitcher:Widget Switcher Object Reference), Index=@n9.Index)
   n10 = Get WidgetSwitcherConfigurations() -> (WidgetSwitcherConfigurations:Array of Minigame Buttons Widget Switcher Configuration Structures)
   n11 = Is Valid Index!(TargetArray:Array of Minigame Buttons Widget Switcher Configuration Structures=@n10, IndexToTest:int=@n9.Index) -> (ReturnValue:bool)
   > Set CurrentWidgetSwitcherConfiguration(CurrentWidgetSwitcherConfiguration:Minigame Buttons Widget Switcher Configuration Structure="Get (a ref)"(
     Array:Array of Minigame Buttons Widget Switcher Configuration Structures=@n10
     Dimension 1=Select(Option 0:int=0, Option 1:int=@n9.Index, Index=@n11) -> (ReturnValue:int)
    ) -> (Output:"Minigame Buttons Widget Switcher Configuration Structure (by ref)")) -> (Output_Get:Minigame Buttons Widget Switcher Configuration Structure)
   > UpdateButtonsFromButtonState()
   Event OnInitialized()
   n12 = Get EndGameButton() -> (EndGameButton:Icon Text Button Object Reference)
   n13 = Get StartGameButton() -> (StartGameButton:Icon Text Button Object Reference)
   n14 = Get EndGameButton2() -> (EndGameButton2:Icon Text Button Object Reference)
   n15 = Get StartGameButton2() -> (StartGameButton2:Icon Text Button Object Reference)
   > Set WidgetSwitcherConfigurations(WidgetSwitcherConfigurations:Array of Minigame Buttons Widget Switcher Configuration Structures=Make Array!(
     "[0]":Minigame Buttons Widget Switcher Configuration Structure=Make Minigame Buttons Widget Switcher Configuration!(
      ActiveEndGameButton_2_#1:Icon Text Button Object Reference=@n12
      ActiveStartGameButton_4_#2:Icon Text Button Object Reference=@n13
      StartGameText_7_#3:Text="LOCTEXT(\"START GAME\")"
      PleaseWaitText_9_#4:Text="LOCTEXT(\"PLEASE WAIT...\")"
     ) -> (MinigameButtonsWidgetSwitcherConfiguration:Minigame Buttons Widget Switcher Configuration Structure) {is=title K2Node_MakeStruct}
     "[1]":Minigame Buttons Widget Switcher Configuration Structure=Make Minigame Buttons Widget Switcher Configuration(
      ActiveEndGameButton_2_#1=@n14
      ActiveStartGameButton_4_#2=@n15
      StartGameText_7_#3="LOCTEXT(\"Start Game\")"
      PleaseWaitText_9_#4="LOCTEXT(\"Please Wait...\")"
     )
    ) -> (Array:Array of Minigame Buttons Widget Switcher Configuration Structures)) -> (Output_Get:Array of Minigame Buttons Widget Switcher Configuration Structures)
  ]}
 OnGetDesiredFocusTarget {nodes [
   OnGetDesiredFocusTarget() {is=title K2Node_FunctionEntry}
   > Return Node!(execute, ReturnValue:Widget Object Reference=Get CurrentButton())
  ]}
 SetButtonState {nodes [
   n1 = SetButtonState!() -> (then, Value:MinigameButtonsState Enum=NewEnumerator6) {is=title K2Node_FunctionEntry}
   > Branch(Condition="Not Equal (Enum)"!(A:MinigameButtonsState Enum=@n1, B:MinigameButtonsState Enum=Get ButtonState() -> (ButtonState:MinigameButtonsState Enum=NewEnumerator6)) -> (ReturnValue:bool))
   > Set ButtonState(ButtonState:MinigameButtonsState Enum=Get Value() -> (Value:MinigameButtonsState Enum=NewEnumerator6)) -> (Output_Get:MinigameButtonsState Enum=NewEnumerator6)
   > UpdateButtonsFromButtonState()
  ]}
 StartButtonClicked {nodes [StartButtonClicked() {is=title K2Node_FunctionEntry}]}
 StartButtonHovered {nodes [StartButtonHovered!() -> (then, Button:Common Button Legacy Object Reference) {is=title K2Node_FunctionEntry}]}
 StartButtonUnhovered {nodes [StartButtonUnhovered!() -> (then, Button:Common Button Legacy Object Reference) {is=title K2Node_FunctionEntry}]}
 UpdateButtons {nodes [
   UpdateButtons() {is=title K2Node_FunctionEntry}
   > n1 = Sequence!(execute) -> (then_0, then_1, then_2:Exec)
   > Set UseButtonState(UseButtonState:MinigameButtonsState Enum=NewEnumerator0) -> (Output_Get:MinigameButtonsState Enum=NewEnumerator6)
   n2 = Cast To FortPlayerControllerAthena(execute=@n1.then_1, Object=GetOwningPlayer())
   > Is Valid(InputObject=@n2)
   > n3 = Set CurrentMinigame(CurrentMinigame:Fort Minigame Object Reference=GetMinigame!(self:Fort Player Controller Gameplay Object Reference=@n2) -> (ReturnValue:Fort Minigame Object Reference)) -> (Output_Get:Fort Minigame Object Reference)
   > Is Valid(InputObject=@n3)
   > n4 = Branch(Condition=IsCuratedHub!(self:Fort Minigame Object Reference=@n3) -> (ReturnValue:bool))
   n5 = Branch(execute=@n4.else, Condition=NOT Boolean(A=CanStartMinigame!(self:Fort Minigame Object Reference=@n3) -> (ReturnValue:bool)))
   n6 = IsSpatialGameplayInterfaceReady!(
    self:Fort Minigame Buttons Object Reference
    SpatialAsActor:Actor Object Reference=Get Volume!(self:Fort Minigame Object Reference=@n3) -> (Volume:Actor Object Reference)
   ) -> (ReturnValue:bool)
   Set UseButtonState(
    execute=@n5.else
    UseButtonState=Select!(
     PreGame:MinigameButtonsState Enum=Select!(Option 0:MinigameButtonsState Enum=NewEnumerator3, Option 1:MinigameButtonsState Enum=NewEnumerator2, Index:bool=@n6) -> (ReturnValue)
     Setup:MinigameButtonsState Enum=NewEnumerator6
     Transitioning:MinigameButtonsState Enum=NewEnumerator6
     WaitingForCameras:MinigameButtonsState Enum=NewEnumerator6
     Warmup:MinigameButtonsState Enum=NewEnumerator1
     InProgress:MinigameButtonsState Enum=NewEnumerator1
     PostGameTimeDilation:MinigameButtonsState Enum=NewEnumerator6
     PostRoundEnd:MinigameButtonsState Enum=NewEnumerator6
     PostGameEnd:MinigameButtonsState Enum=NewEnumerator3
     PostGameAbandon:MinigameButtonsState Enum=NewEnumerator3
     PostGameReset:MinigameButtonsState Enum=NewEnumerator6
     Teardown:MinigameButtonsState Enum=NewEnumerator6
     Index:EFortMinigameState Enum=Get CurrentState!(self:Fort Minigame Object Reference=@n3) -> (CurrentState:EFortMinigameState Enum=PreGame)
    ) -> (ReturnValue:MinigameButtonsState Enum)
   ) -> (Output_Get=NewEnumerator6)
   SetButtonState(execute=@n1.then_2, Value=Get UseButtonState() -> (UseButtonState:MinigameButtonsState Enum=NewEnumerator6)) {is=title K2Node_CallFunction}
  ]}
 UpdateButtonsFromButtonState {nodes [
   UpdateButtonsFromButtonState() {is=title K2Node_FunctionEntry}
   > Set ButtonConfiguration(ButtonConfiguration:Minigame Buttons State Configuration Structure=Find(
     TargetMap:Map of MinigameButtonsState Enums to Minigame Buttons State Configuration Structures=Get ButtonConfigurations() -> (ButtonConfigurations:Map of MinigameButtonsState Enums to Minigame Buttons State Configuration Structures=NewEnumerator6)
     Key:"MinigameButtonsState Enum (by ref)"=Get ButtonState() -> (ButtonState=NewEnumerator6)
    ) -> (Value:Minigame Buttons State Configuration Structure)) -> (Output_Get:Minigame Buttons State Configuration Structure)
   n1 = Break Minigame Buttons State Configuration!(MinigameButtonsStateConfiguration:"Minigame Buttons State Configuration Structure (by ref)"=Get ButtonConfiguration() -> (ButtonConfiguration:Minigame Buttons State Configuration Structure)) -> (
    StartGameButtonVisibility_2_#5:ESlateVisibility Enum
    StartGameButtonEnabled_7_#6:bool=~
    EndGameButtonVisibility_4_#7:ESlateVisibility Enum
   ) {is=title K2Node_BreakStruct}
   n2 = Break Minigame Buttons Widget Switcher Configuration!(MinigameButtonsWidgetSwitcherConfiguration:"Minigame Buttons Widget Switcher Configuration Structure (by ref)"=Get CurrentWidgetSwitcherConfiguration() -> (CurrentWidgetSwitcherConfiguration:Minigame Buttons Widget Switcher Configuration Structure)) -> (
    ActiveEndGameButton_2_#1:Icon Text Button Object Reference
    ActiveStartGameButton_4_#2:Icon Text Button Object Reference
   ) {is=title K2Node_BreakStruct}
   > SetVisibility(self=@n2, InVisibility=@n1.EndGameButtonVisibility_4_#7)
   > SetVisibility(self=@n2.ActiveStartGameButton_4_#2, InVisibility=@n1)
   > SetIsEnabled(
    self=@n2.ActiveStartGameButton_4_#2
    bInIsEnabled=@n1.StartGameButtonEnabled_7_#6
   )
   n3 = Break Minigame Buttons Widget Switcher Configuration!(MinigameButtonsWidgetSwitcherConfiguration=Get CurrentWidgetSwitcherConfiguration()) -> (StartGameText_7_#3:Text, PleaseWaitText_9_#4:Text)
   > Set Text!(
    execute
    self:Icon Text Button Object Reference=@n2.ActiveStartGameButton_4_#2
    ButtonText:Text=Select(
     Option 0:Text=@n3.PleaseWaitText_9_#4
     Option 1:Text=@n3
     Index=@n1.StartGameButtonEnabled_7_#6
    ) -> (ReturnValue:Text)
   ) -> (then) {is=title K2Node_CallFunction}
   n4 = Break Minigame Buttons Widget Switcher Configuration!(MinigameButtonsWidgetSwitcherConfiguration=Get CurrentWidgetSwitcherConfiguration()) -> (
    ActiveEndGameButton_2_#1:Icon Text Button Object Reference
    ActiveStartGameButton_4_#2:Icon Text Button Object Reference
   )
   n5 = Break Minigame Buttons State Configuration(MinigameButtonsStateConfiguration=Get ButtonConfiguration()) -> (StartGameButtonEnabled_7_#6=~)
   n6 = "Equal (Enum)"(A:ESlateVisibility Enum=@n5.EndGameButtonVisibility_4_#7, B:ESlateVisibility Enum=Visible)
   > Set CurrentButton(CurrentButton:Icon Text Button Object Reference=Select(
     Option 0:Icon Text Button Object Reference=Select(Option 1=@n4, Index=@n6)
     Option 1:Icon Text Button Object Reference=@n4.ActiveStartGameButton_4_#2
     Index="Equal (Enum)"(A=@n5, B=Visible)
    ) -> (ReturnValue:Icon Text Button Object Reference)) -> (Output_Get:Icon Text Button Object Reference)
  ]}
]
variables [
 ButtonConfigurations {default {
   EndGame {endGameButtonVisibility=Visible, startGameButtonEnabled=true, startGameButtonVisibility=Collapsed}
   EndGameNoClick {endGameButtonVisibility=SelfHitTestInvisible, startGameButtonEnabled=true, startGameButtonVisibility=Collapsed}
   Hidden {endGameButtonVisibility=Collapsed, startGameButtonEnabled=true, startGameButtonVisibility=Collapsed}
   StartGame {endGameButtonVisibility=Collapsed, startGameButtonEnabled=true, startGameButtonVisibility=Visible}
   StartGameDisabled {endGameButtonVisibility=Collapsed, startGameButtonEnabled=false, startGameButtonVisibility=Visible}
   StartGameNoClick {endGameButtonVisibility=Collapsed, startGameButtonEnabled=true, startGameButtonVisibility=SelfHitTestInvisible}
  }}
 ButtonState {type=MinigameButtonsState, default=Invalid}
 CurrentButton {type=IconTextButton_C, default=null}
 CurrentWidgetSwitcherConfiguration {
  type=MinigameButtonsWidgetSwitcherConfiguration
  default {activeEndGameButton=null, activeStartGameButton=null, pleaseWaitText="", startGameText=""}
 }
 EndButtonClicked {}
 EndButtonHovered {}
 EndButtonUnhovered {}
 StartButtonClicked {}
 StartButtonHovered {}
 StartButtonUnhovered {}
 WidgetSwitcherConfigurations {type="array<MinigameButtonsWidgetSwitcherConfiguration>", default []}
]
widgets {
 info {inheritedWidgetCount=0, parentClass=/Script/FortniteUI.FortMinigameButtons, rootWidgetClass=&Overlay, widgetCount=8}
 widgets [
  Overlay_0 {widgetClassPath=&Overlay, widget=Overlay_0}
  ButtonSwitcher {
   widgetClassPath=&WidgetSwitcher
   bIsVariable=true
   parent=Overlay_0
   widget=ButtonSwitcher
   slot {class=OverlaySlot, properties {horizontalAlignment=HAlign_Fill, verticalAlignment=VAlign_Fill}}
  }
  State1 {widgetClassPath=&Overlay, parent=ButtonSwitcher, widget=State1, slot {class=WidgetSwitcherSlot}}
  EndGameButton {
   widgetClassPath=&IconTextButton
   bIsVariable=true
   parent=State1
   widget=EndGameButton
   properties {
    buttonText=END GAME
    contentAlignment=HAlign_Center
    minHeight=100
    minWidth=400
    navigation=&self:WidgetTree.EndGameButton.WidgetNavigation_0
    onButtonBaseClicked=(&self:WidgetTree.EndGameButton.HandleOnButtonClicked)
    onButtonBaseDoubleClicked=(&self:WidgetTree.EndGameButton.HandleOnButtonDoubleClicked)
    onButtonBaseHovered=(&self:WidgetTree.EndGameButton.HandleOnButtonHovered)
    onButtonBaseUnhovered=(&self:WidgetTree.EndGameButton.HandleOnButtonUnhovered)
    onSelectedChangedBase=(&self:WidgetTree.EndGameButton.HandleOnSelectedChanged)
    style=/FortUILibrary/Foundation/Buttons/ButtonStyles/Skew/ButtonStyle-Skew_Red.ButtonStyle-Skew_Red_C
    visibility=Collapsed
    pressedSlateSoundOverride {resourceObject=/Game/Sounds/Fort_UI_Sounds/FrontEnd/New_FrontEnd/Fortnite_UI_StartButton_03_Cue}
    triggeringInputAction {dataTable=&InputActionDataTable, rowName=GenericAccept}
   }
   slot {class=OverlaySlot, properties {horizontalAlignment=HAlign_Fill, verticalAlignment=VAlign_Fill}}
  }
  StartGameButton {
   widgetClassPath=&IconTextButton
   bIsVariable=true
   parent=State1
   widget=StartGameButton
   properties {
    buttonText=START GAME
    contentAlignment=HAlign_Center
    minHeight=100
    minWidth=400
    navigation=&self:WidgetTree.StartGameButton.WidgetNavigation_0
    onButtonBaseClicked=(&self:WidgetTree.StartGameButton.HandleOnButtonClicked)
    onButtonBaseDoubleClicked=(&self:WidgetTree.StartGameButton.HandleOnButtonDoubleClicked)
    onButtonBaseHovered=(&self:WidgetTree.StartGameButton.HandleOnButtonHovered)
    onButtonBaseUnhovered=(&self:WidgetTree.StartGameButton.HandleOnButtonUnhovered)
    onSelectedChangedBase=(&self:WidgetTree.StartGameButton.HandleOnSelectedChanged)
    style=/FortUILibrary/Foundation/Buttons/ButtonStyles/ButtonStyle_Feature_L_Yellow.ButtonStyle_Feature_L_Yellow_C
    iconBrush {
     bIsDynamicallyLoaded=false
     drawAs=NoDrawType
     imageType=NoImage
     mirroring=NoMirror
     resourceName=None
     resourceObject=null
     tiling=NoTile
     imageSize=<0 0>
     margin=<ltrb 0 0 0 0>
     outlineSettings {
      bUseBrushTransparency=false
      roundingType=HalfHeightRadius
      width=0
      color {colorUseRule=UseColor_Specified, specifiedColor=<rgba 0 0 0 0>}
      cornerRadii=<0 0 0 1>
     }
     tintColor {colorUseRule=UseColor_Specified, specifiedColor=<rgba 0 0 0 1>}
     uVRegion {bIsValid=false, max=<0 0>, min=<0 0>}
    }
    pressedSlateSoundOverride {resourceObject=/Game/Sounds/Fort_UI_Sounds/FrontEnd/SC_UI_MatchMaking_PlayButton_Pressed_Creative_02}
    triggeringInputAction {dataTable=&InputActionDataTable, rowName=GenericAccept}
   }
   slot {class=OverlaySlot, properties {horizontalAlignment=HAlign_Fill, verticalAlignment=VAlign_Fill}}
  }
  State2 {widgetClassPath=&Overlay, parent=ButtonSwitcher, widget=State2, slot {class=WidgetSwitcherSlot}}
  EndGameButton2 {
   widgetClassPath=&IconTextButton
   bIsVariable=true
   parent=State2
   widget=EndGameButton2
   properties {
    buttonText=End Game
    minHeight=100
    minWidth=400
    onButtonBaseClicked=(&self:WidgetTree.EndGameButton2.HandleOnButtonClicked)
    onButtonBaseDoubleClicked=(&self:WidgetTree.EndGameButton2.HandleOnButtonDoubleClicked)
    onButtonBaseHovered=(&self:WidgetTree.EndGameButton2.HandleOnButtonHovered)
    onButtonBaseUnhovered=(&self:WidgetTree.EndGameButton2.HandleOnButtonUnhovered)
    onSelectedChangedBase=(&self:WidgetTree.EndGameButton2.HandleOnSelectedChanged)
    style=&ButtonStyle-Skew_Red_C
    visibility=Collapsed
    pressedSlateSoundOverride {resourceObject=/Game/Sounds/Fort_UI_Sounds/FrontEnd/New_FrontEnd/Fortnite_UI_MainMenu_Tab_Hover_02_Cue}
   }
   slot {class=OverlaySlot, properties {horizontalAlignment=HAlign_Fill, verticalAlignment=VAlign_Fill}}
  }
  StartGameButton2 {
   widgetClassPath=&IconTextButton
   bIsVariable=true
   parent=State2
   widget=StartGameButton2
   properties {
    buttonText=Start Game
    minHeight=100
    minWidth=400
    onButtonBaseClicked=(&self:WidgetTree.StartGameButton2.HandleOnButtonClicked)
    onButtonBaseDoubleClicked=(&self:WidgetTree.StartGameButton2.HandleOnButtonDoubleClicked)
    onButtonBaseHovered=(&self:WidgetTree.StartGameButton2.HandleOnButtonHovered)
    onButtonBaseUnhovered=(&self:WidgetTree.StartGameButton2.HandleOnButtonUnhovered)
    onSelectedChangedBase=(&self:WidgetTree.StartGameButton2.HandleOnSelectedChanged)
    style=&ButtonStyle_Feature_L_Yellow_C
    pressedSlateSoundOverride {resourceObject=&Fortnite_UI_MainMenu_Tab_Hover_02_Cue}
   }
   slot {class=OverlaySlot, properties {horizontalAlignment=HAlign_Fill, verticalAlignment=VAlign_Fill}}
  }
 ]
}
