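# Modal confirmation panel, a FortActivatablePanel, asking the player to confirm clearing their saved island persistence data ("Reset Progress").
# OK asks the owning FortPlayerController to clear island data, then pops the panel, which also pops if the cast fails.
# Cancel and the touch-only close button just pop it from the modal layer.
# SetDescription lets callers replace the body text shown inside the blurred lightbox layout.
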WidgetBlueprint /FortniteCoreUI/PlayerClearPersistenceIslandDataPromptWidget
graphs [
 EventGraph {nodes [
   n1 = "On Button Clicked (Button_CloseTouch)"!() -> (OutputDelegate, then, Button:Common Button Legacy Object Reference) {is=title K2Node_ComponentBoundEvent}
   PopActivatablePanelInModalLayer!(execute=[@n3.then, @n1.then, @n4.then, @n2.CastFailed], Panel:Common Activatable Panel Legacy Object Reference=Self-Reference()) -> (then)
   "On Button Clicked (ApplyButton)"!() -> (OutputDelegate, then, Button:Common Button Legacy Object Reference) {is=title K2Node_ComponentBoundEvent}
   > n2 = Cast To FortPlayerController(Object=GetOwningPlayer())
   > n3 = PlayerRequestClearIslandData(self:Fort Player Controller Object Reference=@n2)
   n4 = "On Button Clicked (CancelButton)"!() -> (OutputDelegate, then, Button:Common Button Legacy Object Reference) {is=title K2Node_ComponentBoundEvent}
  ]}
 SetDescription {nodes [
   n1 = SetDescription!() -> (then, InDescription:Text) {is=title K2Node_FunctionEntry}
   > "SetText (Text)"(self=Get Text_CurrentDescription() -> (Text_CurrentDescription:Common Text Object Reference), InText=@n1)
  ]}
]
variables []
widgets {
 info {
  inheritedWidgetCount=0
  namedSlotCount=1
  parentClass=/Script/FortniteUI.FortActivatablePanel
  rootWidgetClass=&Border
  widgetCount=18
 }
 widgets [
  RootBorder {
   widgetClassPath=&Border
   widget=RootBorder
   properties {
    background {
     bIsDynamicallyLoaded=false
     drawAs=NoDrawType
     imageType=NoImage
     mirroring=NoMirror
     resourceName=None
     resourceObject=null
     tiling=NoTile
     imageSize=<32 32>
     margin=<ltrb 0 0 0 0>
     outlineSettings {
      bUseBrushTransparency=false
      roundingType=HalfHeightRadius
      width=0
      color {colorUseRule=UseColor_Specified, specifiedColor=<rgba 0 0 0 0>}
      cornerRadii=<0 0 0 0>
     }
     tintColor {colorUseRule=UseColor_Specified, specifiedColor=<rgba 1 1 1 1>}
     uVRegion {bIsValid=false, max=<0 0>, min=<0 0>}
    }
    padding=<ltrb 0 0 0 0>
   }
  }
  BackgroundBlur_2 {
   widgetClassPath=&BackgroundBlur
   parent=RootBorder
   widget=BackgroundBlur_2
   properties {blurStrength=15}
   slot {class=BorderSlot, properties {padding=<ltrb 0 0 0 0>}}
  }
  Lightbox {
   widgetClassPath=/Game/UI/Foundation/Lightbox/Lightbox
   bIsVariable=true
   parent=BackgroundBlur_2
   widget=Lightbox
   slot {class=BackgroundBlurSlot, properties {padding=<ltrb 0 0 0 0>}}
  }
  RootOverlay {widgetClassPath=&Overlay, namedSlotHost=Lightbox, widget=RootOverlay}
  TouchToCloseZone {
   widgetClassPath=&CommonBorder
   parent=RootOverlay
   widget=TouchToCloseZone
   properties {horizontalAlignment=HAlign_Center, verticalAlignment=VAlign_Center, padding=<ltrb 0 0 0 0>}
   slot {class=OverlaySlot, properties {horizontalAlignment=HAlign_Fill, verticalAlignment=VAlign_Fill}}
  }
  SizeBox_4 {
   widgetClassPath=&SizeBox
   parent=TouchToCloseZone
   widget=SizeBox_4
   properties {bOverride_WidthOverride=true, widthOverride=1000}
   slot {class=BorderSlot, properties {horizontalAlignment=HAlign_Center, verticalAlignment=VAlign_Center, padding=<ltrb 0 0 0 0>}}
  }
  CommonBorder_10 {
   widgetClassPath=&CommonBorder
   parent=SizeBox_4
   widget=CommonBorder_10
   properties {
    style=/FortUILibrary/Styles/Border/Border-MainModal.Border-MainModal_C
    visibility=Visible
    background {
     bIsDynamicallyLoaded=false
     drawAs=Image
     imageType=NoImage
     mirroring=NoMirror
     resourceName=None
     resourceObject=/FortUILibrary/Styles/Border/Materials/M-Border-MainModal
     tiling=NoTile
     imageSize=<128 32>
     margin=<ltrb 0 0 0 0>
     outlineSettings {
      bUseBrushTransparency=false
      roundingType=HalfHeightRadius
      width=0
      color {colorUseRule=UseColor_Specified, specifiedColor=<rgba 0 0 0 0>}
      cornerRadii=<0 0 0 0>
     }
     tintColor {colorUseRule=UseColor_Specified, specifiedColor=<rgba 1 1 1 1>}
     uVRegion {bIsValid=false, max=<0 0>, min=<0 0>}
    }
    padding=<ltrb 0 0 0 0>
   }
   slot {class=SizeBoxSlot}
  }
  VerticalBox_3 {
   widgetClassPath=&VerticalBox
   parent=CommonBorder_10
   widget=VerticalBox_3
   properties {navigation=&self:WidgetTree.VerticalBox_3.WidgetNavigation_0}
   slot {class=BorderSlot, properties {padding=<ltrb 0 0 0 0>}}
  }
  CommonBorder_Modal {
   widgetClassPath=&CommonBorder
   parent=VerticalBox_3
   widget=CommonBorder_Modal
   properties {style=/FortUILibrary/Styles/Border/Border-ModalHeader.Border-ModalHeader_C, padding=<ltrb 0 0 0 0>}
   slot {class=VerticalBoxSlot}
  }
  HorizontalBox_1 {
   widgetClassPath=&HorizontalBox
   parent=CommonBorder_Modal
   widget=HorizontalBox_1
   slot {class=BorderSlot, properties {padding=<ltrb 0 0 0 0>}}
  }
  Spacer_0 {widgetClassPath=&Spacer, parent=HorizontalBox_1, widget=Spacer_0, properties {size=<30 100>}, slot {class=HorizontalBoxSlot}}
  Title {
   widgetClassPath=&CommonTextBlock
   bIsVariable=true
   parent=HorizontalBox_1
   widget=Title
   properties {
    style=/FortUILibrary/Styles/Typography/Foundation/TextStyle-Header-L.TextStyle-Header-L_C
    text=Reset Progress
    font {
     bForceMonospaced=false
     bMaterialIsStencil=false
     fontMaterial=null
     fontObject=&Burbank
     letterSpacing=0
     monospacedWidth=1
     size=48
     skewAmount=0
     typefaceFontName=Burbank-Black
     outlineSettings {
      bApplyOutlineToDropShadows=false
      bMiteredCorners=false
      bSeparateFillAlpha=false
      outlineMaterial=null
      outlineSize=0
      outlineColor=<rgba 0 0 0 1>
     }
    }
    shadowOffset=<0 0>
   }
   slot {class=HorizontalBoxSlot, properties {horizontalAlignment=HAlign_Left, verticalAlignment=VAlign_Center, padding=<ltrb 8 0 0 0>}}
  }
  VisibilityBorder_Touch {
   widgetClassPath=&CommonHardwareVisibilityBorder
   parent=HorizontalBox_1
   widget=VisibilityBorder_Touch
   properties {visibilityQuery {
     autoDescription=" ANY( Input.Touch )"
     tokenStreamVersion=0
     userDescription=""
     queryTokenStream [0,1,1,1,0]
     tagDictionary [{tagName=Input.Touch}]
    }}
   slot {class=HorizontalBoxSlot, properties {horizontalAlignment=HAlign_Right, size {sizeRule=Fill, value=1}}}
  }
  Button_CloseTouch {
   widgetClassPath=/FortUILibrary/Foundation/Buttons/CloseButton
   bIsVariable=true
   parent=VisibilityBorder_Touch
   widget=Button_CloseTouch
   properties {
    onButtonBaseClicked=(&self:WidgetTree.Button_CloseTouch.HandleOnButtonClicked)
    onButtonBaseDoubleClicked=(&self:WidgetTree.Button_CloseTouch.HandleOnButtonDoubleClicked)
    onButtonBaseHovered=(&self:WidgetTree.Button_CloseTouch.HandleOnButtonHovered)
    onButtonBaseUnhovered=(&self:WidgetTree.Button_CloseTouch.HandleOnButtonUnhovered)
    onSelectedChangedBase=(&self:WidgetTree.Button_CloseTouch.HandleOnSelectedChanged)
    renderTransform {angle=0, scale=<0.8 0.8>, shear=<0 0>, translation=<0 0>}
   }
   slot {class=BorderSlot, properties {padding=<ltrb 0 0 0 0>}}
  }
  Text_CurrentDescription {
   widgetClassPath=&CommonTextBlock
   bIsVariable=true
   parent=VerticalBox_3
   widget=Text_CurrentDescription
   properties {
    autoWrapText=true
    justification=Center
    style=/FortUILibrary/Styles/Typography/Foundation/TextStyle-Base-M-B_Black.TextStyle-Base-M-B_Black_C
    text=Do you want to clear your saved data?
    colorAndOpacity {colorUseRule=UseColor_Specified, specifiedColor=<rgba 0 0 0 1>}
    font {
     bForceMonospaced=false
     bMaterialIsStencil=false
     fontMaterial=null
     fontObject=&NotoSans
     letterSpacing=0
     monospacedWidth=1
     size=24
     skewAmount=0
     typefaceFontName=NotoSans-Bold
     outlineSettings {
      bApplyOutlineToDropShadows=false
      bMiteredCorners=false
      bSeparateFillAlpha=false
      outlineMaterial=null
      outlineSize=0
      outlineColor=<rgba 0 0 0 1>
     }
    }
    shadowOffset=<0 0>
   }
   slot {class=VerticalBoxSlot, properties {padding=<ltrb 30 30 30 0>}}
  }
  HorizontalBox_6 {
   widgetClassPath=&HorizontalBox
   parent=VerticalBox_3
   widget=HorizontalBox_6
   slot {class=VerticalBoxSlot, properties {horizontalAlignment=HAlign_Right, padding=<ltrb 10 10 20 20>}}
  }
  ApplyButton {
   widgetClassPath=&IconTextButton
   bIsVariable=true
   parent=HorizontalBox_6
   widget=ApplyButton
   properties {
    buttonText=OK
    onButtonBaseClicked=(&self:WidgetTree.ApplyButton.HandleOnButtonClicked)
    onButtonBaseDoubleClicked=(&self:WidgetTree.ApplyButton.HandleOnButtonDoubleClicked)
    onButtonBaseHovered=(&self:WidgetTree.ApplyButton.HandleOnButtonHovered)
    onButtonBaseUnhovered=(&self:WidgetTree.ApplyButton.HandleOnButtonUnhovered)
    onSelectedChangedBase=(&self:WidgetTree.ApplyButton.HandleOnSelectedChanged)
    triggeringInputAction {dataTable=&InputActionDataTable, rowName=Apply}
   }
   slot {class=HorizontalBoxSlot, properties {padding=<ltrb 0 0 24 0>}}
  }
  CancelButton {
   widgetClassPath=&IconTextButton
   bIsVariable=true
   parent=HorizontalBox_6
   widget=CancelButton
   properties {
    buttonText=Cancel
    onButtonBaseClicked=(&self:WidgetTree.CancelButton.HandleOnButtonClicked)
    onButtonBaseDoubleClicked=(&self:WidgetTree.CancelButton.HandleOnButtonDoubleClicked)
    onButtonBaseHovered=(&self:WidgetTree.CancelButton.HandleOnButtonHovered)
    onButtonBaseUnhovered=(&self:WidgetTree.CancelButton.HandleOnButtonUnhovered)
    onSelectedChangedBase=(&self:WidgetTree.CancelButton.HandleOnSelectedChanged)
    triggeringInputAction {dataTable=&InputActionDataTable, rowName=Cancel}
   }
   slot {class=HorizontalBoxSlot}
  }
 ]
}
