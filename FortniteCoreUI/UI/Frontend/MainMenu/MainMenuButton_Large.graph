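# Large main menu button widget, a FortMainMenuButton child with an icon, uppercase label and input action glyph over a material-driven sidebar border.
# Hover plays a short animation, reversed on unhover, that darkens and nudges the icon and text right while animating the border material's highlight parameters.
# When the input glyph updates, its dynamic material gets the dark variant with no outline.
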WidgetBlueprint /FortniteCoreUI/UI/Frontend/MainMenu/MainMenuButton_Large
animations [Hovered {
  binding_count=3
  length_seconds=0.15
  object=&self:Hovered
  sequence {
   bindings [
    Image_Icon {
     id=#1
     possessed_class=Image
     tracks [
      {
       type=MovieSceneColorTrack
       display_name=Color and Opacity
       sections [{
         type=MovieSceneColorSection
         channels [
          Color.R {type=MovieSceneScriptingFloatChannel, keys [[0,0.3185],[9,0]]}
          Color.G {type=MovieSceneScriptingFloatChannel, keys [[0,0.7231],[9,0.004]]}
          Color.B {type=MovieSceneScriptingFloatChannel, keys [[0,1],[9,0.0723]]}
          Color.A {type=MovieSceneScriptingFloatChannel, const [1,2]}
         ]
        }]
      }
      {
       type=MovieScene2DTransformTrack
       sections [{
         type=MovieScene2DTransformSection
         end_frame=5
         start_frame=0
         channels [
          Translation.X {type=MovieSceneScriptingFloatChannel, keys [[0,0],[5,6]]}
          Translation.Y {type=MovieSceneScriptingFloatChannel, const [0,1]}
          Angle {type=MovieSceneScriptingFloatChannel, const [0,1]}
          Scale.X {type=MovieSceneScriptingFloatChannel, const [1,1]}
          Scale.Y {type=MovieSceneScriptingFloatChannel, const [1,1]}
          Shear.X {type=MovieSceneScriptingFloatChannel, const [0,1]}
          Shear.Y {type=MovieSceneScriptingFloatChannel, const [0,1]}
         ]
        }]
      }
     ]
    }
    Text_Content {
     id=#2
     possessed_class=CommonTextBlock
     tracks [
      {
       type=MovieSceneColorTrack
       display_name=Color and Opacity
       sections [{
         type=MovieSceneColorSection
         channels [
          Color.R {type=MovieSceneScriptingFloatChannel, keys [[0,0.3185],[9,0]]}
          Color.G {type=MovieSceneScriptingFloatChannel, keys [[0,0.7231],[9,0.004]]}
          Color.B {type=MovieSceneScriptingFloatChannel, keys [[0,1],[9,0.0723]]}
          Color.A {type=MovieSceneScriptingFloatChannel, const [1,2]}
         ]
        }]
      }
      {
       type=MovieScene2DTransformTrack
       sections [{
         type=MovieScene2DTransformSection
         end_frame=5
         start_frame=0
         channels [
          Translation.X {type=MovieSceneScriptingFloatChannel, keys [[0,0],[5,6]]}
          Translation.Y {type=MovieSceneScriptingFloatChannel, const [0,1]}
          Angle {type=MovieSceneScriptingFloatChannel, const [0,1]}
          Scale.X {type=MovieSceneScriptingFloatChannel, const [1,1]}
          Scale.Y {type=MovieSceneScriptingFloatChannel, const [1,1]}
          Shear.X {type=MovieSceneScriptingFloatChannel, const [0,1]}
          Shear.Y {type=MovieSceneScriptingFloatChannel, const [0,1]}
         ]
        }]
      }
     ]
    }
    ContentBorder {
     id=#3
     possessed_class=Border
     tracks [{
       type=MovieSceneWidgetMaterialTrack
       sections [{
         type=MovieSceneParameterSection
         channels [
          AnimateHighlight {type=MovieSceneScriptingFloatChannel, keys [[0,0],[6,1],[9,1]]}
          white_Burst_start {type=MovieSceneScriptingFloatChannel, keys [[0,0.5],[6,0.4]]}
          White_sharpness {type=MovieSceneScriptingFloatChannel, keys [[0,1.5],[6,2.5]]}
          ShowTrimAlways {type=MovieSceneScriptingFloatChannel, const [0,1]}
          LeftBar_Position {type=MovieSceneScriptingFloatChannel, const [-0.02,2]}
          Tri Tile Amount {type=MovieSceneScriptingFloatChannel, keys [[0,0],[1,0],[2,5],[6,5]]}
          LeftBar_UnHovered_ColorR {type=MovieSceneScriptingFloatChannel, const [0.9301,1]}
          LeftBar_UnHovered_ColorG {type=MovieSceneScriptingFloatChannel, const [1,1]}
          LeftBar_UnHovered_ColorB {type=MovieSceneScriptingFloatChannel, const [0.0103,1]}
          LeftBar_UnHovered_ColorA {type=MovieSceneScriptingFloatChannel, const [1,1]}
          RGB_SolidFillR {type=MovieSceneScriptingFloatChannel, const [0.0024,1]}
          RGB_SolidFillG {type=MovieSceneScriptingFloatChannel, const [0.0513,1]}
          RGB_SolidFillB {type=MovieSceneScriptingFloatChannel, const [0.3864,1]}
          RGB_SolidFillA {type=MovieSceneScriptingFloatChannel, const [1,1]}
          ShineRGBAR {type=MovieSceneScriptingFloatChannel, const [1,1]}
          ShineRGBAG {type=MovieSceneScriptingFloatChannel, const [1,1]}
          ShineRGBAB {type=MovieSceneScriptingFloatChannel, const [1,1]}
          ShineRGBAA {type=MovieSceneScriptingFloatChannel, const [0,1]}
         ]
        }]
      }]
    }
   ]
   display_rate {numerator=60}
  }
 }]
graphs [EventGraph {nodes [
   "On Input Icon Updated (InputActionWidget)"() {is=title K2Node_ComponentBoundEvent}
   > n1 = GetIconDynamicMaterial!(execute, self:Common Action Widget Object Reference=Get InputActionWidget() -> (InputActionWidget:Common Action Widget Object Reference)) -> (then, ReturnValue:Material Instance Dynamic Object Reference)
   > SetScalarParameterValue(self=@n1, ParameterName=Use Dark, Value=1)
   > SetScalarParameterValue(self=@n1, ParameterName=Outline Opacity, Value=0)
   Event On Hovered()
   > PlayAnimation(InAnimation=Get Hovered() -> (Hovered:Widget Animation Object Reference))
   Event On Unhovered()
   > PlayAnimationReverse(InAnimation=Get Hovered())
  ]}]
variables []
widgets {
 info {inheritedWidgetCount=4, parentClass=/Script/FortniteUI.FortMainMenuButton, rootWidgetClass=&Border, widgetCount=7}
 widgets [
  ContentBorder {
   widgetClassPath=&Border
   bInherited=true
   widget=ContentBorder
   properties {
    background {
     bIsDynamicallyLoaded=false
     drawAs=Image
     imageType=NoImage
     mirroring=NoMirror
     resourceName=None
     resourceObject=/FortUILibrary/Foundation/Buttons/Materials/MI_UI_Sidebar_Large_Button
     tiling=NoTile
     imageSize=<32 32>
     margin=<ltrb 0 0 0 0>
     outlineSettings {
      bUseBrushTransparency=false
      roundingType=HalfHeightRadius
      width=0
      color {colorUseRule=UseColor_Specified, specifiedColor=<rgba 0 0 0 0>}
      cornerRadii=<0 0 0 0>
     }
     tintColor {colorUseRule=UseColor_Specified, specifiedColor=<rgba 1 1 1 1>}
     uVRegion {bIsValid=false, max=<0 0>, min=<0 0>}
    }
    padding=<ltrb 12 0 12 0>
   }
  }
  HorizontalBox {
   widgetClassPath=&HorizontalBox
   parent=ContentBorder
   widget=HorizontalBox
   properties {flowDirectionPreference=Culture}
   slot {class=BorderSlot, properties {padding=<ltrb 12 0 12 0>}}
  }
  Spacer_133 {widgetClassPath=&Spacer, parent=HorizontalBox, widget=Spacer_133, properties {size=<10 1>}, slot {class=HorizontalBoxSlot}}
  Image_Icon {
   widgetClassPath=&Image
   bInherited=true
   parent=HorizontalBox
   widget=Image_Icon
   properties {
    brush {
     bIsDynamicallyLoaded=false
     drawAs=NoDrawType
     imageType=NoImage
     mirroring=NoMirror
     resourceName=None
     resourceObject=null
     tiling=NoTile
     imageSize=<0 0>
     margin=<ltrb 0 0 0 0>
     outlineSettings {
      bUseBrushTransparency=false
      roundingType=HalfHeightRadius
      width=0
      color {colorUseRule=UseColor_Specified, specifiedColor=<rgba 0 0 0 0>}
      cornerRadii=<0 0 0 0>
     }
     tintColor {colorUseRule=UseColor_Specified, specifiedColor=<rgba 1 1 1 1>}
     uVRegion {bIsValid=false, max=<0 0>, min=<0 0>}
    }
    colorAndOpacity=<rgba 0.3185 0.7231 1 1>
   }
   slot {class=HorizontalBoxSlot, properties {horizontalAlignment=HAlign_Center, verticalAlignment=VAlign_Center, padding=<ltrb 4 4 4 4>}}
  }
  Text_Content {
   widgetClassPath=&CommonTextBlock
   bInherited=true
   parent=HorizontalBox
   widget=Text_Content
   properties {
    mobileFontSizeMultiplier=1.34
    scrollStyle=/FortUILibrary/Styles/Typography/Foundation/TextScrollStyle-NoFade.TextScrollStyle-NoFade_C
    style=/FortUILibrary/Foundation/Buttons/ButtonStyles/Athena/TextStyle_BurbSmBk_15_LightBlue.TextStyle_BurbSmBk_15_LightBlue_C
    text=Menu Settings
    textTransformPolicy=ToUpper
    colorAndOpacity {colorUseRule=UseColor_Specified, specifiedColor=<rgba 0.3185 0.7231 1 1>}
    font {
     bForceMonospaced=false
     bMaterialIsStencil=false
     fontMaterial=null
     fontObject=&BurbankSmall
     letterSpacing=0
     monospacedWidth=1
     size=15
     skewAmount=0
     typefaceFontName=Black
     outlineSettings {
      bApplyOutlineToDropShadows=false
      bMiteredCorners=false
      bSeparateFillAlpha=false
      outlineMaterial=null
      outlineSize=0
      outlineColor=<rgba 0 0 0 1>
     }
    }
    shadowOffset=<0 0>
   }
   slot {
    class=HorizontalBoxSlot
    properties {horizontalAlignment=HAlign_Left, verticalAlignment=VAlign_Center, padding=<ltrb 0 0 8 0>, size {sizeRule=Fill, value=1}}
   }
  }
  InputActionWidget {
   widgetClassPath=&CommonActionWidget
   bInherited=true
   parent=HorizontalBox
   widget=InputActionWidget
   properties {
    progressMaterialParam=Percentage
    progressMaterialBrush {
     bIsDynamicallyLoaded=false
     drawAs=Image
     imageType=NoImage
     mirroring=NoMirror
     resourceName=None
     resourceObject=/Game/UI/Materials/HollowCircularProgressBar
     tiling=NoTile
     imageSize=<36 36>
     margin=<ltrb 0 0 0 0>
     outlineSettings {
      bUseBrushTransparency=false
      roundingType=HalfHeightRadius
      width=0
      color {colorUseRule=UseColor_Specified, specifiedColor=<rgba 0 0 0 0>}
      cornerRadii=<0 0 0 0>
     }
     tintColor {colorUseRule=UseColor_Specified, specifiedColor=<rgba 1 1 0 1>}
     uVRegion {bIsValid=false, max=<0 0>, min=<0 0>}
    }
   }
   slot {class=HorizontalBoxSlot, properties {horizontalAlignment=HAlign_Right, verticalAlignment=VAlign_Center, padding=<ltrb 4 0 10 0>}}
  }
  Spacer_203 {widgetClassPath=&Spacer, parent=HorizontalBox, widget=Spacer_203, properties {size=<10 1>}, slot {class=HorizontalBoxSlot}}
 ]
}
